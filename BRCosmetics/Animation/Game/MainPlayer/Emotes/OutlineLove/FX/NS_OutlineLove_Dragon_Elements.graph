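# Niagara system for the dragon elements of the OutlineLove emote, with flame wings, tail ashes and timed spawn and despawn bursts.
# GPU ashes follow tail bones, while wing meshes track finger sockets and scale with User.Fade_out_geo.
# Merged burst emitters switch motion and size by spawn group, using visibility tags and per-renderer attributes for smoke, fire, embers, distortion and light.
# Front-end modifiers vary some values between menu and gameplay.
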NiagaraSystem /BRCosmetics/Animation/Game/MainPlayer/Emotes/OutlineLove/FX/NS_OutlineLove_Dragon_Elements
dependencies {
 usedDataInterfaces [
  &NiagaraDataInterfaceColorCurve
  &NiagaraDataInterfaceCurve
  &NiagaraDataInterfaceSpriteRendererInfo
  &NiagaraDataInterfaceArrayFloat3
  &NiagaraDataInterfaceArrayFloat
  &NiagaraDataInterfaceMeshRendererInfo
  &NiagaraDataInterfaceSkeletalMesh
 ]
 usedDynamicInputs [
  &MakeFloatFromVector
  &SelectVectorFromArray
  &ReturnExecIndex
  &SelectFloatFromArray
  &SimulationPosition
 ]
 usedModules [
  &SystemState
  &self:SystemSpawnScript.SystemScriptSource.SystemScriptGraph.NiagaraNodeAssignment_0.SetVariables_#1
  &self:SystemSpawnScript.SystemScriptSource.SystemScriptGraph.NiagaraNodeAssignment_4.SetVariables_#2
  &self:SystemSpawnScript.SystemScriptSource.SystemScriptGraph.NiagaraNodeAssignment_6.SetVariables_#3
  &self:SystemSpawnScript.SystemScriptSource.SystemScriptGraph.NiagaraNodeAssignment_7.SetVariables_#4
 ]
}
emitters {
 Ashes {
  simTarget=GPUComputeSim
  emitterUpdateScript [
   &EmitterState(Life Cycle Mode=::NewEnumerator1, Loop Behavior=::NewEnumerator1, Loop Duration=3.8, UseLoopDelay=false)
   &SpawnRate(Spawn Probability=0.7, SpawnRate=150, Use Spawn Probability=true)
  ]
  particleSpawnScript [
   /SetVariables(Particles.VisibilityTag=&SpawnGroup()) {isSetParameters=true}
   &FN_SetSpawnGroup(Number of Spawn Groups=3)
   &InitializeParticle(
    Color=<rgba 10 1.0749 0 1>
    Lifetime=&MergeHelper_FloatSwitch(
     0=&RandomRangeFloat(Minimum=0.3, Maximum=1)
     1=&RandomRangeFloat(Minimum=0.2, Maximum=0.5)
     2=&RandomRangeFloat(Minimum=0.2, Maximum=0.5)
     3=?
     Index=&SpawnGroup()
     Num=::NewEnumerator1
    )
    Sprite Rotation Mode=::NewEnumerator2
    Sprite Size Max=<2.5 2.5>
    Sprite Size Min=<1 0>
    Sprite Size Mode=::NewEnumerator2
    UsePositionOffset=false
    Write Material Random=true
   )
   &ShapeLocation(Box Size=<10 10 1>, MaskSpawnGroup=false, Override Seed=false, Shape Primitive=::NewEnumerator2, Surface Only=false, UseOwnerScale=true) {module=ShapeLocation001}
   &ApplyInitialForces()
   &SkeletalMeshLocation(
    Bone / Socket Index=&SpawnGroup()
    Bone Sampling Mode=::NewEnumerator4
    MaskSpawnGroup=false
    Sampled Position Coordinate Space=::Local
    Sampled Position Offset=<0 10 0>
    Skeletal Mesh=$System.tail
   )
  ]
  particleUpdateScript [
   &ParticleState()
   &Color(Color=&ColorFromCurve(ColorCurve=DataInterface() {propertyValues {
       AlphaCurve {keys [<at 0 1 RCIM_Cubic>,<at 1 0 RCIM_Cubic>]}
       BlueCurve {keys [<at 0 0 RCIM_Cubic>,<at 0.1147 0>,<at 0.5143 0 RCIM_Cubic>]}
       GreenCurve {keys [<at 0 9.1707 RCIM_Cubic>,<at 0.1147 3.0139>,<at 0.5143 0.0332 RCIM_Cubic>]}
       RedCurve {keys [<at 0 20 RCIM_Cubic>,<at 0.1147 10>,<at 0.5143 1 RCIM_Cubic>]}
      }}))
   /SetVariables(Particles.Glow03_Color=&MultiplyLinearColor(Linear Color=$Particles.Color, Scale Factor=<1 1 1 0.04>)) {isSetParameters=true}
   /SetVariables(Particles.Glow03_Size=&Multiply_Vector2D_ByFloat(
     Float=&FloatFromCurve(
      FloatCurve=DataInterface() {propertyValues {Curve {keys [
          <at 0 0 RCIM_Cubic>
          <at 0.2 1.3 RCIM_Cubic>
          {arriveTangent=-0.1423, interpMode=RCIM_Cubic, leaveTangent=-0.1423, tangentMode=RCTM_User, time=1, value=0}
         ]}}}
      Scale Curve=15
     )
     Vector2D=<2 2>
    )) {isSetParameters=true}
   &ScaleColor(
    Scale Alpha=&FloatFromCurve(FloatCurve=DataInterface() {propertyValues {Curve {keys [<at 0.3 1>,<at 0.4 1>,<at 0.9174 1>,<at 1 0>]}}})
    ScaleA=true
    ScaleRGB=true
   )
   &ScaleSpriteSize(Scale Factor=&Vector2DFromFloat(Value=&FloatFromCurve(FloatCurve=DataInterface() {propertyValues {Curve {keys [
          <at 0 0 RCIM_Cubic>
          {interpMode=RCIM_Cubic, leaveTangent=-3.0296, tangentMode=RCTM_Break, time=0.2, value=1.2}
          {arriveTangent=-1.8107, interpMode=RCIM_Cubic, leaveTangent=-1.8107, tangentMode=RCTM_User, time=1, value=0}
         ]}}}))) {module=ScaleSpriteSize001}
   &GravityForce(Gravity=<0 0 -100>)
   &CurlNoiseForce(
    Noise Frequency=350
    Noise Strength=&Multiply_Float(A=233, B=1.5)
    Pan Noise=true
    Pan Noise Field=<100 0 1>
    Random Seed=10
   )
   &Drag(
    Drag=&MergeHelper_FloatSwitch(
     0=&RandomRangeFloat(Minimum=1, Maximum=5)
     1=&RandomRangeFloat(Minimum=1, Maximum=5)
     2=&RandomRangeFloat(Minimum=1, Maximum=5)
     3=?
     Index=0
     Num=::NewEnumerator1
    )
    Use Linear Drag=true
    Use Rotational Drag=false
   )
   &SolveForcesAndVelocity(Clamp Velocity=false, Limit Acceleration=false)
   &ScaleSpriteSizeBySpeed(Max Scale Factor=<5 6>, Velocity Threshold=356)
   /SetVariables(
    Particles.Ashes Color:LinearColor=&MakeLinearColorFromVectorAndFloat(
     "Vector (RGB)"=&MakeVectorFromLinearColorRGB(Color=&ColorFromCurve(ColorCurve=DataInterface() {propertyValues {
         AlphaCurve {keys [<at 0 1 RCIM_Cubic>]}
         BlueCurve {keys [<at 0 1.473 RCIM_Cubic>,<at 0.6088 0.0879 RCIM_Cubic>]}
         GreenCurve {keys [<at 0 11.4982 RCIM_Cubic>,<at 0.6088 0.0919 RCIM_Cubic>]}
         RedCurve {keys [<at 0 50 RCIM_Cubic>,<at 0.6088 0.099 RCIM_Cubic>]}
        }}))
     "Float (Alpha)"=&FloatFromCurve(FloatCurve=DataInterface() {propertyValues {Curve {keys [<at 0.8 1 RCIM_Cubic>,<at 1 0>]}}})
    )
    Particles.Ashes SpriteSize:Vector2f=&Vector2DFromFloat(Value=&RandomRangeFloat(Minimum=0.5, Maximum=5))
    Particles.Vector 4:Vector4f=&MakeVector4(X=0, Y=0, Z=0, W=0)
   ) {isSetParameters=true}
   &SubUVAnimation(
    Output Alternate Renderer Binding=true
    "Sprite Renderer "=DataInterface() {propertyValues {SpriteRenderer=&self:Ashes_0.NiagaraSpriteRendererProperties_3}}
    UseEndFrame=false
    UseStartFrame=false
   ) {module=SubUVAnimation001}
   /SetVariables(
    Particles.Fire Color:LinearColor=&MakeLinearColorFromVectorAndFloat(
     "Vector (RGB)"=&Multiply_VectorByFloat(
      Vector=&MakeVectorFromLinearColorRGB(Color=<rgba 20 1.962 0 1>)
      Float=&FloatFromCurve(FloatCurve=DataInterface() {propertyValues {Curve {
          defaultValue=1
          keys [{interpMode=RCIM_Cubic, tangentMode=RCTM_User, time=0, value=1}, {interpMode=RCIM_Cubic, tangentMode=RCTM_User, time=1, value=0}]
         }}})
     )
     "Float (Alpha)"=1
    )
    Particles.Fire DynParam:Vector4f=&MakeVector4(
     X=&FloatFromCurve(FloatCurve=DataInterface() {propertyValues {Curve {keys [<at 0 1>]}}})
     Y=0.5
     Z=&FloatFromCurve(FloatCurve=DataInterface() {propertyValues {Curve {keys [<at 0 0.5 RCIM_Cubic>, {arriveTangent=1.5, interpMode=RCIM_Cubic, leaveTangent=1.5, time=0.5235, value=0.5}, <at 1 2>]}}})
     W=0
    )
    Particles.fire SpriteSize:Vector2f=&Multiply_Vector2D_ByFloat(
     Float=&FloatFromCurve(FloatCurve=DataInterface() {propertyValues {Curve {keys [
          {interpMode=RCIM_Cubic, tangentMode=RCTM_Break, time=0.0078, value=0}
          {interpMode=RCIM_Cubic, tangentMode=RCTM_Break, time=0.0578, value=2}
          {arriveTangent=-1.5342, interpMode=RCIM_Cubic, leaveTangent=-1.5342, time=0.1668, value=0.9624}
          {arriveTangent=-1.155, interpMode=RCIM_Cubic, leaveTangent=-1.155, time=0.7786, value=0.8942}
          <at 1 0>
         ]}}})
     Vector2D=&RandomRangeVector2D(Minimum=<12 12>, Maximum=<20 20>)
    )
   ) {isSetParameters=true}
   &SubUVAnimation(
    Output Alternate Renderer Binding=true
    "Sprite Renderer "=DataInterface() {propertyValues {SpriteRenderer=&self:Ashes_0.NiagaraSpriteRendererProperties_4}}
    UseEndFrame=false
    UseStartFrame=false
   ) {module=SubUVAnimation002}
  ]
  rendererClasses [&NiagaraSpriteRendererProperties]
  renderers [{rendererClass=&NiagaraSpriteRendererProperties}]
 }
 Dragons_despawn002 {
  emitterUpdateScript [
   &EmitterState(Life Cycle Mode=::NewEnumerator1, Loop Behavior=::NewEnumerator1, Loop Delay=3.68, Loop Duration=2.7, UseLoopDelay=true)
   &SpawnBurst_Instantaneous(Spawn Group=6, Use Spawn Probability=false, UseLoopCountLimit=false) {module=SpawnBurst_Instantaneous005}
   &SpawnBurst_Instantaneous(Spawn Count=50, Spawn Group=4, Use Spawn Probability=false, UseLoopCountLimit=false) {module=SpawnBurst_Instantaneous004}
   &SpawnBurst_Instantaneous(Spawn Count=35, Spawn Group=3, Use Spawn Probability=false, UseLoopCountLimit=false) {module=SpawnBurst_Instantaneous003}
   &SpawnBurst_Instantaneous(Spawn Count=20, Spawn Group=2, Use Spawn Probability=false) {module=SpawnBurst_Instantaneous002}
   &SpawnBurst_Instantaneous(Spawn Count=3, Spawn Group=5, Use Spawn Probability=false)
   &SpawnBurst_Instantaneous(Spawn Count=12, Use Spawn Probability=false) {module=SpawnBurst_Instantaneous001}
  ]
  particleSpawnScript [
   /SetVariables(Particles.VisibilityTag=&SpawnGroup()) {isSetParameters=true, module=SetVariables_#5}
   &PartitionParticles(Number Of Partitions=3)
   /SetVariables(Particles.InitSpriteRotRate=&UniformRangedFloat(Minimum=-1, Maximum=1)) {isSetParameters=true, module=SetVariables_#6}
   &InitializeParticle(
    Color=&MergeHelper_ColorSwitch(
     0=<rgba 0.5255 0.5255 0.5255 1>
     1=<rgba 0.6031 0.5599 0.463 0.2>
     2=<rgba 5 3.7043 1.51 1>
     3=?
     Index=&SpawnGroup()
     Num=::NewEnumerator1
    )
    Lifetime=&MergeHelper_FloatSwitch(
     0=&UniformRangedFloat(Minimum=0.3, Maximum=0.55)
     1=0.25
     2=&UniformRangedFloat(Minimum=0.5, Maximum=1.25)
     3=&UniformRangedFloat(Minimum=0.25, Maximum=0.75)
     4=&UniformRangedFloat(Minimum=0.3, Maximum=0.75)
     5=0.25
     6=0.4
     Index=&SpawnGroup()
     Num=::NewEnumerator5
    )
    Position Offset=<0 -25 185>
    Sprite Rotation Mode=::NewEnumerator2
    Sprite Size Mode=::NewEnumerator3
    Uniform Sprite Size=&MergeHelper_FloatSwitch(
     0=&UniformRangedFloat(Minimum=8, Maximum=20)
     1=65
     2=&RandomRangeFloat(Minimum=10, Maximum=25)
     3=?
     Index=&SpawnGroup()
     Num=::NewEnumerator1
    )
    UsePositionOffset=true
    Write Material Random=true
   )
   &ShapeLocation(
    MaskSpawnGroup=false
    Override Seed=false
    Sphere Radius=&MergeHelper_FloatSwitch(0=7, 1=0, 2=4, 3=5, 4=10, 5=0, Index=&SpawnGroup(), Num=::NewEnumerator4)
    Sphere Surface Distribution=1
    UseOwnerScale=true
   )
   &AddVelocity(
    Cone Axis=&MergeHelper_VectorSwitch(0=<-1 0 -1>, 1=<0.3 0 1>, 2=<1 0 1>, 3=?, Index=$Particles.PartitionParticles.Partition, Num=::NewEnumerator1)
    Module.UseConeFalloff=false
    Module.UseVelDistribution=false
    Override Seed=false
    Velocity Mode=::NewEnumerator2
    Velocity Speed=&FloatFromCurve(FloatCurve=DataInterface() {propertyValues {Curve {keys [<at 0 1>,<at 1 0>]}}}, Scale Curve=60)
   )
   /SetVariables(
    Particles.Smoke Velocity:Vector3f=&Multiply_Vector(
     A=$Particles.ShapeLocation.ShapeNormal
     B=&VectorFromFloat(Value=&RandomRangeFloat(Minimum=15, Maximum=50))
    )
    Particles.Sparkles Velocity:Vector3f=&UniformRangedVector(Minimum=<-500 -200 -200>, Maximum=<5000 200 200>)
    Particles.airglow CameraOffset=10
    Particles.smoke CameraOffset:NiagaraFloat=&RandomRangeFloat(Minimum=10, Maximum=25)
   ) {isSetParameters=true, module=SetVariables_#7}
   &AccelerationForce(Acceleration=&MergeHelper_VectorSwitch(
     0=&UniformRangedVector(Minimum=<0 0 -25>, Maximum=<0 0 -75>)
     1=<0 0 0>
     2=&RandomRangeVector(Minimum=<0 0 0>, Maximum=<0 0 0>)
     3=?
     Index=&SpawnGroup()
     Num=::NewEnumerator1
    ))
   &ApplyInitialForces()
   &ShapeLocation(
    Disc Coverage=0.3738
    MaskSpawnGroup=true
    Offset=<0 0 10>
    Offset Mode=::NewEnumerator0
    Override Seed=false
    Ring Radius=30
    Rotation Axis=<0 1 0>
    Rotation Mode=::NewEnumerator1
    Shape Primitive=::NewEnumerator4
    Spawn Group Mask=5
    UseOwnerScale=true
   ) {module=ShapeLocation001}
   &SkeletalMeshLocation(
    Bone / Socket Index=$Particles.PartitionParticles.Partition
    Bone Sampling Mode=::NewEnumerator4
    MaskSpawnGroup=false
    Position Sampling=::NewEnumerator1
    Sampled Position Offset=<0 5 0>
    Sampled Position Offset Coordinate Space:ENiagaraSkelMeshTransforms=::NewEnumerator2
    Skeletal Mesh=$System.tail
    Spawn Group Mask=2
   )
  ]
  particleUpdateScript [
   &ParticleState()
   &SpriteRotationRate(Rotation Rate=&MergeHelper_FloatSwitch(
     0=&Multiply_Float(
      A=$Particles.InitSpriteRotRate
      B=&FloatFromCurve(
       FloatCurve=DataInterface() {propertyValues {Curve {keys [
           {arriveTangent=-1.9214, interpMode=RCIM_Cubic, leaveTangent=-1.9214, tangentMode=RCTM_User, time=-0.0046, value=1.0113}
           {arriveTangent=-0.2496, interpMode=RCIM_Cubic, leaveTangent=-0.2496, tangentMode=RCTM_User, time=0.9896, value=0.16}
          ]}}}
       CurveIndex=$Emitter.Age
      )
     )
     1=0
     2=5
     3=?
     Index=&SpawnGroup()
     Num=::NewEnumerator1
    ))
   &ScaleSpriteSize(
    Scale Sprite Size Mode=::NewEnumerator0
    Uniform Scale Factor=&MergeHelper_FloatSwitch(
     0=&FloatFromCurve(
      FloatCurve=DataInterface() {propertyValues {Curve {keys [
          {arriveTangent=5.2354, interpMode=RCIM_Cubic, leaveTangent=5.2354, tangentMode=RCTM_User, time=-0.0046, value=0.25}
          <at 0.1439 1.4923>
          <at 1.0011 3.4809 RCIM_Cubic>
         ]}}}
      Scale Curve=&RandomRangeFloat(Minimum=0.5, Maximum=3)
     )
     1=&FloatFromCurve(FloatCurve=DataInterface() {propertyValues {Curve {
         defaultValue=1
         keys [{interpMode=RCIM_Cubic, tangentMode=RCTM_User, time=0, value=0}, {interpMode=RCIM_Cubic, tangentMode=RCTM_User, time=1, value=1}]
        }}})
     2=50
     3=0
     4=&FloatFromCurve(
      FloatCurve=DataInterface() {propertyValues {Curve {keys [<at 0.8 1 RCIM_Cubic>, {arriveTangent=-6.6126, interpMode=RCIM_Cubic, leaveTangent=-6.6126, tangentMode=RCTM_User, time=1, value=0}]}}}
      Scale Curve=0.1
     )
     Index=&SpawnGroup()
     Num=::NewEnumerator3
    )
   )
   &CurlNoiseForce(
    Bias Noise=false
    Coordinate Space:ENiagaraCoordinateSpace=::Local
    Noise Frequency=&MergeHelper_FloatSwitch(0=0, 1=0, 2=0.02, 3=0.368, 4=0, 5=0, Index=&SpawnGroup(), Num=::NewEnumerator4)
    Noise Quality / Cost=::NewEnumerator0
    Noise Strength=&MergeHelper_FloatSwitch(
     0=0
     1=0
     2=500
     3=2.6618
     4=&FloatFromCurve(FloatCurve=DataInterface() {propertyValues {Curve {keys [<at 0 1>,<at 1 0>]}}}, Scale Curve=3000)
     5=0
     Index=&SpawnGroup()
     Num=::NewEnumerator4
    )
    Pan Noise=false
    Random Seed=0
    Randomize Noise Sample=false
    Write to Intrinsic Parameters=false
   )
   &Drag(
    Drag=&MergeHelper_FloatSwitch(
     0=5
     1=0
     2=&FloatFromCurve(FloatCurve=DataInterface() {propertyValues {Curve {keys [<at 0 0.5>,<at 0.5 0.5>,<at 1 1>]}}}, Scale Curve=9)
     3=&FloatFromCurve(
      FloatCurve=DataInterface() {propertyValues {Curve {keys [<at 0 0.5>,<at 0.5 0.5>,<at 1 1>]}}}
      Scale Curve=&RandomRangeFloat(Minimum=0, Maximum=9)
     )
     4=&FloatFromCurve(FloatCurve=DataInterface() {propertyValues {Curve {keys [<at 0 0.5141>,<at 0.5078 0.7118>,<at 0.9984 0.4918>]}}}, Scale Curve=5)
     5=7
     Index=&SpawnGroup()
     Num=::NewEnumerator4
    )
    Use Linear Drag=true
    Use Rotational Drag=true
   )
   &SubUVAnimation(
    Output Alternate Renderer Binding=true
    Random Start Frame=true
    "Sprite Renderer "=DataInterface() {propertyValues {SpriteRenderer=&self:Dragons_despawn002_0.NiagaraSpriteRendererProperties_1}}
    SubUV Lookup Index Scale=0.2
    UseEndFrame=false
    UseStartFrame=false
   )
   &SubUVAnimation(
    Output Alternate Renderer Binding=true
    "Sprite Renderer "=DataInterface() {propertyValues {SpriteRenderer=&self:Dragons_despawn002_0.NiagaraSpriteRendererProperties_0}}
    UseEndFrame=false
    UseStartFrame=false
   ) {module=SubUVAnimation001}
   &SubUVAnimation(
    Output Alternate Renderer Binding=true
    "Sprite Renderer "=DataInterface() {propertyValues {SpriteRenderer=&self:Dragons_despawn002_0.NiagaraSpriteRendererProperties_4}}
    UseEndFrame=false
    UseStartFrame=false
   ) {module=SubUVAnimation002}
   /SetVariables(
    Particles.Smoke Color:LinearColor=&MakeLinearColorFromVectorAndFloat(
     "Vector (RGB)"=&MakeVectorFromLinearColorRGB(Color=<rgba 0.0656 0.0656 0.0656 1>)
     "Float (Alpha)"=&FloatFromCurve(FloatCurve=DataInterface() {propertyValues {Curve {
         defaultValue=1
         keys [
          {arriveTangent=9.0725, interpMode=RCIM_Cubic, leaveTangent=9.0725, tangentMode=RCTM_User, time=0, value=0}
          {arriveTangent=-2.5249, interpMode=RCIM_Cubic, leaveTangent=-2.5249, tangentMode=RCTM_User, time=0.0826, value=1.0001}
          {arriveTangent=-0.0405, interpMode=RCIM_Cubic, leaveTangent=-0.0405, tangentMode=RCTM_User, time=1, value=0}
         ]
        }}})
    )
    Particles.Smoke Dyn Param:Vector4f=&MakeVector4(X=0, Y=0, Z=0, W=0)
   ) {isSetParameters=true, module=SetVariables_#8}
   /SetVariables(
    Particles.soft glow Color:LinearColor=&MakeLinearColorFromVectorAndFloat(
     "Vector (RGB)"=&MakeVectorFromLinearColorRGB(Color=<rgba 1 0.6731 0.3597 1>)
     "Float (Alpha)"=&FloatFromCurve(FloatCurve=DataInterface() {propertyValues {Curve {keys [<at 0 1>,<at 1 0>]}}}, Scale Curve=0.5)
    )
    Particles.soft glow SpriteSize:Vector2f=&Vector2DFromFloat(Value=&Multiply_Float(A=&FloatFromCurve(FloatCurve=DataInterface() {propertyValues {Curve {keys [<at 0 1>,<at 1 0>]}}}), B=25))
   ) {isSetParameters=true, module=SetVariables_#9}
   /SetVariables(
    Particles.Embers Color:LinearColor=&MakeLinearColorFromVectorAndFloat(
     "Vector (RGB)"=&MakeVectorFromLinearColorRGB(Color=&ColorFromCurve(ColorCurve=DataInterface() {propertyValues {
         AlphaCurve {keys [<at 0 1 RCIM_Cubic>]}
         BlueCurve {keys [<at 0 7.8125 RCIM_Cubic>,<at 0.2499 0.7812>,<at 0.7114 0.2083 RCIM_Cubic>]}
         GreenCurve {keys [<at 0 34.1758 RCIM_Cubic>,<at 0.2499 4.8887>,<at 0.7114 0.8934 RCIM_Cubic>]}
         RedCurve {keys [<at 0 50 RCIM_Cubic>,<at 0.2499 10>,<at 0.7114 4 RCIM_Cubic>]}
        }}))
     "Float (Alpha)"=1
    )
    Particles.Embers SpriteSize:Vector2f=&Multiply_Vector2D_ByFloat(
     Float=&FloatFromCurve(
      FloatCurve=DataInterface() {propertyValues {Curve {
         defaultValue=1
         keys [{interpMode=RCIM_Cubic, tangentMode=RCTM_User, time=0, value=0}, {interpMode=RCIM_Cubic, tangentMode=RCTM_User, time=0.1511, value=1}]
        }}}
      Scale Curve=1.2
     )
     Vector2D=&RandomRangeVector2D(Minimum=<0.4 0.8>, Maximum=<0.5 1.5>)
    )
   ) {isSetParameters=true, module=SetVariables_#10}
   /SetVariables(
    Particles.Ashes Color:LinearColor=&MakeLinearColorFromVectorAndFloat(
     "Vector (RGB)"=&MakeVectorFromLinearColorRGB(Color=&ColorFromCurve(ColorCurve=DataInterface() {propertyValues {
         AlphaCurve {keys [<at 0 1 RCIM_Cubic>]}
         BlueCurve {keys [<at 0 1.473 RCIM_Cubic>,<at 0.6088 0.0879 RCIM_Cubic>]}
         GreenCurve {keys [<at 0 11.4982 RCIM_Cubic>,<at 0.6088 0.0919 RCIM_Cubic>]}
         RedCurve {keys [<at 0 50 RCIM_Cubic>,<at 0.6088 0.099 RCIM_Cubic>]}
        }}))
     "Float (Alpha)"=&FloatFromCurve(FloatCurve=DataInterface() {propertyValues {Curve {keys [<at 0.8 1 RCIM_Cubic>,<at 1 0>]}}})
    )
    Particles.Ashes SpriteSize:Vector2f=&Vector2DFromFloat(Value=&RandomRangeFloat(Minimum=0.5, Maximum=5))
    Particles.Vector 4:Vector4f=&MakeVector4(X=0, Y=0, Z=0, W=0)
   ) {isSetParameters=true, module=SetVariables_#11}
   /SetVariables(
    Particles.Fire Color:LinearColor=&MakeLinearColorFromVectorAndFloat(
     "Vector (RGB)"=&Multiply_VectorByFloat(
      Vector=&MakeVectorFromLinearColorRGB(Color=<rgba 20 1.962 0 1>)
      Float=&FloatFromCurve(FloatCurve=DataInterface() {propertyValues {Curve {
          defaultValue=1
          keys [{interpMode=RCIM_Cubic, tangentMode=RCTM_User, time=0, value=1}, {interpMode=RCIM_Cubic, tangentMode=RCTM_User, time=1, value=0}]
         }}})
     )
     "Float (Alpha)"=1
    )
    Particles.Fire DynParam:Vector4f=&MakeVector4(
     X=&FloatFromCurve(FloatCurve=DataInterface() {propertyValues {Curve {keys [<at 0 1>]}}})
     Y=0.5
     Z=&FloatFromCurve(FloatCurve=DataInterface() {propertyValues {Curve {keys [<at 0 0.5 RCIM_Cubic>, {arriveTangent=1.5, interpMode=RCIM_Cubic, leaveTangent=1.5, time=0.5235, value=0.5}, <at 1 2>]}}})
     W=0
    )
    Particles.fire SpriteSize:Vector2f=&Multiply_Vector2D_ByFloat(
     Float=&FloatFromCurve(FloatCurve=DataInterface() {propertyValues {Curve {keys [
          {interpMode=RCIM_Cubic, tangentMode=RCTM_Break, time=0.0079, value=0}
          {interpMode=RCIM_Cubic, tangentMode=RCTM_Break, time=0.0578, value=2}
          {arriveTangent=-1.5342, interpMode=RCIM_Cubic, leaveTangent=-1.5342, time=0.1668, value=0.9624}
          {arriveTangent=-1.155, interpMode=RCIM_Cubic, leaveTangent=-1.155, time=0.7786, value=0.8942}
          <at 1 0>
         ]}}})
     Vector2D=&RandomRangeVector2D(Minimum=<12 12>, Maximum=<20 20>)
    )
   ) {isSetParameters=true, module=SetVariables_#12}
   &GravityForce(Gravity=&MergeHelper_VectorSwitch(
     0=&MergeHelper_VectorSwitch(
      0=&RandomRangeVector(Minimum=<0 -10 0>, Maximum=<1 -30 -100>)
      1=&RandomRangeVector(Minimum=<0 0 -10>, Maximum=<-300 0 50>)
      2=&RandomRangeVector(Minimum=<0 0 10>, Maximum=<150 0 120>)
      3=?
      Index=$Particles.PartitionParticles.Partition
      Num=::NewEnumerator1
     )
     1=<0 0 0>
     2=&MakeVector(
      X=0
      Y=0
      Z=&MergeHelper_FloatSwitch(
       0=&FloatFromCurve(
        FloatCurve=DataInterface() {propertyValues {Curve {keys [<at 0 0.25 RCIM_Cubic>,<at 0.25 1>]}}}
        CurveIndex=$Emitter.Age
        Scale Curve=&RandomRangeFloat(Minimum=-60, Maximum=-350)
       )
       1=&FloatFromCurve(
        FloatCurve=DataInterface() {propertyValues {Curve {keys [<at 0 0.25 RCIM_Cubic>,<at 0.25 1>]}}}
        CurveIndex=$Emitter.Age
        Scale Curve=&RandomRangeFloat(Minimum=-5, Maximum=-500)
       )
       2=&FloatFromCurve(
        FloatCurve=DataInterface() {propertyValues {Curve {keys [<at 0 0.25 RCIM_Cubic>,<at 0.25 1>]}}}
        CurveIndex=$Emitter.Age
        Scale Curve=&RandomRangeFloat(Minimum=-60, Maximum=-350)
       )
       3=?
       Index=$Particles.PartitionParticles.Partition
       Num=::NewEnumerator1
      )
     )
     3=&MergeHelper_VectorSwitch(
      0=&RandomRangeVector(Minimum=<0 -10 0>, Maximum=<1 -30 -50>)
      1=&RandomRangeVector(Minimum=<0 0 -10>, Maximum=<-300 0 50>)
      2=&RandomRangeVector(Minimum=<0 0 10>, Maximum=<150 0 120>)
      3=?
      Index=$Particles.PartitionParticles.Partition
      Num=::NewEnumerator1
     )
     4=&MergeHelper_VectorSwitch(
      0=&RandomRangeVector(Minimum=<0 -10 0>, Maximum=<1 -30 -100>)
      1=&RandomRangeVector(Minimum=<0 0 -10>, Maximum=<-300 0 50>)
      2=&RandomRangeVector(Minimum=<0 0 10>, Maximum=<150 0 120>)
      3=?
      Index=$Particles.PartitionParticles.Partition
      Num=::NewEnumerator1
     )
     5=<0 0 0>
     Index=&SpawnGroup()
     Num=::NewEnumerator4
    ))
   &InheritVelocity(Speed Limit=true, UseSpeedThreshold=false) {enabled=false}
   &InheritSourceMovement(
    Applied Position Scale=&VectorFromFloat(Value=50)
    ApplyForce=false
    ApplyPosition=true
    ApplyPositionFalloff=true
    ApplyPositionFalloffCurve=false
   )
   &SolveForcesAndVelocity(Clamp Velocity=false, Limit Acceleration=false)
   &ScaleSpriteSizeBySpeed(
    Max Scale Factor=&MergeHelper_Vector2Switch(0=<1 1>, 1=<0 0>, 2=<0 0>, 3=<0 0>, 4=<1 5>, Index=&SpawnGroup(), Num=::NewEnumerator3)
    Min Scale Factor=&MergeHelper_Vector2Switch(0=<1 1>, 1=<0 0>, 2=<0 0>, 3=<0 0>, 4=<0 0>, Index=&SpawnGroup(), Num=::NewEnumerator3)
    Sample Scale Factor By Curve=true
    Velocity Threshold=500
   )
   &FN_Light_Attributes(
    Color=<rgba 15 4.9774 1.3498 1>
    Create Color Attribute=true
    Create Exponent Attribute=true
    Create Position Attribute=true
    Create Radius Attribute=true
    Enable Particle Light Enable=true
    Enable Visibility Tag=true
    Light Enabled=true
    Position=<0 0 180>
    Radius=&FloatFromCurve(
     FloatCurve=DataInterface() {propertyValues {Curve {
        defaultValue=1
        keys [{interpMode=RCIM_Cubic, tangentMode=RCTM_User, time=0, value=0}, {interpMode=RCIM_Cubic, tangentMode=RCTM_User, time=0.3, value=1}]
       }}}
     Scale Curve=35
    )
    Scale Color=&FloatFromCurve(FloatCurve=DataInterface() {propertyValues {Curve {keys [
         {interpMode=RCIM_Cubic, tangentMode=RCTM_Break, time=0, value=1}
         {interpMode=RCIM_Cubic, tangentMode=RCTM_Break, time=0.6908, value=1}
         {interpMode=RCIM_Cubic, tangentMode=RCTM_Break, time=1, value=0}
        ]}}})
    VisibilityTag=10
   )
   /SetVariables(Particles.Distortion SpriteSize:Vector2f=<150 150>) {isSetParameters=true}
   &SubUVAnimation(
    Output Alternate Renderer Binding=true
    "Sprite Renderer "=DataInterface() {propertyValues {SpriteRenderer=&self:Dragons_despawn002_0.NiagaraSpriteRendererProperties_6}}
    UseEndFrame=false
    UseStartFrame=false
   ) {module=SubUVAnimation003}
   /SetVariables(
    Particles.FIre2 DynParam:Vector4f=&MakeVector4(
     X=&FloatFromCurve(
      FloatCurve=DataInterface() {propertyValues {Curve {
         defaultValue=1
         keys [{interpMode=RCIM_Cubic, tangentMode=RCTM_User, time=0, value=0}, {interpMode=RCIM_Cubic, tangentMode=RCTM_User, time=1.0015, value=0.6706}]
        }}}
      Scale Curve=0.5
     )
     Y=&RandomRangeFloat(Minimum=0, Maximum=4)
     Z=0
     W=2.5
    )
    Particles.Fire 2 Color:LinearColor=&MakeLinearColorFromVectorAndFloat(
     "Vector (RGB)"=&VectorFromFloat(Value=&FloatFromCurve(
       FloatCurve=DataInterface() {propertyValues {Curve {keys [
           {arriveTangent=-0.4927, interpMode=RCIM_Cubic, leaveTangent=-0.4927, tangentMode=RCTM_User, time=0, value=0.1087}
           {arriveTangent=-0.5399, interpMode=RCIM_Cubic, leaveTangent=-0.5399, tangentMode=RCTM_User, time=0.1564, value=0.3915}
           {arriveTangent=-0.4927, interpMode=RCIM_Cubic, leaveTangent=-0.4927, tangentMode=RCTM_User, time=0.987, value=0.0549}
          ]}}}
       Scale Curve=&RandomRangeFloat(Minimum=2, Maximum=5)
      ))
     "Float (Alpha)"=&FloatFromCurve(
      FloatCurve=DataInterface() {propertyValues {Curve {keys [
          {arriveTangent=3.3783, interpMode=RCIM_Cubic, leaveTangent=3.3783, tangentMode=RCTM_User, time=0, value=0}
          {arriveTangent=0.2941, interpMode=RCIM_Cubic, leaveTangent=0.2941, tangentMode=RCTM_User, time=0.203, value=0.7175}
          {arriveTangent=-1.109, interpMode=RCIM_Cubic, leaveTangent=-1.109, time=0.5122, value=0.5742}
          {arriveTangent=-0.0646, interpMode=RCIM_Cubic, leaveTangent=-0.0646, tangentMode=RCTM_User, time=0.85, value=0}
         ]}}}
      Scale Curve=&Multiply_Float(A=3, B=$System.TimeOfDay)
     )
    )
    Particles.Fire 2 Sprite Size:Vector2f=&Vector2DFromFloat(Value=&FloatFromCurve(
      FloatCurve=DataInterface() {propertyValues {Curve {
         defaultValue=1
         keys [{interpMode=RCIM_Cubic, tangentMode=RCTM_User, time=0, value=0}, {interpMode=RCIM_Cubic, tangentMode=RCTM_User, time=1, value=1}]
        }}}
      Scale Curve=&RandomRangeFloat(Minimum=35, Maximum=50)
     ))
   ) {isSetParameters=true}
  ]
  rendererClasses [&NiagaraSpriteRendererProperties, &NiagaraLightRendererProperties]
  renderers [
   {rendererClass=&NiagaraSpriteRendererProperties}
   {rendererClass=&NiagaraSpriteRendererProperties, rendererIndex=1}
   {rendererClass=&NiagaraSpriteRendererProperties, rendererIndex=2}
   {rendererClass=&NiagaraSpriteRendererProperties, rendererIndex=3}
   {rendererClass=&NiagaraSpriteRendererProperties, rendererIndex=4}
   {rendererClass=&NiagaraLightRendererProperties, rendererIndex=5}
   {rendererClass=&NiagaraSpriteRendererProperties, rendererIndex=6}
  ]
 }
 Mesh_flame_wings_low {
  emitterSpawnScript [/SetVariables(
    Emitter.fade in out wings:NiagaraFloat=&FloatFromCurve(
     FloatCurve=DataInterface() {propertyValues {Curve {keys [<at 0 1>,<at 0.2104 1.3294>,<at 0.3639 0.5855>,<at 0.5584 0.7789>,<at 0.7872 -0.0054>,<at 0.9962 0.3689>]}}}
     CurveIndex=$Emitter.NormalizedLoopAge
     Scale Curve=2.5
    )
    Emitter.final opacity:NiagaraFloat=&FloatFromCurve(
     FloatCurve=DataInterface() {propertyValues {Curve {keys [<at 0 1>,<at 0.2104 1.3294>,<at 0.3639 0.5855>,<at 0.5584 0.7789>,<at 0.7872 -0.0054>,<at 0.9962 0.3689>]}}}
     CurveIndex=$Emitter.NormalizedLoopAge
    )
   ) {isSetParameters=true}]
  emitterUpdateScript [
   &EmitterState(
    Inactive Response=::NewEnumerator1
    Life Cycle Mode=::NewEnumerator1
    Loop Behavior=::NewEnumerator1
    Loop Delay=0.02
    Loop Duration=3.75
    UseLoopDelay=true
   )
   &SpawnBurst_Instantaneous(Spawn Count=12, Use Spawn Probability=false, UseLoopCountLimit=false)
  ]
  particleSpawnScript [
   &InitializeParticle(Lifetime=4.7883, Mesh Scale Mode=::NewEnumerator1, Mesh Uniform Scale Min=1.25, UsePositionOffset=false, Write Material Random=true) {module=InitializeParticle001}
   &FN_SetSpawnGroup(Number of Spawn Groups=6)
   &Color(Color=&MakeLinearColorFromVectorAndFloat(
     "Vector (RGB)"=<1 1 1>
     "Float (Alpha)"=&ClampFloat(Float=&RandomRangeFloat(Minimum=0.15, Maximum=0.3, Evaluation Type=::NewEnumerator1), Min=0.2)
    ))
   &InitialMeshOrientation(
    Mesh Axis To Orient=<0 1 0>
    Mesh Orientation Mode=::NewEnumerator2
    Orientation Vector=<0 0 1>
    Rotation=&MergeHelper_VectorSwitch(
     0=<0.1602 0.4566 0.0801>
     1=<0.1 0 0>
     2=<0.1602 0.4566 0.0801>
     3=<0.1 0 0>
     4=<0.1602 0.4566 0.0801>
     5=<0.1 0 0>
     Index=&SpawnGroup()
     Num=::NewEnumerator4
    )
    Rotation Coordinate Space=::NewEnumerator2
    Use Rotation Vector=true
   )
  ]
  particleUpdateScript [
   &SkeletalMeshLocation(
    Bone / Socket Index=&SpawnGroup()
    MaskSpawnGroup=false
    Mesh Sampling Type=::NewEnumerator5
    Sampled Position Coordinate Space=::Local
    Sampled Position Offset Coordinate Space:ENiagaraSkelMeshTransforms=::NewEnumerator2
    Sampled Scale Coordinate Space:ENiagaraCoordinateSpace=::Simulation
    Skeletal Mesh=$System.wing finger
    Socket Sampling Mode=::NewEnumerator5
    Use Sampled Mesh Attribute as Ribbon ID=false
   ) {module=SkeletalMeshLocation001}
   &ParticleState()
   &ScaleMeshSize(Scale Factor=&Multiply_Vector(A=<3 0.3 1.5>, B=&VectorFromFloat(Value=$User.Fade_out_geo)))
   &DynamicMaterialParameters(
    Index 0 Param 1=&RandomRangeFloat(Minimum=0, Maximum=2)
    Index 0 Param 2=3
    Index 0 Param 3=40
    Module.Index0Float0Write=true
    Module.Index0Float1Write=true
    Module.Index0Float2Write=true
    Module.Index0Float3Write=true
   )
   /SetVariables(Particles.filler wing:LinearColor=<rgba 5 2.4625 0.7914 1>) {isSetParameters=true}
   /SetVariables(Particles.new fire Scale:Vector3f=&Multiply_VectorByFloat(Vector=<1 1 1>, Float=&Multiply_Float(A=0.2, B=$User.Fade_out_geo))) {isSetParameters=true}
   &ScaleColor(
    Scale Alpha=&FloatFromCurve(
     FloatCurve=DataInterface() {propertyValues {Curve {keys [
         <at 0 0 RCIM_Cubic>
         <at 0.1 1>
         {interpMode=RCIM_Cubic, tangentMode=RCTM_Break, time=0.9, value=0.9806}
         {interpMode=RCIM_Cubic, tangentMode=RCTM_Break, time=1.0022, value=0}
        ]}}}
     Scale Curve=1.5
    )
    Scale RGB=&MakeVectorFromLinearColorRGB(Color=<rgba 1 1 1 1>)
    ScaleA=true
    ScaleRGB=true
   )
   /SetVariables(Particles.Wing pair Color 1:LinearColor=&MakeLinearColorFromVectorAndFloat(
     "Vector (RGB)"=&FN_FrontEndVector3Modifier(
      Gameplay=&MakeVectorFromLinearColorRGB(Color=<rgba 1 1 1 0.2>)
      Front End=&MakeVectorFromLinearColorRGB(Color=<rgba 1 1 1 0.5>)
     )
     "Float (Alpha)"=&FloatFromCurve(
      FloatCurve=DataInterface() {propertyValues {Curve {keys [
          <at 0 0 RCIM_Cubic>
          <at 0.4022 1>
          {interpMode=RCIM_Cubic, tangentMode=RCTM_Break, time=0.9, value=0.9806}
          {interpMode=RCIM_Cubic, tangentMode=RCTM_Break, time=1.0022, value=0}
         ]}}}
      Scale Curve=&FN_FrontEndFloatModifier(Gameplay=0.5, Front End=0.3)
     )
    )) {isSetParameters=true}
  ]
  rendererClasses [&NiagaraMeshRendererProperties]
  renderers [
   {rendererClass=&NiagaraMeshRendererProperties, rendererIndex=0}
   {rendererClass=&NiagaraMeshRendererProperties, rendererIndex=1}
   {rendererClass=&NiagaraMeshRendererProperties, rendererIndex=2}
  ]
 }
 Prop_spawn_merged_Blend_space {
  emitterUpdateScript [
   &EmitterState(
    Inactive Response=::NewEnumerator1
    Life Cycle Mode=::NewEnumerator1
    Loop Behavior=::NewEnumerator1
    Loop Delay=0.05
    Loop Duration=2
    UseLoopDelay=true
   )
   &SpawnBurst_Instantaneous(Spawn Count=5, Use Spawn Probability=false, UseLoopCountLimit=false) {module=SpawnBurst_Instantaneous006}
   &SpawnBurst_Instantaneous(Spawn Group=6, Use Spawn Probability=false, UseLoopCountLimit=false) {module=SpawnBurst_Instantaneous005}
   &SpawnBurst_Instantaneous(Spawn Count=50, Spawn Group=4, Use Spawn Probability=false, UseLoopCountLimit=false) {module=SpawnBurst_Instantaneous004}
   &SpawnBurst_Instantaneous(Spawn Count=75, Spawn Group=3, Use Spawn Probability=false, UseLoopCountLimit=false) {module=SpawnBurst_Instantaneous003}
   &SpawnBurst_Instantaneous(Spawn Count=20, Spawn Group=2, Use Spawn Probability=false) {module=SpawnBurst_Instantaneous002}
   &SpawnBurst_Instantaneous(Spawn Count=1, Spawn Group=7, Use Spawn Probability=false)
   &SpawnBurst_Instantaneous(Spawn Count=35, Use Spawn Probability=false) {module=SpawnBurst_Instantaneous001}
  ]
  particleSpawnScript [
   /SetVariables(Particles.VisibilityTag=&SpawnGroup()) {isSetParameters=true, module=SetVariables_#5}
   /SetVariables(Particles.InitSpriteRotRate=&UniformRangedFloat(Minimum=-1, Maximum=1)) {isSetParameters=true, module=SetVariables_#6}
   &InitializeParticle(
    Color=&MergeHelper_ColorSwitch(
     0=<rgba 0.5255 0.5255 0.5255 1>
     1=<rgba 0.6031 0.5599 0.463 0.2>
     2=<rgba 5 3.7043 1.51 1>
     3=?
     Index=&SpawnGroup()
     Num=::NewEnumerator1
    )
    Lifetime=&MergeHelper_FloatSwitch(
     0=&UniformRangedFloat(Minimum=0.3, Maximum=0.55)
     1=0.25
     2=&UniformRangedFloat(Minimum=0.5, Maximum=1.25)
     3=&UniformRangedFloat(Minimum=0.25, Maximum=0.65)
     4=&UniformRangedFloat(Minimum=0.3, Maximum=0.5)
     5=&RandomRangeFloat(Minimum=0.3, Maximum=0.6)
     6=0.6
     7=0.25
     Index=&SpawnGroup()
     Num=::NewEnumerator6
    )
    Position Offset=<5 15 120>
    Sprite Rotation Mode=::NewEnumerator2
    Sprite Size Mode=::NewEnumerator3
    Uniform Sprite Size=&MergeHelper_FloatSwitch(
     0=&UniformRangedFloat(Minimum=8, Maximum=20)
     1=65
     2=&RandomRangeFloat(Minimum=10, Maximum=25)
     3=?
     Index=&SpawnGroup()
     Num=::NewEnumerator1
    )
    UsePositionOffset=true
    Write Material Random=true
   )
   &ShapeLocation(
    MaskSpawnGroup=false
    Override Seed=false
    Sphere Radius=&MergeHelper_FloatSwitch(0=7, 1=0, 2=4, 3=5, 4=10, 5=0, 6=0, Index=&SpawnGroup(), Num=::NewEnumerator5)
    Sphere Surface Distribution=1
    UseOwnerScale=true
   )
   &AddVelocity(
    Rotation Mode=::NewEnumerator5
    Velocity=&MergeHelper_VectorSwitch(
     0=&Multiply_Vector(
      A=$Particles.ShapeLocation.ShapeNormal
      B=&VectorFromFloat(Value=&RandomRangeFloat(Minimum=15, Maximum=40))
     )
     1=<0 0 0>
     2=&UniformRangedVector(Minimum=<-200 0 -150>, Maximum=<200 0 150>)
     3=&FN_FrontEndVector3Modifier(
      Gameplay=&RandomRangeVector(Minimum=<-125 0 -50>, Maximum=<125 1 85>)
      Front End=&RandomRangeVector(Minimum=<-125 10 -50>, Maximum=<300 32 85>)
     )
     4=&RandomRangeVector(Minimum=<-180 -180 50>, Maximum=<180 180 150>)
     5=<0 0 0>
     6=<0 0 0>
     Index=&SpawnGroup()
     Num=::NewEnumerator5
    )
   )
   /SetVariables(
    Particles.Smoke Velocity:Vector3f=&Multiply_Vector(
     A=$Particles.ShapeLocation.ShapeNormal
     B=&VectorFromFloat(Value=&RandomRangeFloat(Minimum=15, Maximum=50))
    )
    Particles.Sparkles Velocity:Vector3f=&UniformRangedVector(Minimum=<-500 -200 -200>, Maximum=<5000 200 200>)
    Particles.airglow CameraOffset=10
    Particles.smoke CameraOffset:NiagaraFloat=&RandomRangeFloat(Minimum=10, Maximum=25)
   ) {isSetParameters=true, module=SetVariables_#7}
   &AccelerationForce(Acceleration=&MergeHelper_VectorSwitch(
     0=&UniformRangedVector(Minimum=<0 0 -25>, Maximum=<0 0 -25>)
     1=<0 0 0>
     2=&RandomRangeVector(Minimum=<0 0 0>, Maximum=<0 0 0>)
     3=&RandomRangeVector(Minimum=<0 50 0>, Maximum=<0 120 1>)
     4=<0 0 0>
     Index=&SpawnGroup()
     Num=::NewEnumerator3
    ))
   &ApplyInitialForces()
  ]
  particleUpdateScript [
   &ParticleState()
   &SpriteRotationRate(Rotation Rate=&MergeHelper_FloatSwitch(
     0=&Multiply_Float(
      A=$Particles.InitSpriteRotRate
      B=&FloatFromCurve(
       FloatCurve=DataInterface() {propertyValues {Curve {keys [
           {arriveTangent=-1.9214, interpMode=RCIM_Cubic, leaveTangent=-1.9214, tangentMode=RCTM_User, time=-0.0046, value=1.0113}
           {arriveTangent=-0.2496, interpMode=RCIM_Cubic, leaveTangent=-0.2496, tangentMode=RCTM_User, time=0.9896, value=0.16}
          ]}}}
       CurveIndex=$Emitter.Age
      )
     )
     1=0
     2=5
     3=?
     Index=&SpawnGroup()
     Num=::NewEnumerator1
    ))
   &ScaleSpriteSize(
    Scale Sprite Size Mode=::NewEnumerator0
    Uniform Scale Factor=&MergeHelper_FloatSwitch(
     0=&FloatFromCurve(
      FloatCurve=DataInterface() {propertyValues {Curve {keys [
          {arriveTangent=5.2354, interpMode=RCIM_Cubic, leaveTangent=5.2354, tangentMode=RCTM_User, time=-0.0046, value=0.25}
          <at 0.1439 1.4923>
          <at 1.0011 3.4809 RCIM_Cubic>
         ]}}}
      Scale Curve=&RandomRangeFloat(Minimum=0.5, Maximum=2)
     )
     1=&FloatFromCurve(
      FloatCurve=DataInterface() {propertyValues {Curve {
         defaultValue=1
         keys [{interpMode=RCIM_Cubic, tangentMode=RCTM_User, time=0, value=0}, {interpMode=RCIM_Cubic, tangentMode=RCTM_User, time=1, value=1}]
        }}}
      Scale Curve=5
     )
     2=50
     3=0
     4=&FloatFromCurve(
      FloatCurve=DataInterface() {propertyValues {Curve {keys [<at 0.8 1 RCIM_Cubic>, {arriveTangent=-6.6126, interpMode=RCIM_Cubic, leaveTangent=-6.6126, tangentMode=RCTM_User, time=1, value=0}]}}}
      Scale Curve=0.1
     )
     Index=&SpawnGroup()
     Num=::NewEnumerator3
    )
   )
   &CurlNoiseForce(
    Bias Noise=false
    Coordinate Space:ENiagaraCoordinateSpace=::Local
    Noise Frequency=&MergeHelper_FloatSwitch(0=0, 1=0, 2=0.02, 3=0, 4=0, 5=0, Index=&SpawnGroup(), Num=::NewEnumerator4)
    Noise Quality / Cost=::NewEnumerator0
    Noise Strength=&MergeHelper_FloatSwitch(
     0=0
     1=0
     2=500
     3=0
     4=&FloatFromCurve(FloatCurve=DataInterface() {propertyValues {Curve {keys [<at 0 1>,<at 1 0>]}}}, Scale Curve=400)
     5=0
     Index=&SpawnGroup()
     Num=::NewEnumerator4
    )
    Pan Noise=false
    Random Seed=0
    Randomize Noise Sample=false
    Write to Intrinsic Parameters=false
   )
   &Drag(
    Drag=&MergeHelper_FloatSwitch(
     0=5
     1=50
     2=&FloatFromCurve(FloatCurve=DataInterface() {propertyValues {Curve {keys [<at 0 0.5>,<at 0.5 0.5>,<at 1 1>]}}}, Scale Curve=9)
     3=&FloatFromCurve(FloatCurve=DataInterface() {propertyValues {Curve {keys [<at 0 0.5141>,<at 0.5078 0.7118>,<at 0.9984 0.4918>]}}}, Scale Curve=3)
     4=&FloatFromCurve(FloatCurve=DataInterface() {propertyValues {Curve {keys [<at 0 0.5141>,<at 0.5078 0.7118>,<at 0.9984 0.4918>]}}}, Scale Curve=10)
     Index=&SpawnGroup()
     Num=::NewEnumerator3
    )
    Use Linear Drag=true
    Use Rotational Drag=true
   )
   &SubUVAnimation(
    Output Alternate Renderer Binding=true
    Random Start Frame=true
    "Sprite Renderer "=DataInterface() {propertyValues {SpriteRenderer=&self:Prop_spawn_merged_Blend_space_0.NiagaraSpriteRendererProperties_1}}
    SubUV Lookup Index Scale=0.2
    UseEndFrame=false
    UseStartFrame=false
   )
   &SubUVAnimation(
    Output Alternate Renderer Binding=true
    "Sprite Renderer "=DataInterface() {propertyValues {SpriteRenderer=&self:Prop_spawn_merged_Blend_space_0.NiagaraSpriteRendererProperties_0}}
    UseEndFrame=false
    UseStartFrame=false
   ) {module=SubUVAnimation001}
   &SubUVAnimation(
    Output Alternate Renderer Binding=true
    "Sprite Renderer "=DataInterface() {propertyValues {SpriteRenderer=&self:Prop_spawn_merged_Blend_space_0.NiagaraSpriteRendererProperties_4}}
    UseEndFrame=false
    UseStartFrame=false
   ) {module=SubUVAnimation002}
   /SetVariables(
    Particles.Smoke Color:LinearColor=&MakeLinearColorFromVectorAndFloat(
     "Vector (RGB)"=&MakeVectorFromLinearColorRGB(Color=&ColorFromCurve(ColorCurve=DataInterface() {propertyValues {
         AlphaCurve {keys [<at 0 1 RCIM_Cubic>,<at 1 0 RCIM_Cubic>]}
         BlueCurve {keys [<at 0 0.0282 RCIM_Cubic>,<at 1 0 RCIM_Cubic>]}
         GreenCurve {keys [<at 0 0.071 RCIM_Cubic>,<at 1 0 RCIM_Cubic>]}
         RedCurve {keys [<at 0 0.2281 RCIM_Cubic>,<at 1 0 RCIM_Cubic>]}
        }}))
     "Float (Alpha)"=&FloatFromCurve(FloatCurve=DataInterface() {propertyValues {Curve {
         defaultValue=1
         keys [
          {arriveTangent=9.0725, interpMode=RCIM_Cubic, leaveTangent=9.0725, tangentMode=RCTM_User, time=0, value=0}
          {arriveTangent=-2.5249, interpMode=RCIM_Cubic, leaveTangent=-2.5249, tangentMode=RCTM_User, time=0.0826, value=1.0001}
          {arriveTangent=-0.0405, interpMode=RCIM_Cubic, leaveTangent=-0.0405, tangentMode=RCTM_User, time=1, value=0}
         ]
        }}})
    )
    Particles.Smoke Dyn Param:Vector4f=&MakeVector4(X=0, Y=0, Z=0, W=0)
   ) {isSetParameters=true, module=SetVariables_#8}
   /SetVariables(
    Particles.Soft Glow sprite size:Vector2f=&Vector2DFromFloat(Value=&FloatFromCurve(
      FloatCurve=DataInterface() {propertyValues {Curve {
         defaultValue=1
         keys [
          {interpMode=RCIM_Cubic, tangentMode=RCTM_User, time=0, value=0}
          {arriveTangent=0.24, interpMode=RCIM_Cubic, leaveTangent=0.24, tangentMode=RCTM_User, time=0.1, value=1}
         ]
        }}}
      Scale Curve=50
     ))
    Particles.soft glow Color:LinearColor=&MakeLinearColorFromVectorAndFloat(
     "Vector (RGB)"=&MakeVectorFromLinearColorRGB(Color=<rgba 5 0.4905 0 1>)
     "Float (Alpha)"=&FloatFromCurve(FloatCurve=DataInterface() {propertyValues {Curve {keys [<at 0 1>,<at 1 0>]}}}, Scale Curve=0.3)
    )
   ) {isSetParameters=true, module=SetVariables_#9}
   /SetVariables(
    Particles.Embers Color:LinearColor=&MakeLinearColorFromVectorAndFloat(
     "Vector (RGB)"=&MakeVectorFromLinearColorRGB(Color=&ColorFromCurve(ColorCurve=DataInterface() {propertyValues {
         AlphaCurve {keys [<at 0 1 RCIM_Cubic>]}
         BlueCurve {keys [<at 0 7.8125 RCIM_Cubic>,<at 0.2499 0.7812>,<at 0.7114 0.2083 RCIM_Cubic>]}
         GreenCurve {keys [<at 0 34.1758 RCIM_Cubic>,<at 0.2499 4.8887>,<at 0.7114 0.8934 RCIM_Cubic>]}
         RedCurve {keys [<at 0 50 RCIM_Cubic>,<at 0.2499 10>,<at 0.7114 4 RCIM_Cubic>]}
        }}))
     "Float (Alpha)"=1
    )
    Particles.Embers SpriteSize:Vector2f=&Multiply_Vector2D_ByFloat(
     Float=&FloatFromCurve(
      FloatCurve=DataInterface() {propertyValues {Curve {
         defaultValue=1
         keys [{interpMode=RCIM_Cubic, tangentMode=RCTM_User, time=0, value=0}, {interpMode=RCIM_Cubic, tangentMode=RCTM_User, time=0.1511, value=1}]
        }}}
      Scale Curve=1.2
     )
     Vector2D=&RandomRangeVector2D(Minimum=<0.4 0.8>, Maximum=<0.5 1.5>)
    )
   ) {isSetParameters=true, module=SetVariables_#10}
   /SetVariables(
    Particles.Ashes Color:LinearColor=&MakeLinearColorFromVectorAndFloat(
     "Vector (RGB)"=&MakeVectorFromLinearColorRGB(Color=&ColorFromCurve(ColorCurve=DataInterface() {propertyValues {
         AlphaCurve {keys [<at 0 1 RCIM_Cubic>]}
         BlueCurve {keys [<at 0 1.473 RCIM_Cubic>,<at 0.6088 0.0879 RCIM_Cubic>]}
         GreenCurve {keys [<at 0 11.4982 RCIM_Cubic>,<at 0.6088 0.0919 RCIM_Cubic>]}
         RedCurve {keys [<at 0 50 RCIM_Cubic>,<at 0.6088 0.099 RCIM_Cubic>]}
        }}))
     "Float (Alpha)"=&FloatFromCurve(FloatCurve=DataInterface() {propertyValues {Curve {keys [<at 0.8 1 RCIM_Cubic>,<at 1 0>]}}})
    )
    Particles.Ashes SpriteSize:Vector2f=&Vector2DFromFloat(Value=&RandomRangeFloat(Minimum=1, Maximum=7))
    Particles.Vector 4:Vector4f=&MakeVector4(X=0, Y=0, Z=0, W=0)
   ) {isSetParameters=true, module=SetVariables_#11}
   /SetVariables(
    Particles.Fire Color:LinearColor=&MakeLinearColorFromVectorAndFloat(
     "Vector (RGB)"=&FN_FrontEndVector3Modifier(
      Gameplay=&Multiply_VectorByFloat(
       Vector=&MakeVectorFromLinearColorRGB(Color=<rgba 5 1.186 0.5217 1>)
       Float=&FloatFromCurve(FloatCurve=DataInterface() {propertyValues {Curve {
           defaultValue=1
           keys [{interpMode=RCIM_Cubic, tangentMode=RCTM_User, time=0, value=1}, {interpMode=RCIM_Cubic, tangentMode=RCTM_User, time=1, value=0}]
          }}})
      )
      Front End=&Multiply_VectorByFloat(
       Vector=&MakeVectorFromLinearColorRGB(Color=<rgba 2.0312 0.5791 0 1>)
       Float=&FloatFromCurve(
        FloatCurve=DataInterface() {propertyValues {Curve {
           defaultValue=1
           keys [{interpMode=RCIM_Cubic, tangentMode=RCTM_User, time=0, value=1}, {interpMode=RCIM_Cubic, tangentMode=RCTM_User, time=1, value=0}]
          }}}
        Scale Curve=1.5
       )
      )
     )
     "Float (Alpha)"=&FloatFromCurve(
      FloatCurve=DataInterface() {propertyValues {Curve {
         defaultValue=1
         keys [{interpMode=RCIM_Cubic, tangentMode=RCTM_User, time=0, value=1}, {interpMode=RCIM_Cubic, tangentMode=RCTM_User, time=1, value=0}]
        }}}
      Scale Curve=&FNTimeOfDayScalar(Night Scale=3, Morning Scale=1.1, Day Scale=1.5, Evening Scale=1.3)
     )
    )
    Particles.Fire DynParam:Vector4f=&MakeVector4(
     X=&FloatFromCurve(FloatCurve=DataInterface() {propertyValues {Curve {keys [<at 0 1>]}}})
     Y=&RandomRangeFloat(Minimum=0, Maximum=4)
     Z=&FloatFromCurve(FloatCurve=DataInterface() {propertyValues {Curve {keys [<at 0 0.5 RCIM_Cubic>, {arriveTangent=1.5, interpMode=RCIM_Cubic, leaveTangent=1.5, time=0.5235, value=0.5}, <at 1 2>]}}})
     W=2.5
    )
    Particles.fire SpriteSize:Vector2f=&Multiply_Vector2D_ByFloat(
     Float=&FloatFromCurve(FloatCurve=DataInterface() {propertyValues {Curve {keys [
          {interpMode=RCIM_Cubic, tangentMode=RCTM_Break, time=0.0078, value=0}
          {interpMode=RCIM_Cubic, tangentMode=RCTM_Break, time=0.0578, value=1.5}
          {arriveTangent=-0.8405, interpMode=RCIM_Cubic, leaveTangent=-0.8405, time=0.1069, value=1}
          {arriveTangent=-1.1197, interpMode=RCIM_Cubic, leaveTangent=-1.1197, time=0.7786, value=0.8942}
          <at 1 0>
         ]}}})
     Vector2D=&RandomRangeVector2D(Minimum=<20 20>, Maximum=<45 45>)
    )
   ) {isSetParameters=true, module=SetVariables_#12}
   &GravityForce(Gravity=&MergeHelper_VectorSwitch(
     0=<0 0 0>
     1=<0 0 0>
     2=&MakeVector(
      X=0
      Y=0
      Z=&FloatFromCurve(
       FloatCurve=DataInterface() {propertyValues {Curve {keys [<at 0 0.25 RCIM_Cubic>,<at 0.25 1>]}}}
       CurveIndex=$Emitter.Age
       Scale Curve=&RandomRangeFloat(Minimum=-60, Maximum=-200)
      )
     )
     3=<0 0 5>
     4=&RandomRangeVector(Minimum=<0 0 -10>, Maximum=<1 1 50>)
     5=<0 0 0>
     6=<0 0 0>
     Index=&SpawnGroup()
     Num=::NewEnumerator5
    ))
   &SolveForcesAndVelocity(Clamp Velocity=false, Limit Acceleration=false)
   &ScaleSpriteSizeBySpeed(
    Max Scale Factor=&MergeHelper_Vector2Switch(0=<1 1>, 1=<0 0>, 2=<0 0>, 3=<2 2>, 4=<1 5>, Index=&SpawnGroup(), Num=::NewEnumerator3)
    Min Scale Factor=&MergeHelper_Vector2Switch(0=<1 1>, 1=<1 1>, 2=<1 1>, 3=<1 1>, 4=<0 0>, Index=&SpawnGroup(), Num=::NewEnumerator3)
    Sample Scale Factor By Curve=true
    Velocity Threshold=500
   )
   &FN_Light_Attributes(
    Color=<rgba 15 4.9774 1.3498 1>
    Create Color Attribute=true
    Create Exponent Attribute=true
    Create Position Attribute=true
    Create Radius Attribute=true
    Enable Particle Light Enable=true
    Enable Visibility Tag=true
    Light Enabled=true
    Position=<0 0 125>
    Radius=&FloatFromCurve(
     FloatCurve=DataInterface() {propertyValues {Curve {
        defaultValue=1
        keys [{interpMode=RCIM_Cubic, tangentMode=RCTM_User, time=0, value=0}, {interpMode=RCIM_Cubic, tangentMode=RCTM_User, time=0.3, value=1}]
       }}}
     Scale Curve=35
    )
    Scale Color=&FloatFromCurve(
     FloatCurve=DataInterface() {propertyValues {Curve {keys [
         {interpMode=RCIM_Cubic, tangentMode=RCTM_Break, time=0, value=1}
         {interpMode=RCIM_Cubic, tangentMode=RCTM_Break, time=0.6908, value=1}
         {interpMode=RCIM_Cubic, tangentMode=RCTM_Break, time=1, value=0}
        ]}}}
     Scale Curve=5
    )
    VisibilityTag=5
   )
   &SubUVAnimation(
    Output Alternate Renderer Binding=true
    "Sprite Renderer "=DataInterface() {propertyValues {SpriteRenderer=&self:Prop_spawn_merged_Blend_space_0.NiagaraSpriteRendererProperties_2}}
    UseEndFrame=false
    UseStartFrame=false
   ) {module=SubUVAnimation003}
  ]
  rendererClasses [&NiagaraSpriteRendererProperties, &NiagaraLightRendererProperties]
  renderers [
   {rendererClass=&NiagaraSpriteRendererProperties}
   {rendererClass=&NiagaraSpriteRendererProperties, rendererIndex=1}
   {rendererClass=&NiagaraSpriteRendererProperties, rendererIndex=2}
   {rendererClass=&NiagaraSpriteRendererProperties, rendererIndex=3}
   {rendererClass=&NiagaraSpriteRendererProperties, rendererIndex=4}
   {rendererClass=&NiagaraLightRendererProperties, rendererIndex=5}
  ]
 }
 dRAGON_LIGHTS {
  emitterUpdateScript [
   &EmitterState(Inactive Response=::NewEnumerator1, Life Cycle Mode=::NewEnumerator1, Loop Behavior=::NewEnumerator1, Loop Duration=3.7, UseLoopDelay=false)
   &SpawnBurst_Instantaneous(Spawn Group=2, Use Spawn Probability=false, UseLoopCountLimit=false) {module=SpawnBurst_Instantaneous002}
   &SpawnBurst_Instantaneous(Spawn Group=1, Spawn Time=0.2, Use Spawn Probability=false, UseLoopCountLimit=false) {module=SpawnBurst_Instantaneous001}
   &SpawnBurst_Instantaneous(Use Spawn Probability=false, UseLoopCountLimit=false)
  ]
  particleSpawnScript [
   /SetVariables(Particles.VisibilityTag=&SpawnGroup()) {isSetParameters=true}
   &PartitionParticles(Number Of Partitions=3)
   &InitializeParticle(
    Lifetime=&MergeHelper_FloatSwitch(0=0.5, 1=2, 2=0.2, 3=?, Index=$Particles.PartitionParticles.Partition, Num=::NewEnumerator1)
    UsePositionOffset=false
    Write Material Random=true
   )
  ]
  particleUpdateScript [
   &ParticleState()
   &SkeletalMeshLocation(
    Bone / Socket Index=$Particles.PartitionParticles.Partition
    Bone Sampling Mode=::NewEnumerator4
    MaskSpawnGroup=false
    Sampled Position Offset Coordinate Space:ENiagaraSkelMeshTransforms=::NewEnumerator2
    Skeletal Mesh=$System.tail
   )
   &InheritSourceMovement(Applied Position Scale=<-1 -1 1>, ApplyForce=false, ApplyPosition=true, ApplyPositionFalloff=false)
   &SolveForcesAndVelocity(Clamp Velocity=false, Limit Acceleration=false)
   &FN_Light_Attributes(
    Color=<rgba 3 1.1224 0.4219 1>
    Create Color Attribute=true
    Create Exponent Attribute=true
    Create Radius Attribute=true
    Enable Particle Light Enable=true
    Exponent=5
    Light Enabled=true
    Radius=25
    Scale Color=&Multiply_Float(
     A=0.5
     B=&Multiply_Float(
      A=&Waveform("Waveform [1]"=::NewEnumerator0, "[1] Frequency"=3, "[1] Amplitude Min/Max"=<0.7 1>, "[1] Scale"=2.6368)
      B=&FNTimeOfDayScalar(Night Scale=0.75, Morning Scale=0.6, Day Scale=1.2, Evening Scale=0.6)
     )
    )
   )
   &Color(Color=&MergeHelper_ColorSwitch(0=<rgba 1 0.023 0.8081 1>, 1=<rgba 1 0.8799 0 0.943>, 2=<rgba 1 0.2549 0 1>, 3=?, Index=$Particles.VisibilityTag, Num=::NewEnumerator1)) {enabled=false}
  ]
  rendererClasses [&NiagaraLightRendererProperties]
  renderers [
   {rendererClass=&NiagaraLightRendererProperties, rendererIndex=0}
   {rendererClass=&NiagaraLightRendererProperties, rendererIndex=1}
   {rendererClass=&NiagaraLightRendererProperties, rendererIndex=2}
  ]
 }
 smoke_and_fire_licks {
  emitterUpdateScript [
   &EmitterState(Life Cycle Mode=::NewEnumerator1, Loop Behavior=::NewEnumerator1, Loop Delay=0.4, Loop Duration=3.25, UseLoopDelay=true)
   &SpawnRate(SpawnRate=120, Use Spawn Probability=false)
   &SpawnPerUnit(Module.Use Max Movement Threshold=true, Module.Use Movement Tolerance=true, Spawn Spacing=1, Use Spawn Probability=false) {enabled=false}
   &SpawnRate(SpawnGroup=1, SpawnRate=100, Use Spawn Probability=false) {module=SpawnRate001}
  ]
  particleSpawnScript [
   /SetVariables(Particles.VisibilityTag=&SpawnGroup()) {isSetParameters=true}
   /SetVariables(Particles.SideIndex:NiagaraInt32=&ModuloInt(A=&ReturnParticleIDAttributes(), B=2)) {isSetParameters=true}
   &PartitionParticles(Number Of Partitions=6)
   &FN_SetSpawnGroup(Number of Spawn Groups=6)
   &InitializeParticle(
    Lifetime=&MergeHelper_FloatSwitch(
     0=&RandomRangeFloat(Minimum=0.1, Maximum=0.5)
     1=&UniformRangedFloat(Minimum=0.3, Maximum=0.8)
     2=?
     3=?
     Index=$Particles.VisibilityTag
     Num=::NewEnumerator0
    )
    Mass Mode=::NewEnumerator1
    Position Offset=<0 0 -115>
    Sprite Rotation Mode=::NewEnumerator2
    Sprite Size Mode=::NewEnumerator1
    Uniform Sprite Size Max=50
    Uniform Sprite Size Min=35
    UsePositionOffset=true
    Write Material Random=true
   ) {module=InitializeParticle001}
   &SkeletalMeshLocation(
    Bone / Socket Index=$Particles.PartitionParticles.Partition
    MaskSpawnGroup=false
    Mesh Sampling Type=::NewEnumerator5
    Sampled Position Coordinate Space=::Local
    Sampled Position Offset=<0 2 0>
    Sampled Scale Coordinate Space:ENiagaraCoordinateSpace=::Simulation
    Skeletal Mesh=$System.wing finger
    Socket Sampling Mode=::NewEnumerator5
   ) {module=SkeletalMeshLocation001}
   &ShapeLocation(
    Cylinder Height=5
    Cylinder Height Midpoint=0
    Cylinder Radius=5
    MaskSpawnGroup=false
    Override Seed=false
    Shape Origin=$Particles.SkeletalMeshLocation001.SampledPosition
    Shape Primitive=::NewEnumerator1
    Surface Only=false
    UseOwnerScale=true
   ) {enabled=false}
   /SetVariables(Particles.Initial fire licks sprite size:NiagaraFloat=&RandomRangeFloat(Minimum=25, Maximum=50)) {isSetParameters=true}
  ]
  particleUpdateScript [
   &ParticleState()
   &Drag(
    Drag=&MergeHelper_FloatSwitch(
     0=5
     1=&RandomRangeFloat(Minimum=2.5, Maximum=5)
     2=0
     3=0
     4=0
     Index=$Particles.VisibilityTag
     Num=::NewEnumerator3
    )
    Use Linear Drag=true
    Use Rotational Drag=false
   )
   &GravityForce(Gravity=&RandomRangeVector(Minimum=<0 0 -10>, Maximum=<1 1 -35>))
   &InheritSourceMovement(Applied Position Scale=<1 1 -1>, ApplyForce=false, ApplyPosition=true, ApplyPositionFalloff=false) {enabled=false}
   &AccelerationForce(Acceleration=&MergeHelper_VectorSwitch(
     0=&UniformRangedVector(Minimum=<0 0 150>, Maximum=<0 0 350>)
     1=<0 0 0>
     2=?
     3=?
     Index=&SpawnGroup()
     Num=::NewEnumerator0
    ))
   &InheritVelocity(Speed Limit=true, UseSpeedThreshold=false) {enabled=false}
   &SolveForcesAndVelocity(Clamp Velocity=false, Limit Acceleration=false)
   /SetVariables(
    Particles.Fire licks Color:LinearColor=&MakeLinearColorFromVectorAndFloat(
     "Vector (RGB)"=&VectorFromFloat(Value=&FloatFromCurve(
       FloatCurve=DataInterface() {propertyValues {Curve {keys [
           {arriveTangent=-0.4927, interpMode=RCIM_Cubic, leaveTangent=-0.4927, tangentMode=RCTM_User, time=0, value=0.1087}
           {arriveTangent=-0.5399, interpMode=RCIM_Cubic, leaveTangent=-0.5399, tangentMode=RCTM_User, time=0.1564, value=0.498}
           {arriveTangent=-0.4927, interpMode=RCIM_Cubic, leaveTangent=-0.4927, tangentMode=RCTM_User, time=0.987, value=0.2}
          ]}}}
       Scale Curve=&RandomRangeFloat(Minimum=2, Maximum=5)
      ))
     "Float (Alpha)"=&FloatFromCurve(
      FloatCurve=DataInterface() {propertyValues {Curve {keys [
          {arriveTangent=3.3783, interpMode=RCIM_Cubic, leaveTangent=3.3783, tangentMode=RCTM_User, time=0, value=0}
          {arriveTangent=0.2941, interpMode=RCIM_Cubic, leaveTangent=0.2941, tangentMode=RCTM_User, time=0.203, value=0.7175}
          {arriveTangent=-1.4438, interpMode=RCIM_Cubic, leaveTangent=-1.4438, time=0.5122, value=0.5742}
          {arriveTangent=-0.0646, interpMode=RCIM_Cubic, leaveTangent=-0.0646, tangentMode=RCTM_User, time=0.7, value=0}
         ]}}}
      Scale Curve=&Multiply_Float(A=1, B=$System.TimeOfDay)
     )
    )
    Particles.Fire licks dynParam:Vector4f=&MakeVector4(
     X=&FloatFromCurve(
      FloatCurve=DataInterface() {propertyValues {Curve {
         defaultValue=1
         keys [{interpMode=RCIM_Cubic, tangentMode=RCTM_User, time=0, value=0}, {interpMode=RCIM_Cubic, tangentMode=RCTM_User, time=1.0145, value=0.4181}]
        }}}
      Scale Curve=0.5
     )
     Y=&RandomRangeFloat(Minimum=0, Maximum=4)
     Z=0
     W=2.5
    )
    Particles.fire SpriteSize:Vector2f=&Multiply_Vector2D(
     A=&Vector2DFromFloat(Value=$Particles.Initial fire licks sprite size)
     B=&Multiply_Vector2D_ByFloat(
      Float=&FloatFromCurve(
       FloatCurve=DataInterface() {propertyValues {Curve {keys [
           {arriveTangent=1.4708, interpMode=RCIM_Cubic, leaveTangent=1.4708, tangentMode=RCTM_User, time=0, value=1.3}
           {arriveTangent=0.525, interpMode=RCIM_Cubic, leaveTangent=0.525, tangentMode=RCTM_User, time=1.0015, value=1.9209}
          ]}}}
       Scale Curve=&RandomRangeFloat(Minimum=0.05, Maximum=0.3)
      )
      Vector2D=<1 1>
     )
    )
   ) {isSetParameters=true}
   /SetVariables(
    Particles.Smoke Size:Vector2f=&Multiply_Vector2D_ByFloat(
     Float=&FloatFromCurve(
      FloatCurve=DataInterface() {propertyValues {Curve {keys [
          {arriveTangent=2.4634, interpMode=RCIM_Cubic, leaveTangent=2.4634, tangentMode=RCTM_User, time=0, value=0}
          {arriveTangent=0.2754, interpMode=RCIM_Cubic, leaveTangent=0.2754, tangentMode=RCTM_User, time=1, value=1}
         ]}}}
      Scale Curve=&Multiply_Float(A=&RandomRangeFloat(Minimum=3, Maximum=7), B=1)
     )
     Vector2D=&RandomRangeVector2D(Minimum=<5 5>, Maximum=<10 10>)
    )
    Particles.Smoke Velocity:Vector3f=&TransformVector(
     Vector=&MakeVector(X=0, Y=0, Z=&RandomRangeFloat(Minimum=85, Maximum=150))
     Source Space=::World
     Destination Space=::Local
    )
    Particles.smoke Dyn Param:Vector4f=&MakeVector4(X=5, Y=0, Z=0, W=0)
    Particles.smoke color:LinearColor=&MakeLinearColorFromVectorAndFloat(
     "Vector (RGB)"=&MakeVectorFromLinearColorRGB(Color=&ColorFromCurve(ColorCurve=DataInterface() {propertyValues {
         AlphaCurve {keys [<at 0 1 RCIM_Cubic>]}
         BlueCurve {keys [<at 0 0.125 RCIM_Cubic>,<at 1 0.0312 RCIM_Cubic>]}
         GreenCurve {keys [<at 0 0.125 RCIM_Cubic>,<at 1 0.0312 RCIM_Cubic>]}
         RedCurve {keys [<at 0 0.125 RCIM_Cubic>,<at 1 0.0312 RCIM_Cubic>]}
        }}))
     "Float (Alpha)"=&FloatFromCurve(
      FloatCurve=DataInterface() {propertyValues {Curve {
         defaultValue=0
         keys [
          {arriveTangent=0.6761, interpMode=RCIM_Cubic, leaveTangent=0.6761, tangentMode=RCTM_User, time=0, value=0}
          {arriveTangent=-0.0275, interpMode=RCIM_Cubic, leaveTangent=-0.0275, tangentMode=RCTM_User, time=0.07, value=0.85}
          {arriveTangent=-0.914, interpMode=RCIM_Cubic, leaveTangent=-0.914, time=0.69, value=0.592}
          {arriveTangent=-0.925, interpMode=RCIM_Cubic, leaveTangent=-0.925, tangentMode=RCTM_User, time=1, value=0}
         ]
        }}}
      Scale Curve=&Multiply_Float(
       A=1
       B=&FNTimeOfDayScalar(Night Scale=1.25, Morning Scale=1.1, Day Scale=1.15, Evening Scale=1.1)
      )
     )
    )
   ) {isSetParameters=true}
   &SubUVAnimation(
    Output Alternate Renderer Binding=true
    "Sprite Renderer "=DataInterface() {propertyValues {SpriteRenderer=&self:smoke_and_fire_licks_0.NiagaraSpriteRendererProperties_1}}
    SubUV Animation Mode:ENiagara_SubUVLookupModeV2=::NewEnumerator1
    SubUV Curve=DataInterface() {propertyValues {Curve {keys [
        {arriveTangent=4.1785, interpMode=RCIM_Cubic, leaveTangent=4.1785, tangentMode=RCTM_User, time=0, value=0}
        {arriveTangent=1.1449, interpMode=RCIM_Cubic, leaveTangent=1.1449, tangentMode=RCTM_User, time=0.2, value=0.6}
        {arriveTangent=0.2086, interpMode=RCIM_Cubic, leaveTangent=0.2086, tangentMode=RCTM_User, time=1, value=1}
       ]}}}
    UseEndFrame=true
    UseStartFrame=true
   ) {module=SubUVAnimation001}
  ]
  rendererClasses [&NiagaraSpriteRendererProperties]
  renderers [
   {rendererClass=&NiagaraSpriteRendererProperties}
   {rendererClass=&NiagaraSpriteRendererProperties, rendererIndex=1}
  ]
 }
 tails {
  emitterUpdateScript [
   &EmitterState(
    Inactive Response=::NewEnumerator1
    Life Cycle Mode=::NewEnumerator1
    Loop Behavior=::NewEnumerator1
    Loop Delay=0.02
    Loop Duration=3.7
    UseLoopDelay=true
   )
   &SpawnBurst_Instantaneous(Spawn Group=1, Use Spawn Probability=false, UseLoopCountLimit=false) {module=SpawnBurst_Instantaneous003}
   &SpawnBurst_Instantaneous(Spawn Group=2, Spawn Time=0.2, Use Spawn Probability=false, UseLoopCountLimit=false) {module=SpawnBurst_Instantaneous001}
   &SpawnBurst_Instantaneous(Spawn Group=3, Use Spawn Probability=false, UseLoopCountLimit=false) {module=SpawnBurst_Instantaneous002}
   &SpawnBurst_Instantaneous(Spawn Count=3, Spawn Time=0.02, Use Spawn Probability=false)
  ]
  particleSpawnScript [
   /SetVariables(Particles.VisibilityTag=&SpawnGroup()) {isSetParameters=true}
   &PartitionParticles(Number Of Partitions=3)
   &InitializeParticle(
    Lifetime=&MergeHelper_FloatSwitch(0=1, 1=0.5, 2=2, 3=0.3, Index=$Particles.PartitionParticles.Partition, Num=::NewEnumerator2)
    Position Mode=::NewEnumerator0
    Write Material Random=true
   ) {module=InitializeParticle001}
   &InitialMeshOrientation(Mesh Orientation Mode=::NewEnumerator2, Orientation Vector=<0 1 0>, Use Rotation Vector=false) {module=InitialMeshOrientation001}
  ]
  particleUpdateScript [
   &ParticleState(Kill Particles When Lifetime Has Elapsed=false, Loop Particles Lifetime=true)
   &SkeletalMeshLocation(
    Bone / Socket Index=$Particles.PartitionParticles.Partition
    Bone Sampling Mode=::NewEnumerator4
    MaskSpawnGroup=false
    Sampled Scale Coordinate Space:ENiagaraCoordinateSpace=::Simulation
    Skeletal Mesh=$System.tail
   ) {module=SkeletalMeshLocation001}
   &UpdateMeshOrientation()
   &ScaleColor(
    Scale Mode=::NewEnumerator1
    Scale RGBA=&MakeVector4_FromLinearColor(LinearColor=&ColorFromCurve(ColorCurve=DataInterface() {propertyValues {
        AlphaCurve {keys [<at 0 0.5>,<at 0.5 1>,<at 1 0.5>]}
        BlueCurve {keys [<at 0 1 RCIM_Cubic>,<at 0.5 2 RCIM_Cubic>,<at 1 1 RCIM_Cubic>]}
        GreenCurve {keys [<at 0 1 RCIM_Cubic>,<at 0.5 2 RCIM_Cubic>,<at 1 1 RCIM_Cubic>]}
        RedCurve {keys [<at 0 1 RCIM_Cubic>,<at 0.5 2 RCIM_Cubic>,<at 1 1 RCIM_Cubic>]}
       }}))
    ScaleRGBA=true
   )
   &DynamicMaterialParameters(
    Index 0 Param 1=&RandomRangeFloat(Minimum=0, Maximum=1)
    Index 0 Param 2=&RandomRangeFloat(Minimum=0, Maximum=1)
    Index 0 Param 3=40
    Module.Index0Float0Write=true
    Module.Index0Float1Write=true
    Module.Index0Float2Write=true
    Module.Index0Float3Write=true
   )
   &SubUVAnimation(
    Mesh Renderer=DataInterface() {propertyValues {MeshRenderer=&self:tails_0.NiagaraMeshRendererProperties_1}}
    Renderer Type:ENiagaraMeshOrSprite=::NewEnumerator1
    SubUV Animation Mode:ENiagara_SubUVLookupModeV2=::NewEnumerator2
    UseEndFrame=false
    UseStartFrame=false
   )
   /SetVariables(Particles.new fire Scale:Vector3f=&Multiply_VectorByFloat(Vector=<1 1 1>, Float=&Multiply_Float(A=0.3, B=$User.Fade_out_geo))) {isSetParameters=true}
   &FN_Light_Attributes(
    Color=<rgba 3 1.1224 0.4219 1>
    Create Color Attribute=true
    Create Exponent Attribute=true
    Create Radius Attribute=true
    Enable Particle Light Enable=true
    Exponent=5
    Light Enabled=true
    Radius=25
    Scale Color=&Multiply_Float(
     A=0.5
     B=&Multiply_Float(
      A=&Waveform("Waveform [1]"=::NewEnumerator0, "[1] Frequency"=3, "[1] Amplitude Min/Max"=<0.7 1>, "[1] Scale"=2.6368)
      B=&FNTimeOfDayScalar(Night Scale=0.75, Morning Scale=0.6, Day Scale=1.2, Evening Scale=0.6)
     )
    )
   )
  ]
  rendererClasses [&NiagaraMeshRendererProperties]
  renderers [{rendererClass=&NiagaraMeshRendererProperties, rendererIndex=0}]
 }
 trail {
  emitterUpdateScript [
   &EmitterState(Inactive Response=::NewEnumerator1, Life Cycle Mode=::NewEnumerator1, Loop Behavior=::NewEnumerator1, Loop Duration=3.78, UseLoopDelay=false)
   &SpawnRate(
    SpawnRate=&Multiply_Float(
     A=60
     B=&FloatFromCurve(
      FloatCurve=DataInterface() {propertyValues {Curve {keys [
          {arriveTangent=1.4364, interpMode=RCIM_Cubic, leaveTangent=1.4364, tangentMode=RCTM_User, time=0, value=0}
          {arriveTangentWeight=0.1817, interpMode=RCIM_Cubic, leaveTangentWeight=0.6939, tangentMode=RCTM_User, time=0.2, value=1}
          {arriveTangent=-0.2326, interpMode=RCIM_Cubic, leaveTangent=-0.2326, time=3.9448, value=0.8066}
          <at 4.5 0 RCIM_Cubic>
         ]}}}
      CurveIndex=$Engine.System.Age
     )
    )
    Use Spawn Probability=false
   ) {module=SpawnRate001}
   &SpawnRate(SpawnRate=60, Use Spawn Probability=false) {enabled=false}
  ]
  particleSpawnScript [
   &InitializeParticle(
    Color Mode=::NewEnumerator0
    Lifetime=0.85
    Ribbon Width=10
    Ribbon Width Mode=::NewEnumerator1
    UsePositionOffset=false
    Write Material Random=true
   )
   /SetVariables(Particles.Random=&RandomRangeFloat(Minimum=0.7, Maximum=1.1)) {isSetParameters=true}
   &FN_SetSpawnGroup(Number of Spawn Groups=3)
   &SkeletalMeshLocation(
    Bone / Socket Index=&SpawnGroup()
    Bone Sampling Mode=::NewEnumerator4
    MaskSpawnGroup=false
    Sampled Position Coordinate Space=::Local
    Sampled Position Offset=<0 3 5>
    Sampled Scale Coordinate Space:ENiagaraCoordinateSpace=::Simulation
    Skeletal Mesh=$System.tail
   )
  ]
  particleUpdateScript [
   &GravityForce(Gravity=&MakeVector(
     X=0
     Y=0
     Z=&FloatFromCurve(
      FloatCurve=DataInterface() {propertyValues {Curve {keys [<at 0.8 1 RCIM_Cubic>, {arriveTangent=-6041.373, interpMode=RCIM_Cubic, leaveTangent=-6041.373, tangentMode=RCTM_User, time=1, value=-500}]}}}
      CurveIndex=$Emitter.NormalizedLoopAge
     )
    ))
   &InheritSourceMovement(Applied Position Scale=<-1 -1 1>, ApplyForce=false, ApplyPosition=true, ApplyPositionFalloff=false) {enabled=false}
   &SolveForcesAndVelocity(Clamp Velocity=false, Limit Acceleration=false)
   &ParticleState()
   &ScaleRibbonWidth(Ribbon Width Scale=&FloatFromCurve(
     FloatCurve=DataInterface() {propertyValues {Curve {keys [
         <at 0 0 RCIM_Cubic>
         <at 0.0869 0.6365 RCIM_Cubic>
         {arriveTangent=-1.1383, interpMode=RCIM_Cubic, leaveTangent=-1.1383, tangentMode=RCTM_User, time=1, value=0}
        ]}}}
     Scale Curve=1.3
    ))
   &DynamicMaterialParameters(
    Index 0 Param 1=1.2375
    Index 0 Param 3=1
    Module.Index0Float0Write=true
    Module.Index0Float1Write=true
    Module.Index0Float2Write=true
    Module.Index0Float3Write=true
   )
   /SetVariables(
    Particles.Main fire DynParams:Vector4f=&MakeVector4(X=1.2375, Y=0, Z=1, W=0)
    Particles.main fire Color:LinearColor=&MakeLinearColorFromVectorAndFloat(
     "Vector (RGB)"=&FN_FrontEndVector3Modifier(
      Gameplay=&MakeVectorFromLinearColorRGB(Color=<rgba 0.3125 0.2601 0.1485 1>)
      Front End=&MakeVectorFromLinearColorRGB(Color=<rgba 2 0.7483 0.2812 1>)
     )
     "Float (Alpha)"=&Lerp_Float(
      A=0.2
      B=0.9
      Alpha=&FloatFromCurve(
       FloatCurve=DataInterface() {propertyValues {Curve {keys [
           {arriveTangent=-0.1961, interpMode=RCIM_Cubic, leaveTangent=-0.1961, tangentMode=RCTM_User, time=0, value=0}
           <at 0.0488 0.9842 RCIM_Cubic>
           {arriveTangent=-2.0216, interpMode=RCIM_Cubic, leaveTangent=-2.0216, tangentMode=RCTM_User, time=0.7, value=0}
          ]}}}
       Scale Curve=&FN_FrontEndFloatModifier(
        Gameplay=&FNTimeOfDayScalar(Night Scale=1, Morning Scale=0.5, Day Scale=0.75, Evening Scale=0.5)
        Front End=0.2
       )
      )
     )
    )
   ) {isSetParameters=true}
   /SetVariables(
    "Particles.Fake Gravity ":Vector3f=&MakeVector(X=0, Y=0, Z=-80)
    Particles.Smoke RibbonWidth:NiagaraFloat=&FloatFromCurve(
     FloatCurve=DataInterface() {propertyValues {Curve {
        defaultValue=1
        keys [
         <at 0.0042 0.4456 RCIM_Cubic>
         {arriveTangent=2.453, interpMode=RCIM_Cubic, leaveTangent=2.453, tangentMode=RCTM_User, time=0.1003, value=0.8631}
         <at 0.9968 1.5 RCIM_Cubic>
        ]
       }}}
     Scale Curve=19.2404
    )
    Particles.Smoke ribbon color:LinearColor=&MakeLinearColorFromVectorAndFloat(
     "Vector (RGB)"=&MakeVectorFromLinearColorRGB(Color=&ColorFromCurve(ColorCurve=DataInterface() {propertyValues {
         AlphaCurve {keys []}
         BlueCurve {keys [<at 0.0028 0.3652>,<at 0.0975 0.1882>,<at 1 0.1099>]}
         GreenCurve {keys [<at 0.0028 0.8302>,<at 0.0975 0.1844>,<at 1 0.1077>]}
         RedCurve {keys [<at 0.0028 3.5>,<at 0.0975 0.1979>,<at 1 0.1156>]}
        }}))
     "Float (Alpha)"=&Multiply_Float(
      A=&FloatFromCurve(
       FloatCurve=DataInterface() {propertyValues {Curve {
          defaultValue=1
          keys [
           {interpMode=RCIM_Cubic, tangentMode=RCTM_User, time=0.0221, value=1}
           {arriveTangent=-1.2069, interpMode=RCIM_Cubic, leaveTangent=-1.2069, tangentMode=RCTM_User, time=0.2163, value=0.6248}
           {interpMode=RCIM_Cubic, tangentMode=RCTM_User, time=1, value=0.0129}
          ]
         }}}
       Scale Curve=&FNTimeOfDayScalar(Night Scale=1.1, Morning Scale=1, Day Scale=1.2, Evening Scale=1)
      )
      B=&FloatFromCurve(
       FloatCurve=DataInterface() {propertyValues {Curve {keys [
           {arriveTangent=4.9615, interpMode=RCIM_Cubic, leaveTangent=4.9615, tangentMode=RCTM_User, time=0, value=0}
           {arriveTangentWeight=0.1681, interpMode=RCIM_Cubic, leaveTangentWeight=0.7352, tangentMode=RCTM_User, time=0.9069, value=1}
           {arriveTangentWeight=0.7925, interpMode=RCIM_Cubic, leaveTangentWeight=0.1009, tangentMode=RCTM_User, time=3.3799, value=1}
           <at 4 0.0396 RCIM_Cubic>
          ]}}}
       CurveIndex=$Emitter.Age
       Scale Curve=&FNTimeOfDayScalar(Night Scale=1.1, Morning Scale=1, Day Scale=1.2, Evening Scale=1)
      )
     )
    )
    Particles.Smoke trail Dynparam:Vector4f=&MakeVector4(X=-2, Y=0, Z=0, W=1)
   ) {isSetParameters=true}
  ]
  rendererClasses [&NiagaraRibbonRendererProperties]
  renderers [
   {rendererClass=&NiagaraRibbonRendererProperties}
   {rendererClass=&NiagaraRibbonRendererProperties, rendererIndex=1}
  ]
 }
}
summary {
 systemName=NS_OutlineLove_Dragon_Elements
 emitters [
  {emitterName=Ashes, simTarget=GPUComputeSim, rendererClasses [&NiagaraSpriteRendererProperties]}
  {
   emitterName=Prop_spawn_merged_Blend_space
   rendererClasses [&NiagaraSpriteRendererProperties, &NiagaraLightRendererProperties]
  }
  {emitterName=trail, rendererClasses [&NiagaraRibbonRendererProperties]}
  {emitterName=dRAGON_LIGHTS, rendererClasses [&NiagaraLightRendererProperties]}
  {emitterName=tails, rendererClasses [&NiagaraMeshRendererProperties]}
  {emitterName=smoke_and_fire_licks, rendererClasses [&NiagaraSpriteRendererProperties]}
  {
   emitterName=Dragons_despawn002
   rendererClasses [&NiagaraSpriteRendererProperties, &NiagaraLightRendererProperties]
  }
  {emitterName=Mesh_flame_wings_low, rendererClasses [&NiagaraMeshRendererProperties]}
 ]
 userVariables [
  User.Anim_Curve_FireSpread {type=NiagaraFloat, defaultValue=1, description=""}
  User.bIsFrontEnd {type=NiagaraBool}
  User.bIsFrontEndPreview {type=NiagaraBool}
  User.CurrentExhaust {type=Vector3f, description="", defaultValue {struct=Vector3f, value=<0 0 0>}}
  User.Fade_out_geo {type=NiagaraFloat, defaultValue=1, description=""}
  User.Fire_size {type=NiagaraFloat, defaultValue=1, description=""}
  User.Fire_trail_ribbon {type=NiagaraFloat, defaultValue=5, description=""}
  User.Intensity {type=NiagaraFloat, defaultValue=0, description=""}
  User.OwnerLocalVelocity {type=Vector3f, description="", defaultValue {struct=Vector3f, value=<0 0 0>}}
  User.Reveal {type=NiagaraFloat, defaultValue=1, description=""}
  User.RibbonIDOffset {type=NiagaraInt32}
  User.TrailWidth {type=NiagaraFloat, defaultValue=0.4, description=""}
 ]
}
