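# Niagara system for fire trails and a despawn burst on the dragons of the OutlineLove emote gadget.
# Fire001 emits sprite flames from skeletal mesh sockets, and Mesh_flames_for_perf pins mesh flames to dragon finger sockets.
# The delayed Dragons_despawn emitter packs several burst layers into one emitter, switching behavior by spawn group across multiple sprite renderers and a light.
# Several alternative ember, fire and mesh flame emitters are disabled.
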NiagaraSystem /BRCosmetics/Animation/Game/MainPlayer/Emotes/OutlineLove/FX/NS_OutlineLove_Fire_Trail
dependencies {
 usedDataInterfaces [
  &NiagaraDataInterfaceColorCurve
  &NiagaraDataInterfaceCurve
  &NiagaraDataInterfaceSpriteRendererInfo
  &NiagaraDataInterfaceArrayFloat
  &NiagaraDataInterfaceArrayFloat3
  &NiagaraDataInterfaceMeshRendererInfo
 ]
 usedDynamicInputs [
  &MakeFloatFromVector
  &SimulationPosition
  &FN_LOD_Distance_Remap
  &MaskFloatBySpawnGroup
  &MaskVector2dBySpawnGroup
 ]
 usedModules [
  &SystemState
  &self:SystemSpawnScript.SystemScriptSource.SystemScriptGraph.NiagaraNodeAssignment_0.SetVariables_#1
  &self:SystemSpawnScript.SystemScriptSource.SystemScriptGraph.NiagaraNodeAssignment_1.SetVariables_#2
  &self:SystemSpawnScript.SystemScriptSource.SystemScriptGraph.NiagaraNodeAssignment_2.SetVariables_#3
  &self:SystemSpawnScript.SystemScriptSource.SystemScriptGraph.NiagaraNodeAssignment_3.SetVariables_#4
 ]
}
emitters {
 Dragons_despawn {
  emitterUpdateScript [
   &EmitterState(Life Cycle Mode=::NewEnumerator1, Loop Behavior=::NewEnumerator1, Loop Delay=3.8, Loop Duration=2.7, UseLoopDelay=true)
   &SpawnBurst_Instantaneous(Spawn Count=5, Use Spawn Probability=false, UseLoopCountLimit=false) {module=SpawnBurst_Instantaneous006}
   &SpawnBurst_Instantaneous(Spawn Group=6, Use Spawn Probability=false, UseLoopCountLimit=false) {module=SpawnBurst_Instantaneous005}
   &SpawnBurst_Instantaneous(Spawn Count=50, Spawn Group=4, Use Spawn Probability=false, UseLoopCountLimit=false) {module=SpawnBurst_Instantaneous004}
   &SpawnBurst_Instantaneous(Spawn Count=35, Spawn Group=3, Use Spawn Probability=false, UseLoopCountLimit=false) {module=SpawnBurst_Instantaneous003}
   &SpawnBurst_Instantaneous(Spawn Count=20, Spawn Group=2, Use Spawn Probability=false) {module=SpawnBurst_Instantaneous002}
   &SpawnBurst_Instantaneous(Spawn Count=3, Spawn Group=5, Use Spawn Probability=false)
   &SpawnBurst_Instantaneous(Spawn Count=12, Use Spawn Probability=false) {module=SpawnBurst_Instantaneous001}
  ]
  particleSpawnScript [
   /SetVariables(Particles.VisibilityTag=&SpawnGroup()) {isSetParameters=true, module=SetVariables_#5}
   /SetVariables(Particles.InitSpriteRotRate=&UniformRangedFloat(Minimum=-1, Maximum=1)) {isSetParameters=true, module=SetVariables_#6}
   &InitializeParticle(
    Color=&MergeHelper_ColorSwitch(
     0=<rgba 0.5255 0.5255 0.5255 1>
     1=<rgba 0.6031 0.5599 0.463 0.2>
     2=<rgba 5 3.7043 1.51 1>
     3=?
     Index=&SpawnGroup()
     Num=::NewEnumerator1
    )
    Lifetime=&MergeHelper_FloatSwitch(
     0=&UniformRangedFloat(Minimum=0.3, Maximum=0.55)
     1=0.25
     2=&UniformRangedFloat(Minimum=0.5, Maximum=1.25)
     3=&UniformRangedFloat(Minimum=0.25, Maximum=0.75)
     4=&UniformRangedFloat(Minimum=0.3, Maximum=0.75)
     5=0.25
     6=0.4
     Index=&SpawnGroup()
     Num=::NewEnumerator5
    )
    Position Offset=<0 -25 185>
    Sprite Rotation Mode=::NewEnumerator2
    Sprite Size Mode=::NewEnumerator3
    Uniform Sprite Size=&MergeHelper_FloatSwitch(
     0=&UniformRangedFloat(Minimum=8, Maximum=20)
     1=65
     2=&RandomRangeFloat(Minimum=10, Maximum=25)
     3=?
     Index=&SpawnGroup()
     Num=::NewEnumerator1
    )
    UsePositionOffset=true
    Write Material Random=true
   )
   &ShapeLocation(
    MaskSpawnGroup=false
    Override Seed=false
    Sphere Radius=&MergeHelper_FloatSwitch(0=7, 1=0, 2=4, 3=5, 4=10, 5=0, Index=&SpawnGroup(), Num=::NewEnumerator4)
    Sphere Surface Distribution=1
    UseOwnerScale=true
   )
   &AddVelocity(Velocity=&MergeHelper_VectorSwitch(
     0=&Multiply_Vector(
      A=$Particles.ShapeLocation.ShapeNormal
      B=&VectorFromFloat(Value=&RandomRangeFloat(Minimum=15, Maximum=40))
     )
     1=<0 0 0>
     2=&UniformRangedVector(Minimum=<-200 0 -110>, Maximum=<200 0 110>)
     3=&RandomRangeVector(Minimum=<-50 -50 -50>, Maximum=<50 50 50>)
     4=&RandomRangeVector(Minimum=<-90 -90 -90>, Maximum=<90 90 90>)
     5=<0 0 0>
     6=<0 0 0>
     Index=&SpawnGroup()
     Num=::NewEnumerator5
    ))
   /SetVariables(
    Particles.Smoke Velocity:Vector3f=&Multiply_Vector(
     A=$Particles.ShapeLocation.ShapeNormal
     B=&VectorFromFloat(Value=&RandomRangeFloat(Minimum=15, Maximum=50))
    )
    Particles.Sparkles Velocity:Vector3f=&UniformRangedVector(Minimum=<-500 -200 -200>, Maximum=<5000 200 200>)
    Particles.airglow CameraOffset=10
    Particles.smoke CameraOffset:NiagaraFloat=&RandomRangeFloat(Minimum=10, Maximum=25)
   ) {isSetParameters=true, module=SetVariables_#7}
   &AccelerationForce(Acceleration=&MergeHelper_VectorSwitch(
     0=&UniformRangedVector(Minimum=<0 0 -25>, Maximum=<0 0 -75>)
     1=<0 0 0>
     2=&RandomRangeVector(Minimum=<0 0 0>, Maximum=<0 0 0>)
     3=?
     Index=&SpawnGroup()
     Num=::NewEnumerator1
    ))
   &ApplyInitialForces()
   &ShapeLocation(
    Disc Coverage=0.3738
    MaskSpawnGroup=true
    Offset=<0 0 10>
    Offset Mode=::NewEnumerator0
    Override Seed=false
    Ring Radius=30
    Rotation Axis=<0 1 0>
    Rotation Mode=::NewEnumerator1
    Shape Primitive=::NewEnumerator4
    Spawn Group Mask=5
    UseOwnerScale=true
   ) {module=ShapeLocation001}
   &PartitionParticles(Number Of Partitions=3)
   &SkeletalMeshLocation(
    Bone / Socket Index=$Particles.PartitionParticles.Partition
    Bone Sampling Mode=::NewEnumerator4
    MaskSpawnGroup=false
    Sampled Position Offset=<0 5 0>
    Sampled Position Offset Coordinate Space:ENiagaraSkelMeshTransforms=::NewEnumerator2
    Skeletal Mesh=$System.Skeletal Mesh Reader
   )
  ]
  particleUpdateScript [
   &ParticleState()
   &SpriteRotationRate(Rotation Rate=&MergeHelper_FloatSwitch(
     0=&Multiply_Float(
      A=$Particles.InitSpriteRotRate
      B=&FloatFromCurve(
       FloatCurve=DataInterface() {propertyValues {Curve {keys [
           {arriveTangent=-1.9214, interpMode=RCIM_Cubic, leaveTangent=-1.9214, tangentMode=RCTM_User, time=-0.0046, value=1.0113}
           {arriveTangent=-0.2496, interpMode=RCIM_Cubic, leaveTangent=-0.2496, tangentMode=RCTM_User, time=0.9896, value=0.16}
          ]}}}
       CurveIndex=$Emitter.Age
      )
     )
     1=0
     2=5
     3=?
     Index=&SpawnGroup()
     Num=::NewEnumerator1
    ))
   &ScaleSpriteSize(
    Scale Sprite Size Mode=::NewEnumerator0
    Uniform Scale Factor=&MergeHelper_FloatSwitch(
     0=&FloatFromCurve(
      FloatCurve=DataInterface() {propertyValues {Curve {keys [
          {arriveTangent=5.2354, interpMode=RCIM_Cubic, leaveTangent=5.2354, tangentMode=RCTM_User, time=-0.0046, value=0.25}
          <at 0.1439 1.4923>
          <at 1.0011 3.4809 RCIM_Cubic>
         ]}}}
      Scale Curve=&RandomRangeFloat(Minimum=0.5, Maximum=3)
     )
     1=&FloatFromCurve(FloatCurve=DataInterface() {propertyValues {Curve {
         defaultValue=1
         keys [{interpMode=RCIM_Cubic, tangentMode=RCTM_User, time=0, value=0}, {interpMode=RCIM_Cubic, tangentMode=RCTM_User, time=1, value=1}]
        }}})
     2=50
     3=0
     4=&FloatFromCurve(
      FloatCurve=DataInterface() {propertyValues {Curve {keys [<at 0.8 1 RCIM_Cubic>, {arriveTangent=-6.6126, interpMode=RCIM_Cubic, leaveTangent=-6.6126, tangentMode=RCTM_User, time=1, value=0}]}}}
      Scale Curve=0.1
     )
     Index=&SpawnGroup()
     Num=::NewEnumerator3
    )
   )
   &CurlNoiseForce(
    Bias Noise=false
    Coordinate Space:ENiagaraCoordinateSpace=::Local
    Noise Frequency=&MergeHelper_FloatSwitch(0=0, 1=0, 2=0.02, 3=0, 4=0, 5=0, Index=&SpawnGroup(), Num=::NewEnumerator4)
    Noise Quality / Cost=::NewEnumerator0
    Noise Strength=&MergeHelper_FloatSwitch(
     0=0
     1=0
     2=500
     3=0
     4=&FloatFromCurve(FloatCurve=DataInterface() {propertyValues {Curve {keys [<at 0 1>,<at 1 0>]}}}, Scale Curve=3000)
     5=0
     Index=&SpawnGroup()
     Num=::NewEnumerator4
    )
    Pan Noise=false
    Random Seed=0
    Randomize Noise Sample=false
    Write to Intrinsic Parameters=false
   )
   &Drag(
    Drag=&MergeHelper_FloatSwitch(
     0=5
     1=0
     2=&FloatFromCurve(FloatCurve=DataInterface() {propertyValues {Curve {keys [<at 0 0.5>,<at 0.5 0.5>,<at 1 1>]}}}, Scale Curve=9)
     3=&FloatFromCurve(FloatCurve=DataInterface() {propertyValues {Curve {keys [<at 0 0.5>,<at 0.5 0.5>,<at 1 1>]}}}, Scale Curve=9)
     4=&FloatFromCurve(FloatCurve=DataInterface() {propertyValues {Curve {keys [<at 0 0.5141>,<at 0.5078 0.7118>,<at 0.9984 0.4918>]}}}, Scale Curve=5)
     5=7
     Index=&SpawnGroup()
     Num=::NewEnumerator4
    )
    Use Linear Drag=true
    Use Rotational Drag=true
   )
   &SubUVAnimation(
    Output Alternate Renderer Binding=true
    Random Start Frame=true
    "Sprite Renderer "=DataInterface() {propertyValues {SpriteRenderer=&self:Dragons_despawn_0.NiagaraSpriteRendererProperties_1}}
    SubUV Lookup Index Scale=0.2
    UseEndFrame=false
    UseStartFrame=false
   )
   &SubUVAnimation(
    Output Alternate Renderer Binding=true
    "Sprite Renderer "=DataInterface() {propertyValues {SpriteRenderer=&self:Dragons_despawn_0.NiagaraSpriteRendererProperties_0}}
    UseEndFrame=false
    UseStartFrame=false
   ) {module=SubUVAnimation001}
   &SubUVAnimation(
    Output Alternate Renderer Binding=true
    "Sprite Renderer "=DataInterface() {propertyValues {SpriteRenderer=&self:Dragons_despawn_0.NiagaraSpriteRendererProperties_4}}
    UseEndFrame=false
    UseStartFrame=false
   ) {module=SubUVAnimation002}
   /SetVariables(
    Particles.Smoke Color:LinearColor=&MakeLinearColorFromVectorAndFloat(
     "Vector (RGB)"=&MakeVectorFromLinearColorRGB(Color=<rgba 0.0656 0.0656 0.0656 1>)
     "Float (Alpha)"=&FloatFromCurve(FloatCurve=DataInterface() {propertyValues {Curve {
         defaultValue=1
         keys [
          {arriveTangent=9.0725, interpMode=RCIM_Cubic, leaveTangent=9.0725, tangentMode=RCTM_User, time=0, value=0}
          {arriveTangent=-2.5249, interpMode=RCIM_Cubic, leaveTangent=-2.5249, tangentMode=RCTM_User, time=0.0826, value=1.0001}
          {arriveTangent=-0.0405, interpMode=RCIM_Cubic, leaveTangent=-0.0405, tangentMode=RCTM_User, time=1, value=0}
         ]
        }}})
    )
    Particles.Smoke Dyn Param:Vector4f=&MakeVector4(X=0, Y=0, Z=0, W=0)
   ) {isSetParameters=true, module=SetVariables_#8}
   /SetVariables(
    Particles.soft glow Color:LinearColor=&MakeLinearColorFromVectorAndFloat(
     "Vector (RGB)"=&MakeVectorFromLinearColorRGB(Color=<rgba 1 0.6731 0.3597 1>)
     "Float (Alpha)"=&FloatFromCurve(FloatCurve=DataInterface() {propertyValues {Curve {keys [<at 0 1>,<at 1 0>]}}}, Scale Curve=0.5)
    )
    Particles.soft glow SpriteSize:Vector2f=&Vector2DFromFloat(Value=&Multiply_Float(A=&FloatFromCurve(FloatCurve=DataInterface() {propertyValues {Curve {keys [<at 0 1>,<at 1 0>]}}}), B=25))
   ) {isSetParameters=true, module=SetVariables_#9}
   /SetVariables(
    Particles.Embers Color:LinearColor=&MakeLinearColorFromVectorAndFloat(
     "Vector (RGB)"=&MakeVectorFromLinearColorRGB(Color=&ColorFromCurve(ColorCurve=DataInterface() {propertyValues {
         AlphaCurve {keys [<at 0 1 RCIM_Cubic>]}
         BlueCurve {keys [<at 0 7.8125 RCIM_Cubic>,<at 0.2499 0.7812>,<at 0.7114 0.2083 RCIM_Cubic>]}
         GreenCurve {keys [<at 0 34.1758 RCIM_Cubic>,<at 0.2499 4.8887>,<at 0.7114 0.8934 RCIM_Cubic>]}
         RedCurve {keys [<at 0 50 RCIM_Cubic>,<at 0.2499 10>,<at 0.7114 4 RCIM_Cubic>]}
        }}))
     "Float (Alpha)"=1
    )
    Particles.Embers SpriteSize:Vector2f=&Multiply_Vector2D_ByFloat(
     Float=&FloatFromCurve(
      FloatCurve=DataInterface() {propertyValues {Curve {
         defaultValue=1
         keys [{interpMode=RCIM_Cubic, tangentMode=RCTM_User, time=0, value=0}, {interpMode=RCIM_Cubic, tangentMode=RCTM_User, time=0.1511, value=1}]
        }}}
      Scale Curve=1.2
     )
     Vector2D=&RandomRangeVector2D(Minimum=<0.4 0.8>, Maximum=<0.5 1.5>)
    )
   ) {isSetParameters=true, module=SetVariables_#10}
   /SetVariables(
    Particles.Ashes Color:LinearColor=&MakeLinearColorFromVectorAndFloat(
     "Vector (RGB)"=&MakeVectorFromLinearColorRGB(Color=&ColorFromCurve(ColorCurve=DataInterface() {propertyValues {
         AlphaCurve {keys [<at 0 1 RCIM_Cubic>]}
         BlueCurve {keys [<at 0 1.473 RCIM_Cubic>,<at 0.6088 0.0879 RCIM_Cubic>]}
         GreenCurve {keys [<at 0 11.4982 RCIM_Cubic>,<at 0.6088 0.0919 RCIM_Cubic>]}
         RedCurve {keys [<at 0 50 RCIM_Cubic>,<at 0.6088 0.099 RCIM_Cubic>]}
        }}))
     "Float (Alpha)"=&FloatFromCurve(FloatCurve=DataInterface() {propertyValues {Curve {keys [<at 0.8 1 RCIM_Cubic>,<at 1 0>]}}})
    )
    Particles.Ashes SpriteSize:Vector2f=&Vector2DFromFloat(Value=&RandomRangeFloat(Minimum=0.5, Maximum=5))
    Particles.Vector 4:Vector4f=&MakeVector4(X=0, Y=0, Z=0, W=0)
   ) {isSetParameters=true, module=SetVariables_#11}
   /SetVariables(
    Particles.Fire Color:LinearColor=&MakeLinearColorFromVectorAndFloat(
     "Vector (RGB)"=&Multiply_VectorByFloat(
      Vector=&MakeVectorFromLinearColorRGB(Color=<rgba 20 1.962 0 1>)
      Float=&FloatFromCurve(FloatCurve=DataInterface() {propertyValues {Curve {
          defaultValue=1
          keys [{interpMode=RCIM_Cubic, tangentMode=RCTM_User, time=0, value=1}, {interpMode=RCIM_Cubic, tangentMode=RCTM_User, time=1, value=0}]
         }}})
     )
     "Float (Alpha)"=1
    )
    Particles.Fire DynParam:Vector4f=&MakeVector4(
     X=&FloatFromCurve(FloatCurve=DataInterface() {propertyValues {Curve {keys [<at 0 1>]}}})
     Y=0.5
     Z=&FloatFromCurve(FloatCurve=DataInterface() {propertyValues {Curve {keys [<at 0 0.5 RCIM_Cubic>, {arriveTangent=1.5, interpMode=RCIM_Cubic, leaveTangent=1.5, time=0.5235, value=0.5}, <at 1 2>]}}})
     W=0
    )
    Particles.fire SpriteSize:Vector2f=&Multiply_Vector2D_ByFloat(
     Float=&FloatFromCurve(FloatCurve=DataInterface() {propertyValues {Curve {keys [
          {interpMode=RCIM_Cubic, tangentMode=RCTM_Break, time=0.0079, value=0}
          {interpMode=RCIM_Cubic, tangentMode=RCTM_Break, time=0.0578, value=2}
          {arriveTangent=-1.5342, interpMode=RCIM_Cubic, leaveTangent=-1.5342, time=0.1668, value=0.9624}
          {arriveTangent=-1.155, interpMode=RCIM_Cubic, leaveTangent=-1.155, time=0.7786, value=0.8942}
          <at 1 0>
         ]}}})
     Vector2D=&RandomRangeVector2D(Minimum=<8 8>, Maximum=<15 15>)
    )
   ) {isSetParameters=true, module=SetVariables_#12}
   &GravityForce(Gravity=&MergeHelper_VectorSwitch(
     0=<0 0 0>
     1=<0 0 0>
     2=&MakeVector(
      X=0
      Y=0
      Z=&FloatFromCurve(
       FloatCurve=DataInterface() {propertyValues {Curve {keys [<at 0 0.25 RCIM_Cubic>,<at 0.25 1>]}}}
       CurveIndex=$Emitter.Age
       Scale Curve=&RandomRangeFloat(Minimum=-60, Maximum=-170)
      )
     )
     3=<0 0 5>
     4=&RandomRangeVector(Minimum=<0 0 -10>, Maximum=<1 1 50>)
     5=<0 0 0>
     Index=&SpawnGroup()
     Num=::NewEnumerator4
    ))
   &SolveForcesAndVelocity(Clamp Velocity=false, Limit Acceleration=false)
   &ScaleSpriteSizeBySpeed(
    Max Scale Factor=&MergeHelper_Vector2Switch(0=<1 1>, 1=<0 0>, 2=<0 0>, 3=<0 0>, 4=<1 5>, Index=&SpawnGroup(), Num=::NewEnumerator3)
    Min Scale Factor=&MergeHelper_Vector2Switch(0=<1 1>, 1=<0 0>, 2=<0 0>, 3=<0 0>, 4=<0 0>, Index=&SpawnGroup(), Num=::NewEnumerator3)
    Sample Scale Factor By Curve=true
    Velocity Threshold=500
   )
   &FN_Light_Attributes(
    Color=<rgba 15 4.9774 1.3498 1>
    Create Color Attribute=true
    Create Exponent Attribute=true
    Create Position Attribute=true
    Create Radius Attribute=true
    Enable Particle Light Enable=true
    Enable Visibility Tag=true
    Light Enabled=true
    Position=<0 0 180>
    Radius=&FloatFromCurve(
     FloatCurve=DataInterface() {propertyValues {Curve {
        defaultValue=1
        keys [{interpMode=RCIM_Cubic, tangentMode=RCTM_User, time=0, value=0}, {interpMode=RCIM_Cubic, tangentMode=RCTM_User, time=0.3, value=1}]
       }}}
     Scale Curve=35
    )
    Scale Color=&FloatFromCurve(FloatCurve=DataInterface() {propertyValues {Curve {keys [
         {interpMode=RCIM_Cubic, tangentMode=RCTM_Break, time=0, value=1}
         {interpMode=RCIM_Cubic, tangentMode=RCTM_Break, time=0.6908, value=1}
         {interpMode=RCIM_Cubic, tangentMode=RCTM_Break, time=1, value=0}
        ]}}})
    VisibilityTag=10
   )
   /SetVariables(Particles.Distortion SpriteSize:Vector2f=<150 150>) {isSetParameters=true, module=SetVariables_#13}
   &SubUVAnimation(
    Output Alternate Renderer Binding=true
    "Sprite Renderer "=DataInterface() {propertyValues {SpriteRenderer=&self:Dragons_despawn_0.NiagaraSpriteRendererProperties_6}}
    UseEndFrame=false
    UseStartFrame=false
   ) {module=SubUVAnimation003}
  ]
  rendererClasses [&NiagaraSpriteRendererProperties, &NiagaraLightRendererProperties]
  renderers [
   {rendererClass=&NiagaraSpriteRendererProperties}
   {rendererClass=&NiagaraSpriteRendererProperties, rendererIndex=1}
   {rendererClass=&NiagaraSpriteRendererProperties, rendererIndex=2}
   {rendererClass=&NiagaraSpriteRendererProperties, rendererIndex=3}
   {rendererClass=&NiagaraSpriteRendererProperties, rendererIndex=4}
   {rendererClass=&NiagaraSpriteRendererProperties, rendererIndex=5}
   {rendererClass=&NiagaraLightRendererProperties, rendererIndex=6}
  ]
 }
 Embers_Burst001 {
  bEnabled=false
  simTarget=GPUComputeSim
  emitterUpdateScript [
   &EmitterState(
    Inactive Response=::NewEnumerator1
    Life Cycle Mode=::NewEnumerator1
    Loop Behavior=::NewEnumerator1
    Loop Delay=2.25
    Loop Duration=4
    UseLoopDelay=true
   )
   &SpawnBurst_Instantaneous(Spawn Count=100, Spawn Time=3.5, Use Spawn Probability=false, UseLoopCountLimit=false) {enabled=false, module=SpawnBurst_Instantaneous002}
   &SpawnBurst_Instantaneous(Spawn Count=100, Spawn Time=0.05, Use Spawn Probability=false, UseLoopCountLimit=false)
   &SpawnRate(
    SpawnRate=&FloatFromCurve(FloatCurve=DataInterface() {propertyValues {Curve {keys [<at 0.9 1>,<at 1 0>]}}}, CurveIndex=$Emitter.NormalizedLoopAge, Scale Curve=15)
    Use Spawn Probability=false
   )
   &SpawnBurst_Instantaneous(Spawn Count=15, Spawn Time=2, Use Spawn Probability=false, UseLoopCountLimit=false) {module=SpawnBurst_Instantaneous001}
  ]
  particleSpawnScript [
   &InitializeParticle(
    Color=<rgba 1 1 1 2>
    Lifetime=&UniformRangedFloat(Minimum=0.3, Maximum=0.8)
    Mass Mode=::NewEnumerator1
    Position Mode=::NewEnumerator0
    Sprite Size=&Vector2DFromFloat(Value=&FloatFromCurve(
      FloatCurve=DataInterface() {propertyValues {Curve {keys [<at 0 1.2>,<at 0.1257 0.9847>,<at 0.2996 0.5>,<at 1 0.3>]}}}
      CurveIndex=&UniformRangedFloat(Minimum=0.5, Maximum=2)
      Scale Curve=5
     ))
    Sprite Size Mode=::NewEnumerator4
    Write Material Random=true
   ) {module=InitializeParticle001}
   &ShapeLocation(MaskSpawnGroup=false, Override Seed=false, Rotation Axis=<0 1 0>, Rotation Mode=::NewEnumerator1, Sphere Radius=25, UseOwnerScale=true)
   &AddVelocity() {module=AddVelocity001}
   &SkeletalMeshLocation(
    Bone Sampling Mode=::NewEnumerator4
    MaskSpawnGroup=false
    Sampled Position Offset=<0 0 20>
    Skeletal Mesh=DataInterface() {propertyValues {
      DefaultMesh=null
      ExcludeBoneName=root
      OptionalComponentClass=null
      PreviewMesh=/BRCosmetics/Animation/Props/Gadgets_OutlineLove/Meshes/OutlineLove_Gadget
      SkinningMode=SkinOnTheFly
      SoftSourceActor=null
      SourceMode=Default
      UvSetIndex=0
      WholeMeshLOD=-1
      bExcludeBone=false
      bReadDeformedGeometry=true
      bRequireCurrentFrameData=true
      ComponentTags []
      FilteredBones [JNT_Portal_FX]
      FilteredSockets []
      MeshUserParameter {parameter {name=None, typeDefHandle {registeredTypeIndex=1}, varData []}}
      SamplingRegions []
     }}
   )
  ]
  particleUpdateScript [
   &ParticleState()
   &ApplyOwnerScaleToAttributes(Scale Initial Velocity=true)
   &ScaleSpriteSize(Scale Factor=<0.5 1>, Scale Sprite Size Mode=::NewEnumerator2)
   &Drag(
    Drag=&Multiply_Float(
     A=&FloatFromCurve(
      FloatCurve=DataInterface() {propertyValues {Curve {keys [<at 0 0.5>, {arriveTangent=0.5, interpMode=RCIM_Cubic, leaveTangent=0.5, time=0.3, value=0.5}, <at 1 1 RCIM_Cubic>]}}}
      Scale Curve=12
     )
     B=1
    )
    Use Linear Drag=true
    Use Rotational Drag=false
   )
   /Niagara/Modules/Update/Forces/CurlNoiseForce(
    Noise Frequency=10
    Noise Strength=&FloatFromCurve(FloatCurve=DataInterface() {propertyValues {Curve {keys [<at 0 1>,<at 1 0>]}}}, Scale Curve=3000)
    Pan Noise=true
    Pan Noise Field=<0.1 0.1 0.5>
   )
   &GravityForce(Gravity=&RandomRangeVector(Minimum=<0 0 900>, Maximum=<0 0 -500>))
   &SolveForcesAndVelocity(Clamp Velocity=false, Limit Acceleration=false)
   &ScaleSpriteSizeBySpeed(
    Max Scale Factor=<0.8 5>
    Min Scale Factor=<0.8 0.8>
    Sample Scale Factor By Curve=true
    Scale Factor Curve=DataInterface() {propertyValues {Curve {keys [
        <at -0.0077 0.0765 RCIM_Cubic>
        {arriveTangent=1.2463, interpMode=RCIM_Cubic, leaveTangent=1.2463, tangentMode=RCTM_User, time=0.9985, value=0.9549}
       ]}}}
    Velocity Threshold=300
   )
   &ScaleColor(
    Scale Alpha=&FloatFromCurve(
     FloatCurve=DataInterface() {propertyValues {Curve {keys [<at 0.25 1>,<at 1 0>]}}}
     Scale Curve=&FNTimeOfDayScalar(Night Scale=2, Morning Scale=1, Evening Scale=0.2)
    )
    Scale RGB=&MakeVectorFromLinearColorRGB(Color=&ColorFromCurve(ColorCurve=DataInterface() {propertyValues {
        AlphaCurve {keys [<at 0 1 RCIM_Cubic>,<at 0.7107 1>]}
        BlueCurve {keys [<at 0 14.6701 RCIM_Cubic>,<at 0.368 3.776>,<at 0.855 0.1753 RCIM_Cubic>]}
        GreenCurve {keys [<at 0 36.7481 RCIM_Cubic>,<at 0.368 13.2325>,<at 0.855 0.3243 RCIM_Cubic>]}
        RedCurve {keys [<at 0 50 RCIM_Cubic>,<at 0.368 25>,<at 0.855 1 RCIM_Cubic>]}
       }}))
    ScaleA=true
    ScaleRGB=true
   )
   &ScaleColor(
    Scale Alpha=&FloatFromCurve(
     FloatCurve=DataInterface() {propertyValues {Curve {keys [
         {arriveTangent=0.5383, interpMode=RCIM_Cubic, leaveTangent=0.5383, tangentMode=RCTM_User, time=0, value=0.03}
         {arriveTangent=0.2783, interpMode=RCIM_Cubic, leaveTangent=0.2783, tangentMode=RCTM_User, time=1, value=1}
        ]}}}
     CurveIndex=$System.NearCameraFade
    )
    ScaleA=true
    ScaleRGB=true
   ) {enabled=false, module=ScaleColor001}
   /SetVariables(Particles.Glow03_Color=&MakeLinearColorFromVectorAndFloat(
     "Vector (RGB)"=&MakeCustomVectorFromBool(
      Bool=$User.bIsFrontEndPreview
      True Vector=&MakeVectorFromLinearColorRGB(Color=<rgba 1 0.28 0.02 1>)
      False Vector=&MakeVectorFromLinearColorRGB(Color=&MultiplyLinearColor(
        Linear Color=&FNTimeOfDayColor(Night Color=<rgba 1 0.3167 0.0354 1>, Morning Color=<rgba 1 0.15 0.02 1>, Day Color=<rgba 1 0.15 0.02 1>, Evening Color=<rgba 1 0.15 0.02 1>)
        Scale Factor=&MakeVector4_FromFloat(Float=&Multiply_Float(A=1, B=&FNTimeOfDayScalar(Night Scale=0.5, Morning Scale=0.5, Evening Scale=0.5)))
       ))
     )
     "Float (Alpha)"=&Multiply_Float(
      A=0.325
      B=&FloatFromCurve(
       FloatCurve=DataInterface() {propertyValues {Curve {keys [<at 0 0>,<at 0.15 0>,<at 0.25 1>]}}}
       CurveIndex=$Emitter.Age
       Scale Curve=&MakeCustomFloatFromBool(Bool=$User.bIsFrontEndPreview, True Float=0.6, False Float=1)
      )
     )
    )) {isSetParameters=true}
   /SetVariables(Particles.Glow03_Size=&Multiply_Vector2D_ByFloat(
     Float=&FloatFromCurve(
      FloatCurve=DataInterface() {propertyValues {Curve {keys [
          <at 0 0 RCIM_Cubic>
          <at 0.2 1.3 RCIM_Cubic>
          {arriveTangent=-0.1423, interpMode=RCIM_Cubic, leaveTangent=-0.1423, tangentMode=RCTM_User, time=1, value=0}
         ]}}}
      Scale Curve=10
     )
     Vector2D=<0.75 1>
    )) {isSetParameters=true}
  ]
  rendererClasses [&NiagaraSpriteRendererProperties]
  renderers [
   {rendererClass=&NiagaraSpriteRendererProperties}
   {rendererClass=&NiagaraSpriteRendererProperties, rendererIndex=1}
  ]
 }
 Fire {
  bEnabled=false
  emitterSpawnScript [/SetVariables(Emitter.velocity flames:NiagaraFloat=&FloatFromCurve(
     FloatCurve=DataInterface() {propertyValues {Curve {
        defaultValue=1
        keys [
         {arriveTangent=9.0725, interpMode=RCIM_Cubic, leaveTangent=9.0725, tangentMode=RCTM_User, time=0, value=0}
         {arriveTangent=-2.5249, interpMode=RCIM_Cubic, leaveTangent=-2.5249, tangentMode=RCTM_User, time=0.1408, value=3}
         {arriveTangent=-0.0405, interpMode=RCIM_Cubic, leaveTangent=-0.0405, tangentMode=RCTM_User, time=0.1916, value=1}
        ]
       }}}
     CurveIndex=$Emitter.NormalizedLoopAge
    )) {isSetParameters=true}]
  emitterUpdateScript [
   &EmitterState(Life Cycle Mode=::NewEnumerator1, Loop Behavior=::NewEnumerator1, Loop Delay=2.25, Loop Duration=1, UseLoopDelay=true)
   &SpawnBurst_Instantaneous(Use Spawn Probability=false, UseLoopCountLimit=false)
  ]
  particleSpawnScript [
   &InitializeParticle(
    Lifetime=3.5
    Mesh Scale=<1.7 1.7 1.7>
    Mesh Scale Mode=::NewEnumerator4
    Position Offset=<0 -23 195>
    UsePositionOffset=true
    Write Material Random=true
   )
   &Color(Color=&MakeLinearColorFromVectorAndFloat("Vector (RGB)"=<3 3 3>, "Float (Alpha)"=1))
   /CascadeToNiagaraConverter/NiagaraScripts/ModuleScripts/AddVectorToPosition(Vector:Vector3f=&TransformVector(Vector=<0 -5 -5>, Destination Space=::Simulation))
  ]
  particleUpdateScript [
   &ParticleState()
   &ScaleColor(
    Scale Alpha=&FloatFromCurve(FloatCurve=DataInterface() {propertyValues {Curve {keys [
         <at 0 0 RCIM_Cubic>
         {arriveTangent=6.7494, interpMode=RCIM_Cubic, leaveTangent=6.7494, time=0.0841, value=1.5}
         {arriveTangent=-0.6021, interpMode=RCIM_Cubic, leaveTangent=-0.6021, time=0.1517, value=1.0242}
         {arriveTangent=0.0001, interpMode=RCIM_Cubic, leaveTangent=0.0001, tangentMode=RCTM_User, time=0.9146, value=1}
         <at 1 0 RCIM_Cubic>
        ]}}})
    ScaleA=true
    ScaleRGB=true
   )
   &DynamicMaterialParameters(Module.Param0WriteEnabled=true, Module.Param1WriteEnabled=true, Module.Param2WriteEnabled=true, Module.Param3WriteEnabled=true)
   /SetVariables(Particles.Scale=&VectorFromFloat(Value=&FloatFromCurve(
      FloatCurve=DataInterface() {propertyValues {Curve {
         defaultValue=1
         keys [{arriveTangent=-0.0405, interpMode=RCIM_Cubic, leaveTangent=-0.0405, tangentMode=RCTM_User, time=0.2246, value=0.9693}]
        }}}
      Scale Curve=1.7
     ))) {isSetParameters=true}
  ]
  rendererClasses [&NiagaraMeshRendererProperties]
  renderers [
   {rendererClass=&NiagaraMeshRendererProperties, rendererIndex=0}
   {rendererClass=&NiagaraMeshRendererProperties, rendererIndex=1}
  ]
 }
 Fire001 {
  emitterUpdateScript [
   &EmitterState(Inactive Response=::NewEnumerator1, Life Cycle Mode=::NewEnumerator1, Loop Behavior=::NewEnumerator1, Loop Duration=3.7, UseLoopDelay=false)
   &SpawnBurst_Instantaneous(Spawn Count=12, Use Spawn Probability=false, UseLoopCountLimit=false) {enabled=false}
   &SpawnBurst_Instantaneous(Spawn Count=3, Spawn Group=1, Use Spawn Probability=false, UseLoopCountLimit=false) {enabled=false, module=SpawnBurst_Instantaneous001}
   &SpawnRate(SpawnRate=300, Use Spawn Probability=false)
   &SpawnRate(SpawnGroup=1, SpawnRate=8, Use Spawn Probability=false) {enabled=false, module=SpawnRate001}
  ]
  particleSpawnScript [
   &PartitionParticles(Number Of Partitions=6) {module=PartitionParticles001}
   /SetVariables(Particles.VisibilityTag=&SpawnGroup()) {isSetParameters=true}
   &InitializeParticle(
    Lifetime=&MergeHelper_FloatSwitch(
     0=&RandomRangeFloat(Minimum=0.1, Maximum=0.3)
     1=&RandomRangeFloat(Minimum=0.5, Maximum=0.9)
     2=?
     3=?
     Index=&SpawnGroup()
     Num=::NewEnumerator0
    )
    Sprite Size=<2.5 1.5>
    Sprite Size Mode=::NewEnumerator4
    UsePositionOffset=false
    Write Material Random=true
   )
   &AddVelocity(Velocity=<0 0 -50>) {enabled=false, module=AddVelocity001}
   &SkeletalMeshLocation(
    Bone / Socket Index=$Particles.PartitionParticles001.Partition
    MaskSpawnGroup=false
    Mesh Sampling Type=::NewEnumerator5
    Position Sampling=::NewEnumerator1
    Skeletal Mesh=$System.Skeletal Mesh Reader
    Socket Sampling Mode=::NewEnumerator5
   )
   &DynamicMaterialParameters(
    Index 0 Param 2=&RandomRangeFloat(Minimum=0, Maximum=4)
    Index 0 Param 4=2.5
    Module.Param0WriteEnabled=false
    Module.Param1WriteEnabled=true
    Module.Param2WriteEnabled=false
    Module.Param3WriteEnabled=true
   ) {module=DynamicMaterialParameters001}
   &ShapeLocation(Box Size=<5 10 1>, MaskSpawnGroup=false, Override Seed=false, Shape Primitive=::NewEnumerator2, Surface Only=false, UseOwnerScale=true)
  ]
  particleUpdateScript [
   &ParticleState()
   &CameraOffset(Camera Offset Amount=&FloatFromCurve(
     FloatCurve=DataInterface() {propertyValues {Curve {defaultValue=1, keys [<at 0 0 RCIM_Cubic>,<at 0.4 1 RCIM_Cubic>]}}}
     Scale Curve=&RandomRangeFloat(Minimum=18, Maximum=25)
    ))
   &Drag(
    Drag=&MergeHelper_FloatSwitch(
     0=&RandomRangeFloat(Minimum=5, Maximum=50)
     1=&RandomRangeFloat(Minimum=2, Maximum=3)
     2=?
     3=?
     Index=&SpawnGroup()
     Num=::NewEnumerator0
    )
    Use Linear Drag=true
    Use Rotational Drag=true
   ) {enabled=false}
   &GravityForce(Gravity=&MakeVector(
     X=0
     Y=0
     Z=&FloatFromCurve(
      FloatCurve=DataInterface() {propertyValues {Curve {
         defaultValue=1
         keys [
          {arriveTangent=-0.006, interpMode=RCIM_Cubic, leaveTangent=-0.006, tangentMode=RCTM_User, time=0, value=0}
          {arriveTangent=3.3339, interpMode=RCIM_Cubic, leaveTangent=3.3339, tangentMode=RCTM_User, time=1, value=1}
         ]
        }}}
      Scale Curve=&RandomRangeFloat(Minimum=-250, Maximum=-425)
     )
    )) {enabled=false}
   &InheritSourceMovement(ApplyForce=false, ApplyPosition=true, ApplyPositionFalloff=false)
   &SolveForcesAndVelocity(Clamp Velocity=false, Limit Acceleration=false)
   &ScaleSpriteSizeBySpeed(Max Scale Factor=<2 6>)
   &ScaleSpriteSize(Scale Factor=&Multiply_Vector2D_ByFloat(
     Float=&FloatFromCurve(
      FloatCurve=DataInterface() {propertyValues {Curve {keys [
          {arriveTangent=4.1785, interpMode=RCIM_Cubic, leaveTangent=4.1785, tangentMode=RCTM_User, time=0, value=0}
          {arriveTangent=1.1179, interpMode=RCIM_Cubic, leaveTangent=1.1179, tangentMode=RCTM_User, time=0.2042, value=0.6228}
          {interpMode=RCIM_Cubic, leaveTangent=0.474, time=1, value=0.9805}
         ]}}}
      Scale Curve=&RandomRangeFloat(Minimum=0.5, Maximum=4)
     )
     Vector2D=<0.8 1>
    ))
   &SpriteRotationRate(Rotation Rate=&Lerp_Float(A=-10, B=10, Alpha=$Particles.Initial.MaterialRandom)) {module=SpriteRotationRate001}
   &ScaleColor(
    Scale Alpha=&FloatFromCurve(FloatCurve=DataInterface() {propertyValues {Curve {
        defaultValue=1
        keys [
         {arriveTangent=0.0866, interpMode=RCIM_Cubic, leaveTangent=0.0866, tangentMode=RCTM_User, time=0, value=0}
         {arriveTangent=2.5076, interpMode=RCIM_Cubic, leaveTangent=2.5076, tangentMode=RCTM_User, time=0.2, value=0.1072}
         {arriveTangent=-0.1236, interpMode=RCIM_Cubic, leaveTangent=-0.1236, tangentMode=RCTM_User, time=0.5, value=0.9731}
         {arriveTangent=-2.491, interpMode=RCIM_Cubic, leaveTangent=-2.491, tangentMode=RCTM_User, time=1, value=0}
        ]
       }}})
    Scale RGB=&MakeVectorFromLinearColorRGB(Color=&ColorFromCurve(ColorCurve=DataInterface() {propertyValues {
        AlphaCurve {keys [<at 0 1 RCIM_Cubic>]}
        BlueCurve {keys [<at 0 0 RCIM_Cubic>,<at 0.5931 0 RCIM_Cubic>,<at 1 0>]}
        GreenCurve {keys [<at 0 1.046 RCIM_Cubic>, {arriveTangent=-0.9672, interpMode=RCIM_Cubic, leaveTangent=-0.9672, time=0.5931, value=0.1575}, <at 1 0.0788>]}
        RedCurve {keys [<at 0 4.5 RCIM_Cubic>, {arriveTangent=-3.5, interpMode=RCIM_Cubic, leaveTangent=-3.5, time=0.5931, value=2}, <at 1 1>]}
       }}))
    ScaleA=true
    ScaleRGB=true
   ) {module=ScaleColor002}
   &DynamicMaterialParameters(
    Index 0 Param 1=&FloatFromCurve(FloatCurve=DataInterface() {propertyValues {Curve {
        defaultValue=0
        keys [{arriveTangent=2.1289, interpMode=RCIM_Cubic, leaveTangent=2.1289, tangentMode=RCTM_User, time=0, value=0}, <at 1 1 RCIM_Cubic>]
       }}})
    Module.Index0Float0Write=true
    Module.Index0Float1Write=false
    Module.Index0Float2Write=false
    Module.Index0Float3Write=false
   )
  ]
  rendererClasses [&NiagaraSpriteRendererProperties]
  renderers [
   {rendererClass=&NiagaraSpriteRendererProperties}
   {rendererClass=&NiagaraSpriteRendererProperties, rendererIndex=1}
  ]
 }
 MeshFlames001 {
  bEnabled=false
  emitterUpdateScript [
   &EmitterState(
    Inactive Response=::NewEnumerator1
    Life Cycle Mode=::NewEnumerator1
    Loop Behavior=::NewEnumerator1
    Loop Delay=0.02
    Loop Duration=3.5
    UseLoopDelay=true
   )
   &SpawnRate(SpawnRate=64, Use Spawn Probability=false) {enabled=false}
   &SpawnBurst_Instantaneous(Spawn Count=6, Use Spawn Probability=false, UseLoopCountLimit=false)
  ]
  particleSpawnScript [
   &PartitionParticles(Number Of Partitions=6)
   &FN_SetSpawnGroup(Normalize Spawn Group=true, Particles Per Group=3, Pattern=::NewEnumerator0)
   &InitializeParticle(
    Color=&MergeHelper_ColorSwitch(
     0=<rgba 1 0 0.6498 1>
     1=<rgba 1 0 0.6498 1>
     2=<rgba 0.1776 1 0.076 1>
     3=<rgba 0.1776 1 0.076 1>
     4=<rgba 0.0617 0 1 1>
     5=<rgba 0.0617 0 1 1>
     Index=$Particles.PartitionParticles.Partition
     Num=::NewEnumerator4
    )
    Lifetime=5
    Sprite Size Mode=::NewEnumerator3
    UsePositionOffset=false
    Write Material Random=true
   ) {module=InitializeParticle001}
   &InitialMeshOrientation(Mesh Axis To Orient=<0 1 0>, Mesh Orientation Mode=::NewEnumerator2, Orientation Vector=<0 0 1>, Use Rotation Vector=false)
  ]
  particleUpdateScript [
   &ParticleState()
   &SkeletalMeshLocation(
    Bone / Socket Index=$Particles.PartitionParticles.Partition
    MaskSpawnGroup=false
    Mesh Sampling Type=::NewEnumerator5
    Sampled Scale Coordinate Space:ENiagaraCoordinateSpace=::Simulation
    Skeletal Mesh=DataInterface() {propertyValues {
      DefaultMesh=null
      ExcludeBoneName=root
      OptionalComponentClass=null
      PreviewMesh=&OutlineLove_Gadget
      SkinningMode=SkinOnTheFly
      SoftSourceActor=null
      SourceMode=Default
      UvSetIndex=0
      WholeMeshLOD=-1
      bExcludeBone=false
      bReadDeformedGeometry=true
      bRequireCurrentFrameData=true
      ComponentTags []
      FilteredBones [JNT_Dragon01_Elbow_l, JNT_Dragon01_Elbow_r, JNT_Dragon02_Elbow_l, JNT_Dragon02_Elbow_r, JNT_Dragon03_Elbow_r, JNT_Dragon03_Elbow_l]
      FilteredSockets [
       FX_Dragon01_Elbow_Socket_r
       FX_Dragon01_Elbow_Socket_l
       FX_Dragon02_Elbow_Socket_r
       FX_Dragon02_Elbow_Socket_l
       FX_Dragon03_Elbow_Socket_r
       FX_Dragon03_Elbow_Socket_l
      ]
      MeshUserParameter {parameter {name=None, typeDefHandle {registeredTypeIndex=1}, varData []}}
      SamplingRegions []
     }}
    Socket Sampling Mode=::NewEnumerator5
   ) {module=SkeletalMeshLocation001}
   &DynamicMaterialParameters(
    Index 0 Param 1=&RandomRangeFloat(Minimum=0, Maximum=50)
    Index 0 Param 4=5.6056
    Module.Index0Float0Write=true
    Module.Index0Float1Write=true
    Module.Index0Float2Write=true
    Module.Index0Float3Write=true
   )
  ]
  rendererClasses [&NiagaraSpriteRendererProperties]
  renderers [{rendererClass=&NiagaraSpriteRendererProperties}]
 }
 Mesh_flames_for_perf {
  emitterUpdateScript [
   &EmitterState(
    Inactive Response=::NewEnumerator1
    Life Cycle Mode=::NewEnumerator1
    Loop Behavior=::NewEnumerator1
    Loop Delay=0.02
    Loop Duration=3.7
    UseLoopDelay=true
   )
   &SpawnRate(SpawnRate=60, Use Spawn Probability=false)
   &SpawnBurst_Instantaneous(Spawn Count=18, Use Spawn Probability=false, UseLoopCountLimit=false) {enabled=false}
  ]
  particleSpawnScript [
   &FN_SetSpawnGroup(Number of Spawn Groups=18)
   &InitializeParticle(
    Lifetime Max=1
    Lifetime Min=0.5
    Lifetime Mode=::NewEnumerator1
    Mesh Scale Mode=::NewEnumerator1
    Mesh Uniform Scale Min=1.25
    UsePositionOffset=false
    Write Material Random=true
   ) {module=InitializeParticle001}
   &Color(Color=&MakeLinearColorFromVectorAndFloat(
     "Vector (RGB)"=<1 1 1>
     "Float (Alpha)"=&ClampFloat(Float=&RandomRangeFloat(Minimum=0.15, Maximum=0.3, Evaluation Type=::NewEnumerator1), Min=0.2)
    ))
   &InitialMeshOrientation(Mesh Axis To Orient=<0 1 0>, Mesh Orientation Mode=::NewEnumerator2, Orientation Vector=<0 0 1>, Use Rotation Vector=false)
  ]
  particleUpdateScript [
   &SkeletalMeshLocation(
    Bone / Socket Index=$Particles.SpawnGroup
    MaskSpawnGroup=false
    Mesh Sampling Type=::NewEnumerator5
    Sampled Scale Coordinate Space:ENiagaraCoordinateSpace=::Simulation
    Skeletal Mesh=DataInterface() {propertyValues {
      DefaultMesh=null
      ExcludeBoneName=root
      OptionalComponentClass=null
      PreviewMesh=&OutlineLove_Gadget
      SkinningMode=SkinOnTheFly
      SoftSourceActor=null
      SourceMode=Default
      UvSetIndex=0
      WholeMeshLOD=-1
      bExcludeBone=false
      bReadDeformedGeometry=true
      bRequireCurrentFrameData=true
      ComponentTags []
      FilteredBones [JNT_Dragon01_Elbow_l, JNT_Dragon01_Elbow_r, JNT_Dragon02_Elbow_l, JNT_Dragon02_Elbow_r, JNT_Dragon03_Elbow_r, JNT_Dragon03_Elbow_l]
      FilteredSockets [
       FX_Dragon01_Finger_Socket_r
       FX_Dragon01_FingerA_Socket_r
       FX_Dragon01_Finger_end_Socket_r
       FX_Dragon01_Finger_Socket_l
       FX_Dragon01_FingerA_Socket_l
       FX_Dragon01_Finger_end_Socket_l
       FX_Dragon02_Finger_Socket_r
       FX_Dragon02_FingerA_Socket_r
       FX_Dragon02_Finger_end_Socket_r
       FX_Dragon02_Finger_Socket_l
       FX_Dragon02_FingerA_Socket_l
       FX_Dragon02_Finger_end_Socket_l
       FX_Dragon03_Finger_Socket_r
       FX_Dragon03_FingerA_Socket_r
       FX_Dragon03_Finger_end_Socket_r
       FX_Dragon03_Finger_Socket_l
       FX_Dragon03_FingerA_Socket_l
       FX_Dragon03_Finger_end_Socket_l
      ]
      MeshUserParameter {parameter {name=None, typeDefHandle {registeredTypeIndex=1}, varData []}}
      SamplingRegions []
     }}
    Socket Sampling Mode=::NewEnumerator5
   ) {module=SkeletalMeshLocation001}
   &ParticleState()
   &UpdateMeshOrientation()
   &ScaleMeshSize(Scale Factor=<2 0.3 1.5>)
   &ScaleColor(
    Scale Alpha=&FloatFromCurve(FloatCurve=DataInterface() {propertyValues {Curve {keys [<at 0 0 RCIM_Cubic>,<at 0.4 1>]}}})
    Scale RGB=&MakeVectorFromLinearColorRGB(Color=<rgba 1 0.887 0.0302 1>)
    ScaleA=true
    ScaleRGB=true
   )
   &DynamicMaterialParameters(
    Index 0 Param 1=&RandomRangeFloat(Minimum=0, Maximum=50)
    Index 0 Param 4=5.6056
    Module.Index0Float0Write=true
    Module.Index0Float1Write=true
    Module.Index0Float2Write=true
    Module.Index0Float3Write=true
   )
   /SetVariables(Particles.fire licks dynparam:Vector4f=&MakeVector4(X=0, Y=0, Z=20, W=0)) {isSetParameters=true, module=SetVariables_#14}
  ]
  rendererClasses [&NiagaraMeshRendererProperties]
  renderers [
   {rendererClass=&NiagaraMeshRendererProperties, rendererIndex=0}
   {rendererClass=&NiagaraMeshRendererProperties, rendererIndex=1}
   {rendererClass=&NiagaraMeshRendererProperties, rendererIndex=2}
   {rendererClass=&NiagaraMeshRendererProperties, rendererIndex=3}
   {rendererClass=&NiagaraMeshRendererProperties, rendererIndex=4}
  ]
 }
 Mesh_flames_for_perf001 {
  bEnabled=false
  emitterUpdateScript [
   &EmitterState(
    Inactive Response=::NewEnumerator1
    Life Cycle Mode=::NewEnumerator1
    Loop Behavior=::NewEnumerator1
    Loop Delay=0.02
    Loop Duration=3.7
    UseLoopDelay=true
   )
   &SpawnRate(SpawnRate=60, Use Spawn Probability=false)
   &SpawnBurst_Instantaneous(Spawn Count=18, Use Spawn Probability=false, UseLoopCountLimit=false) {enabled=false}
  ]
  particleSpawnScript [
   &FN_SetSpawnGroup(Number of Spawn Groups=18)
   &InitializeParticle(
    Lifetime Max=1
    Lifetime Min=0.5
    Lifetime Mode=::NewEnumerator1
    Mesh Scale Mode=::NewEnumerator1
    Mesh Uniform Scale Min=1.25
    UsePositionOffset=false
    Write Material Random=true
   ) {module=InitializeParticle001}
   &Color(Color=&MakeLinearColorFromVectorAndFloat(
     "Vector (RGB)"=<1 1 1>
     "Float (Alpha)"=&ClampFloat(Float=&RandomRangeFloat(Minimum=0.15, Maximum=0.3, Evaluation Type=::NewEnumerator1), Min=0.2)
    ))
   &InitialMeshOrientation(Mesh Axis To Orient=<-1 0 0>, Mesh Orientation Mode=::NewEnumerator2, Orientation Vector=<0 0 1>, Use Rotation Vector=false)
  ]
  particleUpdateScript [
   &SkeletalMeshLocation(
    Bone / Socket Index=$Particles.SpawnGroup
    MaskSpawnGroup=false
    Mesh Sampling Type=::NewEnumerator5
    Sampled Scale Coordinate Space:ENiagaraCoordinateSpace=::Simulation
    Skeletal Mesh=DataInterface() {propertyValues {
      DefaultMesh=null
      ExcludeBoneName=root
      OptionalComponentClass=null
      PreviewMesh=&OutlineLove_Gadget
      SkinningMode=SkinOnTheFly
      SoftSourceActor=null
      SourceMode=Default
      UvSetIndex=0
      WholeMeshLOD=-1
      bExcludeBone=false
      bReadDeformedGeometry=true
      bRequireCurrentFrameData=true
      ComponentTags []
      FilteredBones [JNT_Dragon01_Elbow_l, JNT_Dragon01_Elbow_r, JNT_Dragon02_Elbow_l, JNT_Dragon02_Elbow_r, JNT_Dragon03_Elbow_r, JNT_Dragon03_Elbow_l]
      FilteredSockets [
       FX_Dragon01_Finger_Socket_r
       FX_Dragon01_FingerA_Socket_r
       FX_Dragon01_Finger_end_Socket_r
       FX_Dragon01_Finger_Socket_l
       FX_Dragon01_FingerA_Socket_l
       FX_Dragon01_Finger_end_Socket_l
       FX_Dragon02_Finger_Socket_r
       FX_Dragon02_FingerA_Socket_r
       FX_Dragon02_Finger_end_Socket_r
       FX_Dragon02_Finger_Socket_l
       FX_Dragon02_FingerA_Socket_l
       FX_Dragon02_Finger_end_Socket_l
       FX_Dragon03_Finger_Socket_r
       FX_Dragon03_FingerA_Socket_r
       FX_Dragon03_Finger_end_Socket_r
       FX_Dragon03_Finger_Socket_l
       FX_Dragon03_FingerA_Socket_l
       FX_Dragon03_Finger_end_Socket_l
      ]
      MeshUserParameter {parameter {name=None, typeDefHandle {registeredTypeIndex=1}, varData []}}
      SamplingRegions []
     }}
    Socket Sampling Mode=::NewEnumerator5
   ) {module=SkeletalMeshLocation001}
   &ParticleState()
   &UpdateMeshOrientation()
   &ScaleMeshSize(Scale Factor=<2 0.3 1.5>)
   &ScaleColor(
    Scale Alpha=&FloatFromCurve(FloatCurve=DataInterface() {propertyValues {Curve {keys [<at 0 0 RCIM_Cubic>,<at 0.4 1>]}}})
    Scale RGB=&MakeVectorFromLinearColorRGB(Color=<rgba 1 0.887 0.0302 1>)
    ScaleA=true
    ScaleRGB=true
   )
   &DynamicMaterialParameters(
    Index 0 Param 1=&RandomRangeFloat(Minimum=0, Maximum=50)
    Index 0 Param 4=5.6056
    Module.Index0Float0Write=true
    Module.Index0Float1Write=true
    Module.Index0Float2Write=true
    Module.Index0Float3Write=true
   )
   /SetVariables(Particles.fire licks dynparam:Vector4f=&MakeVector4(X=0, Y=0, Z=20, W=0)) {isSetParameters=true, module=SetVariables_#14}
  ]
  rendererClasses [&NiagaraMeshRendererProperties]
  renderers [
   {rendererClass=&NiagaraMeshRendererProperties, rendererIndex=0}
   {rendererClass=&NiagaraMeshRendererProperties, rendererIndex=1}
   {rendererClass=&NiagaraMeshRendererProperties, rendererIndex=2}
   {rendererClass=&NiagaraMeshRendererProperties, rendererIndex=3}
   {rendererClass=&NiagaraMeshRendererProperties, rendererIndex=4}
  ]
 }
 Minimal {
  bEnabled=false
  emitterUpdateScript [
   &EmitterState(Life Cycle Mode=::NewEnumerator1, Loop Behavior=::NewEnumerator1, Loop Delay=3.2, Loop Duration=2, UseLoopDelay=true)
   &SpawnBurst_Instantaneous(Spawn Count=30, Use Spawn Probability=false, UseLoopCountLimit=false)
  ]
  particleSpawnScript [
   &InitializeParticle(
    Color=<rgba 1.4 0.7736 0.0185 1>
    Lifetime Max=0.6
    Lifetime Min=0.25
    Lifetime Mode=::NewEnumerator1
    Position Offset=<0 0 200>
    Position Offset Coordinate Space=::Local
    Sprite Size Mode=::NewEnumerator1
    Uniform Sprite Size Max=6
    Uniform Sprite Size Min=2
    UsePositionOffset=true
    Write Material Random=true
   )
   &ShapeLocation(MaskSpawnGroup=false, Override Seed=false, Sphere Radius=2, UseOwnerScale=true)
   &AddVelocity(
    Module.Use Falloff Distance=false
    Override Seed=false
    Velocity Mode=::NewEnumerator1
    Velocity Speed=&RandomRangeFloat(Minimum=45, Maximum=100)
   )
  ]
  particleUpdateScript [
   &Drag(Drag=3, Use Linear Drag=true, Use Rotational Drag=true)
   /Niagara/Modules/Update/Forces/V2/CurlNoiseForce(
    Bias Noise=false
    Noise Frequency=10
    Noise Strength=&RandomRangeFloat(Minimum=50, Maximum=120)
    Pan Noise=false
   ) {enabled=false}
   &SolveForcesAndVelocity(Clamp Velocity=false, Limit Acceleration=false)
   &ParticleState()
   &ScaleSpriteSizeBySpeed(Max Scale Factor=<1.2 2>, Min Scale Factor=<0.8 1.4>)
   &ScaleSpriteSize(Uniform Curve Sprite Scale=DataInterface() {propertyValues {Curve {keys [{tangentMode=RCTM_User, time=0, value=0}, <at 0.2 1>, <at 0.6 1>, {tangentMode=RCTM_User, time=1, value=0}]}}})
   &DynamicMaterialParameters(
    Index 0 Param 1=&FloatFromCurve(FloatCurve=DataInterface() {propertyValues {Curve {keys [{interpMode=RCIM_Cubic, leaveTangent=3.2748, tangentMode=RCTM_Break, time=0.4, value=0}, <at 1 1>]}}})
    Index 0 Param 2=&RandomRangeFloat(Minimum=0, Maximum=1)
    Index 0 Param 3=&RandomRangeFloat(Minimum=0, Maximum=1)
    Module.Index0Float0Write=true
    Module.Index0Float1Write=true
    Module.Index0Float2Write=true
    Module.Index0Float3Write=true
   )
  ]
  rendererClasses [&NiagaraSpriteRendererProperties]
  renderers [{rendererClass=&NiagaraSpriteRendererProperties}]
 }
 Minimal004 {
  bEnabled=false
  emitterUpdateScript [
   &EmitterState(
    Inactive Response=::NewEnumerator1
    Life Cycle Mode=::NewEnumerator1
    Loop Behavior=::NewEnumerator1
    Loop Delay=0.02
    Loop Duration=3.7
    UseLoopDelay=true
   )
   &SpawnRate(SpawnRate=60, Use Spawn Probability=false)
   &SpawnBurst_Instantaneous(Use Spawn Probability=false, UseLoopCountLimit=false)
  ]
  particleSpawnScript [
   &FN_SetSpawnGroup(Number of Spawn Groups=18)
   &InitializeParticle(Lifetime=5, UsePositionOffset=false, Write Material Random=true)
   &Color(Color=&MakeLinearColorFromVectorAndFloat(
     "Vector (RGB)"=<1 1 1>
     "Float (Alpha)"=&ClampFloat(Float=&RandomRangeFloat(Minimum=0.15, Maximum=0.3, Evaluation Type=::NewEnumerator1), Min=0.2)
    ))
   &InitialMeshOrientation(Mesh Axis To Orient=<0 1 0>, Mesh Orientation Mode=::NewEnumerator2, Orientation Vector=<0 0 1>, Use Rotation Vector=false)
  ]
  particleUpdateScript [
   &ParticleState()
   &ScaleMeshSize(Scale Factor=<1.9 1 1.5>)
   &UpdateMeshOrientation()
   &SkeletalMeshLocation(
    Bone / Socket Index=$Particles.SpawnGroup
    MaskSpawnGroup=false
    Mesh Sampling Type=::NewEnumerator5
    Sampled Scale Coordinate Space:ENiagaraCoordinateSpace=::Simulation
    Skeletal Mesh=DataInterface() {propertyValues {
      DefaultMesh=null
      ExcludeBoneName=root
      OptionalComponentClass=null
      PreviewMesh=&OutlineLove_Gadget
      SkinningMode=SkinOnTheFly
      SoftSourceActor=null
      SourceMode=Default
      UvSetIndex=0
      WholeMeshLOD=-1
      bExcludeBone=false
      bReadDeformedGeometry=true
      bRequireCurrentFrameData=true
      ComponentTags []
      FilteredBones [JNT_Dragon01_Elbow_l, JNT_Dragon01_Elbow_r, JNT_Dragon02_Elbow_l, JNT_Dragon02_Elbow_r, JNT_Dragon03_Elbow_r, JNT_Dragon03_Elbow_l]
      FilteredSockets [
       FX_Dragon01_Finger_Socket_r
       FX_Dragon01_FingerA_Socket_r
       FX_Dragon01_Finger_end_Socket_r
       FX_Dragon01_Finger_Socket_l
       FX_Dragon01_FingerA_Socket_l
       FX_Dragon01_Finger_end_Socket_l
       FX_Dragon02_Finger_Socket_r
       FX_Dragon02_FingerA_Socket_r
       FX_Dragon02_Finger_end_Socket_r
       FX_Dragon02_Finger_Socket_l
       FX_Dragon02_FingerA_Socket_l
       FX_Dragon02_Finger_end_Socket_l
       FX_Dragon03_Finger_Socket_r
       FX_Dragon03_FingerA_Socket_r
       FX_Dragon03_Finger_end_Socket_r
       FX_Dragon03_Finger_Socket_l
       FX_Dragon03_FingerA_Socket_l
       FX_Dragon03_Finger_end_Socket_l
      ]
      MeshUserParameter {parameter {name=None, typeDefHandle {registeredTypeIndex=1}, varData []}}
      SamplingRegions []
     }}
    Socket Sampling Mode=::NewEnumerator5
   ) {module=SkeletalMeshLocation001}
   &ScaleColor(
    Scale Alpha=&FloatFromCurve(FloatCurve=DataInterface() {propertyValues {Curve {keys [<at 0 0 RCIM_Cubic>,<at 0.4 1>]}}})
    Scale RGB=&MakeVectorFromLinearColorRGB(Color=<rgba 1 0.887 0.0302 1>)
    ScaleA=true
    ScaleRGB=true
   )
   &DynamicMaterialParameters(
    Index 0 Param 1=&RandomRangeFloat(Minimum=0, Maximum=50)
    Index 0 Param 4=5.6056
    Module.Index0Float0Write=true
    Module.Index0Float1Write=true
    Module.Index0Float2Write=true
    Module.Index0Float3Write=true
   )
   /SetVariables() {isSetParameters=true}
  ]
  rendererClasses [&NiagaraMeshRendererProperties]
  renderers [
   {rendererClass=&NiagaraMeshRendererProperties, rendererIndex=0}
   {rendererClass=&NiagaraMeshRendererProperties, rendererIndex=1}
  ]
 }
 Prop_spawn_end001 {
  emitterUpdateScript [
   &EmitterState(Life Cycle Mode=::NewEnumerator1, Loop Behavior=::NewEnumerator1, Loop Delay=5.76, Loop Duration=2.3, UseLoopDelay=true)
   &SpawnBurst_Instantaneous(Spawn Count=5, Use Spawn Probability=false, UseLoopCountLimit=false) {enabled=false, module=SpawnBurst_Instantaneous006}
   &SpawnBurst_Instantaneous(Spawn Group=6, Use Spawn Probability=false, UseLoopCountLimit=false) {module=SpawnBurst_Instantaneous005}
   &SpawnBurst_Instantaneous(Spawn Count=50, Spawn Group=4, Use Spawn Probability=false, UseLoopCountLimit=false) {module=SpawnBurst_Instantaneous004}
   &SpawnBurst_Instantaneous(Spawn Count=35, Spawn Group=3, Use Spawn Probability=false, UseLoopCountLimit=false) {module=SpawnBurst_Instantaneous003}
   &SpawnBurst_Instantaneous(Spawn Count=20, Spawn Group=2, Use Spawn Probability=false) {module=SpawnBurst_Instantaneous002}
   &SpawnBurst_Instantaneous(Spawn Count=1, Spawn Group=5, Use Spawn Probability=false)
   &SpawnBurst_Instantaneous(Spawn Count=35, Use Spawn Probability=false) {module=SpawnBurst_Instantaneous001}
  ]
  particleSpawnScript [
   /SetVariables(Particles.VisibilityTag=&SpawnGroup()) {isSetParameters=true, module=SetVariables_#5}
   /SetVariables(Particles.InitSpriteRotRate=&UniformRangedFloat(Minimum=-1, Maximum=1)) {isSetParameters=true, module=SetVariables_#6}
   &InitializeParticle(
    Color=&MergeHelper_ColorSwitch(
     0=<rgba 0.5255 0.5255 0.5255 1>
     1=<rgba 0.6031 0.5599 0.463 0.2>
     2=<rgba 5 3.7043 1.51 1>
     3=?
     Index=&SpawnGroup()
     Num=::NewEnumerator1
    )
    Lifetime=&MergeHelper_FloatSwitch(
     0=&UniformRangedFloat(Minimum=0.3, Maximum=0.55)
     1=0.25
     2=&UniformRangedFloat(Minimum=0.3, Maximum=0.65)
     3=&UniformRangedFloat(Minimum=0.25, Maximum=0.6)
     4=&UniformRangedFloat(Minimum=0.45, Maximum=0.75)
     5=0.25
     6=0.4
     7=0
     8=0
     9=0
     Index=&SpawnGroup()
     Num=::NewEnumerator8
    )
    Position Offset=<0 -25 185>
    Sprite Rotation Mode=::NewEnumerator2
    Sprite Size Mode=::NewEnumerator3
    Uniform Sprite Size=&MergeHelper_FloatSwitch(
     0=&UniformRangedFloat(Minimum=8, Maximum=20)
     1=65
     2=&RandomRangeFloat(Minimum=10, Maximum=25)
     3=?
     Index=&SpawnGroup()
     Num=::NewEnumerator1
    )
    UsePositionOffset=true
    Write Material Random=true
   )
   &ShapeLocation(
    MaskSpawnGroup=false
    Override Seed=false
    Sphere Radius=&MergeHelper_FloatSwitch(0=7, 1=0, 2=10, 3=5, 4=10, 5=0, Index=&SpawnGroup(), Num=::NewEnumerator4)
    Sphere Surface Distribution=1
    UseOwnerScale=true
   )
   &AddVelocity(Velocity=&MergeHelper_VectorSwitch(
     0=<0 0 150>
     1=<0 0 0>
     2=&UniformRangedVector(Minimum=<-200 0 -110>, Maximum=<200 0 110>)
     3=&RandomRangeVector(Minimum=<-100 0 -50>, Maximum=<100 1 150>)
     4=&RandomRangeVector(Minimum=<-180 -180 -180>, Maximum=<180 180 180>)
     5=<0 0 0>
     6=<0 0 0>
     Index=&SpawnGroup()
     Num=::NewEnumerator5
    ))
   /SetVariables(
    Particles.Smoke Velocity:Vector3f=&Multiply_Vector(
     A=$Particles.ShapeLocation.ShapeNormal
     B=&VectorFromFloat(Value=&RandomRangeFloat(Minimum=15, Maximum=50))
    )
    Particles.Sparkles Velocity:Vector3f=&UniformRangedVector(Minimum=<-500 -200 -200>, Maximum=<5000 200 200>)
    Particles.airglow CameraOffset=10
    Particles.smoke CameraOffset:NiagaraFloat=&RandomRangeFloat(Minimum=10, Maximum=25)
   ) {isSetParameters=true, module=SetVariables_#7}
   &AccelerationForce(Acceleration=&MergeHelper_VectorSwitch(
     0=&UniformRangedVector(Minimum=<0 0 -25>, Maximum=<0 0 -75>)
     1=<0 0 0>
     2=&RandomRangeVector(Minimum=<0 0 0>, Maximum=<0 0 0>)
     3=?
     Index=&SpawnGroup()
     Num=::NewEnumerator1
    ))
   &ApplyInitialForces()
   &ShapeLocation(
    Disc Coverage=0.3738
    MaskSpawnGroup=true
    Offset=<0 0 10>
    Offset Mode=::NewEnumerator0
    Override Seed=false
    Ring Radius=30
    Rotation Axis=<0 1 0>
    Rotation Mode=::NewEnumerator1
    Shape Primitive=::NewEnumerator4
    Spawn Group Mask=5
    UseOwnerScale=true
   ) {module=ShapeLocation001}
  ]
  particleUpdateScript [
   &ParticleState()
   &SpriteRotationRate(Rotation Rate=&MergeHelper_FloatSwitch(
     0=&Multiply_Float(
      A=$Particles.InitSpriteRotRate
      B=&FloatFromCurve(
       FloatCurve=DataInterface() {propertyValues {Curve {keys [
           {arriveTangent=-1.9214, interpMode=RCIM_Cubic, leaveTangent=-1.9214, tangentMode=RCTM_User, time=-0.0046, value=1.0113}
           {arriveTangent=-0.2496, interpMode=RCIM_Cubic, leaveTangent=-0.2496, tangentMode=RCTM_User, time=0.9896, value=0.16}
          ]}}}
       CurveIndex=$Emitter.Age
      )
     )
     1=0
     2=5
     3=?
     Index=&SpawnGroup()
     Num=::NewEnumerator1
    ))
   &ScaleSpriteSize(
    Scale Sprite Size Mode=::NewEnumerator0
    Uniform Scale Factor=&MergeHelper_FloatSwitch(
     0=&FloatFromCurve(
      FloatCurve=DataInterface() {propertyValues {Curve {keys [
          {arriveTangent=5.2354, interpMode=RCIM_Cubic, leaveTangent=5.2354, tangentMode=RCTM_User, time=-0.0046, value=0.25}
          <at 0.1439 1.4923>
          <at 1.0011 3.4809 RCIM_Cubic>
         ]}}}
      Scale Curve=&RandomRangeFloat(Minimum=0.5, Maximum=1.5)
     )
     1=&FloatFromCurve(FloatCurve=DataInterface() {propertyValues {Curve {
         defaultValue=1
         keys [{interpMode=RCIM_Cubic, tangentMode=RCTM_User, time=0, value=0}, {interpMode=RCIM_Cubic, tangentMode=RCTM_User, time=1, value=1}]
        }}})
     2=50
     3=0
     4=&FloatFromCurve(FloatCurve=DataInterface() {propertyValues {Curve {keys [<at 0.8 1 RCIM_Cubic>, {arriveTangent=-6.6126, interpMode=RCIM_Cubic, leaveTangent=-6.6126, tangentMode=RCTM_User, time=1, value=0}]}}})
     Index=&SpawnGroup()
     Num=::NewEnumerator3
    )
   )
   &CurlNoiseForce(
    Bias Noise=false
    Coordinate Space:ENiagaraCoordinateSpace=::Local
    Noise Frequency=&MergeHelper_FloatSwitch(0=0, 1=0, 2=0.02, 3=0, 4=0, 5=0, Index=&SpawnGroup(), Num=::NewEnumerator4)
    Noise Quality / Cost=::NewEnumerator0
    Noise Strength=&MergeHelper_FloatSwitch(
     0=0
     1=0
     2=500
     3=0
     4=&FloatFromCurve(FloatCurve=DataInterface() {propertyValues {Curve {keys [<at 0 1>,<at 1 0>]}}}, Scale Curve=3000)
     5=0
     Index=&SpawnGroup()
     Num=::NewEnumerator4
    )
    Pan Noise=false
    Random Seed=0
    Randomize Noise Sample=false
    Write to Intrinsic Parameters=false
   )
   &Drag(
    Drag=&MergeHelper_FloatSwitch(
     0=5
     1=0
     2=&FloatFromCurve(FloatCurve=DataInterface() {propertyValues {Curve {keys [<at 0 0.5>,<at 0.5 0.5>,<at 1 1>]}}}, Scale Curve=9)
     3=&FloatFromCurve(FloatCurve=DataInterface() {propertyValues {Curve {keys [<at 0 0.5>,<at 0.5 0.5>,<at 1 1>]}}}, Scale Curve=9)
     4=&FloatFromCurve(FloatCurve=DataInterface() {propertyValues {Curve {keys [<at 0 0.5141>,<at 0.5078 0.7118>,<at 0.9984 0.4918>]}}}, Scale Curve=10)
     Index=&SpawnGroup()
     Num=::NewEnumerator3
    )
    Use Linear Drag=true
    Use Rotational Drag=true
   )
   &SubUVAnimation(
    Output Alternate Renderer Binding=true
    Random Start Frame=true
    "Sprite Renderer "=DataInterface() {propertyValues {SpriteRenderer=&self:Prop_spawn_end001_0.NiagaraSpriteRendererProperties_1}}
    SubUV Lookup Index Scale=0.2
    UseEndFrame=false
    UseStartFrame=false
   )
   &SubUVAnimation(
    Output Alternate Renderer Binding=true
    "Sprite Renderer "=DataInterface() {propertyValues {SpriteRenderer=&self:Prop_spawn_end001_0.NiagaraSpriteRendererProperties_0}}
    UseEndFrame=false
    UseStartFrame=false
   ) {module=SubUVAnimation001}
   &SubUVAnimation(
    Output Alternate Renderer Binding=true
    "Sprite Renderer "=DataInterface() {propertyValues {SpriteRenderer=&self:Prop_spawn_end001_0.NiagaraSpriteRendererProperties_4}}
    UseEndFrame=false
    UseStartFrame=false
   ) {module=SubUVAnimation002}
   /SetVariables(
    Particles.Smoke Color:LinearColor=&MakeLinearColorFromVectorAndFloat(
     "Vector (RGB)"=&MakeVectorFromLinearColorRGB(Color=&ColorFromCurve(ColorCurve=DataInterface() {propertyValues {
         AlphaCurve {keys [<at 0 1 RCIM_Cubic>,<at 1 0 RCIM_Cubic>]}
         BlueCurve {keys [<at 0 0 RCIM_Cubic>,<at 1 0 RCIM_Cubic>]}
         GreenCurve {keys [<at 0 0.012 RCIM_Cubic>,<at 1 0 RCIM_Cubic>]}
         RedCurve {keys [<at 0 0.0365 RCIM_Cubic>,<at 1 0 RCIM_Cubic>]}
        }}))
     "Float (Alpha)"=&FloatFromCurve(
      FloatCurve=DataInterface() {propertyValues {Curve {
         defaultValue=1
         keys [
          {arriveTangent=9.0725, interpMode=RCIM_Cubic, leaveTangent=9.0725, tangentMode=RCTM_User, time=0, value=0}
          {arriveTangent=0.1098, interpMode=RCIM_Cubic, leaveTangent=0.1098, tangentMode=RCTM_User, time=0.0826, value=1.0001}
          {arriveTangent=0.0535, interpMode=RCIM_Cubic, leaveTangent=0.0535, tangentMode=RCTM_User, time=0.6499, value=1}
          {arriveTangent=-1.7285, interpMode=RCIM_Cubic, leaveTangent=-1.7285, tangentMode=RCTM_User, time=1, value=0}
         ]
        }}}
      Scale Curve=&Multiply_Float(A=1, B=&FNTimeOfDayScalar(Night Scale=1, Morning Scale=1, Evening Scale=0))
     )
    )
    Particles.Smoke Dyn Param:Vector4f=&MakeVector4(X=0, Y=0, Z=0, W=0)
   ) {isSetParameters=true, module=SetVariables_#8}
   /SetVariables(
    Particles.soft glow Color:LinearColor=&MakeLinearColorFromVectorAndFloat(
     "Vector (RGB)"=&MakeVectorFromLinearColorRGB(Color=<rgba 1 0.6731 0.3597 1>)
     "Float (Alpha)"=&FloatFromCurve(FloatCurve=DataInterface() {propertyValues {Curve {keys [<at 0 1>,<at 1 0>]}}})
    )
    Particles.soft glow Position:NiagaraPosition=&TransformVector(Vector=<0 0 190>, Destination Space=::Simulation)
   ) {isSetParameters=true, module=SetVariables_#9}
   /SetVariables(
    Particles.Embers Color:LinearColor=&MakeLinearColorFromVectorAndFloat(
     "Vector (RGB)"=&MakeVectorFromLinearColorRGB(Color=&ColorFromCurve(ColorCurve=DataInterface() {propertyValues {
         AlphaCurve {keys [<at 0 1 RCIM_Cubic>]}
         BlueCurve {keys [<at 0 7.8125 RCIM_Cubic>,<at 0.2499 0.7812>,<at 0.7114 0.2083 RCIM_Cubic>]}
         GreenCurve {keys [<at 0 34.1758 RCIM_Cubic>,<at 0.2499 4.8887>,<at 0.7114 0.8934 RCIM_Cubic>]}
         RedCurve {keys [<at 0 50 RCIM_Cubic>,<at 0.2499 10>,<at 0.7114 4 RCIM_Cubic>]}
        }}))
     "Float (Alpha)"=2
    )
    Particles.Embers SpriteSize:Vector2f=&Multiply_Vector2D_ByFloat(
     Float=&FloatFromCurve(
      FloatCurve=DataInterface() {propertyValues {Curve {
         defaultValue=1
         keys [
          {interpMode=RCIM_Cubic, tangentMode=RCTM_User, time=0, value=0}
          {interpMode=RCIM_Cubic, tangentMode=RCTM_User, time=0.1511, value=1}
          <at 0.35 0 RCIM_Cubic>
         ]
        }}}
      Scale Curve=2.5
     )
     Vector2D=&RandomRangeVector2D(Minimum=<0.4 0.8>, Maximum=<0.5 1>)
    )
   ) {isSetParameters=true, module=SetVariables_#10}
   /SetVariables(
    Particles.Ashes Color:LinearColor=&MakeLinearColorFromVectorAndFloat(
     "Vector (RGB)"=&MakeVectorFromLinearColorRGB(Color=&ColorFromCurve(ColorCurve=DataInterface() {propertyValues {
         AlphaCurve {keys [<at 0 1 RCIM_Cubic>]}
         BlueCurve {keys [<at 0 1.473 RCIM_Cubic>,<at 0.6088 0.0879 RCIM_Cubic>]}
         GreenCurve {keys [<at 0 11.4982 RCIM_Cubic>,<at 0.6088 0.0919 RCIM_Cubic>]}
         RedCurve {keys [<at 0 50 RCIM_Cubic>,<at 0.6088 0.099 RCIM_Cubic>]}
        }}))
     "Float (Alpha)"=&FloatFromCurve(FloatCurve=DataInterface() {propertyValues {Curve {keys [<at 0.8 1 RCIM_Cubic>,<at 0.85 0>]}}})
    )
    Particles.Ashes SpriteSize:Vector2f=&Vector2DFromFloat(Value=&FloatFromCurve(
      FloatCurve=DataInterface() {propertyValues {Curve {
         defaultValue=1
         keys [
          {interpMode=RCIM_Cubic, tangentMode=RCTM_User, time=0, value=0}
          {interpMode=RCIM_Cubic, tangentMode=RCTM_User, time=0.5, value=1}
          {tangentMode=RCTM_User, time=1, value=0}
         ]
        }}}
      Scale Curve=&RandomRangeFloat(Minimum=1, Maximum=7)
     ))
    Particles.Vector 4:Vector4f=&MakeVector4(X=0, Y=0, Z=0, W=0)
   ) {isSetParameters=true, module=SetVariables_#11}
   /SetVariables(
    Particles.Fire Color:LinearColor=&MakeLinearColorFromVectorAndFloat(
     "Vector (RGB)"=&Multiply_VectorByFloat(
      Vector=&MakeVectorFromLinearColorRGB(Color=<rgba 8 0.7848 0 1>)
      Float=&FloatFromCurve(FloatCurve=DataInterface() {propertyValues {Curve {
          defaultValue=1
          keys [{interpMode=RCIM_Cubic, tangentMode=RCTM_User, time=0, value=1}, {interpMode=RCIM_Cubic, tangentMode=RCTM_User, time=0.75, value=0}]
         }}})
     )
     "Float (Alpha)"=1
    )
    Particles.Fire DynParam:Vector4f=&MakeVector4(
     X=&FloatFromCurve(FloatCurve=DataInterface() {propertyValues {Curve {
         defaultValue=0
         keys [{arriveTangent=2.1289, interpMode=RCIM_Cubic, leaveTangent=2.1289, tangentMode=RCTM_User, time=0, value=0}, <at 1.0039 1 RCIM_Cubic>]
        }}})
     Y=0.5
     Z=0
     W=0
    )
    Particles.fire SpriteSize:Vector2f=&Multiply_Vector2D_ByFloat(
     Float=&FloatFromCurve(FloatCurve=DataInterface() {propertyValues {Curve {keys [
          {interpMode=RCIM_Cubic, tangentMode=RCTM_Break, time=0.0079, value=0}
          {interpMode=RCIM_Cubic, tangentMode=RCTM_Break, time=0.0417, value=3}
          {arriveTangent=0.3384, interpMode=RCIM_Cubic, leaveTangent=0.3384, tangentMode=RCTM_User, time=0.1668, value=1.3}
          {arriveTangent=-0.0782, interpMode=RCIM_Cubic, leaveTangent=-0.0782, tangentMode=RCTM_User, time=0.7507, value=1.3}
          <at 0.8222 0>
         ]}}})
     Vector2D=&RandomRangeVector2D(Minimum=<15 15>, Maximum=<20 20>)
    )
   ) {isSetParameters=true, module=SetVariables_#12}
   &GravityForce(Gravity=&MergeHelper_VectorSwitch(
     0=<0 0 0>
     1=<0 0 0>
     2=&MakeVector(
      X=0
      Y=0
      Z=&FloatFromCurve(
       FloatCurve=DataInterface() {propertyValues {Curve {keys [<at 0 0.25 RCIM_Cubic>,<at 0.25 1>]}}}
       CurveIndex=$Emitter.Age
       Scale Curve=&RandomRangeFloat(Minimum=-60, Maximum=-170)
      )
     )
     3=<0 0 5>
     4=&RandomRangeVector(Minimum=<0 0 -10>, Maximum=<1 1 -350>)
     5=<0 0 0>
     Index=&SpawnGroup()
     Num=::NewEnumerator4
    ))
   &SolveForcesAndVelocity(Clamp Velocity=false, Limit Acceleration=false)
   &ScaleSpriteSizeBySpeed(
    Max Scale Factor=&MergeHelper_Vector2Switch(0=<1 1>, 1=<0 0>, 2=<0 0>, 3=<0 0>, 4=<1 5>, Index=&SpawnGroup(), Num=::NewEnumerator3)
    Min Scale Factor=&MergeHelper_Vector2Switch(0=<1 1>, 1=<0 0>, 2=<0 0>, 3=<0 0>, 4=<0 0>, Index=&SpawnGroup(), Num=::NewEnumerator3)
    Sample Scale Factor By Curve=true
    Velocity Threshold=500
   )
   &FN_Light_Attributes(
    Color=<rgba 15 4.9774 1.3498 1>
    Create Color Attribute=true
    Create Exponent Attribute=true
    Create Position Attribute=true
    Create Radius Attribute=true
    Enable Particle Light Enable=true
    Enable Visibility Tag=true
    Light Enabled=true
    Position=<0 0 180>
    Radius=&FloatFromCurve(
     FloatCurve=DataInterface() {propertyValues {Curve {
        defaultValue=1
        keys [{interpMode=RCIM_Cubic, tangentMode=RCTM_User, time=0, value=0}, {interpMode=RCIM_Cubic, tangentMode=RCTM_User, time=0.3, value=1}]
       }}}
     Scale Curve=35
    )
    Scale Color=&FloatFromCurve(FloatCurve=DataInterface() {propertyValues {Curve {keys [
         {interpMode=RCIM_Cubic, tangentMode=RCTM_Break, time=0, value=1}
         {interpMode=RCIM_Cubic, tangentMode=RCTM_Break, time=0.6908, value=1}
         {interpMode=RCIM_Cubic, tangentMode=RCTM_Break, time=1, value=0}
        ]}}})
    VisibilityTag=9
   )
   /SetVariables(Particles.Distortion SpriteSize:Vector2f=<150 150>) {isSetParameters=true, module=SetVariables_#13}
  ]
  rendererClasses [&NiagaraSpriteRendererProperties, &NiagaraLightRendererProperties]
  renderers [
   {rendererClass=&NiagaraSpriteRendererProperties}
   {rendererClass=&NiagaraSpriteRendererProperties, rendererIndex=1}
   {rendererClass=&NiagaraSpriteRendererProperties, rendererIndex=2}
   {rendererClass=&NiagaraSpriteRendererProperties, rendererIndex=3}
   {rendererClass=&NiagaraSpriteRendererProperties, rendererIndex=4}
   {rendererClass=&NiagaraLightRendererProperties, rendererIndex=5}
  ]
 }
 Prop_spawn_merged_Blend_space {
  emitterUpdateScript [
   &EmitterState(
    Inactive Response=::NewEnumerator1
    Life Cycle Mode=::NewEnumerator1
    Loop Behavior=::NewEnumerator1
    Loop Delay=0.05
    Loop Duration=2
    UseLoopDelay=true
   )
   &SpawnBurst_Instantaneous(Spawn Count=5, Use Spawn Probability=false, UseLoopCountLimit=false) {module=SpawnBurst_Instantaneous006}
   &SpawnBurst_Instantaneous(Spawn Group=6, Use Spawn Probability=false, UseLoopCountLimit=false) {module=SpawnBurst_Instantaneous005}
   &SpawnBurst_Instantaneous(Spawn Count=50, Spawn Group=4, Use Spawn Probability=false, UseLoopCountLimit=false) {module=SpawnBurst_Instantaneous004}
   &SpawnBurst_Instantaneous(Spawn Count=50, Spawn Group=3, Use Spawn Probability=false, UseLoopCountLimit=false) {module=SpawnBurst_Instantaneous003}
   &SpawnBurst_Instantaneous(Spawn Count=20, Spawn Group=2, Use Spawn Probability=false) {module=SpawnBurst_Instantaneous002}
   &SpawnBurst_Instantaneous(Spawn Count=1, Spawn Group=7, Use Spawn Probability=false)
   &SpawnBurst_Instantaneous(Spawn Count=35, Use Spawn Probability=false) {module=SpawnBurst_Instantaneous001}
  ]
  particleSpawnScript [
   /SetVariables(Particles.VisibilityTag=&SpawnGroup()) {isSetParameters=true, module=SetVariables_#5}
   /SetVariables(Particles.InitSpriteRotRate=&UniformRangedFloat(Minimum=-1, Maximum=1)) {isSetParameters=true, module=SetVariables_#6}
   &InitializeParticle(
    Color=&MergeHelper_ColorSwitch(
     0=<rgba 0.5255 0.5255 0.5255 1>
     1=<rgba 0.6031 0.5599 0.463 0.2>
     2=<rgba 5 3.7043 1.51 1>
     3=?
     Index=&SpawnGroup()
     Num=::NewEnumerator1
    )
    Lifetime=&MergeHelper_FloatSwitch(
     0=&UniformRangedFloat(Minimum=0.3, Maximum=0.55)
     1=0.25
     2=&UniformRangedFloat(Minimum=0.5, Maximum=1.25)
     3=&UniformRangedFloat(Minimum=0.25, Maximum=0.65)
     4=&UniformRangedFloat(Minimum=0.3, Maximum=0.5)
     5=&RandomRangeFloat(Minimum=0.3, Maximum=0.6)
     6=0.6
     7=0.25
     Index=&SpawnGroup()
     Num=::NewEnumerator6
    )
    Position Offset=<0 15 120>
    Sprite Rotation Mode=::NewEnumerator2
    Sprite Size Mode=::NewEnumerator3
    Uniform Sprite Size=&MergeHelper_FloatSwitch(
     0=&UniformRangedFloat(Minimum=8, Maximum=20)
     1=65
     2=&RandomRangeFloat(Minimum=10, Maximum=25)
     3=?
     Index=&SpawnGroup()
     Num=::NewEnumerator1
    )
    UsePositionOffset=true
    Write Material Random=true
   )
   &ShapeLocation(
    MaskSpawnGroup=false
    Override Seed=false
    Sphere Radius=&MergeHelper_FloatSwitch(0=7, 1=0, 2=4, 3=5, 4=10, 5=0, 6=0, Index=&SpawnGroup(), Num=::NewEnumerator5)
    Sphere Surface Distribution=1
    UseOwnerScale=true
   )
   &AddVelocity(Velocity=&MergeHelper_VectorSwitch(
     0=&Multiply_Vector(
      A=$Particles.ShapeLocation.ShapeNormal
      B=&VectorFromFloat(Value=&RandomRangeFloat(Minimum=15, Maximum=40))
     )
     1=<0 0 0>
     2=&UniformRangedVector(Minimum=<-200 0 -150>, Maximum=<200 0 150>)
     3=&RandomRangeVector(Minimum=<-100 0 20>, Maximum=<100 1 85>)
     4=&RandomRangeVector(Minimum=<-180 -180 50>, Maximum=<180 180 150>)
     5=<0 0 0>
     6=<0 0 0>
     Index=&SpawnGroup()
     Num=::NewEnumerator5
    ))
   /SetVariables(
    Particles.Smoke Velocity:Vector3f=&Multiply_Vector(
     A=$Particles.ShapeLocation.ShapeNormal
     B=&VectorFromFloat(Value=&RandomRangeFloat(Minimum=15, Maximum=50))
    )
    Particles.Sparkles Velocity:Vector3f=&UniformRangedVector(Minimum=<-500 -200 -200>, Maximum=<5000 200 200>)
    Particles.airglow CameraOffset=10
    Particles.smoke CameraOffset:NiagaraFloat=&RandomRangeFloat(Minimum=10, Maximum=25)
   ) {isSetParameters=true, module=SetVariables_#7}
   &AccelerationForce(Acceleration=&MergeHelper_VectorSwitch(
     0=&UniformRangedVector(Minimum=<0 0 -25>, Maximum=<0 0 15>)
     1=<0 0 0>
     2=&RandomRangeVector(Minimum=<0 0 0>, Maximum=<0 0 0>)
     3=?
     Index=&SpawnGroup()
     Num=::NewEnumerator1
    ))
   &ApplyInitialForces()
  ]
  particleUpdateScript [
   &ParticleState()
   &SpriteRotationRate(Rotation Rate=&MergeHelper_FloatSwitch(
     0=&Multiply_Float(
      A=$Particles.InitSpriteRotRate
      B=&FloatFromCurve(
       FloatCurve=DataInterface() {propertyValues {Curve {keys [
           {arriveTangent=-1.9214, interpMode=RCIM_Cubic, leaveTangent=-1.9214, tangentMode=RCTM_User, time=-0.0046, value=1.0113}
           {arriveTangent=-0.2496, interpMode=RCIM_Cubic, leaveTangent=-0.2496, tangentMode=RCTM_User, time=0.9896, value=0.16}
          ]}}}
       CurveIndex=$Emitter.Age
      )
     )
     1=0
     2=5
     3=?
     Index=&SpawnGroup()
     Num=::NewEnumerator1
    ))
   &ScaleSpriteSize(
    Scale Sprite Size Mode=::NewEnumerator0
    Uniform Scale Factor=&MergeHelper_FloatSwitch(
     0=&FloatFromCurve(
      FloatCurve=DataInterface() {propertyValues {Curve {keys [
          {arriveTangent=5.2354, interpMode=RCIM_Cubic, leaveTangent=5.2354, tangentMode=RCTM_User, time=-0.0046, value=0.25}
          <at 0.1439 1.4923>
          <at 1.0011 3.4809 RCIM_Cubic>
         ]}}}
      Scale Curve=&RandomRangeFloat(Minimum=0.5, Maximum=2)
     )
     1=&FloatFromCurve(
      FloatCurve=DataInterface() {propertyValues {Curve {
         defaultValue=1
         keys [{interpMode=RCIM_Cubic, tangentMode=RCTM_User, time=0, value=0}, {interpMode=RCIM_Cubic, tangentMode=RCTM_User, time=1, value=1}]
        }}}
      Scale Curve=5
     )
     2=50
     3=0
     4=&FloatFromCurve(
      FloatCurve=DataInterface() {propertyValues {Curve {keys [<at 0.8 1 RCIM_Cubic>, {arriveTangent=-6.6126, interpMode=RCIM_Cubic, leaveTangent=-6.6126, tangentMode=RCTM_User, time=1, value=0}]}}}
      Scale Curve=0.1
     )
     Index=&SpawnGroup()
     Num=::NewEnumerator3
    )
   )
   &CurlNoiseForce(
    Bias Noise=false
    Coordinate Space:ENiagaraCoordinateSpace=::Local
    Noise Frequency=&MergeHelper_FloatSwitch(0=0, 1=0, 2=0.02, 3=0, 4=0, 5=0, Index=&SpawnGroup(), Num=::NewEnumerator4)
    Noise Quality / Cost=::NewEnumerator0
    Noise Strength=&MergeHelper_FloatSwitch(
     0=0
     1=0
     2=500
     3=0
     4=&FloatFromCurve(FloatCurve=DataInterface() {propertyValues {Curve {keys [<at 0 1>,<at 1 0>]}}}, Scale Curve=400)
     5=0
     Index=&SpawnGroup()
     Num=::NewEnumerator4
    )
    Pan Noise=false
    Random Seed=0
    Randomize Noise Sample=false
    Write to Intrinsic Parameters=false
   )
   &Drag(
    Drag=&MergeHelper_FloatSwitch(
     0=5
     1=50
     2=&FloatFromCurve(FloatCurve=DataInterface() {propertyValues {Curve {keys [<at 0 0.5>,<at 0.5 0.5>,<at 1 1>]}}}, Scale Curve=9)
     3=&FloatFromCurve(FloatCurve=DataInterface() {propertyValues {Curve {keys [<at 0 0.5141>,<at 0.5078 0.7118>,<at 0.9984 0.4918>]}}}, Scale Curve=3)
     4=&FloatFromCurve(FloatCurve=DataInterface() {propertyValues {Curve {keys [<at 0 0.5141>,<at 0.5078 0.7118>,<at 0.9984 0.4918>]}}}, Scale Curve=10)
     Index=&SpawnGroup()
     Num=::NewEnumerator3
    )
    Use Linear Drag=true
    Use Rotational Drag=true
   )
   &SubUVAnimation(
    Output Alternate Renderer Binding=true
    Random Start Frame=true
    "Sprite Renderer "=DataInterface() {propertyValues {SpriteRenderer=&self:Prop_spawn_merged_Blend_space_0.NiagaraSpriteRendererProperties_1}}
    SubUV Lookup Index Scale=0.2
    UseEndFrame=false
    UseStartFrame=false
   )
   &SubUVAnimation(
    Output Alternate Renderer Binding=true
    "Sprite Renderer "=DataInterface() {propertyValues {SpriteRenderer=&self:Prop_spawn_merged_Blend_space_0.NiagaraSpriteRendererProperties_0}}
    UseEndFrame=false
    UseStartFrame=false
   ) {module=SubUVAnimation001}
   &SubUVAnimation(
    Output Alternate Renderer Binding=true
    "Sprite Renderer "=DataInterface() {propertyValues {SpriteRenderer=&self:Prop_spawn_merged_Blend_space_0.NiagaraSpriteRendererProperties_4}}
    UseEndFrame=false
    UseStartFrame=false
   ) {module=SubUVAnimation002}
   /SetVariables(
    Particles.Smoke Color:LinearColor=&MakeLinearColorFromVectorAndFloat(
     "Vector (RGB)"=&MakeVectorFromLinearColorRGB(Color=&ColorFromCurve(ColorCurve=DataInterface() {propertyValues {
         AlphaCurve {keys [<at 0 1 RCIM_Cubic>,<at 1 0 RCIM_Cubic>]}
         BlueCurve {keys [<at 0 0.0282 RCIM_Cubic>,<at 1 0 RCIM_Cubic>]}
         GreenCurve {keys [<at 0 0.071 RCIM_Cubic>,<at 1 0 RCIM_Cubic>]}
         RedCurve {keys [<at 0 0.2281 RCIM_Cubic>,<at 1 0 RCIM_Cubic>]}
        }}))
     "Float (Alpha)"=&FloatFromCurve(FloatCurve=DataInterface() {propertyValues {Curve {
         defaultValue=1
         keys [
          {arriveTangent=9.0725, interpMode=RCIM_Cubic, leaveTangent=9.0725, tangentMode=RCTM_User, time=0, value=0}
          {arriveTangent=-2.5249, interpMode=RCIM_Cubic, leaveTangent=-2.5249, tangentMode=RCTM_User, time=0.0826, value=1.0001}
          {arriveTangent=-0.0405, interpMode=RCIM_Cubic, leaveTangent=-0.0405, tangentMode=RCTM_User, time=1, value=0}
         ]
        }}})
    )
    Particles.Smoke Dyn Param:Vector4f=&MakeVector4(X=0, Y=0, Z=0, W=0)
   ) {isSetParameters=true, module=SetVariables_#8}
   /SetVariables(
    Particles.Soft Glow sprite size:Vector2f=&Vector2DFromFloat(Value=&FloatFromCurve(
      FloatCurve=DataInterface() {propertyValues {Curve {
         defaultValue=1
         keys [
          {interpMode=RCIM_Cubic, tangentMode=RCTM_User, time=0, value=0}
          {arriveTangent=0.24, interpMode=RCIM_Cubic, leaveTangent=0.24, tangentMode=RCTM_User, time=0.1, value=1}
         ]
        }}}
      Scale Curve=50
     ))
    Particles.soft glow Color:LinearColor=&MakeLinearColorFromVectorAndFloat(
     "Vector (RGB)"=&MakeVectorFromLinearColorRGB(Color=<rgba 5 0.4905 0 1>)
     "Float (Alpha)"=&FloatFromCurve(FloatCurve=DataInterface() {propertyValues {Curve {keys [<at 0 1>,<at 1 0>]}}}, Scale Curve=0.3)
    )
   ) {isSetParameters=true, module=SetVariables_#9}
   /SetVariables(
    Particles.Embers Color:LinearColor=&MakeLinearColorFromVectorAndFloat(
     "Vector (RGB)"=&MakeVectorFromLinearColorRGB(Color=&ColorFromCurve(ColorCurve=DataInterface() {propertyValues {
         AlphaCurve {keys [<at 0 1 RCIM_Cubic>]}
         BlueCurve {keys [<at 0 7.8125 RCIM_Cubic>,<at 0.2499 0.7812>,<at 0.7114 0.2083 RCIM_Cubic>]}
         GreenCurve {keys [<at 0 34.1758 RCIM_Cubic>,<at 0.2499 4.8887>,<at 0.7114 0.8934 RCIM_Cubic>]}
         RedCurve {keys [<at 0 50 RCIM_Cubic>,<at 0.2499 10>,<at 0.7114 4 RCIM_Cubic>]}
        }}))
     "Float (Alpha)"=1
    )
    Particles.Embers SpriteSize:Vector2f=&Multiply_Vector2D_ByFloat(
     Float=&FloatFromCurve(
      FloatCurve=DataInterface() {propertyValues {Curve {
         defaultValue=1
         keys [{interpMode=RCIM_Cubic, tangentMode=RCTM_User, time=0, value=0}, {interpMode=RCIM_Cubic, tangentMode=RCTM_User, time=0.1511, value=1}]
        }}}
      Scale Curve=1.2
     )
     Vector2D=&RandomRangeVector2D(Minimum=<0.4 0.8>, Maximum=<0.5 1.5>)
    )
   ) {isSetParameters=true, module=SetVariables_#10}
   /SetVariables(
    Particles.Ashes Color:LinearColor=&MakeLinearColorFromVectorAndFloat(
     "Vector (RGB)"=&MakeVectorFromLinearColorRGB(Color=&ColorFromCurve(ColorCurve=DataInterface() {propertyValues {
         AlphaCurve {keys [<at 0 1 RCIM_Cubic>]}
         BlueCurve {keys [<at 0 1.473 RCIM_Cubic>,<at 0.6088 0.0879 RCIM_Cubic>]}
         GreenCurve {keys [<at 0 11.4982 RCIM_Cubic>,<at 0.6088 0.0919 RCIM_Cubic>]}
         RedCurve {keys [<at 0 50 RCIM_Cubic>,<at 0.6088 0.099 RCIM_Cubic>]}
        }}))
     "Float (Alpha)"=&FloatFromCurve(FloatCurve=DataInterface() {propertyValues {Curve {keys [<at 0.8 1 RCIM_Cubic>,<at 1 0>]}}})
    )
    Particles.Ashes SpriteSize:Vector2f=&Vector2DFromFloat(Value=&RandomRangeFloat(Minimum=1, Maximum=7))
    Particles.Vector 4:Vector4f=&MakeVector4(X=0, Y=0, Z=0, W=0)
   ) {isSetParameters=true, module=SetVariables_#11}
   /SetVariables(
    Particles.Fire Color:LinearColor=&MakeLinearColorFromVectorAndFloat(
     "Vector (RGB)"=&Multiply_VectorByFloat(
      Vector=&MakeVectorFromLinearColorRGB(Color=<rgba 20 1.962 0 1>)
      Float=&FloatFromCurve(FloatCurve=DataInterface() {propertyValues {Curve {
          defaultValue=1
          keys [{interpMode=RCIM_Cubic, tangentMode=RCTM_User, time=0, value=1}, {interpMode=RCIM_Cubic, tangentMode=RCTM_User, time=1, value=0}]
         }}})
     )
     "Float (Alpha)"=&FloatFromCurve(
      FloatCurve=DataInterface() {propertyValues {Curve {
         defaultValue=1
         keys [{interpMode=RCIM_Cubic, tangentMode=RCTM_User, time=0, value=1}, {interpMode=RCIM_Cubic, tangentMode=RCTM_User, time=1, value=0}]
        }}}
      Scale Curve=&FNTimeOfDayScalar(Night Scale=3, Morning Scale=1, Day Scale=1.5, Evening Scale=1)
     )
    )
    Particles.Fire DynParam:Vector4f=&MakeVector4(
     X=&FloatFromCurve(FloatCurve=DataInterface() {propertyValues {Curve {keys [<at 0 1>]}}})
     Y=&RandomRangeFloat(Minimum=0, Maximum=4)
     Z=&FloatFromCurve(FloatCurve=DataInterface() {propertyValues {Curve {keys [<at 0 0.5 RCIM_Cubic>, {arriveTangent=1.5, interpMode=RCIM_Cubic, leaveTangent=1.5, time=0.5235, value=0.5}, <at 1 2>]}}})
     W=2.5
    )
    Particles.fire SpriteSize:Vector2f=&Multiply_Vector2D_ByFloat(
     Float=&FloatFromCurve(FloatCurve=DataInterface() {propertyValues {Curve {keys [
          {interpMode=RCIM_Cubic, tangentMode=RCTM_Break, time=0.0078, value=0}
          {interpMode=RCIM_Cubic, tangentMode=RCTM_Break, time=0.0578, value=2}
          {arriveTangent=-1.5342, interpMode=RCIM_Cubic, leaveTangent=-1.5342, time=0.1668, value=0.9624}
          {arriveTangent=-1.155, interpMode=RCIM_Cubic, leaveTangent=-1.155, time=0.7786, value=0.8942}
          <at 1 0>
         ]}}})
     Vector2D=&RandomRangeVector2D(Minimum=<10 10>, Maximum=<35 35>)
    )
   ) {isSetParameters=true, module=SetVariables_#12}
   &GravityForce(Gravity=&MergeHelper_VectorSwitch(
     0=<0 0 0>
     1=<0 0 0>
     2=&MakeVector(
      X=0
      Y=0
      Z=&FloatFromCurve(
       FloatCurve=DataInterface() {propertyValues {Curve {keys [<at 0 0.25 RCIM_Cubic>,<at 0.25 1>]}}}
       CurveIndex=$Emitter.Age
       Scale Curve=&RandomRangeFloat(Minimum=-60, Maximum=-200)
      )
     )
     3=<0 0 5>
     4=&RandomRangeVector(Minimum=<0 0 -10>, Maximum=<1 1 50>)
     5=<0 0 0>
     6=<0 0 0>
     Index=&SpawnGroup()
     Num=::NewEnumerator5
    ))
   &SolveForcesAndVelocity(Clamp Velocity=false, Limit Acceleration=false)
   &ScaleSpriteSizeBySpeed(
    Max Scale Factor=&MergeHelper_Vector2Switch(0=<1 1>, 1=<0 0>, 2=<0 0>, 3=<1 1>, 4=<1 5>, Index=&SpawnGroup(), Num=::NewEnumerator3)
    Min Scale Factor=&MergeHelper_Vector2Switch(0=<1 1>, 1=<1 1>, 2=<1 1>, 3=<1 1>, 4=<0 0>, Index=&SpawnGroup(), Num=::NewEnumerator3)
    Sample Scale Factor By Curve=true
    Velocity Threshold=500
   )
   &FN_Light_Attributes(
    Color=<rgba 15 4.9774 1.3498 1>
    Create Color Attribute=true
    Create Exponent Attribute=true
    Create Position Attribute=true
    Create Radius Attribute=true
    Enable Particle Light Enable=true
    Enable Visibility Tag=true
    Light Enabled=true
    Position=<0 0 125>
    Radius=&FloatFromCurve(
     FloatCurve=DataInterface() {propertyValues {Curve {
        defaultValue=1
        keys [{interpMode=RCIM_Cubic, tangentMode=RCTM_User, time=0, value=0}, {interpMode=RCIM_Cubic, tangentMode=RCTM_User, time=0.3, value=1}]
       }}}
     Scale Curve=35
    )
    Scale Color=&FloatFromCurve(
     FloatCurve=DataInterface() {propertyValues {Curve {keys [
         {interpMode=RCIM_Cubic, tangentMode=RCTM_Break, time=0, value=1}
         {interpMode=RCIM_Cubic, tangentMode=RCTM_Break, time=0.6908, value=1}
         {interpMode=RCIM_Cubic, tangentMode=RCTM_Break, time=1, value=0}
        ]}}}
     Scale Curve=5
    )
    VisibilityTag=5
   )
   &SubUVAnimation(
    Output Alternate Renderer Binding=true
    "Sprite Renderer "=DataInterface() {propertyValues {SpriteRenderer=&self:Prop_spawn_merged_Blend_space_0.NiagaraSpriteRendererProperties_2}}
    UseEndFrame=false
    UseStartFrame=false
   ) {module=SubUVAnimation003}
  ]
  rendererClasses [&NiagaraSpriteRendererProperties, &NiagaraLightRendererProperties]
  renderers [
   {rendererClass=&NiagaraSpriteRendererProperties}
   {rendererClass=&NiagaraSpriteRendererProperties, rendererIndex=1}
   {rendererClass=&NiagaraSpriteRendererProperties, rendererIndex=2}
   {rendererClass=&NiagaraSpriteRendererProperties, rendererIndex=3}
   {rendererClass=&NiagaraSpriteRendererProperties, rendererIndex=4}
   {rendererClass=&NiagaraLightRendererProperties, rendererIndex=5}
  ]
 }
 Prop_spawn_merged_Blend_space001 {
  emitterUpdateScript [
   &EmitterState(Life Cycle Mode=::NewEnumerator1, Loop Behavior=::NewEnumerator1, Loop Delay=2.25, Loop Duration=2.7, UseLoopDelay=true)
   &SpawnRate(SpawnGroup=5, SpawnRate=100, Use Spawn Probability=false)
   &SpawnBurst_Instantaneous(Spawn Count=5, Use Spawn Probability=false, UseLoopCountLimit=false) {module=SpawnBurst_Instantaneous006}
   &SpawnBurst_Instantaneous(Spawn Group=6, Use Spawn Probability=false, UseLoopCountLimit=false) {module=SpawnBurst_Instantaneous005}
   &SpawnBurst_Instantaneous(Spawn Count=50, Spawn Group=4, Use Spawn Probability=false, UseLoopCountLimit=false) {module=SpawnBurst_Instantaneous004}
   &SpawnBurst_Instantaneous(Spawn Count=35, Spawn Group=3, Use Spawn Probability=false, UseLoopCountLimit=false) {module=SpawnBurst_Instantaneous003}
   &SpawnBurst_Instantaneous(Spawn Count=20, Spawn Group=2, Use Spawn Probability=false) {module=SpawnBurst_Instantaneous002}
   &SpawnBurst_Instantaneous(Spawn Count=1, Spawn Group=1, Use Spawn Probability=false)
   &SpawnBurst_Instantaneous(Spawn Count=35, Use Spawn Probability=false) {module=SpawnBurst_Instantaneous001}
  ]
  particleSpawnScript [
   /SetVariables(Particles.VisibilityTag=&SpawnGroup()) {isSetParameters=true, module=SetVariables_#5}
   /SetVariables(Particles.InitSpriteRotRate=&UniformRangedFloat(Minimum=-1, Maximum=1)) {isSetParameters=true, module=SetVariables_#6}
   &InitializeParticle(
    Color=&MergeHelper_ColorSwitch(
     0=<rgba 0.5255 0.5255 0.5255 1>
     1=<rgba 0.6031 0.5599 0.463 0.2>
     2=<rgba 5 3.7043 1.51 1>
     3=?
     Index=&SpawnGroup()
     Num=::NewEnumerator1
    )
    Lifetime=&MergeHelper_FloatSwitch(
     0=&UniformRangedFloat(Minimum=0.3, Maximum=0.55)
     1=0.25
     2=&UniformRangedFloat(Minimum=0.5, Maximum=1.25)
     3=&UniformRangedFloat(Minimum=0.25, Maximum=0.75)
     4=&UniformRangedFloat(Minimum=0.3, Maximum=0.9)
     5=&RandomRangeFloat(Minimum=0.3, Maximum=0.6)
     6=0.4
     Index=&SpawnGroup()
     Num=::NewEnumerator5
    )
    Position Offset=<0 -25 185>
    Sprite Rotation Mode=::NewEnumerator2
    Sprite Size Mode=::NewEnumerator3
    Uniform Sprite Size=&MergeHelper_FloatSwitch(
     0=&UniformRangedFloat(Minimum=8, Maximum=20)
     1=65
     2=&RandomRangeFloat(Minimum=10, Maximum=25)
     3=?
     Index=&SpawnGroup()
     Num=::NewEnumerator1
    )
    UsePositionOffset=true
    Write Material Random=true
   )
   &ShapeLocation(
    MaskSpawnGroup=false
    Override Seed=false
    Sphere Radius=&MergeHelper_FloatSwitch(0=7, 1=0, 2=4, 3=5, 4=10, 5=0, Index=&SpawnGroup(), Num=::NewEnumerator4)
    Sphere Surface Distribution=1
    UseOwnerScale=true
   )
   &AddVelocity(Velocity=&MergeHelper_VectorSwitch(
     0=&Multiply_Vector(
      A=$Particles.ShapeLocation.ShapeNormal
      B=&VectorFromFloat(Value=&RandomRangeFloat(Minimum=15, Maximum=40))
     )
     1=<0 0 0>
     2=&UniformRangedVector(Minimum=<-200 0 -110>, Maximum=<200 0 110>)
     3=&RandomRangeVector(Minimum=<-100 0 -50>, Maximum=<100 1 150>)
     4=&RandomRangeVector(Minimum=<-180 -180 50>, Maximum=<180 180 150>)
     5=<0 0 0>
     6=<0 0 0>
     Index=&SpawnGroup()
     Num=::NewEnumerator5
    ))
   /SetVariables(
    Particles.Smoke Velocity:Vector3f=&Multiply_Vector(
     A=$Particles.ShapeLocation.ShapeNormal
     B=&VectorFromFloat(Value=&RandomRangeFloat(Minimum=15, Maximum=50))
    )
    Particles.Sparkles Velocity:Vector3f=&UniformRangedVector(Minimum=<-500 -200 -200>, Maximum=<5000 200 200>)
    Particles.airglow CameraOffset=10
    Particles.smoke CameraOffset:NiagaraFloat=&RandomRangeFloat(Minimum=10, Maximum=25)
   ) {isSetParameters=true, module=SetVariables_#7}
   &AccelerationForce(Acceleration=&MergeHelper_VectorSwitch(
     0=&UniformRangedVector(Minimum=<0 0 -25>, Maximum=<0 0 -75>)
     1=<0 0 0>
     2=&RandomRangeVector(Minimum=<0 0 0>, Maximum=<0 0 0>)
     3=?
     Index=&SpawnGroup()
     Num=::NewEnumerator1
    ))
   &ApplyInitialForces()
   &ShapeLocation(
    Disc Coverage=0.3738
    MaskSpawnGroup=true
    Offset=<0 0 10>
    Offset Mode=::NewEnumerator0
    Override Seed=false
    Ring Radius=30
    Rotation Axis=<0 1 0>
    Rotation Mode=::NewEnumerator1
    Shape Primitive=::NewEnumerator4
    Spawn Group Mask=5
    UseOwnerScale=true
   ) {module=ShapeLocation001}
  ]
  particleUpdateScript [
   &ParticleState()
   &SpriteRotationRate(Rotation Rate=&MergeHelper_FloatSwitch(
     0=&Multiply_Float(
      A=$Particles.InitSpriteRotRate
      B=&FloatFromCurve(
       FloatCurve=DataInterface() {propertyValues {Curve {keys [
           {arriveTangent=-1.9214, interpMode=RCIM_Cubic, leaveTangent=-1.9214, tangentMode=RCTM_User, time=-0.0046, value=1.0113}
           {arriveTangent=-0.2496, interpMode=RCIM_Cubic, leaveTangent=-0.2496, tangentMode=RCTM_User, time=0.9896, value=0.16}
          ]}}}
       CurveIndex=$Emitter.Age
      )
     )
     1=0
     2=5
     3=?
     Index=&SpawnGroup()
     Num=::NewEnumerator1
    ))
   &ScaleSpriteSize(
    Scale Sprite Size Mode=::NewEnumerator0
    Uniform Scale Factor=&MergeHelper_FloatSwitch(
     0=&FloatFromCurve(
      FloatCurve=DataInterface() {propertyValues {Curve {keys [
          {arriveTangent=5.2354, interpMode=RCIM_Cubic, leaveTangent=5.2354, tangentMode=RCTM_User, time=-0.0046, value=0.25}
          <at 0.1439 1.4923>
          <at 1.0011 3.4809 RCIM_Cubic>
         ]}}}
      Scale Curve=&RandomRangeFloat(Minimum=0.5, Maximum=3)
     )
     1=&FloatFromCurve(FloatCurve=DataInterface() {propertyValues {Curve {
         defaultValue=1
         keys [{interpMode=RCIM_Cubic, tangentMode=RCTM_User, time=0, value=0}, {interpMode=RCIM_Cubic, tangentMode=RCTM_User, time=1, value=1}]
        }}})
     2=50
     3=0
     4=&FloatFromCurve(
      FloatCurve=DataInterface() {propertyValues {Curve {keys [<at 0.8 1 RCIM_Cubic>, {arriveTangent=-6.6126, interpMode=RCIM_Cubic, leaveTangent=-6.6126, tangentMode=RCTM_User, time=1, value=0}]}}}
      Scale Curve=0.1
     )
     Index=&SpawnGroup()
     Num=::NewEnumerator3
    )
   )
   &CurlNoiseForce(
    Bias Noise=false
    Coordinate Space:ENiagaraCoordinateSpace=::Local
    Noise Frequency=&MergeHelper_FloatSwitch(0=0, 1=0, 2=0.02, 3=0, 4=0, 5=0, Index=&SpawnGroup(), Num=::NewEnumerator4)
    Noise Quality / Cost=::NewEnumerator0
    Noise Strength=&MergeHelper_FloatSwitch(
     0=0
     1=0
     2=500
     3=0
     4=&FloatFromCurve(FloatCurve=DataInterface() {propertyValues {Curve {keys [<at 0 1>,<at 1 0>]}}}, Scale Curve=3000)
     5=0
     Index=&SpawnGroup()
     Num=::NewEnumerator4
    )
    Pan Noise=false
    Random Seed=0
    Randomize Noise Sample=false
    Write to Intrinsic Parameters=false
   )
   &Drag(
    Drag=&MergeHelper_FloatSwitch(
     0=5
     1=0
     2=&FloatFromCurve(FloatCurve=DataInterface() {propertyValues {Curve {keys [<at 0 0.5>,<at 0.5 0.5>,<at 1 1>]}}}, Scale Curve=9)
     3=&FloatFromCurve(FloatCurve=DataInterface() {propertyValues {Curve {keys [<at 0 0.5>,<at 0.5 0.5>,<at 1 1>]}}}, Scale Curve=9)
     4=&FloatFromCurve(FloatCurve=DataInterface() {propertyValues {Curve {keys [<at 0 0.5141>,<at 0.5078 0.7118>,<at 0.9984 0.4918>]}}}, Scale Curve=10)
     Index=&SpawnGroup()
     Num=::NewEnumerator3
    )
    Use Linear Drag=true
    Use Rotational Drag=true
   )
   &SubUVAnimation(
    Output Alternate Renderer Binding=true
    Random Start Frame=true
    "Sprite Renderer "=DataInterface() {propertyValues {SpriteRenderer=&self:Prop_spawn_merged_Blend_space001_0.NiagaraSpriteRendererProperties_1}}
    SubUV Lookup Index Scale=0.2
    UseEndFrame=false
    UseStartFrame=false
   )
   &SubUVAnimation(
    Output Alternate Renderer Binding=true
    "Sprite Renderer "=DataInterface() {propertyValues {SpriteRenderer=&self:Prop_spawn_merged_Blend_space001_0.NiagaraSpriteRendererProperties_0}}
    UseEndFrame=false
    UseStartFrame=false
   ) {module=SubUVAnimation001}
   &SubUVAnimation(
    Output Alternate Renderer Binding=true
    "Sprite Renderer "=DataInterface() {propertyValues {SpriteRenderer=&self:Prop_spawn_merged_Blend_space001_0.NiagaraSpriteRendererProperties_4}}
    UseEndFrame=false
    UseStartFrame=false
   ) {module=SubUVAnimation002}
   /SetVariables(
    Particles.Smoke Color:LinearColor=&MakeLinearColorFromVectorAndFloat(
     "Vector (RGB)"=&MakeVectorFromLinearColorRGB(Color=<rgba 0.0656 0.0656 0.0656 1>)
     "Float (Alpha)"=&FloatFromCurve(FloatCurve=DataInterface() {propertyValues {Curve {
         defaultValue=1
         keys [
          {arriveTangent=9.0725, interpMode=RCIM_Cubic, leaveTangent=9.0725, tangentMode=RCTM_User, time=0, value=0}
          {arriveTangent=-2.5249, interpMode=RCIM_Cubic, leaveTangent=-2.5249, tangentMode=RCTM_User, time=0.0826, value=1.0001}
          {arriveTangent=-0.0405, interpMode=RCIM_Cubic, leaveTangent=-0.0405, tangentMode=RCTM_User, time=1, value=0}
         ]
        }}})
    )
    Particles.Smoke Dyn Param:Vector4f=&MakeVector4(X=0, Y=0, Z=0, W=0)
   ) {isSetParameters=true, module=SetVariables_#8}
   /SetVariables(Particles.soft glow Color:LinearColor=&MakeLinearColorFromVectorAndFloat(
     "Vector (RGB)"=&MakeVectorFromLinearColorRGB(Color=<rgba 1 0.6731 0.3597 1>)
     "Float (Alpha)"=&FloatFromCurve(FloatCurve=DataInterface() {propertyValues {Curve {keys [<at 0 1>,<at 1 0>]}}}, Scale Curve=0.5)
    )) {isSetParameters=true, module=SetVariables_#9}
   /SetVariables(
    Particles.Embers Color:LinearColor=&MakeLinearColorFromVectorAndFloat(
     "Vector (RGB)"=&MakeVectorFromLinearColorRGB(Color=&ColorFromCurve(ColorCurve=DataInterface() {propertyValues {
         AlphaCurve {keys [<at 0 1 RCIM_Cubic>]}
         BlueCurve {keys [<at 0 7.8125 RCIM_Cubic>,<at 0.2499 0.7812>,<at 0.7114 0.2083 RCIM_Cubic>]}
         GreenCurve {keys [<at 0 34.1758 RCIM_Cubic>,<at 0.2499 4.8887>,<at 0.7114 0.8934 RCIM_Cubic>]}
         RedCurve {keys [<at 0 50 RCIM_Cubic>,<at 0.2499 10>,<at 0.7114 4 RCIM_Cubic>]}
        }}))
     "Float (Alpha)"=1
    )
    Particles.Embers SpriteSize:Vector2f=&Multiply_Vector2D_ByFloat(
     Float=&FloatFromCurve(
      FloatCurve=DataInterface() {propertyValues {Curve {
         defaultValue=1
         keys [{interpMode=RCIM_Cubic, tangentMode=RCTM_User, time=0, value=0}, {interpMode=RCIM_Cubic, tangentMode=RCTM_User, time=0.1511, value=1}]
        }}}
      Scale Curve=1.2
     )
     Vector2D=&RandomRangeVector2D(Minimum=<0.4 0.8>, Maximum=<0.5 1.5>)
    )
   ) {isSetParameters=true, module=SetVariables_#10}
   /SetVariables(
    Particles.Ashes Color:LinearColor=&MakeLinearColorFromVectorAndFloat(
     "Vector (RGB)"=&MakeVectorFromLinearColorRGB(Color=&ColorFromCurve(ColorCurve=DataInterface() {propertyValues {
         AlphaCurve {keys [<at 0 1 RCIM_Cubic>]}
         BlueCurve {keys [<at 0 1.473 RCIM_Cubic>,<at 0.6088 0.0879 RCIM_Cubic>]}
         GreenCurve {keys [<at 0 11.4982 RCIM_Cubic>,<at 0.6088 0.0919 RCIM_Cubic>]}
         RedCurve {keys [<at 0 50 RCIM_Cubic>,<at 0.6088 0.099 RCIM_Cubic>]}
        }}))
     "Float (Alpha)"=&FloatFromCurve(FloatCurve=DataInterface() {propertyValues {Curve {keys [<at 0.8 1 RCIM_Cubic>,<at 1 0>]}}})
    )
    Particles.Ashes SpriteSize:Vector2f=&Vector2DFromFloat(Value=&RandomRangeFloat(Minimum=1, Maximum=7))
    Particles.Vector 4:Vector4f=&MakeVector4(X=0, Y=0, Z=0, W=0)
   ) {isSetParameters=true, module=SetVariables_#11}
   /SetVariables(
    Particles.Fire Color:LinearColor=&MakeLinearColorFromVectorAndFloat(
     "Vector (RGB)"=&Multiply_VectorByFloat(
      Vector=&MakeVectorFromLinearColorRGB(Color=<rgba 20 1.962 0 1>)
      Float=&FloatFromCurve(FloatCurve=DataInterface() {propertyValues {Curve {
          defaultValue=1
          keys [{interpMode=RCIM_Cubic, tangentMode=RCTM_User, time=0, value=1}, {interpMode=RCIM_Cubic, tangentMode=RCTM_User, time=1, value=0}]
         }}})
     )
     "Float (Alpha)"=1
    )
    Particles.Fire DynParam:Vector4f=&MakeVector4(
     X=&FloatFromCurve(FloatCurve=DataInterface() {propertyValues {Curve {keys [<at 0 1>]}}})
     Y=0.5
     Z=&FloatFromCurve(FloatCurve=DataInterface() {propertyValues {Curve {keys [<at 0 0.5 RCIM_Cubic>, {arriveTangent=1.5, interpMode=RCIM_Cubic, leaveTangent=1.5, time=0.5235, value=0.5}, <at 1 2>]}}})
     W=0
    )
    Particles.fire SpriteSize:Vector2f=&Multiply_Vector2D_ByFloat(
     Float=&FloatFromCurve(FloatCurve=DataInterface() {propertyValues {Curve {keys [
          {interpMode=RCIM_Cubic, tangentMode=RCTM_Break, time=0.0079, value=0}
          {interpMode=RCIM_Cubic, tangentMode=RCTM_Break, time=0.0578, value=2}
          {arriveTangent=-1.5342, interpMode=RCIM_Cubic, leaveTangent=-1.5342, time=0.1668, value=0.9624}
          {arriveTangent=-1.155, interpMode=RCIM_Cubic, leaveTangent=-1.155, time=0.7786, value=0.8942}
          <at 1 0>
         ]}}})
     Vector2D=&RandomRangeVector2D(Minimum=<20 0>, Maximum=<35 35>)
    )
   ) {isSetParameters=true, module=SetVariables_#12}
   &GravityForce(Gravity=&MergeHelper_VectorSwitch(
     0=<0 0 0>
     1=<0 0 0>
     2=&MakeVector(
      X=0
      Y=0
      Z=&FloatFromCurve(
       FloatCurve=DataInterface() {propertyValues {Curve {keys [<at 0 0.25 RCIM_Cubic>,<at 0.25 1>]}}}
       CurveIndex=$Emitter.Age
       Scale Curve=&RandomRangeFloat(Minimum=-60, Maximum=-170)
      )
     )
     3=<0 0 5>
     4=&RandomRangeVector(Minimum=<0 0 -10>, Maximum=<1 1 50>)
     5=<0 0 0>
     Index=&SpawnGroup()
     Num=::NewEnumerator4
    ))
   &SolveForcesAndVelocity(Clamp Velocity=false, Limit Acceleration=false)
   &ScaleSpriteSizeBySpeed(
    Max Scale Factor=&MergeHelper_Vector2Switch(0=<1 1>, 1=<0 0>, 2=<0 0>, 3=<0 0>, 4=<1 5>, Index=&SpawnGroup(), Num=::NewEnumerator3)
    Min Scale Factor=&MergeHelper_Vector2Switch(0=<1 1>, 1=<0 0>, 2=<0 0>, 3=<0 0>, 4=<0 0>, Index=&SpawnGroup(), Num=::NewEnumerator3)
    Sample Scale Factor By Curve=true
    Velocity Threshold=500
   )
   &FN_Light_Attributes(
    Color=<rgba 15 4.9774 1.3498 1>
    Create Color Attribute=true
    Create Exponent Attribute=true
    Create Position Attribute=true
    Create Radius Attribute=true
    Enable Particle Light Enable=true
    Enable Visibility Tag=true
    Light Enabled=true
    Position=<0 0 180>
    Radius=&FloatFromCurve(
     FloatCurve=DataInterface() {propertyValues {Curve {
        defaultValue=1
        keys [{interpMode=RCIM_Cubic, tangentMode=RCTM_User, time=0, value=0}, {interpMode=RCIM_Cubic, tangentMode=RCTM_User, time=0.3, value=1}]
       }}}
     Scale Curve=35
    )
    Scale Color=&FloatFromCurve(FloatCurve=DataInterface() {propertyValues {Curve {keys [
         {interpMode=RCIM_Cubic, tangentMode=RCTM_Break, time=0, value=1}
         {interpMode=RCIM_Cubic, tangentMode=RCTM_Break, time=0.6908, value=1}
         {interpMode=RCIM_Cubic, tangentMode=RCTM_Break, time=1, value=0}
        ]}}})
    VisibilityTag=10
   )
   /SetVariables(Particles.Distortion SpriteSize:Vector2f=<150 150>) {isSetParameters=true, module=SetVariables_#13}
   &SubUVAnimation(
    Output Alternate Renderer Binding=true
    "Sprite Renderer "=DataInterface() {propertyValues {SpriteRenderer=&self:Prop_spawn_merged_Blend_space001_0.NiagaraSpriteRendererProperties_6}}
    UseEndFrame=false
    UseStartFrame=false
   ) {module=SubUVAnimation003}
  ]
  rendererClasses [&NiagaraSpriteRendererProperties, &NiagaraLightRendererProperties]
  renderers [
   {rendererClass=&NiagaraSpriteRendererProperties}
   {rendererClass=&NiagaraSpriteRendererProperties, rendererIndex=1}
   {rendererClass=&NiagaraSpriteRendererProperties, rendererIndex=2}
   {rendererClass=&NiagaraSpriteRendererProperties, rendererIndex=3}
   {rendererClass=&NiagaraSpriteRendererProperties, rendererIndex=4}
   {rendererClass=&NiagaraSpriteRendererProperties, rendererIndex=5}
   {rendererClass=&NiagaraLightRendererProperties, rendererIndex=6}
  ]
 }
 Sparks001 {
  simTarget=GPUComputeSim
  emitterUpdateScript [
   &EmitterState(Life Cycle Mode=::NewEnumerator1, Loop Behavior=::NewEnumerator1, Loop Duration=3.8, UseLoopDelay=false)
   &SpawnRate(SpawnGroup=1, SpawnRate=50, Use Spawn Probability=false) {module=SpawnRate001}
   &SpawnRate(SpawnRate=80, Use Spawn Probability=false)
  ]
  particleSpawnScript [
   &InitializeParticle(
    Color=<rgba 10 1.0749 0 1>
    Lifetime Max=0.3
    Lifetime Min=0.6
    Lifetime Mode=::NewEnumerator1
    Sprite Size Max=<2.5 2.5>
    Sprite Size Min=<1 0>
    Sprite Size Mode=::NewEnumerator2
    UsePositionOffset=false
    Write Material Random=true
   )
   &ShapeLocation(Box Size=<10 10 1>, MaskSpawnGroup=false, Override Seed=false, Shape Primitive=::NewEnumerator2, Surface Only=false, UseOwnerScale=true) {module=ShapeLocation001}
   &ApplyInitialForces()
   &FN_SetSpawnGroup(Number of Spawn Groups=3)
   &SkeletalMeshLocation(
    Bone / Socket Index=&SpawnGroup()
    Bone Sampling Mode=::NewEnumerator4
    MaskSpawnGroup=false
    Skeletal Mesh=$System.Skeletal Mesh Reader
   )
  ]
  particleUpdateScript [
   &ParticleState()
   &Color(Color=&ColorFromCurve(ColorCurve=DataInterface() {propertyValues {
       AlphaCurve {keys [<at 0 1 RCIM_Cubic>,<at 1 0 RCIM_Cubic>]}
       BlueCurve {keys [<at 0 0 RCIM_Cubic>,<at 0.1147 0>,<at 0.5143 0 RCIM_Cubic>]}
       GreenCurve {keys [<at 0 9.1707 RCIM_Cubic>,<at 0.1147 3.0139>,<at 0.5143 0.0332 RCIM_Cubic>]}
       RedCurve {keys [<at 0 20 RCIM_Cubic>,<at 0.1147 10>,<at 0.5143 1 RCIM_Cubic>]}
      }}))
   /SetVariables(Particles.Glow03_Color=&MultiplyLinearColor(Linear Color=$Particles.Color, Scale Factor=<1 1 1 0.04>)) {isSetParameters=true}
   /SetVariables(Particles.Glow03_Size=&Multiply_Vector2D_ByFloat(
     Float=&FloatFromCurve(
      FloatCurve=DataInterface() {propertyValues {Curve {keys [
          <at 0 0 RCIM_Cubic>
          <at 0.2 1.3 RCIM_Cubic>
          {arriveTangent=-0.1423, interpMode=RCIM_Cubic, leaveTangent=-0.1423, tangentMode=RCTM_User, time=1, value=0}
         ]}}}
      Scale Curve=15
     )
     Vector2D=<2 2>
    )) {isSetParameters=true}
   &ScaleColor(
    Scale Alpha=&FloatFromCurve(FloatCurve=DataInterface() {propertyValues {Curve {keys [<at 0.3 1>,<at 0.4 1>,<at 1 1>]}}})
    ScaleA=true
    ScaleRGB=true
   )
   &ScaleSpriteSize(Scale Factor=&Vector2DFromFloat(Value=&FloatFromCurve(FloatCurve=DataInterface() {propertyValues {Curve {keys [
          <at 0 0 RCIM_Cubic>
          {interpMode=RCIM_Cubic, leaveTangent=-3.0296, tangentMode=RCTM_Break, time=0.2, value=1.2}
          {arriveTangent=-1.8107, interpMode=RCIM_Cubic, leaveTangent=-1.8107, tangentMode=RCTM_User, time=1, value=0}
         ]}}}))) {module=ScaleSpriteSize001}
   &GravityForce(Gravity=<0 0 -100>)
   &CurlNoiseForce(
    Noise Frequency=100
    Noise Strength=&Multiply_Float(A=233, B=1.5)
    Pan Noise=true
    Pan Noise Field=<100 0 1>
    Random Seed=10
   )
   &Drag(Drag=&RandomRangeFloat(Minimum=1, Maximum=5), Use Linear Drag=true, Use Rotational Drag=false)
   &SolveForcesAndVelocity(Clamp Velocity=false, Limit Acceleration=false)
   &ScaleSpriteSizeBySpeed(Max Scale Factor=<1 4>, Velocity Threshold=356)
  ]
  rendererClasses [&NiagaraSpriteRendererProperties]
  renderers [
   {rendererClass=&NiagaraSpriteRendererProperties}
   {rendererClass=&NiagaraSpriteRendererProperties, rendererIndex=1}
   {rendererClass=&NiagaraSpriteRendererProperties, rendererIndex=2}
  ]
 }
 Trail003 {
  bEnabled=false
  emitterUpdateScript [
   &EmitterState(Inactive Response=::NewEnumerator1, Life Cycle Mode=::NewEnumerator1, Loop Duration=1, UseLoopDelay=false)
   &SpawnRate(SpawnRate=60, Use Spawn Probability=false)
  ]
  particleSpawnScript [
   &InitializeParticle(Lifetime=0.5, Ribbon Width=100, Ribbon Width Mode=::NewEnumerator1, UsePositionOffset=false, Write Material Random=true)
   &AddVelocity(Velocity=<400 0 0>)
  ]
  particleUpdateScript [
   &ParticleState()
   &ScaleColor(
    Scale Alpha=&FloatFromCurve(FloatCurve=DataInterface() {propertyValues {Curve {keys [<at 0 0 RCIM_Cubic>,<at 0.1 1 RCIM_Cubic>,<at 1 0 RCIM_Cubic>]}}})
    ScaleA=true
    ScaleRGB=true
   )
   &SolveForcesAndVelocity(Clamp Velocity=false, Limit Acceleration=false)
   &DynamicMaterialParameters(
    Index 0 Param 1=$Engine.Time
    Module.Index0Float0Write=true
    Module.Index0Float1Write=false
    Module.Index0Float2Write=false
    Module.Index0Float3Write=false
   )
  ]
  rendererClasses [&NiagaraRibbonRendererProperties]
  renderers [{rendererClass=&NiagaraRibbonRendererProperties}]
 }
 Wings {
  emitterUpdateScript [
   &EmitterState(
    Inactive Response=::NewEnumerator1
    Life Cycle Mode=::NewEnumerator1
    Loop Behavior=::NewEnumerator1
    Loop Delay=0.02
    Loop Duration=3.7
    UseLoopDelay=true
   )
   &SpawnRate(SpawnRate=60, Use Spawn Probability=false) {enabled=false}
   &SpawnBurst_Instantaneous(Spawn Count=6, Use Spawn Probability=false)
   &SpawnBurst_Instantaneous(Spawn Count=3, Spawn Time=0.2374, Use Spawn Probability=false, UseLoopCountLimit=false) {enabled=false, module=SpawnBurst_Instantaneous001}
  ]
  particleSpawnScript [
   &PartitionParticles(Number Of Partitions=6)
   /Niagara/Modules/Spawn/Initialization/InitializeParticle(
    Color=<rgba 1.7375 0 2 0.2>
    Mesh Scale=&Multiply_VectorByFloat(
     Vector=<1 0.8 1>
     Float=&SelectFloatFromArray(
      Direct Array Index=&ReturnExecIndex()
      Float Selection Array=DataInterface() {propertyValues {FloatData [0.8,0.8,0.3]}}
     )
    )
    Module.Write Color=false
    Module.Write Lifetime=true
    Module.Write Mass=false
    Module.Write Material Random=true
    Module.Write Position=false
    Module.Write Scale=true
    Module.Write SpriteRotation=false
    Module.Write SpriteSize=true
    Sprite Size=<8 8>
   ) {enabled=false}
   /Niagara/Modules/Spawn/Initialization/V2/InitializeParticle(
    Mesh Scale=&Multiply_VectorByFloat(
     Vector=<0.7 0.5 1>
     Float=&SelectFloatFromArray(
      Direct Array Index=&ReturnExecIndex()
      Float Selection Array=DataInterface() {propertyValues {FloatData [0.5,0.5,0.3]}}
     )
    )
    Mesh Scale Mode=::NewEnumerator4
    Position Mode=::NewEnumerator0
    Write Material Random=true
   ) {module=InitializeParticle001}
   &InitialMeshOrientation(
    Mesh Orientation Mode=::NewEnumerator1
    Rotation=&SelectVectorFromArray(
     Direct Array Index=&ReturnExecIndex()
     Vector Selection Array=DataInterface() {propertyValues {FloatData [<0.25 0 0>, <-0.01 -0.37 0.35>, <3.3239 2.6261 0.1263>]}}
    )
    Rotation Coordinate Space=::NewEnumerator2
    Use Rotation Vector=true
   ) {enabled=false}
   &InitialMeshOrientation(Mesh Orientation Mode=::NewEnumerator2, Orientation Vector=<0 1 0>, Use Rotation Vector=false) {module=InitialMeshOrientation001}
  ]
  particleUpdateScript [
   &ParticleState(Kill Particles When Lifetime Has Elapsed=false, Loop Particles Lifetime=true)
   &SkeletalMeshLocation(
    Bone / Socket Index=$Particles.PartitionParticles.Partition
    MaskSpawnGroup=false
    Mesh Sampling Type=::NewEnumerator5
    Sampled Scale Coordinate Space:ENiagaraCoordinateSpace=::Simulation
    Skeletal Mesh=$System.dragon 1
    Socket Sampling Mode=::NewEnumerator5
   ) {module=SkeletalMeshLocation001}
   &UpdateMeshOrientation()
   /SetVariables(Particles.LightCol:LinearColor=&MultiplyLinearColor(
     Linear Color=<rgba 0.5 0.089 0 1>
     Scale Factor=&MakeVector4_FromFloat(Float=&Sine(Normalized Angle=$Emitter.Age, Period=0.2, Scale=0.1, Bias=1))
    )) {isSetParameters=true}
   &ScaleColor(
    Scale Mode=::NewEnumerator1
    Scale RGBA=&MakeVector4_FromLinearColor(LinearColor=&ColorFromCurve(ColorCurve=DataInterface() {propertyValues {
        AlphaCurve {keys [<at 0 0.5>,<at 0.5 1>,<at 1 0.5>]}
        BlueCurve {keys [<at 0 1 RCIM_Cubic>,<at 0.5 2 RCIM_Cubic>,<at 1 1 RCIM_Cubic>]}
        GreenCurve {keys [<at 0 1 RCIM_Cubic>,<at 0.5 2 RCIM_Cubic>,<at 1 1 RCIM_Cubic>]}
        RedCurve {keys [<at 0 1 RCIM_Cubic>,<at 0.5 2 RCIM_Cubic>,<at 1 1 RCIM_Cubic>]}
       }}))
    ScaleRGBA=true
   )
   &DynamicMaterialParameters(
    Index 0 Param 3=10
    Module.Param0WriteEnabled=true
    Module.Param0WriteEnabled1=true
    Module.Param1WriteEnabled=true
    Module.Param1WriteEnabled1=true
    Module.Param2WriteEnabled=true
    Module.Param2WriteEnabled1=true
    Module.Param3WriteEnabled=true
    Module.Param3WriteEnabled1=true
    Write Parameter Index 1=true
   )
   &SubUVAnimation(
    Mesh Renderer=DataInterface() {propertyValues {MeshRenderer=&self:Wings_0.NiagaraMeshRendererProperties_1}}
    Renderer Type:ENiagaraMeshOrSprite=::NewEnumerator1
    SubUV Animation Mode:ENiagara_SubUVLookupModeV2=::NewEnumerator2
    UseEndFrame=false
    UseStartFrame=false
   )
   /SetVariables(Particles.new fire Scale:Vector3f=&Multiply_VectorByFloat(Vector=<1 1 1>, Float=0.2)) {isSetParameters=true}
  ]
  rendererClasses [
   &NiagaraSpriteRendererProperties
   &NiagaraMeshRendererProperties
   &NiagaraLightRendererProperties
  ]
  renderers [
   {rendererClass=&NiagaraSpriteRendererProperties}
   {rendererClass=&NiagaraMeshRendererProperties, rendererIndex=1}
   {rendererClass=&NiagaraLightRendererProperties, rendererIndex=2}
   {rendererClass=&NiagaraMeshRendererProperties, rendererIndex=3}
  ]
 }
 fire_wings_licks {
  bEnabled=false
  emitterUpdateScript [
   &EmitterState(Life Cycle Mode=::NewEnumerator1, Loop Behavior=::NewEnumerator1, Loop Duration=3.5, UseLoopDelay=false)
   &SpawnRate(SpawnGroup=1, SpawnRate=50, Use Spawn Probability=false) {module=SpawnRate001}
   &SpawnBurst_Instantaneous(Spawn Count=$Emitter.ParticlesToSpawn, Spawn Probability=0.5, Use Spawn Probability=false, UseLoopCountLimit=false) {enabled=false, module=SpawnBurst_Instantaneous001}
   &SpawnBurst_Instantaneous(Spawn Count=12, Use Spawn Probability=false, UseLoopCountLimit=false) {enabled=false}
   &SpawnRate(SpawnRate=50, Use Spawn Probability=false)
  ]
  particleSpawnScript [
   /SetVariables(Particles.VisibilityTag=&SpawnGroup()) {isSetParameters=true}
   &InitializeParticle(
    Color=<rgba 1 1 1 0.5>
    Lifetime=&MergeHelper_FloatSwitch(
     0=&UniformRangedFloat(Minimum=0.45, Maximum=0.75)
     1=&UniformRangedFloat(Minimum=0.15, Maximum=0.5)
     2=0
     3=0
     4=0
     Index=$Particles.VisibilityTag
     Num=::NewEnumerator3
    )
    Mass Mode=::NewEnumerator1
    Position Offset=<0 0 -115>
    Sprite Rotation Mode=::NewEnumerator2
    Sprite Size=<23 23>
    Sprite Size Mode=::NewEnumerator4
    UsePositionOffset=false
    Write Material Random=true
   ) {module=InitializeParticle001}
   &FN_SetSpawnGroup(Number of Spawn Groups=6)
   &AddVelocity(Rotation Coordinate Space=::Simulation)
   &SkeletalMeshLocation(
    Bone / Socket Index=$Particles.SpawnGroup
    MaskSpawnGroup=false
    Mesh Sampling Type=::NewEnumerator5
    Sampled Orientation Coordinate Space=::Local
    Sampled Scale Coordinate Space:ENiagaraCoordinateSpace=::Simulation
    Skeletal Mesh=DataInterface() {propertyValues {
      DefaultMesh=null
      ExcludeBoneName=root
      OptionalComponentClass=null
      PreviewMesh=&OutlineLove_Gadget
      SkinningMode=SkinOnTheFly
      SoftSourceActor=null
      SourceMode=Default
      UvSetIndex=0
      WholeMeshLOD=-1
      bExcludeBone=false
      bReadDeformedGeometry=true
      bRequireCurrentFrameData=true
      ComponentTags []
      FilteredBones [JNT_Dragon01_Elbow_l, JNT_Dragon01_Elbow_r, JNT_Dragon02_Elbow_l, JNT_Dragon02_Elbow_r, JNT_Dragon03_Elbow_r, JNT_Dragon03_Elbow_l]
      FilteredSockets [
       FX_Dragon01_Elbow_Socket_R
       FX_Dragon01_Elbow_Socket_l
       FX_Dragon02_Elbow_Socket_r
       FX_Dragon02_Elbow_Socket_l
       FX_Dragon03_Elbow_Socket_r
       FX_Dragon03_Elbow_Socket_l
      ]
      MeshUserParameter {parameter {name=None, typeDefHandle {registeredTypeIndex=1}, varData []}}
      SamplingRegions []
     }}
    Socket Sampling Mode=::NewEnumerator5
   ) {module=SkeletalMeshLocation001}
  ]
  particleUpdateScript [
   &ParticleState()
   &SpriteFacingAndAlignment(Sprite Alignment=<0 0 0>, Sprite Facing=<0 0 1>, WriteAlignment=false, WriteFacing=true)
   &ScaleSpriteSize(Scale Factor=&Vector2DFromFloat(Value=&FloatFromCurve(
      FloatCurve=DataInterface() {propertyValues {Curve {keys [
          <at 0 0.7 RCIM_Cubic>
          {arriveTangent=4, interpMode=RCIM_Cubic, leaveTangent=4, time=0.1496, value=2}
          {arriveTangent=-0.565, interpMode=RCIM_Cubic, leaveTangent=-0.565, tangentMode=RCTM_User, time=0.2, value=1.5}
          {arriveTangent=-1.5177, interpMode=RCIM_Cubic, leaveTangent=-1.5177, tangentMode=RCTM_User, time=1, value=0.75}
         ]}}}
      Scale Curve=&RandomRangeFloat(Minimum=0.25, Maximum=0.75)
     ))) {module=ScaleSpriteSize001}
   &GravityForce()
   &Drag(
    Drag=&MergeHelper_FloatSwitch(0=5, 1=0, 2=0, 3=0, 4=0, Index=$Particles.Initial.VisibilityTag, Num=::NewEnumerator3)
    Use Linear Drag=true
    Use Rotational Drag=true
   )
   &InheritVelocity(Speed Limit=true, UseSpeedThreshold=false)
   &SolveForcesAndVelocity(Clamp Velocity=false, Limit Acceleration=false)
   &ScaleColor(
    Scale Alpha=&Multiply_Float(
     A=&Multiply_Float(
      A=&FloatFromCurve(FloatCurve=DataInterface() {propertyValues {Curve {keys [
           {arriveTangent=16.5612, interpMode=RCIM_Cubic, leaveTangentWeight=0.3333, time=0.1, value=0}
           {arriveTangentWeight=0.1333, interpMode=RCIM_Cubic, leaveTangentWeight=0.1667, tangentMode=RCTM_User, time=0.5, value=1}
           {arriveTangentWeight=0.1667, interpMode=RCIM_Cubic, tangentMode=RCTM_User, time=1, value=0}
          ]}}})
      B=&Multiply_Float(
       A=&RandomRangeFloat(Minimum=1, Maximum=1)
       B=&UniformAOrBFloat(A=0, Distribution Weight=1, Evaluation Type=::NewEnumerator0, Recalculate Random Each Loop=true, Override Seed=false)
      )
     )
     B=1
    )
    Scale RGB=&Multiply_VectorByFloat(
     Vector=&Multiply_VectorByFloat(
      Vector=<1 1 1>
      Float=&FloatFromCurve(FloatCurve=DataInterface() {propertyValues {Curve {keys [
           {arriveTangent=-5.044, interpMode=RCIM_Cubic, leaveTangent=-5.044, tangentMode=RCTM_User, time=0.25, value=2}
           {arriveTangent=-11.3866, interpMode=RCIM_Cubic, leaveTangent=-11.3866, tangentMode=RCTM_User, time=0.4, value=1}
          ]}}})
     )
     Float=&FloatFromCurve(
      FloatCurve=DataInterface() {propertyValues {Curve {keys [
          {arriveTangent=9.0725, interpMode=RCIM_Cubic, time=0, value=0}
          <at 0.1114 0.8492 RCIM_Cubic>
          {arriveTangent=-3.3432, interpMode=RCIM_Cubic, leaveTangent=-3.3432, tangentMode=RCTM_User, time=1, value=0}
         ]}}}
      Scale Curve=&Multiply_Float(A=&RandomRangeFloat(Minimum=0.5, Maximum=1), B=1)
     )
    )
    ScaleA=true
    ScaleRGB=true
   ) {enabled=false}
   &SubUVAnimation()
   &ScaleColor(
    Scale Alpha=&Multiply_Float(
     A=&FloatFromCurve(FloatCurve=DataInterface() {propertyValues {Curve {keys [
          {arriveTangent=16.5612, interpMode=RCIM_Cubic, leaveTangent=16.5612, leaveTangentWeight=0.3333, tangentMode=RCTM_User, time=0, value=0}
          {arriveTangentWeight=0.0333, interpMode=RCIM_Cubic, leaveTangentWeight=0.3, tangentMode=RCTM_User, time=0.1, value=1}
          {arriveTangentWeight=0.1333, interpMode=RCIM_Cubic, leaveTangentWeight=0.1667, tangentMode=RCTM_User, time=0.5, value=1}
          {arriveTangentWeight=0.1667, interpMode=RCIM_Cubic, tangentMode=RCTM_User, time=1, value=0}
         ]}}})
     B=&RandomRangeFloat(Minimum=0.5, Maximum=1)
    )
    Scale RGB=&Multiply_VectorByFloat(
     Vector=<1 1 1>
     Float=&FloatFromCurve(
      FloatCurve=DataInterface() {propertyValues {Curve {keys [
          {arriveTangent=9.0725, interpMode=RCIM_Cubic, leaveTangent=9.0725, tangentMode=RCTM_User, time=0, value=0}
          {arriveTangent=-2.5249, interpMode=RCIM_Cubic, leaveTangent=-2.5249, tangentMode=RCTM_User, time=0.1408, value=0.9847}
          {arriveTangent=-0.0405, interpMode=RCIM_Cubic, leaveTangent=-0.0405, tangentMode=RCTM_User, time=1, value=0}
         ]}}}
      Scale Curve=&RandomRangeFloat(Minimum=0.5, Maximum=1)
     )
    )
    ScaleA=true
    ScaleRGB=true
   ) {module=ScaleColor001}
   &DynamicMaterialParameters(
    Index 0 Param 1=&RandomRangeFloat(Minimum=2, Maximum=1)
    Index 0 Param 2=&RandomRangeFloat(Minimum=0, Maximum=2)
    Index 0 Param 3=8
    Index 0 Param 4=$Particles.SubImageIndex
    Index 1 Param 1=&RandomRangeFloat(Minimum=-1, Maximum=1)
    Index 1 Param 2=&RandomRangeFloat(Minimum=-1, Maximum=1)
    Index 1 Param 3=&FloatFromCurve(FloatCurve=DataInterface() {propertyValues {Curve {keys [
         {arriveTangent=4.1785, interpMode=RCIM_Cubic, leaveTangent=4.1785, tangentMode=RCTM_User, time=0, value=0}
         {arriveTangent=0.9995, interpMode=RCIM_Cubic, leaveTangent=0.9995, tangentMode=RCTM_User, time=0.2223, value=0.6825}
         {arriveTangent=-0.0405, interpMode=RCIM_Cubic, leaveTangent=-0.0405, tangentMode=RCTM_User, time=0.7, value=0.9}
         <at 1 0.7>
        ]}}})
    Index 2 Param 4=1
    Module.Param0WriteEnabled=true
    Module.Param0WriteEnabled1=true
    Module.Param0WriteEnabled2=false
    Module.Param1WriteEnabled=true
    Module.Param1WriteEnabled1=true
    Module.Param1WriteEnabled2=false
    Module.Param2WriteEnabled=true
    Module.Param2WriteEnabled1=true
    Module.Param2WriteEnabled2=false
    Module.Param3WriteEnabled=true
    Module.Param3WriteEnabled1=false
    Module.Param3WriteEnabled2=true
    Write Parameter Index 1=true
    Write Parameter Index 2=true
   )
   /SetVariables(
    Particles.fire licks SpriteSize:Vector2f=&Vector2DFromFloat(Value=&RandomRangeFloat(Minimum=5, Maximum=10))
    Particles.fire licks color:LinearColor=&MakeLinearColorFromVectorAndFloat(
     "Vector (RGB)"=<1 1 1>
     "Float (Alpha)"=&FloatFromCurve(FloatCurve=DataInterface() {propertyValues {Curve {keys [<at 0 1>,<at 0.8631 0.1788>]}}})
    )
   ) {isSetParameters=true}
   /SetVariables(
    Particles.Ashes Color:LinearColor=&MakeLinearColorFromVectorAndFloat(
     "Vector (RGB)"=&MakeVectorFromLinearColorRGB(Color=&ColorFromCurve(ColorCurve=DataInterface() {propertyValues {
         AlphaCurve {keys [<at 0 1 RCIM_Cubic>]}
         BlueCurve {keys [<at 0 1.473 RCIM_Cubic>,<at 0.6088 0.0879 RCIM_Cubic>]}
         GreenCurve {keys [<at 0 11.4982 RCIM_Cubic>,<at 0.6088 0.0919 RCIM_Cubic>]}
         RedCurve {keys [<at 0 50 RCIM_Cubic>,<at 0.6088 0.099 RCIM_Cubic>]}
        }}))
     "Float (Alpha)"=&FloatFromCurve(FloatCurve=DataInterface() {propertyValues {Curve {keys [<at 0.8 1 RCIM_Cubic>,<at 1 0>]}}})
    )
    Particles.Ashes SpriteSize:Vector2f=&Vector2DFromFloat(Value=&RandomRangeFloat(Minimum=1, Maximum=7))
    Particles.Vector 4:Vector4f=&MakeVector4(X=0, Y=0, Z=0, W=0)
   ) {isSetParameters=true}
   &SkeletalMeshLocation(
    Bone / Socket Index=$Particles.SpawnGroup
    MaskSpawnGroup=false
    Mesh Sampling Type=::NewEnumerator5
    Sampled Orientation Coordinate Space=::Local
    Sampled Scale Coordinate Space:ENiagaraCoordinateSpace=::Simulation
    Skeletal Mesh=DataInterface() {propertyValues {
      DefaultMesh=null
      ExcludeBoneName=root
      OptionalComponentClass=null
      PreviewMesh=&OutlineLove_Gadget
      SkinningMode=SkinOnTheFly
      SoftSourceActor=null
      SourceMode=Default
      UvSetIndex=0
      WholeMeshLOD=-1
      bExcludeBone=false
      bReadDeformedGeometry=true
      bRequireCurrentFrameData=true
      ComponentTags []
      FilteredBones [JNT_Dragon01_Elbow_l, JNT_Dragon01_Elbow_r, JNT_Dragon02_Elbow_l, JNT_Dragon02_Elbow_r, JNT_Dragon03_Elbow_r, JNT_Dragon03_Elbow_l]
      FilteredSockets [
       FX_Dragon01_Elbow_Socket_R
       FX_Dragon01_Elbow_Socket_l
       FX_Dragon02_Elbow_Socket_r
       FX_Dragon02_Elbow_Socket_l
       FX_Dragon03_Elbow_Socket_r
       FX_Dragon03_Elbow_Socket_l
      ]
      MeshUserParameter {parameter {name=None, typeDefHandle {registeredTypeIndex=1}, varData []}}
      SamplingRegions []
     }}
    Socket Sampling Mode=::NewEnumerator5
   ) {module=SkeletalMeshLocation003}
  ]
  rendererClasses [&NiagaraSpriteRendererProperties]
  renderers [
   {rendererClass=&NiagaraSpriteRendererProperties}
   {rendererClass=&NiagaraSpriteRendererProperties, rendererIndex=1}
   {rendererClass=&NiagaraSpriteRendererProperties, rendererIndex=2}
  ]
 }
 glow_on_symbol {bEnabled=false, rendererClasses [], renderers []}
 trail {
  emitterUpdateScript [
   &EmitterState(Inactive Response=::NewEnumerator1, Life Cycle Mode=::NewEnumerator1, Loop Behavior=::NewEnumerator1, Loop Duration=3.7, UseLoopDelay=false)
   &SpawnRate(SpawnRate=60, Use Spawn Probability=false)
  ]
  particleSpawnScript [
   &InitializeParticle(
    Color Mode=::NewEnumerator0
    Lifetime=0.5
    Ribbon Width=10
    Ribbon Width Mode=::NewEnumerator1
    UsePositionOffset=false
    Write Material Random=true
   )
   /SetVariables(Particles.Random=&RandomRangeFloat(Minimum=0.7, Maximum=1.1)) {isSetParameters=true, module=SetVariables_#15}
   &FN_SetSpawnGroup(Number of Spawn Groups=3)
   &SkeletalMeshLocation(
    Bone / Socket Index=&SpawnGroup()
    Bone Sampling Mode=::NewEnumerator4
    MaskSpawnGroup=false
    Sampled Position Offset=<0 5 0>
    Sampled Position Offset Coordinate Space:ENiagaraSkelMeshTransforms=::NewEnumerator2
    Skeletal Mesh=$System.Skeletal Mesh Reader
   )
  ]
  particleUpdateScript [
   &GravityForce(Gravity=&MakeVector(
     X=0
     Y=0
     Z=&FloatFromCurve(
      FloatCurve=DataInterface() {propertyValues {Curve {keys [<at 0.8 1 RCIM_Cubic>, {arriveTangent=-6041.373, interpMode=RCIM_Cubic, leaveTangent=-6041.373, tangentMode=RCTM_User, time=1, value=-500}]}}}
      CurveIndex=$Emitter.NormalizedLoopAge
     )
    ))
   &InheritSourceMovement(Applied Position Scale=<-1 -1 1>, ApplyForce=false, ApplyPosition=true, ApplyPositionFalloff=false)
   &SolveForcesAndVelocity(Clamp Velocity=false, Limit Acceleration=false)
   &ParticleState()
   &ScaleRibbonWidth(Ribbon Width Scale=&FloatFromCurve(FloatCurve=DataInterface() {propertyValues {Curve {keys [<at 0.0026 0.6424 RCIM_Cubic>, {arriveTangent=-1.7794, interpMode=RCIM_Cubic, leaveTangent=-1.7794, tangentMode=RCTM_User, time=1, value=0}]}}}))
   &DynamicMaterialParameters(
    Index 0 Param 1=1.2375
    Index 0 Param 3=1
    Module.Index0Float0Write=true
    Module.Index0Float1Write=true
    Module.Index0Float2Write=true
    Module.Index0Float3Write=true
   )
   /SetVariables(
    Particles.Filler DynParam:Vector4f=&MakeVector4(
     X=&Multiply_Float(A=0, B=$Particles.Random)
     Y=&FloatFromCurve(FloatCurve=DataInterface() {propertyValues {Curve {keys [{tangentMode=RCTM_User, time=0, value=1}, {tangentMode=RCTM_User, time=0.5, value=1}]}}})
     Z=&FloatFromCurve(FloatCurve=DataInterface() {propertyValues {Curve {keys [<at 0.4 0>,<at 1 1>]}}})
     W=0
    )
    Particles.Filler DynParam 1:Vector4f=&MakeVector4(
     X=&Multiply_Float(A=0.2, B=$Particles.Random)
     Y=&FloatFromCurve(FloatCurve=DataInterface() {propertyValues {Curve {keys [{tangentMode=RCTM_User, time=0, value=1}, {tangentMode=RCTM_User, time=1, value=1}]}}})
     Z=&FloatFromCurve(FloatCurve=DataInterface() {propertyValues {Curve {keys [<at 0.7 0>,<at 1 0>]}}})
     W=0
    )
    Particles.Filler DynParam 2:Vector4f=&MakeVector4(X=0, Y=0.8, Z=0, W=0)
    Particles.Filler RibbonWidth=10
    Particles.filler trail Color:LinearColor=&MakeLinearColorFromVectorAndFloat(
     "Vector (RGB)"=<1 1 1>
     "Float (Alpha)"=&FloatFromCurve(FloatCurve=DataInterface() {propertyValues {Curve {keys [<at 0 1>,<at 0.7181 0.0188>]}}})
    )
   ) {isSetParameters=true, module=SetVariables_#16}
   /SetVariables(
    Particles.Color=&MakeLinearColorFromVectorAndFloat(
     "Vector (RGB)"=<1 1 1>
     "Float (Alpha)"=&Lerp_Float(
      A=0.5
      B=0.9
      Alpha=&FloatFromCurve(
       FloatCurve=DataInterface() {propertyValues {Curve {keys [
           {arriveTangent=-0.1961, interpMode=RCIM_Cubic, leaveTangent=-0.1961, tangentMode=RCTM_User, time=0, value=1}
           {arriveTangent=-2.0216, interpMode=RCIM_Cubic, leaveTangent=-2.0216, tangentMode=RCTM_User, time=0.7, value=0}
          ]}}}
       Scale Curve=0.5
      )
     )
    )
    Particles.Main fire DynParams:Vector4f=&MakeVector4(X=1.2375, Y=0, Z=1, W=0)
   ) {isSetParameters=true, module=SetVariables_#17}
  ]
  rendererClasses [&NiagaraRibbonRendererProperties]
  renderers [
   {rendererClass=&NiagaraRibbonRendererProperties}
   {rendererClass=&NiagaraRibbonRendererProperties, rendererIndex=1}
  ]
 }
 trail002 {
  emitterUpdateScript [
   &EmitterState(Inactive Response=::NewEnumerator1, Life Cycle Mode=::NewEnumerator1, Loop Behavior=::NewEnumerator1, Loop Duration=3.7, UseLoopDelay=false)
   &SpawnBurst_Instantaneous(Spawn Count=3, Use Spawn Probability=false, UseLoopCountLimit=false)
  ]
  particleSpawnScript [
   &InitializeParticle(
    Color Mode=::NewEnumerator0
    Lifetime=3.7
    Ribbon Width=10
    Ribbon Width Mode=::NewEnumerator1
    UsePositionOffset=false
    Write Material Random=true
   )
   &PartitionParticles(Number Of Partitions=3)
  ]
  particleUpdateScript [
   &ParticleState()
   &SkeletalMeshLocation(
    Bone / Socket Index=$Particles.PartitionParticles.Partition
    Bone Sampling Mode=::NewEnumerator4
    MaskSpawnGroup=false
    Sampled Position Offset=<0 5 0>
    Sampled Position Offset Coordinate Space:ENiagaraSkelMeshTransforms=::NewEnumerator2
    Skeletal Mesh=$System.Skeletal Mesh Reader
   )
   &InheritSourceMovement(Applied Position Scale=<-1 -1 1>, ApplyForce=false, ApplyPosition=true, ApplyPositionFalloff=false)
   &SolveForcesAndVelocity(Clamp Velocity=false, Limit Acceleration=false)
   &FN_Light_Attributes(
    Color=<rgba 15 1.4715 0 1>
    Create Color Attribute=true
    Create Exponent Attribute=true
    Create Radius Attribute=true
    Enable Particle Light Enable=true
    Exponent=5
    Light Enabled=true
    Radius=25
    Scale Color=&Multiply_Float(
     A=0.5
     B=&Multiply_Float(
      A=&Waveform("Waveform [1]"=::NewEnumerator0, "[1] Frequency"=3, "[1] Amplitude Min/Max"=<0.7 1>, "[1] Scale"=2.6368)
      B=&FNTimeOfDayScalar(Night Scale=0.75, Morning Scale=0.6, Day Scale=0.45, Evening Scale=0.6)
     )
    )
   )
  ]
  rendererClasses [&NiagaraLightRendererProperties]
  renderers [{rendererClass=&NiagaraLightRendererProperties, rendererIndex=0}]
 }
 win_trails_test {
  bEnabled=false
  emitterUpdateScript [
   &EmitterState(Inactive Response=::NewEnumerator1, Life Cycle Mode=::NewEnumerator1, Loop Behavior=::NewEnumerator1, Loop Duration=3.7, UseLoopDelay=false)
   &FN_SpawnAnimTrail(AnimTrailWidthMode=::FromFirst, Number of Anim Trails=18, Switch on ENiagara_AnimTrail_Source:ENiagara_AnimTrail_Source=::NewEnumerator1) {enabled=false}
   &SpawnRate(SpawnRate=200, Use Spawn Probability=false)
  ]
  particleSpawnScript [
   &PartitionParticles(Number Of Partitions=18, Partition Mode:ENiagara_PartitionMode_v2=::NewEnumerator1)
   &InitializeParticle(
    Color Mode=::NewEnumerator0
    Lifetime=0.5
    Ribbon Width=35
    Ribbon Width Mode=::NewEnumerator1
    UsePositionOffset=false
    Write Material Random=true
   )
   /SetVariables(Particles.Random=&RandomRangeFloat(Minimum=0.7, Maximum=1.1)) {isSetParameters=true, module=SetVariables_#15}
   &FN_SetSpawnGroup(Number of Spawn Groups=18)
   &SkeletalMeshLocation(
    Bone / Socket Index=&SpawnGroup()
    Bone Sampling Mode=::NewEnumerator4
    MaskSpawnGroup=false
    Sampled Position Offset=<0 5 0>
    Sampled Position Offset Coordinate Space:ENiagaraSkelMeshTransforms=::NewEnumerator2
    Skeletal Mesh=$System.Skeletal Mesh Reader
   ) {enabled=false}
   &SkeletalMeshLocation(
    Bone / Socket Index=$Particles.SpawnGroup
    MaskSpawnGroup=false
    Mesh Sampling Type=::NewEnumerator5
    Sampled Scale Coordinate Space:ENiagaraCoordinateSpace=::Simulation
    Skeletal Mesh=DataInterface() {propertyValues {
      DefaultMesh=null
      ExcludeBoneName=root
      OptionalComponentClass=null
      PreviewMesh=&OutlineLove_Gadget
      SkinningMode=SkinOnTheFly
      SoftSourceActor=null
      SourceMode=Default
      UvSetIndex=0
      WholeMeshLOD=-1
      bExcludeBone=false
      bReadDeformedGeometry=true
      bRequireCurrentFrameData=true
      ComponentTags []
      FilteredBones [JNT_Dragon01_Elbow_l, JNT_Dragon01_Elbow_r, JNT_Dragon02_Elbow_l, JNT_Dragon02_Elbow_r, JNT_Dragon03_Elbow_r, JNT_Dragon03_Elbow_l]
      FilteredSockets [
       FX_Dragon01_Finger_Socket_r
       FX_Dragon01_FingerA_Socket_r
       FX_Dragon01_Finger_end_Socket_r
       FX_Dragon01_Finger_Socket_l
       FX_Dragon01_FingerA_Socket_l
       FX_Dragon01_Finger_end_Socket_l
       FX_Dragon02_Finger_Socket_r
       FX_Dragon02_FingerA_Socket_r
       FX_Dragon02_Finger_end_Socket_r
       FX_Dragon02_Finger_Socket_l
       FX_Dragon02_FingerA_Socket_l
       FX_Dragon02_Finger_end_Socket_l
       FX_Dragon03_Finger_Socket_r
       FX_Dragon03_FingerA_Socket_r
       FX_Dragon03_Finger_end_Socket_r
       FX_Dragon03_Finger_Socket_l
       FX_Dragon03_FingerA_Socket_l
       FX_Dragon03_Finger_end_Socket_l
      ]
      MeshUserParameter {parameter {name=None, typeDefHandle {registeredTypeIndex=1}, varData []}}
      SamplingRegions []
     }}
    Socket Sampling Mode=::NewEnumerator5
   ) {module=SkeletalMeshLocation001}
  ]
  particleUpdateScript [
   &GravityForce(Gravity=&MakeVector(
     X=0
     Y=0
     Z=&FloatFromCurve(
      FloatCurve=DataInterface() {propertyValues {Curve {keys [<at 0.8 1 RCIM_Cubic>, {arriveTangent=-6041.373, interpMode=RCIM_Cubic, leaveTangent=-6041.373, tangentMode=RCTM_User, time=1, value=-500}]}}}
      CurveIndex=$Emitter.NormalizedLoopAge
     )
    ))
   &InheritSourceMovement(Applied Position Scale=<-1 -1 1>, ApplyForce=false, ApplyPosition=true, ApplyPositionFalloff=false)
   &ScaleColor(
    Scale Alpha=&FloatFromCurve(FloatCurve=DataInterface() {propertyValues {Curve {keys [<at 0 0 RCIM_Cubic>,<at 0.1 1 RCIM_Cubic>,<at 1 0 RCIM_Cubic>]}}})
    ScaleA=true
    ScaleRGB=true
   )
   &SolveForcesAndVelocity(Clamp Velocity=false, Limit Acceleration=false)
   &ParticleState()
   &ScaleRibbonWidth(Ribbon Width Scale=&FloatFromCurve(FloatCurve=DataInterface() {propertyValues {Curve {keys [<at 0.0026 0.6424 RCIM_Cubic>, {arriveTangent=-1.7794, interpMode=RCIM_Cubic, leaveTangent=-1.7794, tangentMode=RCTM_User, time=1, value=0}]}}}))
   /SetVariables(
    Particles.Filler DynParam:Vector4f=&MakeVector4(
     X=&Multiply_Float(A=0, B=$Particles.Random)
     Y=&FloatFromCurve(FloatCurve=DataInterface() {propertyValues {Curve {keys [{tangentMode=RCTM_User, time=0, value=1}, {tangentMode=RCTM_User, time=0.5, value=1}]}}})
     Z=&FloatFromCurve(FloatCurve=DataInterface() {propertyValues {Curve {keys [<at 0.4 0>,<at 1 1>]}}})
     W=0
    )
    Particles.Filler DynParam 1:Vector4f=&MakeVector4(
     X=&Multiply_Float(A=0.2, B=$Particles.Random)
     Y=&FloatFromCurve(FloatCurve=DataInterface() {propertyValues {Curve {keys [{tangentMode=RCTM_User, time=0, value=1}, {tangentMode=RCTM_User, time=1, value=1}]}}})
     Z=&FloatFromCurve(FloatCurve=DataInterface() {propertyValues {Curve {keys [<at 0.7 0>,<at 1 0>]}}})
     W=0
    )
    Particles.Filler DynParam 2:Vector4f=&MakeVector4(X=0, Y=0.8, Z=0, W=0)
    Particles.Filler RibbonWidth=10
   ) {isSetParameters=true, module=SetVariables_#16}
   /SetVariables(Particles.Main fire DynParams:Vector4f=&MakeVector4(X=1.2375, Y=0, Z=1, W=0)) {isSetParameters=true, module=SetVariables_#17}
   /SetVariables(Particles.new fire dynparam:Vector4f=&MakeVector4(X=$Engine.Time, Y=0, Z=0, W=0)) {isSetParameters=true}
   &DynamicMaterialParameters(
    Index 0 Param 1=$Engine.Time
    Index 0 Param 4=65
    Module.Index0Float0Write=true
    Module.Index0Float1Write=true
    Module.Index0Float2Write=true
    Module.Index0Float3Write=true
   )
  ]
  rendererClasses [&NiagaraRibbonRendererProperties]
  renderers [{rendererClass=&NiagaraRibbonRendererProperties}]
 }
}
summary {
 systemName=NS_OutlineLove_Fire_Trail
 emitters [
  {emitterName=Sparks001, simTarget=GPUComputeSim, rendererClasses [&NiagaraSpriteRendererProperties]}
  {
   emitterName=Prop_spawn_merged_Blend_space
   rendererClasses [&NiagaraSpriteRendererProperties, &NiagaraLightRendererProperties]
  }
  {bEnabled=false, emitterName=fire_wings_licks, rendererClasses [&NiagaraSpriteRendererProperties]}
  {emitterName=trail, rendererClasses [&NiagaraRibbonRendererProperties]}
  {bEnabled=false, emitterName=Minimal, rendererClasses [&NiagaraSpriteRendererProperties]}
  {
   emitterName=Prop_spawn_merged_Blend_space001
   rendererClasses [&NiagaraSpriteRendererProperties, &NiagaraLightRendererProperties]
  }
  {bEnabled=false, emitterName=Embers_Burst001, simTarget=GPUComputeSim, rendererClasses [&NiagaraSpriteRendererProperties]}
  {bEnabled=false, emitterName=glow_on_symbol, rendererClasses []}
  {emitterName=Fire001, rendererClasses [&NiagaraSpriteRendererProperties]}
  {emitterName=Mesh_flames_for_perf, rendererClasses [&NiagaraMeshRendererProperties]}
  {bEnabled=false, emitterName=MeshFlames001, rendererClasses [&NiagaraSpriteRendererProperties]}
  {
   emitterName=Wings
   rendererClasses [
    &NiagaraSpriteRendererProperties
    &NiagaraMeshRendererProperties
    &NiagaraLightRendererProperties
   ]
  }
  {emitterName=trail002, rendererClasses [&NiagaraLightRendererProperties]}
  {bEnabled=false, emitterName=Minimal004, rendererClasses [&NiagaraMeshRendererProperties]}
  {
   emitterName=Dragons_despawn
   rendererClasses [&NiagaraSpriteRendererProperties, &NiagaraLightRendererProperties]
  }
  {bEnabled=false, emitterName=Trail003, rendererClasses [&NiagaraRibbonRendererProperties]}
  {bEnabled=false, emitterName=win_trails_test, rendererClasses [&NiagaraRibbonRendererProperties]}
  {bEnabled=false, emitterName=Fire, rendererClasses [&NiagaraMeshRendererProperties]}
  {
   emitterName=Prop_spawn_end001
   rendererClasses [&NiagaraSpriteRendererProperties, &NiagaraLightRendererProperties]
  }
  {bEnabled=false, emitterName=Mesh_flames_for_perf001, rendererClasses [&NiagaraMeshRendererProperties]}
 ]
 userVariables [
  User.Anim_Curve_FireSpread {type=NiagaraFloat, defaultValue=1, description=""}
  User.bIsFrontEnd {type=NiagaraBool}
  User.bIsFrontEndPreview {type=NiagaraBool}
  User.BPOpacityMult {type=NiagaraFloat, defaultValue=0, description=""}
  User.CurrentExhaust {type=Vector3f, description="", defaultValue {struct=Vector3f, value=<0 0 0>}}
  User.DebugVelocity {type=NiagaraFloat, defaultValue=0, description=""}
  User.FrostyMistAlpha {type=NiagaraFloat, defaultValue=1, description=""}
  User.FrostyMistTrail_Rate {type=NiagaraFloat, defaultValue=1, description=""}
  User.IsR2 {type=NiagaraBool, defaultValue=true}
  User.MistyTrail_Life {type=NiagaraFloat, defaultValue=0.5, description=""}
  User.PlayerSpeed {type=NiagaraFloat, defaultValue=0, description=""}
  User.Reveal {type=NiagaraFloat, defaultValue=1, description=""}
  User.RibbonIDOffset {type=NiagaraInt32}
  User.SkeletalMesh {
   type=NiagaraDataInterfaceSkeletalMesh
   defaultValue {
    struct=NiagaraExt_VariableValue_DataInterface
    value {dataInterface=&self:NiagaraDataInterfaceSkeletalMesh_0}
   }
  }
  User.TrailLengthScale {type=NiagaraFloat, defaultValue=1, description=""}
  User.TrailWidth {type=NiagaraFloat, defaultValue=0.4, description=""}
 ]
}
