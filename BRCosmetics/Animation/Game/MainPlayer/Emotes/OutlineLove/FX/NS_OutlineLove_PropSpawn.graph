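# Niagara system for the OutlineLove emote's prop spawn, combining a short spawn burst with spreading fire effects.
# Each emitter packs several layers via spawn groups, using MergeHelper switches and VisibilityTag to vary their behaviour.
# The fire emitters place flames on a ring mirrored by particle ID parity, with one also sampling the body skeletal mesh.
# Particle lights, flames and foot glow respond to time of day, some also to front-end context.
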NiagaraSystem /BRCosmetics/Animation/Game/MainPlayer/Emotes/OutlineLove/FX/NS_OutlineLove_PropSpawn
dependencies {
 usedDataInterfaces [
  &NiagaraDataInterfaceCurve
  &NiagaraDataInterfaceColorCurve
  &NiagaraDataInterfaceVector2DCurve
  &NiagaraDataInterfaceSpriteRendererInfo
  &NiagaraDataInterfaceSkeletalMesh
 ]
 usedDynamicInputs [
  &Sine
  &SimulationPosition
  &NormalizeVector
  &SubtractVectorFromPosition
 ]
 usedModules [
  &SystemState
  &self:SystemSpawnScript.SystemScriptSource.SystemScriptGraph.NiagaraNodeAssignment_0.SetVariables_#1
  &SampleSkeletalMesh
  &self:SystemSpawnScript.SystemScriptSource.SystemScriptGraph.NiagaraNodeAssignment_1.SetVariables_#2
  &self:SystemSpawnScript.SystemScriptSource.SystemScriptGraph.NiagaraNodeAssignment_2.SetVariables_#3
  &self:SystemSpawnScript.SystemScriptSource.SystemScriptGraph.NiagaraNodeAssignment_3.SetVariables_#4
  &self:SystemSpawnScript.SystemScriptSource.SystemScriptGraph.NiagaraNodeAssignment_4.SetVariables_#5
  &self:SystemSpawnScript.SystemScriptSource.SystemScriptGraph.NiagaraNodeAssignment_5.SetVariables_#6
 ]
}
emitters {
 Prop_spawn_merged_Blend_space {
  emitterUpdateScript [
   &EmitterState(
    Inactive Response=::NewEnumerator1
    Life Cycle Mode=::NewEnumerator1
    Loop Behavior=::NewEnumerator1
    Loop Delay=0.1
    Loop Duration=0.5
    UseLoopDelay=true
   )
   &SpawnBurst_Instantaneous(Spawn Count=13, Spawn Group=2, Use Spawn Probability=false) {module=SpawnBurst_Instantaneous002}
   &SpawnBurst_Instantaneous(Spawn Count=1, Spawn Group=1, Use Spawn Probability=false)
   &SpawnBurst_Instantaneous(Spawn Count=10, Use Spawn Probability=false) {module=SpawnBurst_Instantaneous001}
  ]
  particleSpawnScript [
   /SetVariables(Particles.VisibilityTag=&SpawnGroup()) {isSetParameters=true}
   /SetVariables(Particles.InitSpriteRotRate=&UniformRangedFloat(Minimum=-1, Maximum=1)) {isSetParameters=true}
   &InitializeParticle(
    Color=&MergeHelper_ColorSwitch(
     0=<rgba 0.5255 0.5255 0.5255 1>
     1=<rgba 0.6031 0.5599 0.463 0.2>
     2=<rgba 5 3.7043 1.51 1>
     3=?
     Index=&SpawnGroup()
     Num=::NewEnumerator1
    )
    Lifetime=&MergeHelper_FloatSwitch(
     0=&UniformRangedFloat(Minimum=0.2, Maximum=0.5)
     1=0.25
     2=&UniformRangedFloat(Minimum=0.3, Maximum=1)
     3=?
     Index=&SpawnGroup()
     Num=::NewEnumerator1
    )
    Position Offset=&FN_FrontEndVector3Modifier(Gameplay=<-10 0 10>, Front End=<0 0 10>)
    Sprite Rotation Mode=::NewEnumerator2
    Sprite Size Mode=::NewEnumerator3
    Uniform Sprite Size=&MergeHelper_FloatSwitch(
     0=&UniformRangedFloat(Minimum=8, Maximum=20)
     1=65
     2=&RandomRangeFloat(Minimum=10, Maximum=25)
     3=?
     Index=&SpawnGroup()
     Num=::NewEnumerator1
    )
    UsePositionOffset=true
    Write Material Random=true
   )
   &ShapeLocation(
    MaskSpawnGroup=false
    Override Seed=false
    Sphere Radius=&MergeHelper_FloatSwitch(0=7, 1=0, 2=10, 3=?, Index=&SpawnGroup(), Num=::NewEnumerator1)
    Sphere Surface Distribution=1
    UseOwnerScale=true
   )
   /SetVariables(
    Particles.Smoke Velocity:Vector3f=&Multiply_Vector(
     A=$Particles.ShapeLocation.ShapeNormal
     B=&VectorFromFloat(Value=&RandomRangeFloat(Minimum=20, Maximum=55))
    )
    Particles.Sparkles Velocity:Vector3f=&UniformRangedVector(Minimum=<-500 -200 -200>, Maximum=<5000 200 200>)
    Particles.airglow CameraOffset=10
    Particles.smoke CameraOffset:NiagaraFloat=&RandomRangeFloat(Minimum=10, Maximum=25)
   ) {isSetParameters=true}
   &AccelerationForce(Acceleration=&MergeHelper_VectorSwitch(
     0=&UniformRangedVector(Minimum=<0 0 -25>, Maximum=<0 0 -75>)
     1=<0 0 0>
     2=&RandomRangeVector(Minimum=<0 0 0>, Maximum=<0 0 0>)
     3=?
     Index=&SpawnGroup()
     Num=::NewEnumerator1
    ))
   &ApplyInitialForces()
   &AddVelocity(Velocity=&MergeHelper_VectorSwitch(
     0=&Multiply_Vector(
      A=$Particles.ShapeLocation.ShapeNormal
      B=&VectorFromFloat(Value=&RandomRangeFloat(Minimum=15, Maximum=40))
     )
     1=<0 0 0>
     2=&UniformRangedVector(Minimum=<-110 -110 -110>, Maximum=<110 110 110>)
     3=?
     Index=&SpawnGroup()
     Num=::NewEnumerator1
    ))
  ]
  particleUpdateScript [
   &ParticleState()
   &SpriteRotationRate(Rotation Rate=&MergeHelper_FloatSwitch(
     0=&Multiply_Float(
      A=$Particles.InitSpriteRotRate
      B=&FloatFromCurve(
       FloatCurve=DataInterface() {propertyValues {Curve {keys [
           {arriveTangent=-1.9214, interpMode=RCIM_Cubic, leaveTangent=-1.9214, tangentMode=RCTM_User, time=-0.0046, value=1.0113}
           {arriveTangent=-0.2496, interpMode=RCIM_Cubic, leaveTangent=-0.2496, tangentMode=RCTM_User, time=0.9896, value=0.16}
          ]}}}
       CurveIndex=$Emitter.Age
      )
     )
     1=0
     2=5
     3=?
     Index=&SpawnGroup()
     Num=::NewEnumerator1
    ))
   &ScaleSpriteSize(
    Scale Sprite Size Mode=::NewEnumerator0
    Uniform Scale Factor=&MergeHelper_FloatSwitch(
     0=&FloatFromCurve(FloatCurve=DataInterface() {propertyValues {Curve {keys [
          {arriveTangent=5.2354, interpMode=RCIM_Cubic, leaveTangent=5.2354, tangentMode=RCTM_User, time=-0.0046, value=0.25}
          <at 0.1439 1.4923>
          <at 1.0011 3.4809 RCIM_Cubic>
         ]}}})
     1=&FloatFromCurve(FloatCurve=DataInterface() {propertyValues {Curve {
         defaultValue=1
         keys [{interpMode=RCIM_Cubic, tangentMode=RCTM_User, time=0, value=0}, {interpMode=RCIM_Cubic, tangentMode=RCTM_User, time=1, value=1}]
        }}})
     2=&FloatFromCurve(
      FloatCurve=DataInterface() {propertyValues {Curve {keys [
          {interpMode=RCIM_Cubic, tangentMode=RCTM_Break, time=-0.0046, value=0}
          {interpMode=RCIM_Cubic, tangentMode=RCTM_Break, time=0.1746, value=0.9952}
          {interpMode=RCIM_Cubic, tangentMode=RCTM_Break, time=1.0011, value=0.002}
         ]}}}
      Scale Curve=0.5
     )
     3=?
     Index=&SpawnGroup()
     Num=::NewEnumerator1
    )
   )
   &ScaleColor(
    Scale Alpha=&MergeHelper_FloatSwitch(
     0=&FloatFromCurve(FloatCurve=DataInterface() {propertyValues {Curve {keys [
          {interpMode=RCIM_Cubic, tangentMode=RCTM_Break, time=0, value=0}
          {interpMode=RCIM_Cubic, tangentMode=RCTM_Break, time=0.2049, value=0.9765}
          {interpMode=RCIM_Cubic, tangentMode=RCTM_Break, time=0.7791, value=0.9107}
          {interpMode=RCIM_Cubic, tangentMode=RCTM_Break, time=1, value=0}
         ]}}})
     1=&FloatFromCurve(FloatCurve=DataInterface() {propertyValues {Curve {keys [
          {interpMode=RCIM_Cubic, tangentMode=RCTM_Break, time=0, value=0}
          {interpMode=RCIM_Cubic, tangentMode=RCTM_Break, time=0.25, value=0.75}
          {interpMode=RCIM_Cubic, tangentMode=RCTM_Break, time=1, value=0}
         ]}}})
     2=&FloatFromCurve(FloatCurve=DataInterface() {propertyValues {Curve {keys [
          {interpMode=RCIM_Cubic, tangentMode=RCTM_User, time=-0.0092, value=0.0144}
          {interpMode=RCIM_Cubic, tangentMode=RCTM_User, time=0.0967, value=0.9767}
          {interpMode=RCIM_Cubic, tangentMode=RCTM_User, time=1.0011, value=0.002}
         ]}}})
     3=?
     Index=&SpawnGroup()
     Num=::NewEnumerator1
    )
    Scale RGB=&MergeHelper_VectorSwitch(0=<1 1 1>, 1=<1 1 1>, 2=<1 1 1>, 3=?, Index=&SpawnGroup(), Num=::NewEnumerator1)
    ScaleA=true
    ScaleRGB=true
   )
   &CurlNoiseForce(
    Bias Noise=false
    Coordinate Space:ENiagaraCoordinateSpace=::Local
    Noise Frequency=&MergeHelper_FloatSwitch(0=0, 1=0, 2=0.02, 3=?, Index=&SpawnGroup(), Num=::NewEnumerator1)
    Noise Quality / Cost=::NewEnumerator0
    Noise Strength=&MergeHelper_FloatSwitch(0=0, 1=0, 2=500, 3=?, Index=&SpawnGroup(), Num=::NewEnumerator1)
    Pan Noise=false
    Random Seed=0
    Randomize Noise Sample=false
    Write to Intrinsic Parameters=false
   )
   &DynamicMaterialParameters(
    Index 0 Param 1=&FloatFromCurve(FloatCurve=DataInterface() {propertyValues {Curve {keys [<at 0 0 RCIM_Cubic>,<at 1.0011 0.2>]}}})
    Index 0 Param 2=&FloatFromCurve(FloatCurve=DataInterface() {propertyValues {Curve {keys [
         {arriveTangent=0.0453, interpMode=RCIM_Cubic, tangentMode=RCTM_Break, time=0, value=-1}
         {arriveTangent=0.0604, interpMode=RCIM_Cubic, leaveTangent=0.0604, tangentMode=RCTM_User, time=1, value=-1}
        ]}}})
    Index 0 Param 3=&FloatFromCurve(FloatCurve=DataInterface() {propertyValues {Curve {keys [{interpMode=RCIM_Cubic, leaveTangent=4.6779, time=0.2851, value=0}]}}})
    Index 1 Param 1=&FloatFromCurve(
     FloatCurve=DataInterface() {propertyValues {Curve {keys [<at 0 0>,<at 0.25 1>]}}}
     Scale Curve=&RandomRangeFloat(Minimum=2, Maximum=1)
    )
    Index 1 Param 2=&FloatFromCurve(
     FloatCurve=DataInterface() {propertyValues {Curve {keys [
         {interpMode=RCIM_Cubic, leaveTangent=3.7422, tangentMode=RCTM_Break, time=0, value=0.2}
         {interpMode=RCIM_Cubic, leaveTangent=0.9066, time=1, value=2}
        ]}}}
     Scale Curve=&RandomRangeFloat(Minimum=0.5, Maximum=1)
    )
    Index 1 Param 3=&FloatFromCurve(FloatCurve=DataInterface() {propertyValues {Curve {keys [
         {interpMode=RCIM_Cubic, leaveTangent=2.8445, tangentMode=RCTM_Break, time=0, value=0.2}
         {interpMode=RCIM_Cubic, leaveTangent=63.2978, time=0.25, value=1}
        ]}}})
    Index 1 Param 4=&FloatFromCurve(
     FloatCurve=DataInterface() {propertyValues {Curve {keys [<at 0 1>,<at 1 0.8>]}}}
     Scale Curve=&RandomRangeFloat(Minimum=0.5, Maximum=1)
    )
    Index 2 Param 1=&FloatFromCurve(FloatCurve=DataInterface() {propertyValues {Curve {keys [{interpMode=RCIM_Cubic, leaveTangent=3.2309, tangentMode=RCTM_Break, time=0.1, value=0}, <at 1 1>]}}})
    Index 2 Param 2=$Particles.MaterialRandom
    Index 2 Param 3=&FloatFromCurve(FloatCurve=DataInterface() {propertyValues {Curve {keys [
         {interpMode=RCIM_Cubic, tangentMode=RCTM_Break, time=0, value=1}
         {interpMode=RCIM_Cubic, leaveTangent=11.1233, tangentMode=RCTM_Break, time=1, value=2}
        ]}}})
    Index 2 Param 4=0.75
    Module.Param0WriteEnabled=true
    Module.Param0WriteEnabled1=true
    Module.Param0WriteEnabled2=true
    Module.Param1WriteEnabled=false
    Module.Param1WriteEnabled1=true
    Module.Param1WriteEnabled2=true
    Module.Param2WriteEnabled=true
    Module.Param2WriteEnabled1=true
    Module.Param2WriteEnabled2=true
    Module.Param3WriteEnabled=true
    Module.Param3WriteEnabled1=true
    Module.Param3WriteEnabled2=true
    Write Parameter Index 1=true
    Write Parameter Index 2=true
   )
   &Drag(
    Drag=&MergeHelper_FloatSwitch(
     0=5
     1=0
     2=&RandomRangeFloat(Minimum=1, Maximum=8)
     3=?
     Index=&SpawnGroup()
     Num=::NewEnumerator1
    )
    Use Linear Drag=true
    Use Rotational Drag=true
   )
   &GravityForce(Gravity=&MergeHelper_VectorSwitch(
     0=<0 0 0>
     1=<0 0 0>
     2=&RandomRangeVector(Minimum=<0 0 50>, Maximum=<0 0 150>)
     3=?
     Index=&SpawnGroup()
     Num=::NewEnumerator1
    ))
   &SolveForcesAndVelocity(Clamp Velocity=false, Limit Acceleration=false)
  ]
  rendererClasses [&NiagaraSpriteRendererProperties]
  renderers [
   {rendererClass=&NiagaraSpriteRendererProperties}
   {rendererClass=&NiagaraSpriteRendererProperties, rendererIndex=1}
   {rendererClass=&NiagaraSpriteRendererProperties, rendererIndex=2}
  ]
 }
 Spreading_fire004 {
  emitterUpdateScript [
   &EmitterState(Life Cycle Mode=::NewEnumerator1, Loop Behavior=::NewEnumerator1, Loop Delay=0.57, Loop Duration=6.5, UseLoopDelay=true)
   &SpawnBurst_Instantaneous(Spawn Count=15, Spawn Group=1, Use Spawn Probability=false, UseLoopCountLimit=false) {enabled=false, module=SpawnBurst_Instantaneous002}
   &SpawnBurst_Instantaneous(Spawn Group=4, Use Spawn Probability=false, UseLoopCountLimit=false) {module=SpawnBurst_Instantaneous001}
   &SpawnBurst_Instantaneous(Spawn Group=3, Use Spawn Probability=false, UseLoopCountLimit=false)
   &SpawnRate(SpawnGroup=2, SpawnRate=35, Use Spawn Probability=false) {module=SpawnRate002}
   &SpawnRate(SpawnGroup=1, SpawnRate=10, Use Spawn Probability=false) {module=SpawnRate003}
   &SpawnRate(SpawnRate=50, Use Spawn Probability=false)
  ]
  particleSpawnScript [
   /SetVariables(Particles.VisibilityTag=&SpawnGroup()) {isSetParameters=true, module=SetVariables_#7}
   /SetVariables(Particles.SideIndex:NiagaraInt32=&ModuloInt(A=&ReturnParticleIDAttributes(), B=2)) {isSetParameters=true}
   &SkeletalMeshLocation(MaskSpawnGroup=false, OverrideRandomSeed=false, Skeletal Mesh=$System.body skeletal mesh)
   &InitializeParticle(
    Color=<rgba 1 1 1 0.5>
    Lifetime=&MergeHelper_FloatSwitch(
     0=&RandomRangeFloat(Minimum=0.2, Maximum=1.5)
     1=&UniformRangedFloat(Minimum=0.65, Maximum=1.5)
     2=&UniformRangedFloat(Minimum=1, Maximum=1.7)
     3=7.5
     4=9
     Index=&SpawnGroup()
     Num=::NewEnumerator3
    )
    Mass Mode=::NewEnumerator1
    Position Offset=<0 0 -115>
    Sprite Rotation Angle=&UniformRangedFloat(Minimum=-40, Maximum=40)
    Sprite Rotation Mode=::NewEnumerator0
    Sprite Size=&Vector2DFromFloat(Value=&UniformRangedFloat(Minimum=14, Maximum=20))
    Sprite Size Mode=::NewEnumerator4
    UsePositionOffset=true
    Write Material Random=true
   ) {module=InitializeParticle001}
   &ShapeLocation(
    MaskSpawnGroup=false
    Non Uniform Scale=&MakeVector(
     X=1
     Y=&Add_Float(
      A=&Multiply_Float(A=2, B=&MakeFloatFromInt(Int=$Particles.SideIndex))
      B=-1
     )
     Z=1
    )
    Override Seed=false
    Radius Position:NiagaraFloat=$System.Ring of fire Radius
    Ring / Disc Distribution Mode:ENiagaraSphereDistributionMode=::NewEnumerator1
    Ring Radius=50
    Rotation Mode=::NewEnumerator2
    Shape Primitive=::NewEnumerator4
    U Position=&ModuloFloat(A=&Multiply_Float(A=$Particles.MaterialRandom, B=1), B=0.3)
    UseOwnerScale=true
    Yaw / Pitch / Roll=<0.6889 0 0>
   )
   /SetVariables(Particles.Initial Main flame Sprite Size:NiagaraFloat=&RandomRangeFloat(Minimum=35, Maximum=50)) {isSetParameters=true}
   /SetVariables(Particles.Glow on feet Initial SpriteSize:Vector2f=&RandomRangeVector2D(Minimum=<20 30>, Maximum=<50 55>)) {isSetParameters=true}
  ]
  particleUpdateScript [
   &ParticleState()
   /SetVariables(Particles.ParticleRelativeLifetimeSprites:NiagaraFloat=&Frac_Float(Float=&Add_Float(A=&Multiply_Float(A=1.5, B=$Emitter.Age), B=&ReturnNormalizedExecIndex()))) {isSetParameters=true, module=SetVariables_#8}
   /SetVariables(Particles.OpacityMultiplierSprites:NiagaraFloat=&FloatFromCurve(
     FloatCurve=DataInterface() {propertyValues {Curve {keys [
         {interpMode=RCIM_Cubic, tangentMode=RCTM_User, time=0, value=0}
         {interpMode=RCIM_Cubic, tangentMode=RCTM_User, time=0.5, value=1}
         {tangentMode=RCTM_User, time=1, value=0}
        ]}}}
     CurveIndex=&Add_Float(
      A=&Multiply_Float(A=$Particles.ParticleRelativeLifetimeSprites, B=-1)
      B=&Multiply_Float(A=50, B=2)
     )
    )) {isSetParameters=true, module=SetVariables_#9}
   /SetVariables(Particles.InterpolateCurveSprites:NiagaraFloat=&FloatFromCurve(
     FloatCurve=DataInterface() {propertyValues {Curve {keys [
         {arriveTangent=-0.6503, interpMode=RCIM_Cubic, leaveTangent=1.4736, tangentMode=RCTM_Break, time=0, value=0}
         {arriveTangent=0.4202, interpMode=RCIM_Cubic, tangentMode=RCTM_Break, time=1, value=1}
        ]}}}
     CurveIndex=$Particles.ParticleRelativeLifetimeSprites
    )) {isSetParameters=true, module=SetVariables_#10}
   &CameraOffset(Camera Offset Amount=5)
   &SpriteRotationRate(Rotation Rate=&MergeHelper_FloatSwitch(0=0, 1=0, 2=1, 3=0, 4=0, Index=&SpawnGroup(), Num=::NewEnumerator3))
   &AccelerationForce(
    Acceleration=&MergeHelper_VectorSwitch(
     0=&UniformRangedVector(Minimum=<0 10 0>, Maximum=<0 -15 500>)
     1=<0 0 0>
     2=<0 0 0>
     3=<0 0 0>
     4=<0 0 0>
     Index=$Particles.VisibilityTag
     Num=::NewEnumerator3
    )
    Coordinate Space=::Local
   )
   &GravityForce(Gravity=&MergeHelper_VectorSwitch(
     0=&UniformRangedVector(Minimum=<0 10 0>, Maximum=<0 -15 0>)
     1=<0 0 0>
     2=&RandomRangeVector(Minimum=<0 0 0>, Maximum=<1 1 115>)
     3=<0 0 0>
     4=<0 0 0>
     Index=&SpawnGroup()
     Num=::NewEnumerator3
    ))
   &SubUVAnimation(
    Output Alternate Renderer Binding=true
    Random Start Frame=true
    "Sprite Renderer "=DataInterface() {propertyValues {SpriteRenderer=&self:Spreading_fire004_0.NiagaraSpriteRendererProperties_2}}
    SubUV Animation Mode:ENiagara_SubUVLookupModeV2=::NewEnumerator1
    UseEndFrame=false
    UseStartFrame=false
   ) {module=SubUVAnimation001}
   /SetVariables(
    Particles.Smoke Size:Vector2f=&Multiply_Vector2D_ByFloat(
     Float=&FloatFromCurve(
      FloatCurve=DataInterface() {propertyValues {Curve {keys [
          {arriveTangent=2.4634, interpMode=RCIM_Cubic, leaveTangent=2.4634, tangentMode=RCTM_User, time=0, value=0.1}
          {arriveTangent=0.2754, interpMode=RCIM_Cubic, leaveTangent=0.2754, tangentMode=RCTM_User, time=1, value=1}
         ]}}}
      Scale Curve=&Multiply_Float(A=&RandomRangeFloat(Minimum=1, Maximum=5), B=$User.Fire_size)
     )
     Vector2D=&RandomRangeVector2D(Minimum=<5 5>, Maximum=<10 10>)
    )
    Particles.Smoke Velocity:Vector3f=&TransformVector(
     Vector=&MakeVector(X=0, Y=0, Z=&RandomRangeFloat(Minimum=85, Maximum=150))
     Source Space=::World
     Destination Space=::Local
    )
    Particles.smoke Dyn Param:Vector4f=&MakeVector4(X=5, Y=0, Z=0, W=0)
    Particles.smoke color:LinearColor=&MakeLinearColorFromVectorAndFloat(
     "Vector (RGB)"=&MakeVectorFromLinearColorRGB(Color=&ColorFromCurve(ColorCurve=DataInterface() {propertyValues {
         AlphaCurve {keys [<at 0 1 RCIM_Cubic>]}
         BlueCurve {keys [<at 0 0.125 RCIM_Cubic>,<at 1 0.0312 RCIM_Cubic>]}
         GreenCurve {keys [<at 0 0.125 RCIM_Cubic>,<at 1 0.0312 RCIM_Cubic>]}
         RedCurve {keys [<at 0 0.125 RCIM_Cubic>,<at 1 0.0312 RCIM_Cubic>]}
        }}))
     "Float (Alpha)"=&FloatFromCurve(
      FloatCurve=DataInterface() {propertyValues {Curve {
         defaultValue=0
         keys [
          {arriveTangent=0.6761, interpMode=RCIM_Cubic, leaveTangent=0.6761, tangentMode=RCTM_User, time=0, value=0}
          {arriveTangent=2.0828, interpMode=RCIM_Cubic, leaveTangent=2.0828, tangentMode=RCTM_User, time=0.2554, value=0.7973}
          {arriveTangent=-0.0405, interpMode=RCIM_Cubic, leaveTangent=-0.0405, tangentMode=RCTM_User, time=1, value=0}
         ]
        }}}
      Scale Curve=0.5
     )
    )
   ) {isSetParameters=true}
   /SetVariables(
    "Particles.Main Fire DynParam ":Vector4f=&MakeVector4(
     X=&FloatFromCurve(
      FloatCurve=DataInterface() {propertyValues {Curve {
         defaultValue=1
         keys [{interpMode=RCIM_Cubic, tangentMode=RCTM_User, time=0, value=0}, {interpMode=RCIM_Cubic, tangentMode=RCTM_User, time=1.0015, value=0.6706}]
        }}}
      Scale Curve=0.85
     )
     Y=&RandomRangeFloat(Minimum=0, Maximum=4)
     Z=0
     W=2.5
    )
    Particles.Main Fire SpriteSize:Vector2f=&Multiply_Vector2D(
     A=&Vector2DFromFloat(Value=$Particles.Initial Main flame Sprite Size)
     B=&Multiply_Vector2D_ByFloat(
      Float=&FloatFromCurve(
       FloatCurve=DataInterface() {propertyValues {Curve {keys [
           {arriveTangent=1.4708, interpMode=RCIM_Cubic, leaveTangent=1.4708, tangentMode=RCTM_User, time=0, value=1.3}
           {arriveTangent=0.0624, interpMode=RCIM_Cubic, leaveTangent=0.0624, time=0.6, value=1.8}
           {interpMode=RCIM_Cubic, tangentMode=RCTM_User, time=1.0543, value=1.3657}
          ]}}}
       Scale Curve=&RandomRangeFloat(Minimum=0.5, Maximum=1)
      )
      Vector2D=<0.8 0.8>
     )
    )
    Particles.Main Flames color:LinearColor=&MakeLinearColorFromVectorAndFloat(
     "Vector (RGB)"=&VectorFromFloat(Value=&FloatFromCurve(
       FloatCurve=DataInterface() {propertyValues {Curve {keys [
           {arriveTangent=-0.4927, interpMode=RCIM_Cubic, leaveTangent=-0.4927, tangentMode=RCTM_User, time=0, value=0.1087}
           {arriveTangent=-0.5399, interpMode=RCIM_Cubic, leaveTangent=-0.5399, tangentMode=RCTM_User, time=0.1564, value=0.3915}
           {arriveTangent=-0.4927, interpMode=RCIM_Cubic, leaveTangent=-0.4927, tangentMode=RCTM_User, time=0.987, value=0.0549}
          ]}}}
       Scale Curve=&RandomRangeFloat(Minimum=2, Maximum=5)
      ))
     "Float (Alpha)"=&RandomRangeFloat(
      Minimum=&FloatFromCurve(
       FloatCurve=DataInterface() {propertyValues {Curve {keys [
           {arriveTangent=3.3783, interpMode=RCIM_Cubic, leaveTangent=3.3783, tangentMode=RCTM_User, time=0, value=0}
           {arriveTangent=0.2941, interpMode=RCIM_Cubic, leaveTangent=0.2941, tangentMode=RCTM_User, time=0.203, value=0.7175}
           {arriveTangent=-1.6053, interpMode=RCIM_Cubic, leaveTangent=-1.6053, time=0.5122, value=0.5742}
           {arriveTangent=-0.0646, interpMode=RCIM_Cubic, leaveTangent=-0.0646, tangentMode=RCTM_User, time=0.65, value=0}
          ]}}}
       Scale Curve=&Multiply_Float(A=&RandomRangeFloat(Minimum=0.3, Maximum=1), B=$System.TimeOfDay)
      )
      Maximum=&FloatFromCurve(
       FloatCurve=DataInterface() {propertyValues {Curve {keys [
           {arriveTangent=3.3783, interpMode=RCIM_Cubic, leaveTangent=3.3783, tangentMode=RCTM_User, time=0, value=0}
           {arriveTangent=0.2941, interpMode=RCIM_Cubic, leaveTangent=0.2941, tangentMode=RCTM_User, time=0.203, value=0.7175}
           {arriveTangent=-1.109, interpMode=RCIM_Cubic, leaveTangent=-1.109, time=0.5122, value=0.5742}
           {arriveTangent=-0.0646, interpMode=RCIM_Cubic, leaveTangent=-0.0646, tangentMode=RCTM_User, time=0.85, value=0}
          ]}}}
       Scale Curve=&Multiply_Float(A=&RandomRangeFloat(Minimum=0.3, Maximum=1), B=$System.TimeOfDay)
      )
     )
    )
   ) {isSetParameters=true}
   /SetVariables(
    Particles.Scorch Color:LinearColor=&MakeLinearColorFromVectorAndFloat(
     "Vector (RGB)"=&MakeVectorFromLinearColorRGB(Color=<rgba 5 1.186 0.5217 1>)
     "Float (Alpha)"=&FloatFromCurve(
      FloatCurve=DataInterface() {propertyValues {Curve {
         defaultValue=1
         keys [
          {interpMode=RCIM_Cubic, tangentMode=RCTM_User, time=0, value=0}
          {interpMode=RCIM_Cubic, tangentMode=RCTM_User, time=0.2654, value=0.9958}
          {arriveTangent=-1.3557, interpMode=RCIM_Cubic, leaveTangent=-1.3557, time=0.7691, value=0.8518}
          {tangentMode=RCTM_User, time=1, value=0}
         ]
        }}}
      Scale Curve=&Multiply_Float(
       A=1
       B=&FN_FrontEndFloatModifier(
        Gameplay=&FNTimeOfDayScalar(Night Scale=10, Morning Scale=10, Day Scale=15, Evening Scale=10)
        Front End=5
       )
      )
     )
    )
    Particles.Scorch SpriteSize:Vector2f=&RandomRangeVector2D(Minimum=<25 25>, Maximum=<35 35>)
    Particles.scorch SpriteFacing:Vector3f=<0 0 1>
   ) {isSetParameters=true}
   &Drag(
    Drag=&MergeHelper_FloatSwitch(
     0=&RandomRangeFloat(Minimum=3, Maximum=8)
     1=0
     2=0
     3=0
     4=0
     Index=&SpawnGroup()
     Num=::NewEnumerator3
    )
    Use Linear Drag=true
    Use Rotational Drag=true
   )
   &SolveForcesAndVelocity(Clamp Velocity=false, Limit Acceleration=false)
   &ScaleSpriteSizeBySpeed(Sample Scale Factor By Curve=true)
   &FN_Light_Attributes(
    Color=<rgba 3 1.1224 0.4219 1>
    Create Color Attribute=true
    Create Exponent Attribute=true
    Create Position Attribute=true
    Create Radius Attribute=true
    Enable Particle Light Enable=true
    Enable Visibility Tag=true
    Exponent=2
    Light Enabled=true
    Position=<0 15 -13>
    Radius=100
    Scale Color=&Multiply_Float(
     A=&FloatFromCurve(
      FloatCurve=DataInterface() {propertyValues {Curve {keys [
          {interpMode=RCIM_Cubic, tangentMode=RCTM_Break, time=0, value=0}
          {interpMode=RCIM_Cubic, tangentMode=RCTM_Break, time=0.0756, value=0.9906}
          {arriveTangent=-0.2434, interpMode=RCIM_Cubic, leaveTangent=-0.2434, time=0.5486, value=0.5552}
          {interpMode=RCIM_Cubic, tangentMode=RCTM_Break, time=1.0034, value=0.7648}
          <at 1.323 0 RCIM_Cubic>
         ]}}}
      Scale Curve=0.5
     )
     B=&Multiply_Float(
      A=&Waveform("Waveform [1]"=::NewEnumerator0, "[1] Phase Scale"=0.2, "[1] Frequency"=3, "[1] Amplitude Min/Max"=<0.7 1>, "[1] Scale"=3)
      B=&FNTimeOfDayScalar(Night Scale=1, Morning Scale=1, Day Scale=1.5, Evening Scale=1)
     )
    )
    VisibilityTag=3
   )
   &FN_Light_Attributes(
    Color=<rgba 3 1.1224 0.4219 1>
    Create Color Attribute=true
    Create Exponent Attribute=true
    Create Position Attribute=true
    Create Radius Attribute=true
    Enable Particle Light Enable=true
    Enable Visibility Tag=true
    Exponent=5
    Light Enabled=true
    Position=&MakeVector(
     X=-8
     Y=&FloatFromCurve(FloatCurve=DataInterface() {propertyValues {Curve {keys [<at 0 -60>,<at 2 1>]}}})
     Z=-5
    )
    Radius=45
    Scale Color=&Multiply_Float(
     A=&FloatFromCurve(
      FloatCurve=DataInterface() {propertyValues {Curve {keys [
          {interpMode=RCIM_Cubic, tangentMode=RCTM_Break, time=0, value=0}
          {interpMode=RCIM_Cubic, tangentMode=RCTM_Break, time=0.0629, value=1}
          {interpMode=RCIM_Cubic, tangentMode=RCTM_Break, time=1.0034, value=0.7648}
          <at 1.323 0 RCIM_Cubic>
         ]}}}
      Scale Curve=0.5
     )
     B=&Waveform(
      "Waveform [1]"=::NewEnumerator1
      "[1] Phase Scale"=5
      "[1] Frequency"=1
      "[1] Amplitude Min/Max"=<0.75 1>
      "[1] Scale"=&FNTimeOfDayScalar(Night Scale=2, Morning Scale=2, Day Scale=1.75, Evening Scale=2)
     )
    )
    VisibilityTag=3
   ) {module=FN_Light_Attributes001}
   /SetVariables(
    Particles.Glow On feet Color:LinearColor=&MakeLinearColorFromVectorAndFloat(
     "Vector (RGB)"=&FN_FrontEndVector3Modifier(
      Gameplay=&MakeVectorFromLinearColorRGB(Color=&FNTimeOfDayColor(
        Night Color=<rgba 5 2.4625 0.7914 1>
        Morning Color=<rgba 2.0312 0.5791 0 1>
        Day Color=<rgba 2.0312 0.5791 0 1>
        Evening Color=<rgba 2.0312 0.5791 0 1>
       ))
      Front End=&MakeVectorFromLinearColorRGB(Color=<rgba 5 2.4625 0.7914 1>)
     )
     "Float (Alpha)"=&MakeFloatFromLinearColor(
      LinearColor=/Niagara/DynamicInputs/TypeConversions/MakeLinearColorFromFloat(Value:NiagaraFloat=&Multiply_Float(
        A=&FloatFromCurve(
         FloatCurve=DataInterface() {propertyValues {Curve {keys [
             {interpMode=RCIM_Cubic, tangentMode=RCTM_Break, time=0, value=0}
             {interpMode=RCIM_Cubic, tangentMode=RCTM_Break, time=0.0756, value=0.9906}
             {arriveTangent=-1.2081, interpMode=RCIM_Cubic, leaveTangent=-1.2081, time=0.5486, value=0.5552}
             <at 0.911 -0.0186 RCIM_Cubic>
            ]}}}
         Scale Curve=0.5
        )
        B=&Multiply_Float(
         A=&Waveform(
          "Waveform [1]"=::NewEnumerator1
          "[1] Phase Scale"=&FloatFromCurve(FloatCurve=DataInterface() {propertyValues {Curve {keys [
               {interpMode=RCIM_Cubic, tangentMode=RCTM_Break, time=0, value=0}
               {interpMode=RCIM_Cubic, tangentMode=RCTM_Break, time=0.0756, value=0.9906}
               {arriveTangent=-1.2081, interpMode=RCIM_Cubic, leaveTangent=-1.2081, time=0.5486, value=0.5552}
               <at 0.911 -0.0186 RCIM_Cubic>
              ]}}})
          "[1] Frequency"=2
          "[1] Amplitude Min/Max"=<0.8 1>
          "[1] Scale"=3
         )
         B=&FN_FrontEndFloatModifier(
          Gameplay=&FNTimeOfDayScalar(Night Scale=0.5, Morning Scale=0.6, Day Scale=0.7, Evening Scale=0.3)
          Front End=0.5
         )
        )
       ))
      Channel:ENiagaraFloatFromLinearColorOptions=::NewEnumerator0
     )
    )
    Particles.Glow On feet SpriteFacing:Vector3f=<0 0 1>
    Particles.Glow On feet SpriteSize:Vector2f=&Multiply_Vector2D_ByFloat(Float=5.2, Vector2D=<50 50>)
    Particles.Glow on feet Position:NiagaraPosition=&AddVectorToPosition(
     Position=$Particles.SkeletalMeshLocation.SampledPosition
     Vector=&FN_FrontEndVector3Modifier(Gameplay=<-15 -25 -116>, Front End=<-15 -20 -116>)
    )
   ) {isSetParameters=true}
   /SetVariables(Particles.Velocity=&TransformVector(
     Vector=&MergeHelper_VectorSwitch(
      0=&RandomRangeVector(Minimum=<0 0 5>, Maximum=<0 0 10>)
      1=<0 0 0>
      2=<0 0 0>
      3=<0 0 0>
      4=<0 0 0>
      Index=$Particles.VisibilityTag
      Num=::NewEnumerator3
     )
     Source Space=::World
     Destination Space=::Local
    )) {enabled=false, isSetParameters=true}
  ]
  rendererClasses [&NiagaraSpriteRendererProperties, &NiagaraLightRendererProperties]
  renderers [
   {rendererClass=&NiagaraSpriteRendererProperties}
   {rendererClass=&NiagaraSpriteRendererProperties, rendererIndex=1}
   {rendererClass=&NiagaraSpriteRendererProperties, rendererIndex=2}
   {rendererClass=&NiagaraLightRendererProperties, rendererIndex=3}
   {rendererClass=&NiagaraLightRendererProperties, rendererIndex=4}
   {rendererClass=&NiagaraSpriteRendererProperties, rendererIndex=5}
  ]
 }
 Spreading_fire_arm {
  emitterSpawnScript [/SetVariables() {isSetParameters=true}]
  emitterUpdateScript [
   &EmitterState(Life Cycle Mode=::NewEnumerator1, Loop Behavior=::NewEnumerator1, Loop Duration=0.5, UseLoopDelay=false)
   &SpawnBurst_Instantaneous(Spawn Group=2, Use Spawn Probability=false, UseLoopCountLimit=false)
   &SpawnRate(SpawnGroup=1, SpawnRate=50, Use Spawn Probability=false) {module=SpawnRate002}
   &SpawnRate(SpawnRate=200, Use Spawn Probability=false)
  ]
  particleSpawnScript [
   /SetVariables(Particles.VisibilityTag=&SpawnGroup()) {isSetParameters=true, module=SetVariables_#7}
   /SetVariables(Particles.SideIndex:NiagaraInt32=&ModuloInt(A=&ReturnParticleIDAttributes(), B=2)) {isSetParameters=true}
   &InitializeParticle(
    Lifetime=&MergeHelper_FloatSwitch(
     0=&UniformRangedFloat(Minimum=0.5, Maximum=0.7)
     1=&UniformRangedFloat(Minimum=0.65, Maximum=1.3)
     2=0.7
     3=?
     Index=&SpawnGroup()
     Num=::NewEnumerator1
    )
    Mass Mode=::NewEnumerator1
    Position Offset=<0 0 -115>
    Sprite Rotation Mode=::NewEnumerator2
    Sprite Size=&Vector2DFromFloat(Value=&UniformRangedFloat(Minimum=14, Maximum=20))
    Sprite Size Mode=::NewEnumerator4
    UsePositionOffset=true
    Write Material Random=true
   ) {module=InitializeParticle001}
   &ShapeLocation(
    MaskSpawnGroup=false
    Non Uniform Scale=&MakeVector(
     X=1
     Y=&Add_Float(
      A=&Multiply_Float(A=2, B=&MakeFloatFromInt(Int=$Particles.SideIndex))
      B=-1
     )
     Z=1
    )
    Override Seed=false
    Radius Position:NiagaraFloat=$System.Ring of fire Radius
    Ring / Disc Distribution Mode:ENiagaraSphereDistributionMode=::NewEnumerator1
    Ring Radius=50
    Rotation Mode=::NewEnumerator2
    Shape Primitive=::NewEnumerator4
    U Position=&ModuloFloat(A=&Multiply_Float(A=$Emitter.Age, B=0.5), B=0.35)
    UseOwnerScale=true
    Yaw / Pitch / Roll=<0.6889 0 0>
   )
   /SetVariables(Particles.Initial Main flame Sprite Size:NiagaraFloat=&RandomRangeFloat(Minimum=35, Maximum=50)) {isSetParameters=true}
  ]
  particleUpdateScript [
   &ParticleState()
   &GravityForce(Gravity=&MergeHelper_VectorSwitch(
     0=<0 0 50>
     1=<0 0 150>
     2=&RandomRangeVector(Minimum=<0 0 600>, Maximum=<1 1 1500>)
     3=<0 0 0>
     4=<0 0 0>
     Index=&SpawnGroup()
     Num=::NewEnumerator3
    ))
   &SolveForcesAndVelocity(Clamp Velocity=false, Limit Acceleration=false)
   /SetVariables(Particles.ParticleRelativeLifetimeSprites:NiagaraFloat=&Frac_Float(Float=&Add_Float(A=&Multiply_Float(A=1.5, B=$Emitter.Age), B=&ReturnNormalizedExecIndex()))) {isSetParameters=true, module=SetVariables_#8}
   /SetVariables(Particles.OpacityMultiplierSprites:NiagaraFloat=&FloatFromCurve(
     FloatCurve=DataInterface() {propertyValues {Curve {keys [
         {interpMode=RCIM_Cubic, tangentMode=RCTM_User, time=0, value=0}
         {interpMode=RCIM_Cubic, tangentMode=RCTM_User, time=0.5, value=1}
         {tangentMode=RCTM_User, time=1, value=0}
        ]}}}
     CurveIndex=&Add_Float(
      A=&Multiply_Float(A=$Particles.ParticleRelativeLifetimeSprites, B=-1)
      B=&Multiply_Float(A=50, B=2)
     )
    )) {isSetParameters=true, module=SetVariables_#9}
   /SetVariables(Particles.InterpolateCurveSprites:NiagaraFloat=&FloatFromCurve(
     FloatCurve=DataInterface() {propertyValues {Curve {keys [
         {arriveTangent=-0.6503, interpMode=RCIM_Cubic, leaveTangent=1.4736, tangentMode=RCTM_Break, time=0, value=0}
         {arriveTangent=0.4202, interpMode=RCIM_Cubic, tangentMode=RCTM_Break, time=1, value=1}
        ]}}}
     CurveIndex=$Particles.ParticleRelativeLifetimeSprites
    )) {isSetParameters=true, module=SetVariables_#10}
   &ScaleColor(
    Scale Alpha=&Multiply_Float(
     A=&FloatFromCurve(FloatCurve=DataInterface() {propertyValues {Curve {keys [
          {arriveTangent=16.5612, interpMode=RCIM_Cubic, leaveTangent=16.5612, leaveTangentWeight=0.3333, tangentMode=RCTM_User, time=0, value=0}
          {arriveTangentWeight=0.0333, interpMode=RCIM_Cubic, leaveTangentWeight=0.3, tangentMode=RCTM_User, time=0.1, value=1}
          {arriveTangentWeight=0.1333, interpMode=RCIM_Cubic, leaveTangentWeight=0.1667, tangentMode=RCTM_User, time=0.5, value=1}
          {arriveTangentWeight=0.1667, interpMode=RCIM_Cubic, tangentMode=RCTM_User, time=1, value=0}
         ]}}})
     B=&RandomRangeFloat(Minimum=0.5, Maximum=1)
    )
    Scale RGB=&Multiply_VectorByFloat(
     Vector=<1 1 1>
     Float=&FloatFromCurve(
      FloatCurve=DataInterface() {propertyValues {Curve {keys [
          {arriveTangent=9.0725, interpMode=RCIM_Cubic, leaveTangent=9.0725, tangentMode=RCTM_User, time=0, value=0}
          {arriveTangent=-2.5249, interpMode=RCIM_Cubic, leaveTangent=-2.5249, tangentMode=RCTM_User, time=0.1408, value=0.9847}
          {arriveTangent=-0.0405, interpMode=RCIM_Cubic, leaveTangent=-0.0405, tangentMode=RCTM_User, time=1, value=0}
         ]}}}
      Scale Curve=&RandomRangeFloat(Minimum=0.5, Maximum=1)
     )
    )
    ScaleA=true
    ScaleRGB=true
   )
   /SetVariables(
    "Particles.Main Fire DynParam ":Vector4f=&MakeVector4(
     X=&FloatFromCurve(FloatCurve=DataInterface() {propertyValues {Curve {
         defaultValue=1
         keys [{interpMode=RCIM_Cubic, tangentMode=RCTM_User, time=0, value=0}, {interpMode=RCIM_Cubic, tangentMode=RCTM_User, time=1.0015, value=0.6706}]
        }}})
     Y=&RandomRangeFloat(Minimum=0, Maximum=4)
     Z=0
     W=2.5
    )
    Particles.Main Fire SpriteSize:Vector2f=&Multiply_Vector2D(
     A=&Vector2DFromFloat(Value=$Particles.Initial Main flame Sprite Size)
     B=&Multiply_Vector2D_ByFloat(
      Float=&FloatFromCurve(
       FloatCurve=DataInterface() {propertyValues {Curve {keys [
           {arriveTangent=1.4708, interpMode=RCIM_Cubic, leaveTangent=1.4708, tangentMode=RCTM_User, time=0, value=1.3}
           {arriveTangent=0.0624, interpMode=RCIM_Cubic, leaveTangent=0.0624, time=0.6, value=1.8}
           {interpMode=RCIM_Cubic, tangentMode=RCTM_User, time=1.0543, value=1.3657}
          ]}}}
       Scale Curve=&RandomRangeFloat(Minimum=0.5, Maximum=1)
      )
      Vector2D=<0.8 0.8>
     )
    )
    Particles.Main Flames color:LinearColor=&MakeLinearColorFromVectorAndFloat(
     "Vector (RGB)"=&VectorFromFloat(Value=&FloatFromCurve(
       FloatCurve=DataInterface() {propertyValues {Curve {keys [
           {arriveTangent=-0.4927, interpMode=RCIM_Cubic, leaveTangent=-0.4927, tangentMode=RCTM_User, time=0, value=0.1087}
           {arriveTangent=-0.5399, interpMode=RCIM_Cubic, leaveTangent=-0.5399, tangentMode=RCTM_User, time=0.1564, value=0.3915}
           {arriveTangent=-0.4927, interpMode=RCIM_Cubic, leaveTangent=-0.4927, tangentMode=RCTM_User, time=0.987, value=0.0549}
          ]}}}
       Scale Curve=&RandomRangeFloat(Minimum=2, Maximum=5)
      ))
     "Float (Alpha)"=&RandomRangeFloat(
      Minimum=&FloatFromCurve(
       FloatCurve=DataInterface() {propertyValues {Curve {keys [
           {arriveTangent=3.3783, interpMode=RCIM_Cubic, leaveTangent=3.3783, tangentMode=RCTM_User, time=0, value=0}
           {arriveTangent=0.2941, interpMode=RCIM_Cubic, leaveTangent=0.2941, tangentMode=RCTM_User, time=0.203, value=0.7175}
           {arriveTangent=-1.6053, interpMode=RCIM_Cubic, leaveTangent=-1.6053, time=0.5122, value=0.5742}
           {arriveTangent=-0.0646, interpMode=RCIM_Cubic, leaveTangent=-0.0646, tangentMode=RCTM_User, time=0.65, value=0}
          ]}}}
       Scale Curve=&Multiply_Float(A=&RandomRangeFloat(Minimum=0.3, Maximum=1.5), B=$System.TimeOfDay)
      )
      Maximum=&FloatFromCurve(
       FloatCurve=DataInterface() {propertyValues {Curve {keys [
           {arriveTangent=3.3783, interpMode=RCIM_Cubic, leaveTangent=3.3783, tangentMode=RCTM_User, time=0, value=0}
           {arriveTangent=0.2941, interpMode=RCIM_Cubic, leaveTangent=0.2941, tangentMode=RCTM_User, time=0.203, value=0.7175}
           {arriveTangent=-1.109, interpMode=RCIM_Cubic, leaveTangent=-1.109, time=0.5122, value=0.5742}
           {arriveTangent=-0.0646, interpMode=RCIM_Cubic, leaveTangent=-0.0646, tangentMode=RCTM_User, time=0.85, value=0}
          ]}}}
       Scale Curve=&Multiply_Float(A=&RandomRangeFloat(Minimum=0.3, Maximum=1), B=$System.TimeOfDay)
      )
     )
    )
    Particles.Velocity=&TransformVector(Vector=<0 0 1>, Source Space=::World, Destination Space=::Local)
   ) {isSetParameters=true}
   /SetVariables(Particles.embers Velocity:Vector3f=<0 0 70>) {isSetParameters=true}
   &CameraOffset(Camera Offset Amount=5)
   &ScaleSpriteSize(Scale Factor=&Vector2DFromFloat(Value=&FloatFromCurve(
      FloatCurve=DataInterface() {propertyValues {Curve {keys [
          <at 0 0.7 RCIM_Cubic>
          {arriveTangent=4, interpMode=RCIM_Cubic, leaveTangent=4, time=0.1496, value=2}
          {arriveTangent=-0.565, interpMode=RCIM_Cubic, leaveTangent=-0.565, tangentMode=RCTM_User, time=0.2, value=1.5}
          {arriveTangent=-1.5177, interpMode=RCIM_Cubic, leaveTangent=-1.5177, tangentMode=RCTM_User, time=1, value=0.75}
         ]}}}
      Scale Curve=&RandomRangeFloat(Minimum=0.5, Maximum=2.5)
     ))) {module=ScaleSpriteSize001}
   &DynamicMaterialParameters(
    Index 0 Param 1=&RandomRangeFloat(Minimum=2, Maximum=1)
    Index 0 Param 2=&RandomRangeFloat(Minimum=0, Maximum=2)
    Index 0 Param 3=8
    Index 0 Param 4=$Particles.SubImageIndex
    Index 1 Param 1=&RandomRangeFloat(Minimum=-1, Maximum=1)
    Index 1 Param 2=&RandomRangeFloat(Minimum=-1, Maximum=1)
    Index 1 Param 3=&FloatFromCurve(FloatCurve=DataInterface() {propertyValues {Curve {keys [
         {arriveTangent=4.1785, interpMode=RCIM_Cubic, leaveTangent=4.1785, tangentMode=RCTM_User, time=0, value=0}
         {arriveTangent=0.9995, interpMode=RCIM_Cubic, leaveTangent=0.9995, tangentMode=RCTM_User, time=0.2223, value=0.6825}
         {arriveTangent=-0.0405, interpMode=RCIM_Cubic, leaveTangent=-0.0405, tangentMode=RCTM_User, time=0.7, value=0.9}
         <at 1 0.7>
        ]}}})
    Index 2 Param 4=1
    Module.Param0WriteEnabled=true
    Module.Param0WriteEnabled1=true
    Module.Param0WriteEnabled2=false
    Module.Param1WriteEnabled=true
    Module.Param1WriteEnabled1=true
    Module.Param1WriteEnabled2=false
    Module.Param2WriteEnabled=true
    Module.Param2WriteEnabled1=true
    Module.Param2WriteEnabled2=false
    Module.Param3WriteEnabled=true
    Module.Param3WriteEnabled1=false
    Module.Param3WriteEnabled2=true
    Write Parameter Index 1=true
    Write Parameter Index 2=true
   ) {enabled=false}
   &SubUVAnimation(
    Output Alternate Renderer Binding=true
    "Sprite Renderer "=DataInterface() {propertyValues {SpriteRenderer=&self:Spreading_fire_arm_0.NiagaraSpriteRendererProperties_5}}
    SubUV Animation Mode:ENiagara_SubUVLookupModeV2=::NewEnumerator1
    SubUV Curve=DataInterface() {propertyValues {Curve {keys [
        {arriveTangent=4.1785, interpMode=RCIM_Cubic, leaveTangent=4.1785, tangentMode=RCTM_User, time=0, value=0}
        {arriveTangent=1.1449, interpMode=RCIM_Cubic, leaveTangent=1.1449, tangentMode=RCTM_User, time=0.2, value=0.6}
        {arriveTangent=0.2086, interpMode=RCIM_Cubic, leaveTangent=0.2086, tangentMode=RCTM_User, time=1, value=1}
       ]}}}
    UseEndFrame=true
    UseStartFrame=true
   ) {module=SubUVAnimation001}
   /SetVariables(
    Particles.Smoke Size:Vector2f=&Multiply_Vector2D_ByFloat(
     Float=&FloatFromCurve(
      FloatCurve=DataInterface() {propertyValues {Curve {keys [
          {arriveTangent=2.4634, interpMode=RCIM_Cubic, leaveTangent=2.4634, tangentMode=RCTM_User, time=0, value=0.1}
          {arriveTangent=0.2754, interpMode=RCIM_Cubic, leaveTangent=0.2754, tangentMode=RCTM_User, time=1, value=1}
         ]}}}
      Scale Curve=&Multiply_Float(A=&RandomRangeFloat(Minimum=1, Maximum=5), B=$User.Fire_size)
     )
     Vector2D=&RandomRangeVector2D(Minimum=<5 5>, Maximum=<10 10>)
    )
    Particles.Smoke Velocity:Vector3f=&TransformVector(
     Vector=&MakeVector(X=0, Y=0, Z=&RandomRangeFloat(Minimum=85, Maximum=150))
     Source Space=::World
     Destination Space=::Local
    )
    Particles.smoke Dyn Param:Vector4f=&MakeVector4(X=5, Y=0, Z=0, W=0)
    Particles.smoke color:LinearColor=&MakeLinearColorFromVectorAndFloat(
     "Vector (RGB)"=&MakeVectorFromLinearColorRGB(Color=&ColorFromCurve(ColorCurve=DataInterface() {propertyValues {
         AlphaCurve {keys [<at 0 1 RCIM_Cubic>]}
         BlueCurve {keys [<at 0 0 RCIM_Cubic>,<at 1 0 RCIM_Cubic>]}
         GreenCurve {keys [<at 0 0.012 RCIM_Cubic>,<at 1 0 RCIM_Cubic>]}
         RedCurve {keys [<at 0 0.0365 RCIM_Cubic>,<at 1 0 RCIM_Cubic>]}
        }}))
     "Float (Alpha)"=&FloatFromCurve(
      FloatCurve=DataInterface() {propertyValues {Curve {
         defaultValue=0
         keys [
          {arriveTangent=0.6761, interpMode=RCIM_Cubic, leaveTangent=0.6761, tangentMode=RCTM_User, time=0, value=0}
          {arriveTangent=2.0828, interpMode=RCIM_Cubic, leaveTangent=2.0828, tangentMode=RCTM_User, time=0.2554, value=0.7973}
          {arriveTangent=-0.0405, interpMode=RCIM_Cubic, leaveTangent=-0.0405, tangentMode=RCTM_User, time=1, value=0}
         ]
        }}}
      Scale Curve=0.5
     )
    )
   ) {isSetParameters=true}
   &FN_Light_Attributes(
    Color=<rgba 3 1.1224 0.4219 1>
    Create Color Attribute=true
    Create Exponent Attribute=true
    Create Position Attribute=true
    Create Radius Attribute=true
    Enable Particle Light Enable=true
    Enable Visibility Tag=true
    Exponent=2
    Light Enabled=true
    Position=<0 50 -30>
    Radius=&FloatFromCurve(
     FloatCurve=DataInterface() {propertyValues {Curve {
        defaultValue=1
        keys [{interpMode=RCIM_Cubic, tangentMode=RCTM_User, time=0, value=0}, {interpMode=RCIM_Cubic, tangentMode=RCTM_User, time=0.5, value=1}]
       }}}
     Scale Curve=100
    )
    Scale Color=&Multiply_Float(
     A=&FloatFromCurve(FloatCurve=DataInterface() {propertyValues {Curve {keys [<at 0 0>,<at 0.0316 1>,<at 1 0>]}}})
     B=&FNTimeOfDayScalar(Night Scale=0.5, Morning Scale=0.6, Day Scale=0.45, Evening Scale=0.6)
    )
    VisibilityTag=2
   )
  ]
  rendererClasses [&NiagaraSpriteRendererProperties, &NiagaraLightRendererProperties]
  renderers [
   {rendererClass=&NiagaraSpriteRendererProperties}
   {rendererClass=&NiagaraSpriteRendererProperties, rendererIndex=1}
   {rendererClass=&NiagaraLightRendererProperties, rendererIndex=2}
  ]
 }
 fire_elements {
  emitterUpdateScript [
   &EmitterState(Life Cycle Mode=::NewEnumerator1, Loop Behavior=::NewEnumerator1, Loop Delay=1.02, Loop Duration=6.2, UseLoopDelay=true)
   &SpawnRate(SpawnGroup=1, SpawnRate=&Lerp_Float(A=25, B=50, Alpha=$System.Blowing), Use Spawn Probability=false) {module=SpawnRate001}
   &SpawnRate(SpawnRate=&Lerp_Float(A=50, B=200, Alpha=$System.Blowing), Use Spawn Probability=false)
  ]
  particleSpawnScript [
   /SetVariables(Particles.VisibilityTag=&SpawnGroup()) {isSetParameters=true}
   &InitializeParticle(
    Color=&MergeHelper_ColorSwitch(0=<rgba 5 3.7043 1.51 1>, 1=<rgba 5 3.7043 1.51 1>, 2=?, 3=?, Index=&SpawnGroup(), Num=::NewEnumerator0)
    Lifetime=&MergeHelper_FloatSwitch(
     0=&RandomRangeFloat(Minimum=0.3, Maximum=1)
     1=&RandomRangeFloat(Minimum=0.5, Maximum=1.5)
     2=0
     3=0
     4=0
     Index=&SpawnGroup()
     Num=::NewEnumerator3
    )
    Sprite Size Mode=::NewEnumerator3
    Uniform Sprite Size=1.7
    UsePositionOffset=false
    Write Material Random=true
   )
   &AddVelocity(Velocity=&MergeHelper_VectorSwitch(
     0=&RandomRangeVector(Minimum=<-200 -150 20>, Maximum=<200 150 200>)
     1=&RandomRangeVector(Minimum=<-200 0 50>, Maximum=<200 0 300>)
     2=<0 0 0>
     3=<0 0 0>
     4=<0 0 0>
     Index=&SpawnGroup()
     Num=::NewEnumerator3
    ))
   &ShapeLocation(
    Cylinder Height=2
    Cylinder Radius=2
    MaskSpawnGroup=true
    Override Seed=false
    Shape Primitive=::NewEnumerator1
    Surface Only=false
    UseOwnerScale=true
   ) {enabled=false}
   &ShapeLocation(
    Disc Coverage=0.4333
    MaskSpawnGroup=false
    Offset=<0 0 -115>
    Offset Mode=::NewEnumerator0
    Override Seed=false
    Ring Radius=50
    Shape Primitive=::NewEnumerator4
    U Distribution:NiagaraFloat=&Lerp_Float(A=0, B=1, Alpha=$User.Anim_Curve_FireSpread)
    UseOwnerScale=true
   ) {module=ShapeLocation001}
  ]
  particleUpdateScript [
   &AccelerationForce(Acceleration=&MergeHelper_VectorSwitch(
     0=&Multiply_Vector(
      A=&VectorFromFloat(Value=0.5)
      B=&UniformRangedVector(Minimum=<-100 0 50>, Maximum=<100 0 250>)
     )
     1=<0 0 0>
     2=<0 0 0>
     3=<0 0 0>
     4=<0 0 0>
     Index=&SpawnGroup()
     Num=::NewEnumerator3
    ))
   /Niagara/Modules/Update/Forces/CurlNoiseForce(
    Noise Frequency=0.07
    Noise Strength=&FloatFromCurve(FloatCurve=DataInterface() {propertyValues {Curve {keys [<at 0 1>,<at 1 0>]}}}, Scale Curve=150)
    Pan Noise=true
    Pan Noise Field=<0 0 0.5>
   )
   &Drag(
    Drag=&MergeHelper_FloatSwitch(
     0=&FloatFromCurve(FloatCurve=DataInterface() {propertyValues {Curve {keys [<at 0 0.5>,<at 0.5 0.5>,<at 0.8312 0.9812>]}}}, Scale Curve=9)
     1=&FloatFromCurve(
      FloatCurve=DataInterface() {propertyValues {Curve {keys [<at 0 0.5>,<at 0.2033 0.8634>,<at 0.3809 0.5736>,<at 0.4656 0.7864>,<at 0.6848 0.6234>,<at 0.7566 1>]}}}
      Scale Curve=3
     )
     2=0
     3=0
     4=0
     Index=&SpawnGroup()
     Num=::NewEnumerator3
    )
    Use Linear Drag=true
    Use Rotational Drag=true
   )
   &SolveForcesAndVelocity(Clamp Velocity=false, Limit Acceleration=false)
   /SetVariables(
    Particles.Ashes Dyn Param:Vector4f=&MakeVector4(X=1, Y=0, Z=0, W=0)
    Particles.Ashes SpriteSize:Vector2f=&Vector2DFromFloat(Value=&RandomRangeFloat(Minimum=1, Maximum=5))
    "Particles.Ashes color ":LinearColor=&MakeLinearColorFromVectorAndFloat(
     "Vector (RGB)"=&MakeVectorFromLinearColorRGB(Color=&ColorFromCurve(ColorCurve=DataInterface() {propertyValues {
         AlphaCurve {keys [<at 0 1 RCIM_Cubic>]}
         BlueCurve {keys [<at 0 1.473 RCIM_Cubic>,<at 0.6772 0.0879 RCIM_Cubic>]}
         GreenCurve {keys [<at 0 11.4982 RCIM_Cubic>,<at 0.6772 0.0919 RCIM_Cubic>]}
         RedCurve {keys [<at 0 50 RCIM_Cubic>,<at 0.6772 0.099 RCIM_Cubic>]}
        }}))
     "Float (Alpha)"=&FloatFromCurve(FloatCurve=DataInterface() {propertyValues {Curve {keys [
          <at 0 0 RCIM_Cubic>
          {arriveTangent=-0.0126, interpMode=RCIM_Cubic, leaveTangent=-0.0126, tangentMode=RCTM_User, time=0.2355, value=0.9224}
          {arriveTangent=-0.187, interpMode=RCIM_Cubic, leaveTangent=-0.187, tangentMode=RCTM_User, time=0.8, value=1}
          <at 1 0>
         ]}}})
    )
   ) {isSetParameters=true}
   &ParticleState()
   &SpriteRotationRate(Rotation Rate=&MergeHelper_FloatSwitch(0=0, 1=1, 2=0, 3=0, 4=0, Index=&SpawnGroup(), Num=::NewEnumerator3))
   &ScaleSpriteSize(Scale Factor=<0.5 1>, Scale Sprite Size Mode=::NewEnumerator2)
   &ScaleSpriteSizeBySpeed(Max Scale Factor=&MergeHelper_Vector2Switch(0=<0.5 5>, 1=<0 0>, 2=<0 0>, 3=<0 0>, 4=<0 0>, Index=&SpawnGroup(), Num=::NewEnumerator3))
   &SubUVAnimation(
    Output Alternate Renderer Binding=true
    "Sprite Renderer "=DataInterface() {propertyValues {SpriteRenderer=&self:fire_elements_0.NiagaraSpriteRendererProperties_3}}
    UseEndFrame=false
    UseStartFrame=false
   ) {module=SubUVAnimation001}
   /SetVariables(Particles.spark color:LinearColor=&MakeLinearColorFromVectorAndFloat(
     "Vector (RGB)"=&MakeVectorFromLinearColorRGB(Color=&ColorFromCurve(ColorCurve=DataInterface() {propertyValues {
         AlphaCurve {keys [<at 0 1 RCIM_Cubic>]}
         BlueCurve {keys [<at 0 7.8125 RCIM_Cubic>,<at 0.2499 0.7812>,<at 0.7114 0.2083 RCIM_Cubic>]}
         GreenCurve {keys [<at 0 34.1758 RCIM_Cubic>,<at 0.2499 4.8887>,<at 0.7114 0.8934 RCIM_Cubic>]}
         RedCurve {keys [<at 0 50 RCIM_Cubic>,<at 0.2499 10>,<at 0.7114 4 RCIM_Cubic>]}
        }}))
     "Float (Alpha)"=&FloatFromCurve(FloatCurve=DataInterface() {propertyValues {Curve {keys [<at 0.5 1 RCIM_Cubic>,<at 1 0>]}}})
    )) {isSetParameters=true}
  ]
  rendererClasses [&NiagaraSpriteRendererProperties]
  renderers [
   {rendererClass=&NiagaraSpriteRendererProperties}
   {rendererClass=&NiagaraSpriteRendererProperties, rendererIndex=1}
  ]
 }
}
summary {
 systemName=NS_OutlineLove_PropSpawn
 emitters [
  {emitterName=Prop_spawn_merged_Blend_space, rendererClasses [&NiagaraSpriteRendererProperties]}
  {emitterName=fire_elements, rendererClasses [&NiagaraSpriteRendererProperties]}
  {
   emitterName=Spreading_fire_arm
   rendererClasses [&NiagaraSpriteRendererProperties, &NiagaraLightRendererProperties]
  }
  {
   emitterName=Spreading_fire004
   rendererClasses [&NiagaraSpriteRendererProperties, &NiagaraLightRendererProperties]
  }
 ]
 userVariables [
  User.Anim_Curve_FireSpread {type=NiagaraFloat, defaultValue=1, description=""}
  User.bIsFrontEnd {type=NiagaraBool}
  User.bIsFrontEndPreview {type=NiagaraBool}
  User.bisHeadHidden {type=NiagaraBool}
  User.Fire Light Intensity {type=NiagaraFloat, defaultValue=0.5, description=""}
  User.Fire_size {type=NiagaraFloat, defaultValue=1.5, description=""}
  User.Intensity {type=NiagaraFloat, defaultValue=0, description=""}
  User.OwnerLocalVelocity {type=Vector3f, description="", defaultValue {struct=Vector3f, value=<0 0 0>}}
  User.ParaGlideLeanAlpha {type=NiagaraFloat, defaultValue=0, description=""}
  User.Reveal {type=NiagaraFloat, defaultValue=7.5322, description=""}
 ]
}
