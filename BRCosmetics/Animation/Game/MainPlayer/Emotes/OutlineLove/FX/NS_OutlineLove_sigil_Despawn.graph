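# Niagara system for the OutlineLove emote's sigil despawn: a short burst of fire, smoke, embers, ash and glow.
# One emitter fires several bursts positioned from the sigil skeletal mesh, using spawn group to switch each layer's lifetime, motion and forces.
# Spawn group becomes the visibility tag across seven sprite renderers and a light renderer, whose attribute module is disabled.
# Per-layer curves drive color, size and flipbooks, with brightness scaled by time of day and front-end state.
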NiagaraSystem /BRCosmetics/Animation/Game/MainPlayer/Emotes/OutlineLove/FX/NS_OutlineLove_sigil_Despawn
dependencies {
 usedDataInterfaces [
  &NiagaraDataInterfaceCurve
  &NiagaraDataInterfaceSpriteRendererInfo
  &NiagaraDataInterfaceColorCurve
  &NiagaraDataInterfaceSkeletalMesh
 ]
 usedDynamicInputs [&MakeFloatFromVector]
 usedModules [
  &self:SystemSpawnScript.SystemScriptSource.SystemScriptGraph.NiagaraNodeAssignment_0.SetVariables_#1
  &SystemState
  &self:SystemSpawnScript.SystemScriptSource.SystemScriptGraph.NiagaraNodeAssignment_1.SetVariables_#2
  &self:SystemSpawnScript.SystemScriptSource.SystemScriptGraph.NiagaraNodeAssignment_2.SetVariables_#3
 ]
}
emitters {Prop_spawn_end001 {
  emitterUpdateScript [
   &EmitterState(
    Inactive Response=::NewEnumerator1
    Life Cycle Mode=::NewEnumerator1
    Loop Behavior=::NewEnumerator1
    Loop Delay=3.5
    Loop Duration=2.3
    UseLoopDelay=false
   )
   &SpawnBurst_Instantaneous(Spawn Count=7, Spawn Group=5, Spawn Time=0.05, Use Spawn Probability=false, UseLoopCountLimit=false) {module=SpawnBurst_Instantaneous006}
   &SpawnBurst_Instantaneous(Spawn Group=6, Use Spawn Probability=false, UseLoopCountLimit=false) {module=SpawnBurst_Instantaneous005}
   &SpawnBurst_Instantaneous(Spawn Count=36, Spawn Group=4, Use Spawn Probability=false, UseLoopCountLimit=false) {module=SpawnBurst_Instantaneous004}
   &SpawnBurst_Instantaneous(Spawn Count=45, Spawn Group=3, Use Spawn Probability=false, UseLoopCountLimit=false) {module=SpawnBurst_Instantaneous003}
   &SpawnBurst_Instantaneous(Spawn Count=20, Spawn Group=2, Use Spawn Probability=false) {module=SpawnBurst_Instantaneous002}
   &SpawnBurst_Instantaneous(Spawn Count=1, Spawn Group=5, Use Spawn Probability=false) {enabled=false}
   &SpawnBurst_Instantaneous(Spawn Count=35, Use Spawn Probability=false) {module=SpawnBurst_Instantaneous001}
  ]
  particleSpawnScript [
   /SetVariables(Particles.VisibilityTag=&SpawnGroup()) {isSetParameters=true}
   &InitializeParticle(
    Color=&MergeHelper_ColorSwitch(
     0=<rgba 0.5255 0.5255 0.5255 1>
     1=<rgba 0.6031 0.5599 0.463 0.2>
     2=<rgba 5 3.7043 1.51 1>
     3=?
     Index=&SpawnGroup()
     Num=::NewEnumerator1
    )
    Lifetime=&MergeHelper_FloatSwitch(
     0=&UniformRangedFloat(Minimum=0.3, Maximum=0.45)
     1=0.25
     2=&UniformRangedFloat(Minimum=0.3, Maximum=0.65)
     3=&UniformRangedFloat(Minimum=0.25, Maximum=0.5)
     4=&UniformRangedFloat(Minimum=0.45, Maximum=0.8)
     5=&UniformRangedFloat(Minimum=0.3, Maximum=0.6)
     6=0.4
     Index=&SpawnGroup()
     Num=::NewEnumerator5
    )
    Position Mode=::NewEnumerator0
    Sprite Rotation Mode=::NewEnumerator2
    Sprite Size Mode=::NewEnumerator3
    Uniform Sprite Size=&MergeHelper_FloatSwitch(
     0=&UniformRangedFloat(Minimum=8, Maximum=20)
     1=65
     2=&RandomRangeFloat(Minimum=10, Maximum=25)
     3=?
     Index=&SpawnGroup()
     Num=::NewEnumerator1
    )
    Write Material Random=true
   )
   &ShapeLocation(
    MaskSpawnGroup=false
    Override Seed=false
    Sphere Radius=&MergeHelper_FloatSwitch(0=7, 1=0, 2=10, 3=5, 4=10, 5=0, Index=&SpawnGroup(), Num=::NewEnumerator4)
    Sphere Surface Distribution=1
    UseOwnerScale=true
   )
   &AddVelocity(Velocity=&MergeHelper_VectorSwitch(
     0=&RandomRangeVector(Minimum=<-135 -10 -135>, Maximum=<135 10 135>)
     1=<0 0 0>
     2=&UniformRangedVector(Minimum=<-200 0 -110>, Maximum=<200 0 110>)
     3=&RandomRangeVector(Minimum=<-120 -10 -120>, Maximum=<120 10 120>)
     4=&RandomRangeVector(Minimum=<-200 -200 -200>, Maximum=<200 200 200>)
     5=&RandomRangeVector(Minimum=<-150 -150 -150>, Maximum=<150 150 150>)
     6=<0 0 0>
     Index=&SpawnGroup()
     Num=::NewEnumerator5
    ))
   /SetVariables(Particles.InitSpriteRotRate=&UniformRangedFloat(Minimum=-1, Maximum=1)) {isSetParameters=true}
   /SetVariables(
    Particles.Fire CameraOffset:NiagaraFloat=&RandomRangeFloat(Minimum=18, Maximum=35)
    Particles.Smoke Velocity:Vector3f=&Multiply_Vector(
     A=$Particles.ShapeLocation.ShapeNormal
     B=&VectorFromFloat(Value=&RandomRangeFloat(Minimum=15, Maximum=50))
    )
    Particles.Sparkles Velocity:Vector3f=&UniformRangedVector(Minimum=<-500 -200 -200>, Maximum=<5000 200 200>)
    Particles.airglow CameraOffset=10
    Particles.smoke CameraOffset:NiagaraFloat=&RandomRangeFloat(Minimum=10, Maximum=25)
   ) {isSetParameters=true}
   &AccelerationForce(Acceleration=&MergeHelper_VectorSwitch(
     0=&UniformRangedVector(Minimum=<-25 -25 -50>, Maximum=<25 25 50>)
     1=<0 0 0>
     2=&RandomRangeVector(Minimum=<0 0 0>, Maximum=<0 0 0>)
     3=?
     Index=&SpawnGroup()
     Num=::NewEnumerator1
    ))
   &ApplyInitialForces()
   /SetVariables(Particles.Fire Energy Initial Color:LinearColor=&UniformRangedLinearColor(Minimum=<rgba 1 0.0943 0 1>, Maximum=<rgba 1 0.4811 0.1327 1>, Color Channel Mode=::NewEnumerator1)) {isSetParameters=true}
   &SkeletalMeshLocation(
    Bone Sampling Mode=::NewEnumerator4
    MaskSpawnGroup=false
    Sampled Position Coordinate Space=::Local
    Sampled Position Offset=<0 0 45>
    Skeletal Mesh=$System.sigil
   )
   &CameraOffset(Camera Offset Amount=35)
  ]
  particleUpdateScript [
   &ParticleState()
   &SpriteRotationRate(Rotation Rate=&MergeHelper_FloatSwitch(
     0=&Multiply_Float(
      A=$Particles.InitSpriteRotRate
      B=&FloatFromCurve(
       FloatCurve=DataInterface() {propertyValues {Curve {keys [
           {arriveTangent=-1.9214, interpMode=RCIM_Cubic, leaveTangent=-1.9214, tangentMode=RCTM_User, time=-0.0046, value=1.0113}
           {arriveTangent=-0.2496, interpMode=RCIM_Cubic, leaveTangent=-0.2496, tangentMode=RCTM_User, time=0.9896, value=0.16}
          ]}}}
       CurveIndex=$Emitter.Age
      )
     )
     1=0
     2=5
     3=0
     4=0
     5=&Multiply_Float(
      A=&Lerp_Float(A=-200, B=200, Alpha=$Particles.MaterialRandom)
      B=&FloatFromCurve(FloatCurve=DataInterface() {propertyValues {Curve {keys [
           {arriveTangent=-4.4572, interpMode=RCIM_Cubic, leaveTangent=-4.4572, tangentMode=RCTM_User, time=0, value=1}
           {arriveTangent=-1.0628, interpMode=RCIM_Cubic, leaveTangent=-1.0628, tangentMode=RCTM_User, time=0.2, value=0.4}
           <at 1.0011 0.01>
          ]}}})
     )
     Index=&SpawnGroup()
     Num=::NewEnumerator4
    ))
   &ScaleSpriteSize(
    Scale Sprite Size Mode=::NewEnumerator0
    Uniform Scale Factor=&MergeHelper_FloatSwitch(
     0=&FloatFromCurve(
      FloatCurve=DataInterface() {propertyValues {Curve {keys [
          {arriveTangent=5.2354, interpMode=RCIM_Cubic, leaveTangent=5.2354, tangentMode=RCTM_User, time=-0.0046, value=0.25}
          <at 0.1439 1.4923>
          <at 1.0011 3.4809 RCIM_Cubic>
         ]}}}
      Scale Curve=&RandomRangeFloat(Minimum=0.5, Maximum=1.5)
     )
     1=&FloatFromCurve(FloatCurve=DataInterface() {propertyValues {Curve {
         defaultValue=1
         keys [{interpMode=RCIM_Cubic, tangentMode=RCTM_User, time=0, value=0}, {interpMode=RCIM_Cubic, tangentMode=RCTM_User, time=1, value=1}]
        }}})
     2=50
     3=0
     4=&FloatFromCurve(FloatCurve=DataInterface() {propertyValues {Curve {keys [<at 0.8 1 RCIM_Cubic>, {arriveTangent=-6.6126, interpMode=RCIM_Cubic, leaveTangent=-6.6126, tangentMode=RCTM_User, time=1, value=0}]}}})
     Index=&SpawnGroup()
     Num=::NewEnumerator3
    )
   )
   &CurlNoiseForce(
    Bias Noise=false
    Coordinate Space:ENiagaraCoordinateSpace=::Local
    Noise Frequency=&MergeHelper_FloatSwitch(0=0, 1=0, 2=0.02, 3=0, 4=0, 5=0, Index=&SpawnGroup(), Num=::NewEnumerator4)
    Noise Quality / Cost=::NewEnumerator0
    Noise Strength=&MergeHelper_FloatSwitch(
     0=0
     1=0
     2=500
     3=0
     4=&FloatFromCurve(FloatCurve=DataInterface() {propertyValues {Curve {keys [<at 0 1>,<at 1 0>]}}}, Scale Curve=3000)
     5=0
     Index=&SpawnGroup()
     Num=::NewEnumerator4
    )
    Pan Noise=false
    Random Seed=0
    Randomize Noise Sample=false
    Write to Intrinsic Parameters=false
   )
   &Drag(
    Drag=&MergeHelper_FloatSwitch(
     0=5
     1=0
     2=&FloatFromCurve(FloatCurve=DataInterface() {propertyValues {Curve {keys [<at 0 0.5>,<at 0.5 0.5>,<at 1 1>]}}}, Scale Curve=9)
     3=&FloatFromCurve(FloatCurve=DataInterface() {propertyValues {Curve {keys [<at 0 0.5>,<at 0.35 0.5>,<at 1 1>]}}}, Scale Curve=9)
     4=&FloatFromCurve(FloatCurve=DataInterface() {propertyValues {Curve {keys [<at 0 0.5141>,<at 0.5078 0.7118>,<at 0.9984 0.4918>]}}}, Scale Curve=10)
     5=5
     Index=&SpawnGroup()
     Num=::NewEnumerator4
    )
    Use Linear Drag=true
    Use Rotational Drag=true
   )
   &SubUVAnimation(
    Output Alternate Renderer Binding=true
    Random Start Frame=true
    "Sprite Renderer "=DataInterface() {propertyValues {SpriteRenderer=&self:Prop_spawn_end001_0.NiagaraSpriteRendererProperties_1}}
    SubUV Lookup Index Scale=0.2
    UseEndFrame=false
    UseStartFrame=false
   )
   &SubUVAnimation(
    Output Alternate Renderer Binding=true
    "Sprite Renderer "=DataInterface() {propertyValues {SpriteRenderer=&self:Prop_spawn_end001_0.NiagaraSpriteRendererProperties_0}}
    UseEndFrame=false
    UseStartFrame=false
   ) {module=SubUVAnimation001}
   &SubUVAnimation(
    Output Alternate Renderer Binding=true
    "Sprite Renderer "=DataInterface() {propertyValues {SpriteRenderer=&self:Prop_spawn_end001_0.NiagaraSpriteRendererProperties_6}}
    UseEndFrame=false
    UseStartFrame=false
   ) {module=SubUVAnimation002}
   &GravityForce(Gravity=&MergeHelper_VectorSwitch(
     0=<0 0 0>
     1=<0 0 0>
     2=&MakeVector(
      X=0
      Y=0
      Z=&FloatFromCurve(
       FloatCurve=DataInterface() {propertyValues {Curve {keys [<at 0 0.25 RCIM_Cubic>,<at 0.25 1>]}}}
       CurveIndex=$Emitter.Age
       Scale Curve=&RandomRangeFloat(Minimum=-60, Maximum=-170)
      )
     )
     3=<0 0 5>
     4=&RandomRangeVector(Minimum=<0 0 -10>, Maximum=<1 1 -350>)
     5=<0 0 0>
     Index=&SpawnGroup()
     Num=::NewEnumerator4
    ))
   &SolveForcesAndVelocity(Clamp Velocity=false, Limit Acceleration=false)
   &ScaleSpriteSizeBySpeed(
    Max Scale Factor=&MergeHelper_Vector2Switch(0=<1 1>, 1=<0 0>, 2=<0 0>, 3=<2 2>, 4=<1 5>, Index=&SpawnGroup(), Num=::NewEnumerator3)
    Min Scale Factor=&MergeHelper_Vector2Switch(0=<1 1>, 1=<0 0>, 2=<0 0>, 3=<1 1>, 4=<0 0>, Index=&SpawnGroup(), Num=::NewEnumerator3)
    Sample Scale Factor By Curve=true
    Velocity Threshold=500
   )
   &FN_Light_Attributes(
    Color=<rgba 15 4.9774 1.3498 1>
    Create Color Attribute=true
    Create Exponent Attribute=true
    Create Position Attribute=true
    Create Radius Attribute=true
    Enable Particle Light Enable=true
    Enable Visibility Tag=true
    Light Enabled=true
    Position=<0 0 180>
    Radius=&FloatFromCurve(
     FloatCurve=DataInterface() {propertyValues {Curve {
        defaultValue=1
        keys [{interpMode=RCIM_Cubic, tangentMode=RCTM_User, time=0, value=0}, {interpMode=RCIM_Cubic, tangentMode=RCTM_User, time=0.3, value=1}]
       }}}
     Scale Curve=35
    )
    Scale Color=&FloatFromCurve(FloatCurve=DataInterface() {propertyValues {Curve {keys [
         {interpMode=RCIM_Cubic, tangentMode=RCTM_Break, time=0, value=1}
         {interpMode=RCIM_Cubic, tangentMode=RCTM_Break, time=0.6908, value=1}
         {interpMode=RCIM_Cubic, tangentMode=RCTM_Break, time=1, value=0}
        ]}}})
    VisibilityTag=9
   ) {enabled=false}
   /SetVariables(
    Particles.Smoke Color:LinearColor=&MakeLinearColorFromVectorAndFloat(
     "Vector (RGB)"=&MakeVectorFromLinearColorRGB(Color=&ColorFromCurve(ColorCurve=DataInterface() {propertyValues {
         AlphaCurve {keys [<at 0 1 RCIM_Cubic>,<at 1 0 RCIM_Cubic>]}
         BlueCurve {keys [<at 0 0 RCIM_Cubic>,<at 1 0 RCIM_Cubic>]}
         GreenCurve {keys [<at 0 0.012 RCIM_Cubic>,<at 1 0 RCIM_Cubic>]}
         RedCurve {keys [<at 0 0.0365 RCIM_Cubic>,<at 1 0 RCIM_Cubic>]}
        }}))
     "Float (Alpha)"=&FloatFromCurve(
      FloatCurve=DataInterface() {propertyValues {Curve {
         defaultValue=1
         keys [
          {arriveTangent=9.0725, interpMode=RCIM_Cubic, leaveTangent=9.0725, tangentMode=RCTM_User, time=0, value=0}
          {arriveTangent=0.1098, interpMode=RCIM_Cubic, leaveTangent=0.1098, tangentMode=RCTM_User, time=0.0826, value=1.0001}
          {arriveTangent=0.0535, interpMode=RCIM_Cubic, leaveTangent=0.0535, tangentMode=RCTM_User, time=0.6499, value=1}
          {arriveTangent=-1.7285, interpMode=RCIM_Cubic, leaveTangent=-1.7285, tangentMode=RCTM_User, time=1, value=0}
         ]
        }}}
      Scale Curve=&Multiply_Float(A=1, B=&FNTimeOfDayScalar(Night Scale=1, Morning Scale=1, Evening Scale=0))
     )
    )
    Particles.Smoke Dyn Param:Vector4f=&MakeVector4(X=0, Y=0, Z=0, W=0)
   ) {isSetParameters=true}
   /SetVariables(
    Particles.Glow Sprite size:Vector2f=&Vector2DFromFloat(Value=&FloatFromCurve(
      FloatCurve=DataInterface() {propertyValues {Curve {
         defaultValue=1
         keys [{interpMode=RCIM_Cubic, tangentMode=RCTM_User, time=0, value=0}, {interpMode=RCIM_Cubic, tangentMode=RCTM_User, time=1, value=1}]
        }}}
      Scale Curve=90.8315
     ))
    Particles.soft glow Color:LinearColor=&MakeLinearColorFromVectorAndFloat(
     "Vector (RGB)"=&MakeVectorFromLinearColorRGB(Color=<rgba 3 1.3462 0.7292 1>)
     "Float (Alpha)"=&FloatFromCurve(
      FloatCurve=DataInterface() {propertyValues {Curve {keys [<at 0 1>,<at 1 0>]}}}
      Scale Curve=&FNTimeOfDayScalar(Night Scale=1, Morning Scale=1, Day Scale=1.5, Evening Scale=1)
     )
    )
    Particles.soft glow Position:NiagaraPosition=&AddVectorToPosition(Position=$Particles.SkeletalMeshLocation.SampledPosition, Vector=<0 0 46.1712>)
   ) {isSetParameters=true}
   /SetVariables(
    Particles.Embers Color:LinearColor=&MakeLinearColorFromVectorAndFloat(
     "Vector (RGB)"=&MakeVectorFromLinearColorRGB(Color=&ColorFromCurve(ColorCurve=DataInterface() {propertyValues {
         AlphaCurve {keys [<at 0 1 RCIM_Cubic>]}
         BlueCurve {keys [<at 0 7.8125 RCIM_Cubic>,<at 0.2499 0.7812>,<at 0.7114 0.2083 RCIM_Cubic>]}
         GreenCurve {keys [<at 0 34.1758 RCIM_Cubic>,<at 0.2499 4.8887>,<at 0.7114 0.8934 RCIM_Cubic>]}
         RedCurve {keys [<at 0 50 RCIM_Cubic>,<at 0.2499 10>,<at 0.7114 4 RCIM_Cubic>]}
        }}))
     "Float (Alpha)"=2
    )
    Particles.Embers SpriteSize:Vector2f=&Multiply_Vector2D_ByFloat(
     Float=&FloatFromCurve(
      FloatCurve=DataInterface() {propertyValues {Curve {
         defaultValue=1
         keys [
          {interpMode=RCIM_Cubic, tangentMode=RCTM_User, time=0, value=0}
          {interpMode=RCIM_Cubic, tangentMode=RCTM_User, time=0.1511, value=1}
          <at 0.35 0 RCIM_Cubic>
         ]
        }}}
      Scale Curve=2.5
     )
     Vector2D=&RandomRangeVector2D(Minimum=<0.4 0.8>, Maximum=<0.5 1>)
    )
   ) {isSetParameters=true}
   /SetVariables(
    Particles.Ashes Color:LinearColor=&MakeLinearColorFromVectorAndFloat(
     "Vector (RGB)"=&MakeVectorFromLinearColorRGB(Color=&ColorFromCurve(ColorCurve=DataInterface() {propertyValues {
         AlphaCurve {keys [<at 0 1 RCIM_Cubic>]}
         BlueCurve {keys [<at 0 1.473 RCIM_Cubic>,<at 0.6088 0.0879 RCIM_Cubic>]}
         GreenCurve {keys [<at 0 11.4982 RCIM_Cubic>,<at 0.6088 0.0919 RCIM_Cubic>]}
         RedCurve {keys [<at 0 50 RCIM_Cubic>,<at 0.6088 0.099 RCIM_Cubic>]}
        }}))
     "Float (Alpha)"=&FloatFromCurve(FloatCurve=DataInterface() {propertyValues {Curve {keys [<at 0.8 1 RCIM_Cubic>,<at 0.85 0>]}}})
    )
    Particles.Ashes SpriteSize:Vector2f=&Vector2DFromFloat(Value=&FloatFromCurve(
      FloatCurve=DataInterface() {propertyValues {Curve {
         defaultValue=1
         keys [
          {interpMode=RCIM_Cubic, tangentMode=RCTM_User, time=0, value=0}
          {interpMode=RCIM_Cubic, tangentMode=RCTM_User, time=0.5, value=1}
          {tangentMode=RCTM_User, time=1, value=0}
         ]
        }}}
      Scale Curve=&RandomRangeFloat(Minimum=1, Maximum=7)
     ))
    Particles.Vector 4:Vector4f=&MakeVector4(X=0, Y=0, Z=0, W=0)
   ) {isSetParameters=true}
   /SetVariables(
    Particles.Fire Color:LinearColor=&MakeLinearColorFromVectorAndFloat(
     "Vector (RGB)"=&Multiply_VectorByFloat(
      Vector=&MakeVectorFromLinearColorRGB(Color=<rgba 20 1.962 0 1>)
      Float=&FloatFromCurve(FloatCurve=DataInterface() {propertyValues {Curve {
          defaultValue=1
          keys [{interpMode=RCIM_Cubic, tangentMode=RCTM_User, time=0, value=1}, {interpMode=RCIM_Cubic, tangentMode=RCTM_User, time=1, value=0}]
         }}})
     )
     "Float (Alpha)"=&FloatFromCurve(
      FloatCurve=DataInterface() {propertyValues {Curve {
         defaultValue=1
         keys [{interpMode=RCIM_Cubic, tangentMode=RCTM_User, time=0, value=1}, {interpMode=RCIM_Cubic, tangentMode=RCTM_User, time=1, value=0}]
        }}}
      Scale Curve=&FN_FrontEndFloatModifier(
       Gameplay=&FNTimeOfDayScalar(Night Scale=3, Morning Scale=1, Day Scale=1.5, Evening Scale=1)
       Front End=2
      )
     )
    )
    Particles.Fire DynParam:Vector4f=&MakeVector4(
     X=&FloatFromCurve(FloatCurve=DataInterface() {propertyValues {Curve {
         defaultValue=0
         keys [{arriveTangent=2.1289, interpMode=RCIM_Cubic, leaveTangent=2.1289, tangentMode=RCTM_User, time=0, value=0}, <at 1.0039 1 RCIM_Cubic>]
        }}})
     Y=0.5
     Z=0
     W=0
    )
    Particles.fire SpriteSize:Vector2f=&Multiply_Vector2D_ByFloat(
     Float=&FloatFromCurve(FloatCurve=DataInterface() {propertyValues {Curve {keys [
          {interpMode=RCIM_Cubic, tangentMode=RCTM_Break, time=0.0079, value=0.5}
          {interpMode=RCIM_Cubic, tangentMode=RCTM_Break, time=0.0417, value=3}
          {arriveTangent=0.3384, interpMode=RCIM_Cubic, leaveTangent=0.3384, tangentMode=RCTM_User, time=0.1668, value=1.3}
          {arriveTangent=-0.0782, interpMode=RCIM_Cubic, leaveTangent=-0.0782, tangentMode=RCTM_User, time=0.7507, value=1.3}
          <at 0.8222 0>
         ]}}})
     Vector2D=&RandomRangeVector2D(Minimum=<15 15>, Maximum=<20 20>)
    )
   ) {isSetParameters=true}
   /SetVariables(Particles.Distortion SpriteSize:Vector2f=<150 150>) {isSetParameters=true}
   /SetVariables(
    Particles.FIre2 DynParam:Vector4f=&MakeVector4(
     X=&FloatFromCurve(
      FloatCurve=DataInterface() {propertyValues {Curve {
         defaultValue=1
         keys [{interpMode=RCIM_Cubic, tangentMode=RCTM_User, time=0, value=0}, {interpMode=RCIM_Cubic, tangentMode=RCTM_User, time=1.0015, value=0.6706}]
        }}}
      Scale Curve=0.5
     )
     Y=&RandomRangeFloat(Minimum=0, Maximum=4)
     Z=0
     W=2.5
    )
    Particles.Fire 2 Color:LinearColor=&MakeLinearColorFromVectorAndFloat(
     "Vector (RGB)"=&VectorFromFloat(Value=&FloatFromCurve(
       FloatCurve=DataInterface() {propertyValues {Curve {keys [
           {arriveTangent=-0.4927, interpMode=RCIM_Cubic, leaveTangent=-0.4927, tangentMode=RCTM_User, time=0, value=0.1087}
           {arriveTangent=-0.5399, interpMode=RCIM_Cubic, leaveTangent=-0.5399, tangentMode=RCTM_User, time=0.1564, value=0.3915}
           {arriveTangent=-0.4927, interpMode=RCIM_Cubic, leaveTangent=-0.4927, tangentMode=RCTM_User, time=1.0076, value=0.148}
          ]}}}
       Scale Curve=&RandomRangeFloat(Minimum=2, Maximum=5)
      ))
     "Float (Alpha)"=&FloatFromCurve(
      FloatCurve=DataInterface() {propertyValues {Curve {keys [
          {arriveTangent=3.3783, interpMode=RCIM_Cubic, leaveTangent=3.3783, tangentMode=RCTM_User, time=0, value=0}
          {arriveTangent=0.2941, interpMode=RCIM_Cubic, leaveTangent=0.2941, tangentMode=RCTM_User, time=0.203, value=0.7175}
          {arriveTangent=-1.109, interpMode=RCIM_Cubic, leaveTangent=-1.109, time=0.5122, value=0.5742}
          {arriveTangent=-0.0646, interpMode=RCIM_Cubic, leaveTangent=-0.0646, tangentMode=RCTM_User, time=0.85, value=0}
         ]}}}
      Scale Curve=&Multiply_Float(A=2, B=$System.TimeOfDay)
     )
    )
    Particles.Fire 2 Sprite Size:Vector2f=&Vector2DFromFloat(Value=&FloatFromCurve(
      FloatCurve=DataInterface() {propertyValues {Curve {
         defaultValue=1
         keys [{interpMode=RCIM_Cubic, tangentMode=RCTM_User, time=0, value=0}, {interpMode=RCIM_Cubic, tangentMode=RCTM_User, time=1, value=1}]
        }}}
      Scale Curve=&RandomRangeFloat(Minimum=45, Maximum=65)
     ))
   ) {isSetParameters=true}
   /SetVariables(
    Particles.Fire Eenergy SubImageIndex:NiagaraFloat=&Lerp_Float(A=25, B=63, Alpha=$Particles.NormalizedAge)
    Particles.Fire Energy SpriteSize:Vector2f=&Vector2DFromFloat(Value=&Multiply_Float(
      A=&FloatFromCurve(FloatCurve=DataInterface() {propertyValues {Curve {keys [
           {arriveTangent=7.4819, interpMode=RCIM_Cubic, leaveTangent=7.4819, tangentMode=RCTM_User, time=0, value=0.4215}
           {arriveTangent=0.5944, interpMode=RCIM_Cubic, leaveTangent=0.5944, tangentMode=RCTM_User, time=0.1164, value=0.8874}
           {interpMode=RCIM_Cubic, tangentMode=RCTM_User, time=1, value=1}
          ]}}})
      B=&RandomRangeFloat(Minimum=35, Maximum=90)
     ))
    Particles.Fire Energy Update Color:LinearColor=&MultiplyLinearColor(
     Linear Color=$Particles.Fire Energy Initial Color
     Scale Factor=&MakeVector4_FromLinearColor(LinearColor=&MakeLinearColorFromVectorAndFloat(
       "Vector (RGB)"=&VectorFromFloat(Value=&FloatFromCurve(FloatCurve=DataInterface() {propertyValues {Curve {keys [{arriveTangent=-0.1872, interpMode=RCIM_Cubic, leaveTangent=-0.1872, tangentMode=RCTM_User, time=0, value=2}, <at 1 1>]}}}))
       "Float (Alpha)"=&FloatFromCurve(
        FloatCurve=DataInterface() {propertyValues {Curve {keys [
            {arriveTangent=21.9814, interpMode=RCIM_Cubic, leaveTangent=21.9814, tangentMode=RCTM_User, time=0, value=0}
            {arriveTangent=0.1895, interpMode=RCIM_Cubic, leaveTangent=0.1895, tangentMode=RCTM_User, time=0.1, value=0.75}
            {arriveTangent=-2.5555, interpMode=RCIM_Cubic, leaveTangent=-2.5555, tangentMode=RCTM_User, time=1, value=0}
           ]}}}
        Scale Curve=&Multiply_Float(
         A=2
         B=&FN_FrontEndFloatModifier(Gameplay=&FNTimeOfDayScalar(Night Scale=1, Morning Scale=1, Evening Scale=1), Front End=0.5)
        )
       )
      ))
    )
   ) {isSetParameters=true}
  ]
  rendererClasses [&NiagaraSpriteRendererProperties, &NiagaraLightRendererProperties]
  renderers [
   {rendererClass=&NiagaraSpriteRendererProperties}
   {rendererClass=&NiagaraSpriteRendererProperties, rendererIndex=1}
   {rendererClass=&NiagaraSpriteRendererProperties, rendererIndex=2}
   {rendererClass=&NiagaraSpriteRendererProperties, rendererIndex=3}
   {rendererClass=&NiagaraSpriteRendererProperties, rendererIndex=4}
   {rendererClass=&NiagaraSpriteRendererProperties, rendererIndex=5}
   {rendererClass=&NiagaraSpriteRendererProperties, rendererIndex=6}
   {rendererClass=&NiagaraLightRendererProperties, rendererIndex=7}
  ]
 }}
summary {
 systemName=NS_OutlineLove_sigil_Despawn
 emitters [{
   emitterName=Prop_spawn_end001
   rendererClasses [&NiagaraSpriteRendererProperties, &NiagaraLightRendererProperties]
  }]
 userVariables [
  User.bIsFrontEnd {type=NiagaraBool}
  User.bIsFrontEndPreview {type=NiagaraBool}
  User.Erode_circle {type=NiagaraFloat, defaultValue=0.1, description=""}
 ]
}
