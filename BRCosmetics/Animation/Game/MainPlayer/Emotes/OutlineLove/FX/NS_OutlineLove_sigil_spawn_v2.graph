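# Niagara system for the OutlineLove emote that plays a fiery spawn effect on a sigil skeletal mesh.
# GPU embers accompany a delayed fire-ring ignition with layered meshes, a flickering light and erode-driven fading.
# A prop-spawn emitter packs smoke, sparks, fire, glow and energy waves into spawn groups, using group-indexed switches and visibility tags across many renderers.
# Time-of-day and front-end preview modifiers adjust brightness, color and placement.
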NiagaraSystem /BRCosmetics/Animation/Game/MainPlayer/Emotes/OutlineLove/FX/NS_OutlineLove_sigil_spawn_v2
dependencies {
 usedDataInterfaces [
  &NiagaraDataInterfaceCurve
  &NiagaraDataInterfaceSpriteRendererInfo
  &NiagaraDataInterfaceColorCurve
  &NiagaraDataInterfaceSkeletalMesh
 ]
 usedDynamicInputs [
  &MaskVector
  &MakeFloatFromVector
  &FN_LOD_Distance_Remap
  &SimulationPosition
 ]
 usedModules [
  &self:SystemSpawnScript.SystemScriptSource.SystemScriptGraph.NiagaraNodeAssignment_0.SetVariables_#1
  &SystemState
  &self:SystemSpawnScript.SystemScriptSource.SystemScriptGraph.NiagaraNodeAssignment_1.SetVariables_#2
  &self:SystemSpawnScript.SystemScriptSource.SystemScriptGraph.NiagaraNodeAssignment_2.SetVariables_#3
 ]
}
emitters {
 Embers_Burst001 {
  simTarget=GPUComputeSim
  emitterUpdateScript [
   &EmitterState(
    Inactive Response=::NewEnumerator1
    Life Cycle Mode=::NewEnumerator1
    Loop Behavior=::NewEnumerator1
    Loop Delay=0.25
    Loop Duration=4
    UseLoopDelay=true
   )
   &SpawnBurst_Instantaneous(Spawn Count=100, Spawn Time=3.5, Use Spawn Probability=false, UseLoopCountLimit=false) {enabled=false, module=SpawnBurst_Instantaneous002}
   &SpawnBurst_Instantaneous(Spawn Count=100, Spawn Time=0.05, Use Spawn Probability=false, UseLoopCountLimit=false)
   &SpawnRate(
    SpawnRate=&FloatFromCurve(FloatCurve=DataInterface() {propertyValues {Curve {keys [<at 0.9 1>,<at 1 0>]}}}, CurveIndex=$Emitter.NormalizedLoopAge, Scale Curve=15)
    Use Spawn Probability=false
   )
   &SpawnBurst_Instantaneous(Spawn Count=15, Spawn Time=2, Use Spawn Probability=false, UseLoopCountLimit=false) {module=SpawnBurst_Instantaneous001}
  ]
  particleSpawnScript [
   &InitializeParticle(
    Color=<rgba 1 1 1 2>
    Lifetime=&UniformRangedFloat(Minimum=0.3, Maximum=0.8)
    Mass Mode=::NewEnumerator1
    Position Mode=::NewEnumerator0
    Sprite Size=&Vector2DFromFloat(Value=&FloatFromCurve(
      FloatCurve=DataInterface() {propertyValues {Curve {keys [<at 0 1.2>,<at 0.1257 0.9847>,<at 0.2996 0.5>,<at 1 0.3>]}}}
      CurveIndex=&UniformRangedFloat(Minimum=0.5, Maximum=2)
      Scale Curve=5
     ))
    Sprite Size Mode=::NewEnumerator4
    Write Material Random=true
   ) {module=InitializeParticle001}
   &AddVelocity() {module=AddVelocity001}
   &SkeletalMeshLocation(Bone Sampling Mode=::NewEnumerator4, MaskSpawnGroup=false, Sampled Position Offset=<0 0 35>, Skeletal Mesh=$System.sigil)
   &ShapeLocation(MaskSpawnGroup=false, Override Seed=false, Rotation Axis=<0 1 0>, Rotation Mode=::NewEnumerator1, Sphere Radius=25, UseOwnerScale=true)
  ]
  particleUpdateScript [
   &ParticleState()
   &ApplyOwnerScaleToAttributes(Scale Initial Velocity=true)
   &ScaleSpriteSize(Scale Factor=<0.5 1>, Scale Sprite Size Mode=::NewEnumerator2)
   &Drag(
    Drag=&Multiply_Float(
     A=&FloatFromCurve(
      FloatCurve=DataInterface() {propertyValues {Curve {keys [<at 0 0.5>, {arriveTangent=0.5, interpMode=RCIM_Cubic, leaveTangent=0.5, time=0.3, value=0.5}, <at 1 1 RCIM_Cubic>]}}}
      Scale Curve=12
     )
     B=1
    )
    Use Linear Drag=true
    Use Rotational Drag=false
   )
   /Niagara/Modules/Update/Forces/CurlNoiseForce(
    Noise Frequency=10
    Noise Strength=&FloatFromCurve(FloatCurve=DataInterface() {propertyValues {Curve {keys [<at 0 1>,<at 1 0>]}}}, Scale Curve=3000)
    Pan Noise=true
    Pan Noise Field=<0.1 0.1 0.5>
   )
   &GravityForce(Gravity=&RandomRangeVector(Minimum=<0 0 900>, Maximum=<0 0 -500>))
   &SolveForcesAndVelocity(Clamp Velocity=false, Limit Acceleration=false)
   &ScaleSpriteSizeBySpeed(
    Max Scale Factor=<0.8 5>
    Min Scale Factor=<0.8 0.8>
    Sample Scale Factor By Curve=true
    Scale Factor Curve=DataInterface() {propertyValues {Curve {keys [
        <at -0.0077 0.0765 RCIM_Cubic>
        {arriveTangent=1.2463, interpMode=RCIM_Cubic, leaveTangent=1.2463, tangentMode=RCTM_User, time=0.9985, value=0.9549}
       ]}}}
    Velocity Threshold=300
   )
   &ScaleColor(
    Scale Alpha=&FloatFromCurve(
     FloatCurve=DataInterface() {propertyValues {Curve {keys [<at 0.25 1>,<at 1 0>]}}}
     Scale Curve=&FNTimeOfDayScalar(Night Scale=2, Morning Scale=1, Evening Scale=0.2)
    )
    Scale RGB=&MakeVectorFromLinearColorRGB(Color=&ColorFromCurve(ColorCurve=DataInterface() {propertyValues {
        AlphaCurve {keys [<at 0 1 RCIM_Cubic>,<at 0.7107 1>]}
        BlueCurve {keys [<at 0 14.6701 RCIM_Cubic>,<at 0.368 3.776>,<at 0.855 0.1753 RCIM_Cubic>]}
        GreenCurve {keys [<at 0 36.7481 RCIM_Cubic>,<at 0.368 13.2325>,<at 0.855 0.3243 RCIM_Cubic>]}
        RedCurve {keys [<at 0 50 RCIM_Cubic>,<at 0.368 25>,<at 0.855 1 RCIM_Cubic>]}
       }}))
    ScaleA=true
    ScaleRGB=true
   )
   &ScaleColor(
    Scale Alpha=&FloatFromCurve(
     FloatCurve=DataInterface() {propertyValues {Curve {keys [
         {arriveTangent=0.5383, interpMode=RCIM_Cubic, leaveTangent=0.5383, tangentMode=RCTM_User, time=0, value=0.03}
         {arriveTangent=0.2783, interpMode=RCIM_Cubic, leaveTangent=0.2783, tangentMode=RCTM_User, time=1, value=1}
        ]}}}
     CurveIndex=$System.NearCameraFade
    )
    ScaleA=true
    ScaleRGB=true
   ) {enabled=false, module=ScaleColor001}
   /SetVariables(Particles.Glow03_Color=&MakeLinearColorFromVectorAndFloat(
     "Vector (RGB)"=&MakeCustomVectorFromBool(
      Bool=$User.bIsFrontEndPreview
      True Vector=&MakeVectorFromLinearColorRGB(Color=<rgba 1 0.28 0.02 1>)
      False Vector=&MakeVectorFromLinearColorRGB(Color=&MultiplyLinearColor(
        Linear Color=&FNTimeOfDayColor(Night Color=<rgba 1 0.3167 0.0354 1>, Morning Color=<rgba 1 0.15 0.02 1>, Day Color=<rgba 1 0.15 0.02 1>, Evening Color=<rgba 1 0.15 0.02 1>)
        Scale Factor=&MakeVector4_FromFloat(Float=&Multiply_Float(A=1, B=&FNTimeOfDayScalar(Night Scale=0.5, Morning Scale=0.5, Evening Scale=0.5)))
       ))
     )
     "Float (Alpha)"=&Multiply_Float(
      A=0.325
      B=&FloatFromCurve(
       FloatCurve=DataInterface() {propertyValues {Curve {keys [<at 0 0>,<at 0.15 0>,<at 0.25 1>]}}}
       CurveIndex=$Emitter.Age
       Scale Curve=&MakeCustomFloatFromBool(Bool=$User.bIsFrontEndPreview, True Float=0.6, False Float=1)
      )
     )
    )) {isSetParameters=true}
   /SetVariables(Particles.Glow03_Size=&Multiply_Vector2D_ByFloat(
     Float=&FloatFromCurve(
      FloatCurve=DataInterface() {propertyValues {Curve {keys [
          <at 0 0 RCIM_Cubic>
          <at 0.2 1.3 RCIM_Cubic>
          {arriveTangent=-0.1423, interpMode=RCIM_Cubic, leaveTangent=-0.1423, tangentMode=RCTM_User, time=1, value=0}
         ]}}}
      Scale Curve=10
     )
     Vector2D=<0.75 1>
    )) {isSetParameters=true}
  ]
  rendererClasses [&NiagaraSpriteRendererProperties]
  renderers [
   {rendererClass=&NiagaraSpriteRendererProperties}
   {rendererClass=&NiagaraSpriteRendererProperties, rendererIndex=1}
  ]
 }
 Fire_ring_Ignition {
  emitterSpawnScript [/SetVariables(
    Emitter.erode:NiagaraFloat=$User.Erode_circle
    Emitter.velocity flames:NiagaraFloat=&FloatFromCurve(
     FloatCurve=DataInterface() {propertyValues {Curve {
        defaultValue=1
        keys [
         {arriveTangent=9.0725, interpMode=RCIM_Cubic, leaveTangent=9.0725, tangentMode=RCTM_User, time=0, value=0}
         {arriveTangent=-2.5249, interpMode=RCIM_Cubic, leaveTangent=-2.5249, tangentMode=RCTM_User, time=0.1408, value=3}
         {arriveTangent=-0.0405, interpMode=RCIM_Cubic, leaveTangent=-0.0405, tangentMode=RCTM_User, time=0.1916, value=1}
        ]
       }}}
     CurveIndex=$Emitter.NormalizedLoopAge
    )
   ) {isSetParameters=true}]
  emitterUpdateScript [
   &EmitterState(
    Inactive Response=::NewEnumerator1
    Life Cycle Mode=::NewEnumerator1
    Loop Behavior=::NewEnumerator1
    Loop Delay=0.23
    Loop Duration=4
    UseLoopDelay=true
   )
   &SpawnBurst_Instantaneous(Use Spawn Probability=false, UseLoopCountLimit=false)
  ]
  particleSpawnScript [
   /SetVariables(Particles.VisibilityTag=&SpawnGroup()) {isSetParameters=true}
   &InitializeParticle(
    Lifetime=&MergeHelper_FloatSwitch(0=5, 1=0, 2=?, 3=?, Index=&SpawnGroup(), Num=::NewEnumerator0)
    UsePositionOffset=false
    Write Material Random=true
   )
   /SetVariables(Particles.Patch 2 MeshOrientation:Quat4f=<0 0 0.01 1>, Particles.patch MeshOrientation:Quat4f=<0.0092 0 0 1>) {isSetParameters=true}
  ]
  particleUpdateScript [
   &ParticleState()
   &SkeletalMeshLocation(Bone Sampling Mode=::NewEnumerator4, MaskSpawnGroup=false, Position Sampling=::NewEnumerator2, Skeletal Mesh=$System.sigil)
   /SetVariables(
    Particles.Flat Ring DynParam:Vector4f=&MakeVector4(X=3.4925, Y=0, Z=0, W=0)
    Particles.Front Fire Ring Color:LinearColor=&MakeLinearColorFromVectorAndFloat(
     "Vector (RGB)"=&MakeVectorFromLinearColorRGB(Color=<rgba 3 3 3 0.5>)
     "Float (Alpha)"=&FloatFromCurve(
      FloatCurve=DataInterface() {propertyValues {Curve {keys [
          <at 0.0042 -0.0287 RCIM_Cubic>
          {arriveTangent=6.7494, interpMode=RCIM_Cubic, leaveTangent=6.7494, time=0.0883, value=1.4713}
          {arriveTangent=-0.6753, interpMode=RCIM_Cubic, leaveTangent=-0.6753, time=0.156, value=0.9955}
          {arriveTangent=0.0001, interpMode=RCIM_Cubic, leaveTangent=0.0001, tangentMode=RCTM_User, time=0.85, value=0.957}
          <at 0.9354 -0.043 RCIM_Cubic>
         ]}}}
      Scale Curve=&Multiply_Float(
       A=1
       B=&FN_FrontEndFloatModifier(
        Gameplay=&FNTimeOfDayScalar(Night Scale=0.5, Morning Scale=0.3, Day Scale=0.75, Evening Scale=0.3)
        Front End=0.5
       )
      )
     )
    )
    Particles.Front Fire Ring Position:NiagaraPosition=&AddVectorToPosition(Position=$Particles.SkeletalMeshLocation.SampledPosition, Vector=<0 2.8706 40>)
    Particles.Front Fire ring scale:Vector3f=&Multiply_VectorByFloat(Vector=<1 1 1>, Float=1.86)
   ) {isSetParameters=true}
   /SetVariables(
    Particles.Back Fire Ring Color:LinearColor=&MakeLinearColorFromVectorAndFloat(
     "Vector (RGB)"=<3 3 3>
     "Float (Alpha)"=&FloatFromCurve(
      FloatCurve=DataInterface() {propertyValues {Curve {keys [
          <at 0.0042 -0.0287 RCIM_Cubic>
          {arriveTangent=6.7494, interpMode=RCIM_Cubic, leaveTangent=6.7494, time=0.0883, value=1.4713}
          {arriveTangent=-0.6753, interpMode=RCIM_Cubic, leaveTangent=-0.6753, time=0.156, value=0.9955}
          {arriveTangent=0.0001, interpMode=RCIM_Cubic, leaveTangent=0.0001, tangentMode=RCTM_User, time=0.85, value=0.957}
          <at 0.9354 -0.043 RCIM_Cubic>
         ]}}}
      Scale Curve=&Multiply_Float(A=1, B=&FNTimeOfDayScalar(Night Scale=0.5, Morning Scale=0.3, Day Scale=0.3, Evening Scale=0.5))
     )
    )
    Particles.Back Fire Ring Position:NiagaraPosition=&AddVectorToPosition(Position=$Particles.SkeletalMeshLocation.SampledPosition, Vector=<0.1664 0.239 40>)
    Particles.Back Fire Ring Scale:Vector3f=&Multiply_VectorByFloat(Vector=<1 1 1>, Float=1.86)
   ) {isSetParameters=true}
   /SetVariables(
    Particles.Thick Fire Ring CameraOffset=3.3068
    Particles.Thick Fire Ring DynParam:Vector4f=&MakeVector4(X=&RandomRangeFloat(Minimum=0, Maximum=1), Y=0, Z=0, W=0)
    Particles.Thick Fire Ring Position:NiagaraPosition=&AddVectorToPosition(Position=$Particles.SkeletalMeshLocation.SampledPosition, Vector=<0 1.3154 42.5>)
    Particles.Thick Fire Ring Scale:Vector3f=&Multiply_VectorByFloat(Vector=<0.6954 1.0636 1>, Float=0.215)
    Particles.Tick Fire ring Color:LinearColor=&MakeLinearColorFromVectorAndFloat(
     "Vector (RGB)"=&MakeVectorFromLinearColorRGB(Color=<rgba 1 1 1 0.5>)
     "Float (Alpha)"=&FloatFromCurve(
      FloatCurve=DataInterface() {propertyValues {Curve {keys [
          <at 0.0042 -0.0287 RCIM_Cubic>
          {arriveTangent=6.7494, interpMode=RCIM_Cubic, leaveTangent=6.7494, time=0.0883, value=1.4713}
          {arriveTangent=-0.6753, interpMode=RCIM_Cubic, leaveTangent=-0.6753, time=0.156, value=0.9955}
          {arriveTangent=0.0001, interpMode=RCIM_Cubic, leaveTangent=0.0001, tangentMode=RCTM_User, time=0.85, value=0.957}
          <at 0.9354 -0.043 RCIM_Cubic>
         ]}}}
      Scale Curve=&Multiply_Float(
       A=&Lerp_Float(A=0, B=1, Alpha=$User.Erode_circle)
       B=&FN_FrontEndFloatModifier(
        Gameplay=&FNTimeOfDayScalar(Night Scale=0.5, Morning Scale=0.3, Day Scale=0.75, Evening Scale=0.6)
        Front End=1
       )
      )
     )
    )
   ) {isSetParameters=true}
   &FN_Light_Attributes(
    Color=<rgba 3 1.1224 0.4219 1>
    Create Color Attribute=true
    Create Exponent Attribute=true
    Create Position Attribute=true
    Create Radius Attribute=true
    Enable Particle Light Enable=true
    Enable Visibility Tag=true
    Exponent=2
    Light Enabled=true
    Position=&AddVectorToPosition(Position=$Particles.SkeletalMeshLocation.SampledPosition, Vector=<0 0 25>)
    Radius=25
    Scale Color=&Multiply_Float(
     A=&FloatFromCurve(
      FloatCurve=DataInterface() {propertyValues {Curve {keys [
          {interpMode=RCIM_Cubic, tangentMode=RCTM_Break, time=0, value=0}
          {interpMode=RCIM_Cubic, tangentMode=RCTM_Break, time=0.0602, value=0.8601}
          {arriveTangent=-1.089, interpMode=RCIM_Cubic, leaveTangent=-1.089, time=0.6287, value=0.7883}
          {interpMode=RCIM_Cubic, leaveTangent=0.4608, time=0.85, value=0}
         ]}}}
      Scale Curve=0.5
     )
     B=&Multiply_Float(
      A=&Waveform("Waveform [1]"=::NewEnumerator0, "[1] Phase Scale"=0.2, "[1] Frequency"=3, "[1] Amplitude Min/Max"=<0.7 1>, "[1] Scale"=3)
      B=&FNTimeOfDayScalar(Night Scale=1, Morning Scale=0.6, Day Scale=1.5, Evening Scale=0.6)
     )
    )
    VisibilityTag=3
   )
   /SetVariables(
    Particles.Glow Ring Position_top:NiagaraPosition=&AddVectorToPosition(Position=$Particles.SkeletalMeshLocation.SampledPosition, Vector=<0 0.79 38.4231>)
    Particles.Glow ring Position_Bottom:NiagaraPosition=&AddVectorToPosition(Position=$Particles.SkeletalMeshLocation.SampledPosition, Vector=<-0.8074 1.1236 43.7315>)
    Particles.Glow ring Scale:Vector3f=&Multiply_Vector(A=<0.9914 0.9247 1.3329>, B=&VectorFromFloat(Value=2.3836))
    Particles.glow ring color:LinearColor=<rgba 1 0.3741 0.1406 1>
   ) {isSetParameters=true}
   /SetVariables(Particles.random fire licks:Vector3f=&VectorFromFloat(Value=&FloatFromCurve(
      FloatCurve=DataInterface() {propertyValues {Curve {
         defaultValue=1
         keys [{arriveTangent=-0.0405, interpMode=RCIM_Cubic, leaveTangent=-0.0405, tangentMode=RCTM_User, time=0.2246, value=0.9693}]
        }}}
      Scale Curve=1.954
     ))) {isSetParameters=true}
   /SetVariables(
    Particles.Fire licks color:LinearColor=&MakeLinearColorFromVectorAndFloat(
     "Vector (RGB)"=&MakeVectorFromLinearColorRGB(Color=<rgba 3 0.762 0.1861 1>)
     "Float (Alpha)"=&FloatFromCurve(
      FloatCurve=DataInterface() {propertyValues {Curve {keys [
          <at 0.0042 -0.0287 RCIM_Cubic>
          {arriveTangent=6.7494, interpMode=RCIM_Cubic, leaveTangent=6.7494, time=0.0883, value=1.4713}
          {arriveTangent=-0.6753, interpMode=RCIM_Cubic, leaveTangent=-0.6753, time=0.156, value=0.9955}
          {arriveTangent=0.0001, interpMode=RCIM_Cubic, leaveTangent=0.0001, tangentMode=RCTM_User, time=0.85, value=0.957}
          <at 0.9354 -0.043 RCIM_Cubic>
         ]}}}
      Scale Curve=&Multiply_Float(
       A=$User.Erode_circle
       B=&FNTimeOfDayScalar(Night Scale=2, Morning Scale=1.5, Day Scale=5, Evening Scale=1.5)
      )
     )
    )
    Particles.fire licks Position:NiagaraPosition=&AddVectorToPosition(Position=$Particles.SkeletalMeshLocation.SampledPosition, Vector=<0 1.3163 40>)
    Particles.random fire licks scale:Vector3f=&VectorFromFloat(Value=&FloatFromCurve(
      FloatCurve=DataInterface() {propertyValues {Curve {
         defaultValue=1
         keys [{arriveTangent=-0.0405, interpMode=RCIM_Cubic, leaveTangent=-0.0405, tangentMode=RCTM_User, time=0.2246, value=0.9693}]
        }}}
      Scale Curve=1.954
     ))
   ) {isSetParameters=true}
  ]
  rendererClasses [&NiagaraMeshRendererProperties, &NiagaraLightRendererProperties]
  renderers [
   {rendererClass=&NiagaraMeshRendererProperties, rendererIndex=0}
   {rendererClass=&NiagaraMeshRendererProperties, rendererIndex=1}
   {rendererClass=&NiagaraMeshRendererProperties, rendererIndex=2}
   {rendererClass=&NiagaraLightRendererProperties, rendererIndex=3}
   {rendererClass=&NiagaraMeshRendererProperties, rendererIndex=4}
   {rendererClass=&NiagaraMeshRendererProperties, rendererIndex=5}
  ]
 }
 Fire_sigil_Prop_Spawn {
  emitterUpdateScript [
   &EmitterState(
    Inactive Response=::NewEnumerator1
    Life Cycle Mode=::NewEnumerator1
    Loop Behavior=::NewEnumerator1
    Loop Delay=0.01
    Loop Duration=2.7
    UseLoopDelay=true
   )
   &SpawnBurst_Instantaneous(Spawn Group=7, Use Spawn Probability=false, UseLoopCountLimit=false) {module=SpawnBurst_Instantaneous009}
   &SpawnBurst_Instantaneous(Spawn Count=12, Spawn Group=6, Spawn Time=0.05, Use Spawn Probability=false, UseLoopCountLimit=false) {module=SpawnBurst_Instantaneous007}
   &SpawnBurst_Instantaneous(Spawn Count=12, Spawn Group=5, Spawn Time=0.05, Use Spawn Probability=false, UseLoopCountLimit=false) {module=SpawnBurst_Instantaneous006}
   &SpawnBurst_Instantaneous(Spawn Count=0, Spawn Group=4, Spawn Time=0.03, Use Spawn Probability=false, UseLoopCountLimit=false) {module=SpawnBurst_Instantaneous004}
   &SpawnBurst_Instantaneous(Spawn Count=35, Spawn Group=3, Spawn Time=0.015, Use Spawn Probability=false, UseLoopCountLimit=false) {module=SpawnBurst_Instantaneous003}
   &SpawnBurst_Instantaneous(Spawn Count=20, Spawn Group=2, Use Spawn Probability=false) {module=SpawnBurst_Instantaneous002}
   &SpawnBurst_Instantaneous(Spawn Count=1, Spawn Group=1, Use Spawn Probability=false)
   &SpawnBurst_Instantaneous(Spawn Count=35, Use Spawn Probability=false) {module=SpawnBurst_Instantaneous001}
  ]
  particleSpawnScript [
   /SetVariables(Particles.VisibilityTag=&SpawnGroup()) {isSetParameters=true}
   /SetVariables(Particles.InitSpriteRotRate=&UniformRangedFloat(Minimum=-1, Maximum=1)) {isSetParameters=true}
   &InitializeParticle(
    Color=&MergeHelper_ColorSwitch(
     0=<rgba 0.5255 0.5255 0.5255 1>
     1=<rgba 0.6031 0.5599 0.463 0.2>
     2=<rgba 5 3.7043 1.51 1>
     3=?
     Index=&SpawnGroup()
     Num=::NewEnumerator1
    )
    Lifetime=&MergeHelper_FloatSwitch(
     0=&UniformRangedFloat(Minimum=0.3, Maximum=0.55)
     1=0.25
     2=&UniformRangedFloat(Minimum=0.5, Maximum=1.25)
     3=&UniformRangedFloat(Minimum=0.25, Maximum=0.65)
     4=&UniformRangedFloat(Minimum=0.3, Maximum=1.2)
     5=&RandomRangeFloat(Minimum=0.6, Maximum=1.2)
     6=&UniformRangedFloat(Minimum=0.3, Maximum=0.75)
     7=0.6
     Index=&SpawnGroup()
     Num=::NewEnumerator6
    )
    Position Mode=::NewEnumerator0
    Sprite Rotation Mode=::NewEnumerator2
    Sprite Size Mode=::NewEnumerator3
    Uniform Sprite Size=&MergeHelper_FloatSwitch(
     0=&UniformRangedFloat(Minimum=8, Maximum=20)
     1=65
     2=&RandomRangeFloat(Minimum=10, Maximum=25)
     3=?
     Index=&SpawnGroup()
     Num=::NewEnumerator1
    )
    Write Material Random=true
   )
   &ShapeLocation(
    MaskSpawnGroup=false
    Override Seed=false
    Sphere Radius=&MergeHelper_FloatSwitch(0=7, 1=0, 2=4, 3=5, 4=10, 5=0, 6=0, 7=0, Index=&SpawnGroup(), Num=::NewEnumerator6)
    Sphere Surface Distribution=1
    UseOwnerScale=true
   )
   &AddVelocity(Velocity=&MergeHelper_VectorSwitch(
     0=&Multiply_Vector(
      A=$Particles.ShapeLocation.ShapeNormal
      B=&VectorFromFloat(Value=&RandomRangeFloat(Minimum=15, Maximum=40))
     )
     1=<0 0 0>
     2=&UniformRangedVector(Minimum=<-200 0 -110>, Maximum=<200 0 110>)
     3=&RandomRangeVector(Minimum=<-100 0 -50>, Maximum=<100 1 150>)
     4=&RandomRangeVector(Minimum=<-180 -180 50>, Maximum=<180 180 150>)
     5=<0 0 0>
     6=<0 0 0>
     Index=&SpawnGroup()
     Num=::NewEnumerator5
    ))
   /SetVariables(
    Particles.Smoke Velocity:Vector3f=&Multiply_Vector(
     A=$Particles.ShapeLocation.ShapeNormal
     B=&VectorFromFloat(Value=&RandomRangeFloat(Minimum=15, Maximum=50))
    )
    Particles.Sparkles Velocity:Vector3f=&UniformRangedVector(Minimum=<-500 -200 -200>, Maximum=<5000 200 200>)
    Particles.airglow CameraOffset=10
    Particles.fire camera offset:NiagaraFloat=&RandomRangeFloat(Minimum=18, Maximum=35)
    Particles.smoke CameraOffset:NiagaraFloat=&RandomRangeFloat(Minimum=10, Maximum=25)
   ) {isSetParameters=true}
   &AccelerationForce(Acceleration=&MergeHelper_VectorSwitch(
     0=&UniformRangedVector(Minimum=<0 0 -25>, Maximum=<0 0 -75>)
     1=<0 0 0>
     2=&RandomRangeVector(Minimum=<0 0 0>, Maximum=<0 0 0>)
     3=<0 0 0>
     4=<0 0 0>
     5=<0 0 0>
     6=<0 0 0>
     7=<50 0 0>
     Index=&SpawnGroup()
     Num=::NewEnumerator6
    ))
   &ApplyInitialForces()
   &InitialMeshOrientation(
    Mesh Axis To Orient=<0 1 0>
    Mesh Orientation Mode=::NewEnumerator2
    Orientation Vector=<0 0 1>
    Rotation Coordinate Space=::NewEnumerator0
    Use Rotation Vector=false
   )
   /SetVariables(
    Particles.Fire Energy Initial Color:LinearColor=&UniformRangedLinearColor(Minimum=<rgba 1 0.0943 0 1>, Maximum=<rgba 1 0.4811 0.1327 1>, Color Channel Mode=::NewEnumerator1)
    Particles.Red Mist Initial Color:LinearColor=&UniformRangedLinearColor(Minimum=<rgba 1 0.0943 0 0.1>, Maximum=<rgba 1 0.4776 0 0.15>, Color Channel Mode=::NewEnumerator1)
    Particles.Wave Energy initial Color:LinearColor=<rgba 1 0.1924 0 1>
   ) {isSetParameters=true}
   &SkeletalMeshLocation(
    Bone / Socket Index=&SpawnGroup()
    Bone Sampling Mode=::NewEnumerator4
    MaskSpawnGroup=false
    Sampled Position Coordinate Space=::Local
    Sampled Position Offset=&FN_FrontEndVector3Modifier(Gameplay=<-3 0 5>, Front End=<0 0 15>)
    Sampled Position Offset Coordinate Space:ENiagaraSkelMeshTransforms=::NewEnumerator2
    Skeletal Mesh=$System.sigil
   )
  ]
  particleUpdateScript [
   &ParticleState()
   &SpriteRotationRate(Rotation Rate=&MergeHelper_FloatSwitch(
     0=&Multiply_Float(
      A=$Particles.InitSpriteRotRate
      B=&FloatFromCurve(
       FloatCurve=DataInterface() {propertyValues {Curve {keys [
           {arriveTangent=-1.9214, interpMode=RCIM_Cubic, leaveTangent=-1.9214, tangentMode=RCTM_User, time=-0.0046, value=1.0113}
           {arriveTangent=-0.2496, interpMode=RCIM_Cubic, leaveTangent=-0.2496, tangentMode=RCTM_User, time=0.9896, value=0.16}
          ]}}}
       CurveIndex=$Emitter.Age
      )
     )
     1=0
     2=5
     3=0
     4=0
     5=0
     6=&Multiply_Float(
      A=&Lerp_Float(A=-200, B=200, Alpha=$Particles.MaterialRandom)
      B=&FloatFromCurve(FloatCurve=DataInterface() {propertyValues {Curve {keys [
           {arriveTangent=-4.4572, interpMode=RCIM_Cubic, leaveTangent=-4.4572, tangentMode=RCTM_User, time=0, value=1}
           {arriveTangent=-1.0628, interpMode=RCIM_Cubic, leaveTangent=-1.0628, tangentMode=RCTM_User, time=0.2, value=0.4}
           <at 1.0011 0.01>
          ]}}})
     )
     Index=&SpawnGroup()
     Num=::NewEnumerator5
    ))
   &ScaleSpriteSize(
    Scale Sprite Size Mode=::NewEnumerator0
    Uniform Scale Factor=&MergeHelper_FloatSwitch(
     0=&FloatFromCurve(
      FloatCurve=DataInterface() {propertyValues {Curve {keys [
          {arriveTangent=5.2354, interpMode=RCIM_Cubic, leaveTangent=5.2354, tangentMode=RCTM_User, time=-0.0046, value=0.25}
          <at 0.1439 1.4923>
          <at 1.0011 3.4809 RCIM_Cubic>
         ]}}}
      Scale Curve=&RandomRangeFloat(Minimum=0.5, Maximum=3)
     )
     1=&FloatFromCurve(FloatCurve=DataInterface() {propertyValues {Curve {
         defaultValue=1
         keys [{interpMode=RCIM_Cubic, tangentMode=RCTM_User, time=0, value=0}, {interpMode=RCIM_Cubic, tangentMode=RCTM_User, time=1, value=1}]
        }}})
     2=50
     3=0
     4=&FloatFromCurve(
      FloatCurve=DataInterface() {propertyValues {Curve {keys [<at 0.8 1 RCIM_Cubic>, {arriveTangent=-6.6126, interpMode=RCIM_Cubic, leaveTangent=-6.6126, tangentMode=RCTM_User, time=1, value=0}]}}}
      Scale Curve=0.1
     )
     Index=&SpawnGroup()
     Num=::NewEnumerator3
    )
   )
   /Niagara/Modules/Update/Forces/V2/CurlNoiseForce(
    Bias Noise=false
    Coordinate Space:ENiagaraCoordinateSpace=::Local
    Noise Frequency=&MergeHelper_FloatSwitch(0=0, 1=0, 2=0.02, 3=0, 4=0, 5=0, Index=&SpawnGroup(), Num=::NewEnumerator4)
    Noise Quality / Cost=::NewEnumerator0
    Noise Strength=&MergeHelper_FloatSwitch(
     0=0
     1=0
     2=500
     3=0
     4=&FloatFromCurve(FloatCurve=DataInterface() {propertyValues {Curve {keys [<at 0 1>,<at 1 0>]}}}, Scale Curve=3000)
     5=0
     Index=&SpawnGroup()
     Num=::NewEnumerator4
    )
    Pan Noise=false
    Random Seed=0
    Randomize Noise Sample=false
    Write to Intrinsic Parameters=false
   )
   &Drag(
    Drag=&MergeHelper_FloatSwitch(
     0=5
     1=0
     2=&FloatFromCurve(FloatCurve=DataInterface() {propertyValues {Curve {keys [<at 0 0.5>,<at 0.5 0.5>,<at 1 1>]}}}, Scale Curve=9)
     3=&FloatFromCurve(FloatCurve=DataInterface() {propertyValues {Curve {keys [<at 0 0.5>,<at 0.5 0.5>,<at 1 1>]}}}, Scale Curve=9)
     4=&FloatFromCurve(FloatCurve=DataInterface() {propertyValues {Curve {keys [<at 0 0.5141>,<at 0.5078 0.7118>,<at 0.9984 0.4918>]}}}, Scale Curve=10)
     5=5
     6=5
     Index=&SpawnGroup()
     Num=::NewEnumerator5
    )
    Use Linear Drag=true
    Use Rotational Drag=true
   )
   &SubUVAnimation(
    Output Alternate Renderer Binding=true
    Random Start Frame=true
    "Sprite Renderer "=DataInterface() {propertyValues {SpriteRenderer=&self:Fire_sigil_Prop_Spawn_0.NiagaraSpriteRendererProperties_1}}
    SubUV Lookup Index Scale=0.2
    UseEndFrame=false
    UseStartFrame=false
   )
   &SubUVAnimation(
    Output Alternate Renderer Binding=true
    "Sprite Renderer "=DataInterface() {propertyValues {SpriteRenderer=&self:Fire_sigil_Prop_Spawn_0.NiagaraSpriteRendererProperties_0}}
    UseEndFrame=false
    UseStartFrame=false
   ) {module=SubUVAnimation001}
   &SubUVAnimation(
    Output Alternate Renderer Binding=true
    "Sprite Renderer "=DataInterface() {propertyValues {SpriteRenderer=&self:Fire_sigil_Prop_Spawn_0.NiagaraSpriteRendererProperties_4}}
    UseEndFrame=false
    UseStartFrame=false
   ) {module=SubUVAnimation002}
   &GravityForce(Gravity=&MergeHelper_VectorSwitch(
     0=<0 0 0>
     1=<0 0 0>
     2=&MakeVector(
      X=0
      Y=0
      Z=&FloatFromCurve(
       FloatCurve=DataInterface() {propertyValues {Curve {keys [<at 0 0.25 RCIM_Cubic>,<at 0.25 1>]}}}
       CurveIndex=$Emitter.Age
       Scale Curve=&RandomRangeFloat(Minimum=-60, Maximum=-170)
      )
     )
     3=<0 0 5>
     4=&RandomRangeVector(Minimum=<0 0 -10>, Maximum=<1 1 50>)
     5=<0 0 0>
     Index=&SpawnGroup()
     Num=::NewEnumerator4
    ))
   &SolveForcesAndVelocity(Clamp Velocity=false, Limit Acceleration=false)
   &ScaleSpriteSizeBySpeed(
    Max Scale Factor=&MergeHelper_Vector2Switch(0=<1 1>, 1=<0 0>, 2=<0 0>, 3=<0 0>, 4=<1 5>, Index=&SpawnGroup(), Num=::NewEnumerator3)
    Min Scale Factor=&MergeHelper_Vector2Switch(0=<1 1>, 1=<0 0>, 2=<0 0>, 3=<0 0>, 4=<0 0>, Index=&SpawnGroup(), Num=::NewEnumerator3)
    Sample Scale Factor By Curve=true
    Velocity Threshold=500
   )
   &SubUVAnimation(
    Output Alternate Renderer Binding=true
    "Sprite Renderer "=DataInterface() {propertyValues {SpriteRenderer=&self:Fire_sigil_Prop_Spawn_0.NiagaraSpriteRendererProperties_5}}
    UseEndFrame=false
    UseStartFrame=false
   ) {module=SubUVAnimation003}
   /SetVariables(
    Particles.Glow Sprite size:Vector2f=&Vector2DFromFloat(Value=&FloatFromCurve(
      FloatCurve=DataInterface() {propertyValues {Curve {
         defaultValue=1
         keys [{interpMode=RCIM_Cubic, tangentMode=RCTM_User, time=0, value=0}, {interpMode=RCIM_Cubic, tangentMode=RCTM_User, time=1, value=1}]
        }}}
      Scale Curve=90.8315
     ))
    Particles.soft glow Color:LinearColor=&MakeLinearColorFromVectorAndFloat(
     "Vector (RGB)"=&MakeVectorFromLinearColorRGB(Color=<rgba 1 0.6731 0.3597 1>)
     "Float (Alpha)"=&FloatFromCurve(FloatCurve=DataInterface() {propertyValues {Curve {keys [<at 0 1>,<at 1 0>]}}})
    )
    Particles.soft glow Position:NiagaraPosition=&AddVectorToPosition(Position=$Particles.SkeletalMeshLocation.SampledPosition, Vector=<0 0 46.1712>)
   ) {isSetParameters=true}
   /SetVariables(
    Particles.Smoke Color:LinearColor=&MakeLinearColorFromVectorAndFloat(
     "Vector (RGB)"=&MakeVectorFromLinearColorRGB(Color=<rgba 0.0656 0.0656 0.0656 1>)
     "Float (Alpha)"=&FloatFromCurve(FloatCurve=DataInterface() {propertyValues {Curve {
         defaultValue=1
         keys [
          {arriveTangent=9.0725, interpMode=RCIM_Cubic, leaveTangent=9.0725, tangentMode=RCTM_User, time=0, value=0}
          {arriveTangent=-2.5249, interpMode=RCIM_Cubic, leaveTangent=-2.5249, tangentMode=RCTM_User, time=0.0826, value=1.0001}
          {arriveTangent=-0.0405, interpMode=RCIM_Cubic, leaveTangent=-0.0405, tangentMode=RCTM_User, time=1, value=0}
         ]
        }}})
    )
    Particles.Smoke Dyn Param:Vector4f=&MakeVector4(X=0, Y=0, Z=0, W=0)
   ) {isSetParameters=true}
   /SetVariables(
    Particles.Embers Color:LinearColor=&MakeLinearColorFromVectorAndFloat(
     "Vector (RGB)"=&MakeVectorFromLinearColorRGB(Color=&ColorFromCurve(ColorCurve=DataInterface() {propertyValues {
         AlphaCurve {keys [<at 0 1 RCIM_Cubic>]}
         BlueCurve {keys [<at 0 7.8125 RCIM_Cubic>,<at 0.2499 0.7812>,<at 0.7114 0.2083 RCIM_Cubic>]}
         GreenCurve {keys [<at 0 34.1758 RCIM_Cubic>,<at 0.2499 4.8887>,<at 0.7114 0.8934 RCIM_Cubic>]}
         RedCurve {keys [<at 0 50 RCIM_Cubic>,<at 0.2499 10>,<at 0.7114 4 RCIM_Cubic>]}
        }}))
     "Float (Alpha)"=1
    )
    Particles.Embers SpriteSize:Vector2f=&Multiply_Vector2D_ByFloat(
     Float=&FloatFromCurve(
      FloatCurve=DataInterface() {propertyValues {Curve {
         defaultValue=1
         keys [{interpMode=RCIM_Cubic, tangentMode=RCTM_User, time=0, value=0}, {interpMode=RCIM_Cubic, tangentMode=RCTM_User, time=0.1511, value=1}]
        }}}
      Scale Curve=1.2
     )
     Vector2D=&RandomRangeVector2D(Minimum=<0.4 0.8>, Maximum=<0.5 1.5>)
    )
   ) {isSetParameters=true}
   /SetVariables(
    Particles.Ashes Color:LinearColor=&MakeLinearColorFromVectorAndFloat(
     "Vector (RGB)"=&MakeVectorFromLinearColorRGB(Color=&ColorFromCurve(ColorCurve=DataInterface() {propertyValues {
         AlphaCurve {keys [<at 0 1 RCIM_Cubic>]}
         BlueCurve {keys [<at 0 1.473 RCIM_Cubic>,<at 0.6088 0.0879 RCIM_Cubic>]}
         GreenCurve {keys [<at 0 11.4982 RCIM_Cubic>,<at 0.6088 0.0919 RCIM_Cubic>]}
         RedCurve {keys [<at 0 50 RCIM_Cubic>,<at 0.6088 0.099 RCIM_Cubic>]}
        }}))
     "Float (Alpha)"=&FloatFromCurve(FloatCurve=DataInterface() {propertyValues {Curve {keys [<at 0.8 1 RCIM_Cubic>,<at 1 0>]}}})
    )
    Particles.Ashes SpriteSize:Vector2f=&Vector2DFromFloat(Value=&RandomRangeFloat(Minimum=1, Maximum=7))
    Particles.Vector 4:Vector4f=&MakeVector4(X=0, Y=0, Z=0, W=0)
   ) {isSetParameters=true}
   /SetVariables(
    Particles.Fire Color:LinearColor=&MakeLinearColorFromVectorAndFloat(
     "Vector (RGB)"=&Multiply_VectorByFloat(
      Vector=&MakeVectorFromLinearColorRGB(Color=<rgba 20 1.962 0 1>)
      Float=&FloatFromCurve(FloatCurve=DataInterface() {propertyValues {Curve {
          defaultValue=1
          keys [{interpMode=RCIM_Cubic, tangentMode=RCTM_User, time=0, value=1}, {interpMode=RCIM_Cubic, tangentMode=RCTM_User, time=1, value=0}]
         }}})
     )
     "Float (Alpha)"=1
    )
    Particles.Fire DynParam:Vector4f=&MakeVector4(
     X=&FloatFromCurve(FloatCurve=DataInterface() {propertyValues {Curve {keys [<at 0 1>]}}})
     Y=0.5
     Z=&FloatFromCurve(FloatCurve=DataInterface() {propertyValues {Curve {keys [<at 0 0.5 RCIM_Cubic>, {arriveTangent=1.5, interpMode=RCIM_Cubic, leaveTangent=1.5, time=0.5235, value=0.5}, <at 1 2>]}}})
     W=0
    )
    Particles.fire SpriteSize:Vector2f=&Multiply_Vector2D_ByFloat(
     Float=&FloatFromCurve(FloatCurve=DataInterface() {propertyValues {Curve {keys [
          {interpMode=RCIM_Cubic, tangentMode=RCTM_Break, time=0.0079, value=0}
          {interpMode=RCIM_Cubic, tangentMode=RCTM_Break, time=0.0578, value=2}
          {arriveTangent=-1.5342, interpMode=RCIM_Cubic, leaveTangent=-1.5342, time=0.1668, value=0.9624}
          {arriveTangent=-1.155, interpMode=RCIM_Cubic, leaveTangent=-1.155, time=0.7786, value=0.8942}
          <at 1 0>
         ]}}})
     Vector2D=&RandomRangeVector2D(Minimum=<20 0>, Maximum=<35 35>)
    )
   ) {isSetParameters=true}
   /SetVariables(
    Particles.FIre2 DynParam:Vector4f=&MakeVector4(
     X=&FloatFromCurve(
      FloatCurve=DataInterface() {propertyValues {Curve {
         defaultValue=1
         keys [{interpMode=RCIM_Cubic, tangentMode=RCTM_User, time=0, value=0}, {interpMode=RCIM_Cubic, tangentMode=RCTM_User, time=1.0015, value=0.6706}]
        }}}
      Scale Curve=0.5
     )
     Y=&RandomRangeFloat(Minimum=0, Maximum=4)
     Z=0
     W=2.5
    )
    Particles.Fire 2 Color:LinearColor=&MakeLinearColorFromVectorAndFloat(
     "Vector (RGB)"=&VectorFromFloat(Value=&FloatFromCurve(
       FloatCurve=DataInterface() {propertyValues {Curve {keys [
           {arriveTangent=-0.4927, interpMode=RCIM_Cubic, leaveTangent=-0.4927, tangentMode=RCTM_User, time=0, value=0.1087}
           {arriveTangent=-0.5399, interpMode=RCIM_Cubic, leaveTangent=-0.5399, tangentMode=RCTM_User, time=0.1564, value=0.3915}
           {arriveTangent=-0.4927, interpMode=RCIM_Cubic, leaveTangent=-0.4927, tangentMode=RCTM_User, time=0.987, value=0.0549}
          ]}}}
       Scale Curve=&RandomRangeFloat(Minimum=2, Maximum=5)
      ))
     "Float (Alpha)"=&FloatFromCurve(
      FloatCurve=DataInterface() {propertyValues {Curve {keys [
          {arriveTangent=3.3783, interpMode=RCIM_Cubic, leaveTangent=3.3783, tangentMode=RCTM_User, time=0, value=0}
          {arriveTangent=0.2941, interpMode=RCIM_Cubic, leaveTangent=0.2941, tangentMode=RCTM_User, time=0.203, value=0.7175}
          {arriveTangent=-1.109, interpMode=RCIM_Cubic, leaveTangent=-1.109, time=0.5122, value=0.5742}
          {arriveTangent=-0.0646, interpMode=RCIM_Cubic, leaveTangent=-0.0646, tangentMode=RCTM_User, time=0.85, value=0}
         ]}}}
      Scale Curve=&Multiply_Float(A=3, B=$System.TimeOfDay)
     )
    )
    Particles.Fire 2 Sprite Size:Vector2f=&Vector2DFromFloat(Value=&FloatFromCurve(
      FloatCurve=DataInterface() {propertyValues {Curve {
         defaultValue=1
         keys [{interpMode=RCIM_Cubic, tangentMode=RCTM_User, time=0, value=0}, {interpMode=RCIM_Cubic, tangentMode=RCTM_User, time=1, value=1}]
        }}}
      Scale Curve=&RandomRangeFloat(Minimum=45, Maximum=65)
     ))
   ) {isSetParameters=true}
   /SetVariables(
    Particles.Red smoke Sprite Size:Vector2f=&Vector2DFromFloat(Value=&Multiply_Float(
      A=&FloatFromCurve(FloatCurve=DataInterface() {propertyValues {Curve {keys [
           {arriveTangent=7.4819, interpMode=RCIM_Cubic, leaveTangent=7.4819, tangentMode=RCTM_User, time=0, value=0.4215}
           {arriveTangent=0.5944, interpMode=RCIM_Cubic, leaveTangent=0.5944, tangentMode=RCTM_User, time=0.1164, value=0.8874}
           {interpMode=RCIM_Cubic, tangentMode=RCTM_User, time=1, value=1}
          ]}}})
      B=&RandomRangeFloat(Minimum=100, Maximum=150)
     ))
    Particles.Red smoke SubImageIndex:NiagaraFloat=&Lerp_Float(A=25, B=63, Alpha=$Particles.NormalizedAge)
    Particles.Red smoke Update Color:LinearColor=&MultiplyLinearColor(
     Linear Color=$Particles.Initial.Red Mist Initial Color
     Scale Factor=&MakeVector4_FromLinearColor(LinearColor=&MakeLinearColorFromVectorAndFloat(
       "Vector (RGB)"=&VectorFromFloat(Value=&FloatFromCurve(FloatCurve=DataInterface() {propertyValues {Curve {keys [{arriveTangent=-0.1872, interpMode=RCIM_Cubic, leaveTangent=-0.1872, tangentMode=RCTM_User, time=0, value=2}, <at 1 1>]}}}))
       "Float (Alpha)"=&FloatFromCurve(
        FloatCurve=DataInterface() {propertyValues {Curve {keys [
            {arriveTangent=21.9814, interpMode=RCIM_Cubic, leaveTangent=21.9814, tangentMode=RCTM_User, time=0, value=0}
            {arriveTangent=0.1895, interpMode=RCIM_Cubic, leaveTangent=0.1895, tangentMode=RCTM_User, time=0.1, value=0.75}
            {arriveTangent=-2.5555, interpMode=RCIM_Cubic, leaveTangent=-2.5555, tangentMode=RCTM_User, time=1, value=0}
           ]}}}
        Scale Curve=&Multiply_Float(A=1, B=&FNTimeOfDayScalar(Night Scale=1, Morning Scale=1, Evening Scale=1))
       )
      ))
    )
   ) {isSetParameters=true}
   /SetVariables(
    Particles.Fire Eenergy SubImageIndex:NiagaraFloat=&Lerp_Float(A=25, B=63, Alpha=$Particles.NormalizedAge)
    Particles.Fire Energy Update Color:LinearColor=&MultiplyLinearColor(
     Linear Color=$Particles.Fire Energy Initial Color
     Scale Factor=&MakeVector4_FromLinearColor(LinearColor=&MakeLinearColorFromVectorAndFloat(
       "Vector (RGB)"=&VectorFromFloat(Value=&FloatFromCurve(FloatCurve=DataInterface() {propertyValues {Curve {keys [{arriveTangent=-0.1872, interpMode=RCIM_Cubic, leaveTangent=-0.1872, tangentMode=RCTM_User, time=0, value=2}, <at 1 1>]}}}))
       "Float (Alpha)"=&FloatFromCurve(
        FloatCurve=DataInterface() {propertyValues {Curve {keys [
            {arriveTangent=21.9814, interpMode=RCIM_Cubic, leaveTangent=21.9814, tangentMode=RCTM_User, time=0, value=0}
            {arriveTangent=0.1895, interpMode=RCIM_Cubic, leaveTangent=0.1895, tangentMode=RCTM_User, time=0.1, value=0.75}
            {arriveTangent=-2.5555, interpMode=RCIM_Cubic, leaveTangent=-2.5555, tangentMode=RCTM_User, time=1, value=0}
           ]}}}
        Scale Curve=&Multiply_Float(A=2, B=&FNTimeOfDayScalar(Night Scale=1, Morning Scale=1, Evening Scale=1))
       )
      ))
    )
   ) {isSetParameters=true}
   &FN_Light_Attributes(
    Color=<rgba 15 4.9774 1.3498 1>
    Create Color Attribute=true
    Create Exponent Attribute=true
    Create Position Attribute=true
    Create Radius Attribute=true
    Enable Particle Light Enable=true
    Enable Visibility Tag=true
    Light Enabled=true
    Position=<0 0 180>
    Radius=&FloatFromCurve(
     FloatCurve=DataInterface() {propertyValues {Curve {
        defaultValue=1
        keys [{interpMode=RCIM_Cubic, tangentMode=RCTM_User, time=0, value=0}, {interpMode=RCIM_Cubic, tangentMode=RCTM_User, time=0.3, value=1}]
       }}}
     Scale Curve=35
    )
    Scale Color=&FloatFromCurve(FloatCurve=DataInterface() {propertyValues {Curve {keys [
         {interpMode=RCIM_Cubic, tangentMode=RCTM_Break, time=0, value=1}
         {interpMode=RCIM_Cubic, tangentMode=RCTM_Break, time=0.6908, value=1}
         {interpMode=RCIM_Cubic, tangentMode=RCTM_Break, time=1, value=0}
        ]}}})
    VisibilityTag=10
   ) {enabled=false}
   /SetVariables(
    Particles.Wave Energy Scale:Vector3f=&VectorFromFloat(Value=&FloatFromCurve(
      FloatCurve=DataInterface() {propertyValues {Curve {keys [{interpMode=RCIM_Cubic, tangentMode=RCTM_Break, time=0, value=0}, {interpMode=RCIM_Cubic, tangentMode=RCTM_Break, time=0.68, value=2}]}}}
      Scale Curve=0.4
     ))
    Particles.wave energy DynParam:Vector4f=&MakeVector4(X=0, Y=1, Z=1, W=1)
    Particles.wave energy Update Color:LinearColor=&MultiplyLinearColor(
     Linear Color=$Particles.Wave Energy initial Color
     Scale Factor=&MakeVector4_FromLinearColor(LinearColor=&MakeLinearColorFromVectorAndFloat(
       "Vector (RGB)"=&VectorFromFloat(Value=&FloatFromCurve(FloatCurve=DataInterface() {propertyValues {Curve {keys [<at 0 0>, {arriveTangent=-15.019, interpMode=RCIM_Cubic, leaveTangent=-15.019, tangentMode=RCTM_User, time=0.2, value=6}, <at 0.85 0>]}}}))
       "Float (Alpha)"=&FN_FrontEndFloatModifier(
        Gameplay=&FNTimeOfDayScalar(Night Scale=0.1, Morning Scale=0.1, Day Scale=0.1, Evening Scale=0.1)
        Front End=0.5
       )
      ))
    )
   ) {isSetParameters=true}
  ]
  rendererClasses [&NiagaraSpriteRendererProperties, &NiagaraMeshRendererProperties]
  renderers [
   {rendererClass=&NiagaraSpriteRendererProperties}
   {rendererClass=&NiagaraSpriteRendererProperties, rendererIndex=1}
   {rendererClass=&NiagaraSpriteRendererProperties, rendererIndex=2}
   {rendererClass=&NiagaraSpriteRendererProperties, rendererIndex=3}
   {rendererClass=&NiagaraSpriteRendererProperties, rendererIndex=4}
   {rendererClass=&NiagaraSpriteRendererProperties, rendererIndex=5}
   {rendererClass=&NiagaraSpriteRendererProperties, rendererIndex=6}
   {rendererClass=&NiagaraSpriteRendererProperties, rendererIndex=7}
   {rendererClass=&NiagaraMeshRendererProperties, rendererIndex=8}
  ]
 }
 Fire_swirl_in {
  emitterUpdateScript [
   &EmitterState(Life Cycle Mode=::NewEnumerator1, Loop Behavior=::NewEnumerator1, Loop Duration=4, UseLoopDelay=true)
   &SpawnRate(SpawnGroup=1, SpawnRate=30, Use Spawn Probability=false) {enabled=false}
   &SpawnBurst_Instantaneous(Spawn Count=8, Spawn Group=5, Use Spawn Probability=false)
  ]
  particleSpawnScript [
   &InitializeParticle(
    Color=<rgba 3 1.1224 0.4219 1>
    Lifetime Max=0.6
    Lifetime Min=0.4
    Lifetime Mode=::NewEnumerator1
    Mesh Scale=&Multiply_Vector(A=<1 5 2>, B=&VectorFromFloat(Value=0.5))
    Mesh Scale Mode=::NewEnumerator4
    UsePositionOffset=true
    Write Material Random=true
   ) {module=InitializeParticle001}
   &InitialMeshOrientation(
    Mesh Orientation Mode=::NewEnumerator7
    Rotation=&RandomRangeVector(Minimum=<0 -0.01 0>, Maximum=<1 0.01 0>)
    Rotation Coordinate Space=::NewEnumerator0
    Use Rotation Vector=true
   )
   &SkeletalMeshLocation(Bone Sampling Mode=::NewEnumerator4, MaskSpawnGroup=false, Skeletal Mesh=$System.sigil)
   &InitialMeshOrientation(Mesh Orientation Mode=::NewEnumerator2, Orientation Vector=<0 1 0>, Use Rotation Vector=false) {module=InitialMeshOrientation001}
   &CameraOffset(Camera Offset Amount=35)
  ]
  particleUpdateScript [
   &ParticleState()
   &ScaleMeshSize(Scale Factor=&VectorFromFloat(Value=&FloatFromCurve(
      FloatCurve=DataInterface() {propertyValues {Curve {defaultValue=1, keys [{tangentMode=RCTM_User, time=0, value=0}, {tangentMode=RCTM_User, time=1, value=1}]}}}
      Scale Curve=1.3
     )))
   &UpdateMeshOrientation(Rotation Vector=<-1 0 0>)
   &SolveForcesAndVelocity(Clamp Velocity=false, Limit Acceleration=false)
   &DynamicMaterialParameters(
    Index 0 Param 1=&FloatFromCurve(FloatCurve=DataInterface() {propertyValues {Curve {keys [{arriveTangent=3.1068, interpMode=RCIM_Cubic, leaveTangent=3.1068, tangentMode=RCTM_User, time=0, value=0}, <at 1 1 RCIM_Cubic>]}}})
    Index 0 Param 2=&FloatFromCurve(
     FloatCurve=DataInterface() {propertyValues {Curve {keys [{interpMode=RCIM_Cubic, tangentMode=RCTM_Break, time=0, value=1}, <at 1 1.5>]}}}
     Scale Curve=&RandomRangeFloat(Minimum=1, Maximum=2)
    )
    Index 0 Param 3=0.1
    Index 0 Param 4=&RandomRangeFloat(Minimum=-1, Maximum=1)
    Module.Index0Float0Write=true
    Module.Index0Float1Write=true
    Module.Index0Float2Write=true
    Module.Index0Float3Write=true
   )
   &ScaleColor(
    Scale Alpha=&FloatFromCurve(FloatCurve=DataInterface() {propertyValues {Curve {keys [
         {arriveTangent=2.1627, interpMode=RCIM_Cubic, leaveTangent=2.1627, tangentMode=RCTM_User, time=-0.0151, value=0.0235}
         {interpMode=RCIM_Cubic, leaveTangent=0.2258, tangentMode=RCTM_Break, time=0.3183, value=0.9764}
         <at 1 0>
        ]}}})
    Scale RGB=&Multiply_VectorByFloat(
     Vector=<1 1 1>
     Float=&FloatFromCurve(FloatCurve=DataInterface() {propertyValues {Curve {keys [<at 0 0.2>,<at 0.0811 5>,<at 1 0.5>]}}})
    )
    ScaleA=true
    ScaleRGB=true
   )
  ]
  rendererClasses [&NiagaraMeshRendererProperties]
  renderers [{rendererClass=&NiagaraMeshRendererProperties, rendererIndex=0}]
 }
}
summary {
 systemName=NS_OutlineLove_sigil_spawn_v2
 emitters [
  {
   emitterName=Fire_sigil_Prop_Spawn
   rendererClasses [&NiagaraSpriteRendererProperties, &NiagaraMeshRendererProperties]
  }
  {emitterName=Embers_Burst001, simTarget=GPUComputeSim, rendererClasses [&NiagaraSpriteRendererProperties]}
  {emitterName=Fire_swirl_in, rendererClasses [&NiagaraMeshRendererProperties]}
  {
   emitterName=Fire_ring_Ignition
   rendererClasses [&NiagaraMeshRendererProperties, &NiagaraLightRendererProperties]
  }
 ]
 userVariables [
  User.bIsFrontEnd {type=NiagaraBool}
  User.bIsFrontEndPreview {type=NiagaraBool}
  User.Erode_circle {type=NiagaraFloat, defaultValue=0.1, description=""}
 ]
}
